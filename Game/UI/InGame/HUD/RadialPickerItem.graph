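# RadialPickerItem is a HUD widget for one option in a radial picker, showing either an icon with label or a Fortnite item card.
# Setup functions fill the content, apply enabled state, substitute a lock icon when locked and bind an indexed PickerOption key.
# Selection toggles card border and dimming, or plays the icon scale animation forward or reversed from its current time, changing tint and Z-order.
# Mouse press broadcasts On Item Touched.
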WidgetBlueprint /Game/UI/InGame/HUD/RadialPickerItem
animations [Selected {
  binding_count=1
  display_label=""
  length_seconds=0.1003
  object=&self:Selected
  sequence {bindings [ImageIcon {
     id=#1
     possessed_class=CommonLazyImage
     tracks [{
       type=MovieScene2DTransformTrack
       sections [{
         type=MovieScene2DTransformSection
         end_frame=2
         start_frame=0
         channels [
          Translation.X {type=MovieSceneScriptingFloatChannel}
          Translation.Y {type=MovieSceneScriptingFloatChannel}
          Angle {type=MovieSceneScriptingFloatChannel}
          Scale.X {type=MovieSceneScriptingFloatChannel, keys [[0,1],[1.3973,1.2187],[2.006,1.2]]}
          Scale.Y {type=MovieSceneScriptingFloatChannel, keys [[0,1],[1.3973,1.2187],[2.006,1.2]]}
          Shear.X {type=MovieSceneScriptingFloatChannel}
          Shear.Y {type=MovieSceneScriptingFloatChannel}
         ]
        }]
      }]
    }]}
 }]
graphs [
 EventGraph {nodes [Event Construct()]}
 On Item Touched {nodes [On Item Touched!() -> (then, Picked Option:Radial Picker Item Object Reference) {is=title K2Node_FunctionEntry}]}
 OnMouseButtonDown {nodes [
   OnMouseButtonDown!() -> (then, MyGeometry:Geometry Structure, MouseEvent:"Pointer Event Structure (by ref)") {is=title K2Node_FunctionEntry}
   > Call On Item Touched!(execute:Exec, self:Radial Picker Item Object Reference, Picked Option:Radial Picker Item Object Reference=Self-Reference()) -> (then:Exec) {is=title K2Node_CallDelegate}
   > Return Node!(execute, ReturnValue:Event Reply Structure=Handled())
  ]}
 SetImageAndLabelContent {nodes [
   n1 = SetImageAndLabelContent!() -> (
    then
    Text:Text
    Icon:Texture 2D Soft Object Reference
    Brush:Slate Brush Structure
    Option Enabled:bool=~
    RadialItemIndex:int
    IsLocked:bool=~
   ) {is=title K2Node_FunctionEntry}
   > n2 = Set PickerItemText(PickerItemText:Text=@n1) -> (Output_Get:Text)
   > "SetText (Text)"(self=Get TextLabel() -> (TextLabel:Common Text Object Reference), InText=@n2)
   > n3 = Branch(Condition=IsValidSoftObjectReference(SoftObjectReference=@n1.Icon))
   n4 = Get LockedTexture() -> (LockedTexture:Texture 2D Object Reference)
   > n5 = SetBrushFromLazyTexture!(
    execute
    self:Common Lazy Image Object Reference=Get ImageIcon() -> (ImageIcon:Common Lazy Image Object Reference)
    LazyTexture:"Texture 2D Soft Object Reference (by ref)"=Select(
     Option 0:Texture 2D Soft Object Reference=@n1.Icon
     Option 1:Texture 2D Soft Object Reference=Make Soft Reference!(Input:Texture 2D Object Reference=@n4) -> (Output:Texture 2D Soft Object Reference)
     Index=@n1.IsLocked
    ) -> (ReturnValue:Texture 2D Soft Object Reference)
    bMatchSize:bool=true
   ) -> (then)
   SetIsEnabled(execute=[@n5.then, @n8.then], bInIsEnabled=@n1.Option Enabled)
   n6 = "Build String (Integer)"!(
    AppendTo:String
    Prefix:String=PickerOption
    InInt:int=int + int(A=@n1.RadialItemIndex, B=1) -> (ReturnValue=0) {is=title K2Node_CommutativeAssociativeBinaryOperator}
    Suffix:String
   ) -> (ReturnValue:String)
   > SetBoundAction!(
    execute
    self:Fort Keybind Widget Object Reference=Get KeybindWidget() -> (KeybindWidget:Keybind Widget Object Reference)
    NewBoundAction:Name=MakeLiteralName(Value=String To Name(InString=@n6))
   ) -> (then)
   n7 = Get LockedTexture()
   n8 = SetBrush(
    execute=@n3.else
    self=Get ImageIcon()
    InBrush=Select(
     Option 0:Slate Brush Structure=@n1.Brush
     Option 1:Slate Brush Structure=MakeBrushFromTexture!(Texture:Texture 2D Object Reference=@n7, Width:int=0, Height:int=0) -> (ReturnValue:Slate Brush Structure)
     Index=@n1.IsLocked
    ) -> (ReturnValue:Slate Brush Structure)
   )
  ]}
 SetItemContent {nodes [
   n1 = SetItemContent!() -> (then, Item:Fort Item Object Reference, Option Enabled:bool=~, Radial Item Index:int) {is=title K2Node_FunctionEntry}
   > SetItemToRepresent!(
    execute
    self:Fort Item Widget NUI Object Reference=Get PickerItem() -> (PickerItem:Fort Multi Size Item Card Object Reference)
    ItemToRepresent:Fort Item Object Reference=@n1
   ) -> (then)
   > Set PickerItemText(PickerItemText=GetDisplayName!(self:Fort Item Object Reference=@n1) -> (ReturnValue:Text))
   > SetIsEnabled(bInIsEnabled=@n1.Option Enabled)
   > n2 = Branch(Condition=@n1.Option Enabled)
   n3 = Get ImageItemNotSelected() -> (ImageItemNotSelected:Image Object Reference)
   > n4 = SetVisibility(self=@n3, InVisibility=SelfHitTestInvisible)
   n5 = "Build String (Integer)"(Prefix=PickerOption, InInt=int + int(A=@n1.Radial Item Index, B=1) -> (ReturnValue=0))
   SetBoundAction(execute=[@n6.then, @n4.then], self=Get KeybindWidget(), NewBoundAction=MakeLiteralName(Value=String To Name(InString=@n5)))
   n6 = SetVisibility(execute=@n2.else, self=@n3, InVisibility=Hidden)
  ]}
 SetSelected {nodes [
   n1 = SetSelected!() -> (then, IsSelected:bool=~) {is=title K2Node_FunctionEntry}
   n2 = Get ImageOrItemSwitcher() -> (ImageOrItemSwitcher:Common Widget Switcher Legacy Object Reference)
   > n3 = Branch(Condition="Equal (Integer)"(A=GetActiveWidgetIndex!(self:Widget Switcher Object Reference=@n2) -> (ReturnValue:int=0), B=1) {is=title K2Node_CallFunction})
   > SetVisibility(
    self=Get ImageItemNotSelected()
    InVisibility=Select(Option 0:ESlateVisibility Enum=SelfHitTestInvisible, Option 1:ESlateVisibility Enum=Hidden, Index=@n1) -> (ReturnValue:ESlateVisibility Enum)
   )
   > n4 = Branch(Condition=@n1)
   n5 = Get BorderItemBorder() -> (BorderItemBorder:Common Border Object Reference)
   > SetVisibility(self=@n5, InVisibility=SelfHitTestInvisible)
   SetVisibility(execute=@n4.else, self=@n5, InVisibility=Collapsed)
   n6 = Branch(execute=@n3.else, Condition=@n1)
   n7 = Get Selected() -> (Selected:Widget Animation Object Reference)
   n8 = GetAnimationCurrentTime!(self:User Widget Object Reference, InAnimation:Widget Animation Object Reference=@n7) -> (ReturnValue:float)
   > PlayAnimation(InAnimation=@n7, StartAtTime=@n8)
   n9 = SlotAsCanvasSlot!(Widget:Widget Object Reference=Self-Reference()) -> (ReturnValue:Canvas Panel Slot Object Reference)
   > SetZOrder!(execute, self:Canvas Panel Slot Object Reference=@n9, InZOrder:int=1) -> (then)
   > UpdateSelectionColor!(execute, self, Selected:bool=true) -> (then)
   n10 = GetEndTime(self=@n7)
   PlayAnimation(
    execute=@n6.else
    InAnimation=@n7
    StartAtTime=Select(
     Option 0:double=0
     Option 1:double=float - float(A=@n10, B=@n8) {is=title K2Node_CallFunction}
     Index="float > float"(A=@n8, B=0) {is=title K2Node_CallFunction}
    ) -> (ReturnValue:double)
    PlayMode=Reverse
   )
   > SetZOrder(self=@n9, InZOrder=0)
   > UpdateSelectionColor(Selected=false)
   n11 = Get ImageIcon()
   SetRenderScale!(execute, self:Widget Object Reference=@n11, Scale:Vector 2D Structure=(X=1,Y=1)) -> (then)
   SetRenderScale(self=@n11, Scale=(X=1.2,Y=1.2))
  ]}
 SetShowImageAndLabel {nodes [
   n1 = SetShowImageAndLabel!() -> (then, ShowImageAndLabel:bool=~) {is=title K2Node_FunctionEntry}
   > n2 = Branch(Condition=@n1)
   n3 = Get ImageOrItemSwitcher()
   > SetActiveWidgetIndex(self=@n3, Index=0)
   SetActiveWidgetIndex(execute=@n2.else, self=@n3, Index=1)
  ]}
 UpdateSelectionColor {nodes [
   n1 = UpdateSelectionColor!() -> (then, Selected:bool=~) {is=title K2Node_FunctionEntry}
   n2 = GetIsEnabled!(self:Widget Object Reference) -> (ReturnValue:bool)
   n3 = Make Slate Color(SpecifiedColor=Select(
     Option 0:Linear Color Structure=Get DefaultUnselectedColor() -> (DefaultUnselectedColor:Linear Color Structure)
     Option 1:Linear Color Structure=Get DefaultSelectedColor() -> (DefaultSelectedColor:Linear Color Structure)
     Index=AND Boolean(A=@n1, B=@n2)
    ) -> (ReturnValue:Linear Color Structure))
   > SetBrushTintColor!(execute, self:Image Object Reference=Get ImageIcon(), TintColor:Slate Color Structure=@n3) -> (then)
   > SetColorAndOpacity(self=Get TextLabel(), InColorAndOpacity=@n3)
  ]}
]
variables [
 DefaultSelectedColor {type=LinearColor, default=<rgba 1 1 0 1>}
 DefaultUnselectedColor {type=LinearColor, default=<rgba 1 1 1 0.6>}
 LockedTexture {type=Texture2D, default=/Game/UI/Foundation/Textures/Icons/Locks/T-Icon-Lock-128}
 On Item Touched {}
 PickerItemText {type=string, default=""}
]
widgets {
 info {inheritedWidgetCount=0, parentClass=&CommonUserWidget, rootWidgetClass=&VerticalBox, widgetCount=14}
 widgets [
  VerticalBox_2 {widgetClassPath=&VerticalBox, widget=VerticalBox_2}
  KeybindWidget {
   widgetClassPath=/FortUILibrary/Input/KeybindWidget
   bIsVariable=true
   parent=VerticalBox_2
   widget=KeybindWidget
   properties {boundAction=PickerOption1}
   slot {class=VerticalBoxSlot, properties {horizontalAlignment=HAlign_Center, verticalAlignment=VAlign_Center}}
  }
  ImageOrItemSwitcher {
   widgetClassPath=&CommonWidgetSwitcherLegacy
   bIsVariable=true
   parent=VerticalBox_2
   widget=ImageOrItemSwitcher
   properties {transitionDuration=0}
   slot {class=VerticalBoxSlot}
  }
  VerticalBox_0 {
   widgetClassPath=&VerticalBox
   parent=ImageOrItemSwitcher
   widget=VerticalBox_0
   slot {class=WidgetSwitcherSlot, properties {horizontalAlignment=HAlign_Center, verticalAlignment=VAlign_Center}}
  }
  CommonBorder_109 {
   widgetClassPath=&CommonBorder
   parent=VerticalBox_0
   widget=CommonBorder_109
   properties {horizontalAlignment=HAlign_Center, verticalAlignment=VAlign_Center, padding=<ltrb 0 0 0 0>}
   slot {class=VerticalBoxSlot, properties {horizontalAlignment=HAlign_Center, verticalAlignment=VAlign_Center, padding=<ltrb 1 1 1 1>}}
  }
  SizeBox_1 {
   widgetClassPath=&SizeBox
   parent=VerticalBox_0
   widget=SizeBox_1
   properties {bOverride_MaxDesiredHeight=true, bOverride_MaxDesiredWidth=true, maxDesiredHeight=128, maxDesiredWidth=128}
   slot {class=VerticalBoxSlot, properties {horizontalAlignment=HAlign_Center, verticalAlignment=VAlign_Center}}
  }
  ImageIcon {
   widgetClassPath=/Script/CommonUI.CommonLazyImage
   bIsVariable=true
   parent=SizeBox_1
   widget=ImageIcon
   properties {
    brush {
     bIsDynamicallyLoaded=false
     drawAs=Image
     imageType=NoImage
     mirroring=NoMirror
     resourceName=None
     resourceObject=/Game/UI/Foundation/Textures/Icons/Stars/T-Stars-TierNumber-1-32
     tiling=NoTile
     imageSize=<128 128>
     margin=<ltrb 0 0 0 0>
     outlineSettings {
      bUseBrushTransparency=false
      roundingType=HalfHeightRadius
      width=0
      color {colorUseRule=UseColor_Specified, specifiedColor=<rgba 0 0 0 0>}
      cornerRadii=<0 0 0 0>
     }
     tintColor {colorUseRule=UseColor_Specified, specifiedColor=<rgba 1 1 1 1>}
     uVRegion {bIsValid=false, max=<0 0>, min=<0 0>}
    }
    colorAndOpacity=<rgba 1 1 1 0.6>
    renderTransformPivot=<0.5 0.3>
   }
   slot {class=SizeBoxSlot}
  }
  Spacer_0 {widgetClassPath=&Spacer, parent=VerticalBox_0, widget=Spacer_0, slot {class=VerticalBoxSlot}}
  TextLabel {
   widgetClassPath=&CommonTextBlock
   bIsVariable=true
   parent=VerticalBox_0
   widget=TextLabel
   properties {
    justification=Center
    style=/FortUILibrary/Styles/Typography/Foundation/TextStyle-Base-XS-B-S.TextStyle-Base-XS-B-S_C
    text=Label
    font {
     bForceMonospaced=false
     bMaterialIsStencil=false
     fontMaterial=null
     fontObject=&NotoSans
     letterSpacing=0
     monospacedWidth=1
     size=16
     skewAmount=0
     typefaceFontName=NotoSans-Bold
     outlineSettings {
      bApplyOutlineToDropShadows=true
      bMiteredCorners=false
      bSeparateFillAlpha=true
      outlineMaterial=&M-FontShadow
      outlineSize=2
      outlineColor=<rgba 0 0 0 0.5>
     }
    }
    shadowOffset=<0 0>
   }
   slot {class=VerticalBoxSlot, properties {horizontalAlignment=HAlign_Center, verticalAlignment=VAlign_Center}}
  }
  VerticalBox_1 {
   widgetClassPath=&VerticalBox
   parent=ImageOrItemSwitcher
   widget=VerticalBox_1
   slot {class=WidgetSwitcherSlot, properties {horizontalAlignment=HAlign_Center, verticalAlignment=VAlign_Center}}
  }
  Overlay_0 {
   widgetClassPath=&Overlay
   parent=VerticalBox_1
   widget=Overlay_0
   slot {class=VerticalBoxSlot, properties {horizontalAlignment=HAlign_Center, verticalAlignment=VAlign_Center, padding=<ltrb 1 1 1 1>}}
  }
  BorderItemBorder {
   widgetClassPath=&CommonBorder
   bIsVariable=true
   parent=Overlay_0
   widget=BorderItemBorder
   properties {
    style=/FortUILibrary/Styles/Border/Border-1pxOutline.Border-1pxOutline_C
    visibility=Hidden
    brushColor=<rgba 1 1 0 1>
    padding=<ltrb 0 0 0 0>
   }
   slot {class=OverlaySlot, properties {horizontalAlignment=HAlign_Fill, verticalAlignment=VAlign_Fill}}
  }
  PickerItem {
   widgetClassPath=/Script/FortniteUI.FortMultiSizeItemCard
   bIsVariable=true
   parent=Overlay_0
   widget=PickerItem
   properties {bPItemCardSize=M, emptyItemType=AccountItem}
   slot {class=OverlaySlot, properties {horizontalAlignment=HAlign_Fill, verticalAlignment=VAlign_Fill, padding=<ltrb 4 4 4 4>}}
  }
  ImageItemNotSelected {
   widgetClassPath=&Image
   bIsVariable=true
   parent=Overlay_0
   widget=ImageItemNotSelected
   properties {visibility=Hidden, colorAndOpacity=<rgba 0 0 0 0.2>}
   slot {class=OverlaySlot, properties {horizontalAlignment=HAlign_Fill, verticalAlignment=VAlign_Fill, padding=<ltrb 4 4 4 4>}}
  }
 ]
}
