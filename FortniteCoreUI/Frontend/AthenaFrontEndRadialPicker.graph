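# Front-end radial picker widget (FortRadialPicker) that shows options in a ring, with paging between wheels, including emote wheels.
# InitializePicker fills eight pre-placed items from FortPickerContext, creating more as needed, and spaces them evenly by angle.
# Tick converts mouse or thumbstick direction to an angle, highlighting the matching slice and rotating the pointer.
# Clicks, item touches and context events confirm or cancel; closing restores input and reports back to the context.
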WidgetBlueprint /FortniteCoreUI/Frontend/AthenaFrontEndRadialPicker
graphs [
 AcceptChosenOption {nodes [
   n1 = AcceptChosenOption!() -> (then, PickerOption:int) {is=title K2Node_FunctionEntry}
   > SetActiveOption!(execute, self, Option:int=@n1) -> (then)
   > AcceptOption()
  ]}
 AcceptOption {nodes [
   AcceptOption() {is=title K2Node_FunctionEntry}
   > Branch(Condition=CanConfirm!() -> (CanAccept:bool))
   > Set PickerClosing(PickerClosing:bool=true) -> (Output_Get:bool)
   > PlaySound2D(Sound=Get SoundOnAccept() -> (SoundOnAccept:Sound Base Object Reference))
   > ClosePicker()
   > PickerOptionAccepted!(
    execute
    self:Fort Picker Context Object Reference=Get FortPickerContext!(PlayerController:Player Controller Object Reference=GetOwningPlayer()) -> (ReturnValue:Fort Picker Context Object Reference) {is=title K2Node_GetSubsystemFromPC}
    Option:int=Get ActiveOption() -> (ActiveOption:int=0)
   ) -> (then)
  ]}
 BindPickerEvents {nodes [
   BindPickerEvents() {is=title K2Node_FunctionEntry}
   > Bind Event to On Picker Option Chosen(self:Fort Picker Context Object Reference=Get FortPickerContext(PlayerController=GetOwningPlayer()), Delegate=Create Event()) {is=title K2Node_AddDelegate}
   > Bind Event to On Picker Confirm(self:Fort Picker Context Object Reference=Get FortPickerContext(PlayerController=GetOwningPlayer()), Delegate=Create Event()) {is=title K2Node_AddDelegate}
   > Bind Event to On Picker Cancel(self:Fort Picker Context Object Reference=Get FortPickerContext(PlayerController=GetOwningPlayer()), Delegate=Create Event()) {is=title K2Node_AddDelegate}
   > Bind Event to On Picker Option Moved(self:Fort Picker Context Object Reference=Get FortPickerContext(PlayerController=GetOwningPlayer()), Delegate=Create Event()) {is=title K2Node_AddDelegate}
   > Bind Event to On Picker Refresh Items(self:Fort Picker Context Object Reference=Get FortPickerContext(PlayerController=GetOwningPlayer()), Delegate=Create Event()) {is=title K2Node_AddDelegate}
  ]}
 CanConfirm {nodes [
   CanConfirm() {is=title K2Node_FunctionEntry}
   n1 = Get ActiveOption() -> (ActiveOption=0)
   > n2 = Branch(Condition="Not Equal (Integer)"!(A:int=@n1, B:int=-1) -> (ReturnValue:bool) {is=title K2Node_CallFunction})
   n3 = Get PickerOptions() -> (PickerOptions:Array of Radial Picker Item Object References)
   > Return Node!(
    execute
    CanAccept:bool=GetIsEnabled!(self:Widget Object Reference="Get (a copy)"(Array:Array of Radial Picker Item Object References=@n3, Dimension 1=@n1) -> (Output:Radial Picker Item Object Reference)) -> (ReturnValue:bool)
   )
   Return Node(execute=@n2.else, CanAccept=false)
  ]}
 ClearActiveOption {nodes [
   ClearActiveOption() {is=title K2Node_FunctionEntry}
   > Branch(Condition="Not Equal (Integer)"(A=Get ActiveOption() -> (ActiveOption=0), B=-1))
   > SetSelected!(
    execute
    self:Radial Picker Item Object Reference="Get (a copy)"(Array=Get PickerOptions(), Dimension 1=Get ActiveOption() -> (ActiveOption=0))
    IsSelected:bool=false
   ) -> (then)
   > Set ActiveOption(ActiveOption:int=-1) -> (Output_Get:int=0)
  ]}
 ClosePicker {nodes [
   ClosePicker() {is=title K2Node_FunctionEntry}
   > Branch(Condition="Not Equal (Enum)"!(
     A:ESlateVisibility Enum=GetVisibility!(self:Widget Object Reference) -> (ReturnValue:ESlateVisibility Enum=Visible)
     B:ESlateVisibility Enum=Collapsed
    ) -> (ReturnValue:bool))
   > SetVisibility(InVisibility=Collapsed)
   > BroadcastPanelClosed(self:Fort Radial Picker Object Reference)
   > ResetInput()
   > "SetText (Text)"(self=Get Text_OK() -> (Text_OK:Common Text Object Reference))
   > SetRenderOpacity(self=Get KeybindPickerConfirm() -> (KeybindPickerConfirm:Keybind Widget Object Reference), InOpacity=0.5)
   > n1 = For Each Loop(Array:Array of Radial Picker Item Object References=Get PickerOptions()) -> (Array Element:Radial Picker Item Object Reference)
   > SetSelected(self=@n1, IsSelected=false)
   Call Picker Closing!(execute:Exec=@n1.Completed, self:Athena Front End Radial Picker Object Reference) -> (then:Exec) {is=title K2Node_CallDelegate}
  ]}
 Event CancelPicker {nodes [
   Event CancelPicker() {is=title K2Node_FunctionEntry}
   > Set PickerClosing(PickerClosing=true)
   > PlaySound2D(Sound=Get SoundOnCancel() -> (SoundOnCancel:Sound Base Object Reference))
   > ClosePicker() {is=title K2Node_CallFunction}
   > PickerCanceled(self:Fort Picker Context Object Reference=Get FortPickerContext(PlayerController=GetOwningPlayer()))
  ]}
 EventGraph {nodes [
   "On Button Clicked (NextWheelButton)"!() -> (OutputDelegate, then, Button:Common Button Legacy Object Reference) {is=title K2Node_ComponentBoundEvent}
   > LegacySwitchToAdjacentWheel!(execute, self:Fort Picker Context Object Reference=Get FortPickerContext(PlayerController=GetOwningPlayer()), SelectionDirection:int=1) -> (then)
   "On Button Clicked (PreviousWheelButton)"!() -> (OutputDelegate, then, Button:Common Button Legacy Object Reference) {is=title K2Node_ComponentBoundEvent}
   > LegacySwitchToAdjacentWheel(self=Get FortPickerContext(PlayerController=GetOwningPlayer()), SelectionDirection=-1)
   n1 = "On Item Touched (PickerOption3)"!() -> (OutputDelegate, then, Picked Option:Radial Picker Item Object Reference) {is=title K2Node_ComponentBoundEvent}
   > On Item Touched!(execute, self, Touched Item:Radial Picker Item Object Reference=@n1.Picked Option) -> (then)
   n2 = "On Item Touched (PickerOption4)"!() -> (OutputDelegate, then, Picked Option:Radial Picker Item Object Reference) {is=title K2Node_ComponentBoundEvent}
   > On Item Touched(Touched Item=@n2.Picked Option)
   n3 = "On Item Touched (PickerOption0)"!() -> (OutputDelegate, then, Picked Option:Radial Picker Item Object Reference) {is=title K2Node_ComponentBoundEvent}
   > On Item Touched(Touched Item=@n3.Picked Option)
   n4 = "On Item Touched (PickerOption5)"!() -> (OutputDelegate, then, Picked Option:Radial Picker Item Object Reference) {is=title K2Node_ComponentBoundEvent}
   > On Item Touched(Touched Item=@n4.Picked Option)
   n5 = "On Item Touched (PickerOption6)"!() -> (OutputDelegate, then, Picked Option:Radial Picker Item Object Reference) {is=title K2Node_ComponentBoundEvent}
   > On Item Touched(Touched Item=@n5.Picked Option)
   n6 = "On Item Touched (PickerOption7)"!() -> (OutputDelegate, then, Picked Option:Radial Picker Item Object Reference) {is=title K2Node_ComponentBoundEvent}
   > On Item Touched(Touched Item=@n6.Picked Option)
   n7 = "On Item Touched (PickerOption1)"!() -> (OutputDelegate, then, Picked Option:Radial Picker Item Object Reference) {is=title K2Node_ComponentBoundEvent}
   > On Item Touched(Touched Item=@n7.Picked Option)
   n8 = "On Item Touched (PickerOption2)"!() -> (OutputDelegate, then, Picked Option:Radial Picker Item Object Reference) {is=title K2Node_ComponentBoundEvent}
   > On Item Touched(Touched Item=@n8.Picked Option)
   Event Construct()
   > "SetText (Text)"(self=Get Text_OK())
   > SetRenderOpacity(self=Get KeybindPickerConfirm(), InOpacity=0.5)
   > InitializePickerOptions()
   > SetupMobileVisuals()
   > BindPickerEvents() {is=title K2Node_CallFunction}
   Event Tick()
   n9 = GetCurrentInputType(self=Get CommonInputSubsystem!() -> (ReturnValue:Common Input Object Reference) {is=title K2Node_GetBlueprintContextEx})
   > n10 = Branch(Condition="Not Equal (Boolean)"!(A:bool=Get bUsingGamepad() -> (bUsingGamepad:bool), B:bool="Equal (Enum)"(A=@n9, B=Gamepad)) -> (ReturnValue:bool) {is=title K2Node_CallFunction})
   > n11 = SetInputMode()
   n12 = Branch(execute=[@n11.then, @n10.else], Condition=Get PickerClosing() -> (PickerClosing:bool))
   n13 = Get CommonInputSubsystem()
   n14 = Branch(execute=@n12.else, Condition="Equal (Enum)"(A=GetCurrentInputType(self=@n13), B=Touch))
   > n15 = SetVisibility(self=Get ImageRadialPointer() -> (ImageRadialPointer:Image Object Reference), InVisibility=Hidden)
   n16 = Get bUsingGamepad()
   n17 = IsGamepadInPickerDeadZone!(self) -> (bIsInDeadZone:bool)
   n18 = Branch(execute=[@n15.then, @n14.else], Condition=OR Boolean(A=NOT Boolean(A=@n16), B=NOT Boolean(A=@n17)))
   n19 = BreakVector2D(InVec=vector2d / float!(A:Vector 2D Structure=GetViewportSize!() -> (ReturnValue:Vector 2D Structure), B:double=2) -> (ReturnValue:Vector 2D Structure))
   n20 = GetMousePosition!(self:Player Controller Object Reference=GetOwningPlayer()) -> (LocationX:float, LocationY:float, ReturnValue:bool)
   n21 = float - float(A=@n20, B=@n19) {is=title K2Node_CallFunction}
   n22 = float * float(A=float - float(A=@n20.LocationY, B=@n19.Y), B=-1)
   n23 = GetInputAnalogStickState!(
    self:Player Controller Object Reference=GetOwningPlayer()
    WhichStick:EControllerAnalogStick Enum=Get Thumbstick() -> (Thumbstick:EControllerAnalogStick Enum=CAS_LeftStick)
   ) -> (StickX:float, StickY:float)
   n24 = float * float(A=@n23.StickY, B=Get YAxisMultiplier() -> (YAxisMultiplier:double))
   > Set CurrentPointerAngle(CurrentPointerAngle:double=Select(
     Option 0:double="Atan2 (Degrees)"!(Y:double=@n22, X:double=@n21) -> (ReturnValue:double)
     Option 1:double="Atan2 (Degrees)"(Y=@n24, X=@n23)
     Index=Get bUsingGamepad()
    ) -> (ReturnValue:double)) -> (Output_Get:double)
   > n25 = Sequence()
   > n26 = For Each Loop with Break!(Exec:Exec, Array:Array of Radial Picker Item Object References=Get PickerOptions(), Break:Exec=[@n27.else, @n31.else, @n32.then]) -> (LoopBody:Exec, Array Element:Radial Picker Item Object Reference, Array Index:int, Completed:Exec)
   > n27 = Branch(Condition="integer < integer"!(A:int=@n26.Array Index, B:int=Get NumOptions() -> (NumOptions:int=0)) -> (ReturnValue:bool) {is=title K2Node_CallFunction})
   > n28 = GetOptionAngle!(execute, self, Option:int=@n26.Array Index) -> (then, Angle:double)
   > n29 = GetAngleDifference!(execute, self, AngleA:double=@n28, AngleB:double=Get CurrentPointerAngle() -> (CurrentPointerAngle:double)) -> (then, Difference:double)
   n30 = "To Float (Integer)"(InInt=Get NumOptions() -> (NumOptions=0))
   > Branch(Condition="float <= float"!(A:double=@n29, B:double=float / float(A=180, B=@n30) {is=title K2Node_CallFunction}) -> (ReturnValue:bool) {is=title K2Node_CallFunction})
   > n31 = Branch(Condition="Not Equal (Integer)"(A=@n26.Array Index, B=Get ActiveOption() -> (ActiveOption=0)))
   > ClearActiveOption() {is=title K2Node_CallFunction}
   > SetActiveOption(Option=@n26.Array Index)
   > n32 = PlaySound2D(Sound=Get SoundOnSelect() -> (SoundOnSelect:Sound Base Object Reference))
   SetVisibility(execute=[@n25.then_1, @n33.else], self=Get ImageRadialPointer(), InVisibility=HitTestInvisible)
   > SetPointerDirection()
   n33 = Branch(execute=@n18.else, Condition="Equal (Integer)"(A=0, B=-1) {is=title K2Node_CallFunction})
   > SetVisibility(self=Get ImageRadialPointer(), InVisibility=Hidden)
  ]}
 GetAngleDifference {nodes [
   n1 = GetAngleDifference!() -> (then, AngleA:double=~, AngleB:double=~) {is=title K2Node_FunctionEntry}
   n2 = "Absolute (Float)"(A=float - float(A="% (Float)"!(A:double=@n1, B:double=360) -> (ReturnValue:double), B="% (Float)"(A=@n1.AngleB, B=360)))
   > n3 = Branch(Condition="float > float"(A=@n2, B=180) {is=title K2Node_CallFunction})
   > Return Node!(execute, Difference:double=float - float(A=360, B=@n2))
   Return Node(execute=@n3.else, Difference=@n2)
  ]}
 GetOptionAngle {nodes [
   n1 = GetOptionAngle!() -> (then, Option:int) {is=title K2Node_FunctionEntry}
   n2 = "To Float (Integer)"(InInt=@n1)
   n3 = float / float(A=360, B="To Float (Integer)"(InInt=Get NumOptions() -> (NumOptions=0)))
   > Return Node!(execute, Angle:double=float - float(A=90, B=float * float(A=@n3, B=@n2)))
  ]}
 GetOptionPosition {nodes [
   n1 = GetOptionPosition!() -> (then, Option:int) {is=title K2Node_FunctionEntry}
   > n2 = GetOptionAngle(Option=@n1) {is=title K2Node_CallFunction}
   n3 = "Sin (Degrees)"!(A:double=@n2) -> (ReturnValue:double)
   n4 = "Cos (Degrees)"!(A:double=@n2) -> (ReturnValue:double)
   > Return Node!(execute, Position:Vector 2D Structure=MakeVector2D(X=float * float(A=@n4, B=230), Y=float * float(A=@n3, B=-230)))
  ]}
 InitializePicker {nodes [
   n1 = InitializePicker!() -> (then, PickerMode:EFortPickerMode Enum=BuildingCategory, InitalOption:int=0) {is=title K2Node_FunctionEntry}
   > NotifyInitializePicker(self:Fort Radial Picker Object Reference)
   > Set PickerMode(PickerMode:EFortPickerMode Enum=@n1) -> (Output_Get:EFortPickerMode Enum=BuildingCategory)
   > Set bShowingMouseCursor(bShowingMouseCursor:bool=false) -> (Output_Get:bool)
   n2 = GetOwningPlayer()
   > Set NumOptions(NumOptions:int=GetNumPickerOptions!(self:Fort Picker Context Object Reference=Get FortPickerContext(PlayerController=@n2)) -> (ReturnValue:int=0)) -> (Output_Get:int=0)
   > n3 = For Loop(
    FirstIndex=Length(TargetArray:Array of Radial Picker Item Object References=Get PickerOptions())
    LastIndex=int - int(A=Get NumOptions() -> (NumOptions=0), B=1) {is=title K2Node_CallFunction}
   )
   n4 = For Each Loop(Exec=@n3.Completed, Array=Get PickerOptions())
   ClearActiveOption(execute=@n4.Completed)
   > n5 = Branch(Condition="Not Equal (Integer)"(A=@n1.InitalOption, B=-1))
   > n6 = SetActiveOption(Option=@n1.InitalOption)
   SetVisibility(execute=[@n6.then, @n5.else], InVisibility=SelfHitTestInvisible)
   > Set bIsFocusable(bIsFocusable:bool=true) -> (Output_Get:bool)
   > SetInputMode()
   > Set PickerClosing(PickerClosing=false)
   > SetBoundAction!(
    execute
    self:Fort Keybind Widget Object Reference=Get KeybindPickerConfirm()
    NewBoundAction:Name=GetPickerConfirmActionName!(
     self:Fort Global UIContext Object Reference=Get FortGlobalUIContext!() -> (ReturnValue:Fort Global UIContext Object Reference) {is=title K2Node_GetBlueprintContextEx}
     PickerMode:EFortPickerMode Enum=Get PickerMode() -> (PickerMode:EFortPickerMode Enum=BuildingCategory)
    ) -> (ReturnValue:Name=None)
   ) -> (then)
   > n7 = GetCurrentPickerDisplayName!(execute, self:Fort Picker Context Object Reference=Get FortPickerContext(PlayerController=GetOwningPlayer())) -> (then, ReturnValue:Text)
   > "SetText (Text)"(self=Get CurrentWheelDisplayLabel() -> (CurrentWheelDisplayLabel:Common Text Object Reference), InText=@n7)
   > n8 = GetPickerWheelAdjacency!(execute, self:Fort Picker Context Object Reference=Get FortPickerContext(PlayerController=GetOwningPlayer())) -> (then, bCanPageLeft:bool, bCanPageRight:bool)
   n9 = NOT Boolean(A=TextIsEmpty(InText=@n7))
   > SetVisibility(
    self=Get WheelTitleBar() -> (WheelTitleBar:Common Border Object Reference)
    InVisibility=Select(
     Option 0:ESlateVisibility Enum=Collapsed
     Option 1:ESlateVisibility Enum=SelfHitTestInvisible
     Index=OR Boolean!(A=@n8, B=@n8.bCanPageRight, C:bool=@n9) -> (ReturnValue)
    ) -> (ReturnValue:ESlateVisibility Enum)
   )
   n10 = Branch(execute=@n4.LoopBody, Condition="integer < integer"(A=@n4.Array Index, B=Get NumOptions() -> (NumOptions=0)))
   > n11 = GetOptionPosition!(execute, self, Option:int=@n4.Array Index) -> (then, Position:Vector 2D Structure) {is=title K2Node_CallFunction}
   > SetPosition!(
    execute
    self:Canvas Panel Slot Object Reference=SlotAsCanvasSlot!(Widget:Widget Object Reference=@n4) -> (ReturnValue:Canvas Panel Slot Object Reference)
    InPosition:Vector 2D Structure=@n11
   ) -> (then)
   > SetVisibility(self=@n4, InVisibility=Select(Option 0=HitTestInvisible, Option 1=Visible, Index=IsTouchInputPlatform!() -> (ReturnValue:bool)))
   n12 = Get PickerMode() -> (PickerMode=BuildingCategory)
   > n13 = Branch(Condition=OR Boolean!(A="Equal (Enum)"(A:EFortPickerMode Enum=@n12, B:EFortPickerMode Enum=TrapRadial), B="Equal (Enum)"(A=@n12, B=WeaponsSlotted)) -> (ReturnValue))
   > SetShowImageAndLabel!(execute, self:Radial Picker Item Object Reference=@n4, ShowImageAndLabel:bool=false) -> (then)
   n14 = GetRadialPickerOptionItem!(self:Fort Picker Context Object Reference=Get FortPickerContext(PlayerController=GetOwningPlayer()), Index:int=@n4.Array Index) -> (Item:Fort Item Object Reference, bOptionEnabled:bool)
   > SetItemContent!(
    execute
    self:Radial Picker Item Object Reference=@n4
    Item:Fort Item Object Reference=@n14
    Option Enabled:bool=@n14.bOptionEnabled
    Radial Item Index:int=@n4.Array Index
   ) -> (then)
   SetShowImageAndLabel(execute=@n13.else, self=@n4, ShowImageAndLabel=true)
   n15 = GetRadialPickerOptionImageAndLabel!(self:Fort Picker Context Object Reference=Get FortPickerContext(PlayerController=GetOwningPlayer()), Index:int=@n4.Array Index) -> (Label:Text, Brush:Slate Brush Structure, Icon:Texture 2D Soft Object Reference, bOptionEnabled:bool, bIsLocked:bool)
   > SetImageAndLabelContent!(
    execute
    self:Radial Picker Item Object Reference=@n4
    Text:Text=@n15
    Icon:Texture 2D Soft Object Reference=@n15.Icon
    Brush:Slate Brush Structure=@n15.Brush
    Option Enabled:bool=@n15.bOptionEnabled
    RadialItemIndex:int=@n4.Array Index
    IsLocked:bool=@n15.bIsLocked
   ) -> (then)
   SetVisibility(execute=@n10.else, self=@n4, InVisibility=Collapsed)
   n16 = Create Widget!(
    execute:Exec=@n3.LoopBody
    Class:User Widget Class Reference=/Game/UI/InGame/HUD/RadialPickerItem.RadialPickerItem_C
    OwningPlayer:Player Controller Object Reference=GetOwningPlayer()
   ) -> (then:Exec, ReturnValue:Radial Picker Item Object Reference)
   > AddChild(self=Get CanvasPanel_0() -> (CanvasPanel_0:Canvas Panel Object Reference), Content=@n16)
   n17 = SlotAsCanvasSlot(Widget=@n16)
   > SetAnchors!(
    execute
    self:Canvas Panel Slot Object Reference=@n17
    InAnchors:Anchors Structure=Make Anchors!(Minimum:Vector 2D Structure=(X=0.5,Y=0.5), Maximum:Vector 2D Structure=(X=0.5,Y=0.5)) -> (Anchors:Anchors Structure) {is=title K2Node_MakeStruct}
   ) -> (then)
   > SetAlignment!(execute, self:Canvas Panel Slot Object Reference=@n17, InAlignment:Vector 2D Structure=(X=0.5,Y=0.5)) -> (then)
   > Add(TargetArray:Array of Radial Picker Item Object References=Get PickerOptions(), NewItem:"Radial Picker Item Object Reference (by ref)"=@n16) {is=title K2Node_CallArrayFunction}
  ]}
 InitializePickerOptions {nodes [
   InitializePickerOptions() {is=title K2Node_FunctionEntry}
   n1 = Get PickerOptions()
   > Add(TargetArray=@n1, NewItem=Get PickerOption0() -> (PickerOption0:Radial Picker Item Object Reference))
   > Add(TargetArray=@n1, NewItem=Get PickerOption1() -> (PickerOption1:Radial Picker Item Object Reference))
   > Add(TargetArray=@n1, NewItem=Get PickerOption2() -> (PickerOption2:Radial Picker Item Object Reference))
   > Add(TargetArray=@n1, NewItem=Get PickerOption3() -> (PickerOption3:Radial Picker Item Object Reference))
   > Add(TargetArray=@n1, NewItem=Get PickerOption4() -> (PickerOption4:Radial Picker Item Object Reference))
   > Add(TargetArray=@n1, NewItem=Get PickerOption5() -> (PickerOption5:Radial Picker Item Object Reference))
   > Add(TargetArray=@n1, NewItem=Get PickerOption6() -> (PickerOption6:Radial Picker Item Object Reference))
   > Add(TargetArray=@n1, NewItem=Get PickerOption7() -> (PickerOption7:Radial Picker Item Object Reference))
  ]}
 IsGamepadInPickerDeadZone {nodes [
   IsGamepadInPickerDeadZone() {is=title K2Node_FunctionEntry}
   n1 = GetInputAnalogStickState(self=GetOwningPlayer(), WhichStick=Get Thumbstick() -> (Thumbstick=CAS_LeftStick))
   n2 = MakeVector2D(X=@n1, Y=@n1.StickY)
   > Return Node!(execute, bIsInDeadZone:bool="float <= float"(A=Vector2D Length Squared!(A:Vector 2D Structure=@n2) -> (ReturnValue:double), B=0.25))
  ]}
 MoveActiveOption {nodes [
   n1 = MoveActiveOption!() -> (then, MoveOptionDirection:int) {is=title K2Node_FunctionEntry}
   n2 = GetCurrentInputType(self=Get CommonInputSubsystem())
   > Branch(Condition=OR Boolean(
     A="Not Equal (Enum)"(A:ECommonInputType Enum=@n2, B:ECommonInputType Enum=Gamepad)
     B=IsGamepadInPickerDeadZone() {is=title K2Node_CallFunction}
    ))
   > n3 = Set NewActiveOption(NewActiveOption:int=int + int(A=Get ActiveOption() -> (ActiveOption=0), B=@n1) -> (ReturnValue=0) {is=title K2Node_CommutativeAssociativeBinaryOperator}) -> (Output_Get:int=0)
   > n4 = Branch(Condition="integer >= integer"!(A:int=@n3, B:int=Get NumOptions() -> (NumOptions=0)) -> (ReturnValue:bool) {is=title K2Node_CallFunction})
   > n5 = Set NewActiveOption(NewActiveOption=0) -> (Output_Get=0)
   ClearActiveOption(execute=[@n5.then, @n17.then, @n16.else])
   n6 = Get NewActiveOption() -> (NewActiveOption:int=0)
   > SetActiveOption(Option=@n6)
   > PlaySound2D(Sound=Get SoundOnSelect())
   > n7 = GetOptionAngle(Option=@n6)
   > Set CurrentPointerAngle(CurrentPointerAngle=@n7)
   > SetPointerDirection()
   n8 = Get CommonInputSubsystem()
   > n9 = Branch(Condition="Equal (Enum)"(A:ECommonInputType Enum=GetCurrentInputType(self=@n8), B:ECommonInputType Enum=Gamepad))
   n10 = GetOptionPosition(execute=@n9.else, Option=Get NewActiveOption() -> (NewActiveOption=0))
   n11 = BreakVector2D(InVec=vector2d / float(A=GetViewportSize(), B=2))
   n12 = GetViewportScale!() -> (ReturnValue:float)
   n13 = BreakVector2D(InVec=@n10)
   n14 = float * float(A=@n13, B=@n12)
   n15 = float * float(A=@n13.Y, B=@n12)
   > Set Mouse Position!(
    execute
    self:Player Controller Object Reference=GetOwningPlayer()
    X:int=Truncate(A=float + float(A=@n14, B=@n11) {is=title K2Node_CommutativeAssociativeBinaryOperator})
    Y:int=Truncate(A=float + float(A=@n11.Y, B=@n15))
   ) -> (then) {is=title K2Node_CallFunction}
   n16 = Branch(execute=@n4.else, Condition="integer < integer"(A=@n3, B=0))
   > n17 = Set NewActiveOption(NewActiveOption=int - int(A=Get NumOptions() -> (NumOptions=0), B=1)) -> (Output_Get=0)
  ]}
 On Item Touched {nodes [
   n1 = On Item Touched!() -> (then, Touched Item:Radial Picker Item Object Reference) {is=title K2Node_FunctionEntry}
   n2 = Find Item!(TargetArray:Array of Radial Picker Item Object References=Get PickerOptions(), ItemToFind:"Radial Picker Item Object Reference (by ref)"=@n1) -> (ReturnValue)
   > Branch(Condition=Is Valid Index!(TargetArray:Array of Radial Picker Item Object References=Get PickerOptions(), IndexToTest:int=@n2) -> (ReturnValue:bool))
   > n3 = Set ActiveOption(ActiveOption=@n2) -> (Output_Get=0)
   > AcceptChosenOption!(execute, self, PickerOption:int=@n3) -> (then) {is=title K2Node_CallFunction}
  ]}
 OnMouseButtonDown {nodes [
   OnMouseButtonDown!() -> (then, MyGeometry:Geometry Structure, MouseEvent:"Pointer Event Structure (by ref)") {is=title K2Node_FunctionEntry}
   > AcceptOption() {is=title K2Node_CallFunction}
   > Return Node!(execute, ReturnValue:Event Reply Structure=Handled())
  ]}
 OnMouseWheel {nodes [
   n1 = OnMouseWheel!() -> (then, MyGeometry:Geometry Structure, MouseEvent:"Pointer Event Structure (by ref)") {is=title K2Node_FunctionEntry}
   n2 = Get CommonInputSubsystem() {is=title K2Node_GetSubsystem}
   > Branch(Condition="Equal (Enum)"(A=GetCurrentInputType(self=@n2), B=MouseAndKeyboard))
   n3 = "float > float"(A=Get Wheel Delta!(Input:"Pointer Event Structure (by ref)"=@n1.MouseEvent) -> (ReturnValue:float) {is=title K2Node_CallFunction}, B=0)
   > n4 = Branch(Condition=@n3)
   > n5 = Cast To FortPlayerControllerFrontEnd!(execute, Object=GetOwningPlayer()) -> (then, CastFailed, AsFort Player Controller Front End:Fort Player Controller Front End Object Reference) {is=title K2Node_DynamicCast}
   > n6 = EmotePickerNextWheel(self:Fort Player Controller Front End Object Reference=@n5)
   Return Node(execute=[@n6.then, @n8.then, @n5.CastFailed, @n7.CastFailed], ReturnValue=Handled())
   n7 = Cast To FortPlayerControllerFrontEnd(execute=@n4.else, Object=GetOwningPlayer())
   > n8 = EmotePickerPrevWheel(self:Fort Player Controller Front End Object Reference=@n7)
   AND Boolean(A=false, B=@n3)
  ]}
 OnPickerRefreshItems {nodes [
   OnPickerRefreshItems() {is=title K2Node_FunctionEntry}
   > Branch(Condition="Not Equal (Enum)"(A:ESlateVisibility Enum=GetVisibility() -> (ReturnValue=Visible), B:ESlateVisibility Enum=Collapsed))
   > InitializePicker!(
    execute
    self:Fort Radial Picker Object Reference
    PickerMode:EFortPickerMode Enum=Get PickerMode() -> (PickerMode=BuildingCategory)
    InitalOption:int=Get ActiveOption() -> (ActiveOption=0)
   ) -> (then) {is=title K2Node_CallFunction}
  ]}
 OnTouchEnded {nodes [
   OnTouchEnded!() -> (then, MyGeometry:Geometry Structure, InTouchEvent:"Pointer Event Structure (by ref)") {is=title K2Node_FunctionEntry}
   > Event CancelPicker() {is=title K2Node_CallFunction}
   > Return Node(ReturnValue=Handled())
  ]}
 Picker Closing {nodes [Picker Closing() {is=title K2Node_FunctionEntry}]}
 ResetInput {nodes [
   ResetInput() {is=title K2Node_FunctionEntry}
   > RestoreInputAxes(self:Fort Picker Context Object Reference=Get FortPickerContext(PlayerController=GetOwningPlayer()))
   > n1 = Branch(Condition=Get bIgnoringLookInput() -> (bIgnoringLookInput:bool))
   n2 = GetOwningPlayer()
   > SetIgnoreLookInput!(execute, self:Controller Object Reference=@n2, bNewLookInput:bool=false) -> (then)
   > n3 = Set bIgnoringLookInput(bIgnoringLookInput:bool=false) -> (Output_Get:bool)
   Branch(execute=[@n3.then, @n1.else], Condition=Get bIgnoringMoveInput() -> (bIgnoringMoveInput:bool))
   > SetIgnoreMoveInput!(execute, self:Controller Object Reference=@n2, bNewMoveInput:bool=false) -> (then)
   > Set bIgnoringMoveInput(bIgnoringMoveInput:bool=false) -> (Output_Get:bool)
  ]}
 SetActiveOption {nodes [
   n1 = SetActiveOption!() -> (then, Option:int) {is=title K2Node_FunctionEntry}
   n2 = "Get (a copy)"(Array=Get PickerOptions(), Dimension 1=@n1)
   > SetSelected(self=@n2, IsSelected=true)
   > Set ActiveOption(ActiveOption=@n1) -> (Output_Get=0)
   > "SetText (Text)"(self=Get Text_OK(), InText=Get PickerItemText!(self:Radial Picker Item Object Reference=@n2) -> (PickerItemText:Text))
   > SetRenderOpacity(self=Get KeybindPickerConfirm(), InOpacity=1)
  ]}
 SetInputMode {nodes [
   SetInputMode() {is=title K2Node_FunctionEntry}
   n1 = Get CommonInputSubsystem() {is=title K2Node_GetBlueprintContextEx}
   > n2 = Branch(Condition="Equal (Enum)"(A=GetCurrentInputType(self=@n1), B=Touch))
   Set WasUsingGamepad(execute=@n2.else, WasUsingGamepad:bool=Get bUsingGamepad()) -> (Output_Get:bool)
   n3 = Get CommonInputSubsystem()
   > Set bUsingGamepad(bUsingGamepad:bool="Equal (Enum)"(A=GetCurrentInputType(self=@n3), B=Gamepad)) -> (Output_Get:bool)
   > ResetInput() {is=title K2Node_CallFunction}
   > n4 = Branch(Condition=Get bUsingGamepad())
   > SetShowMouseCursor!(execute, self, InShowMouseCursor:bool=false) -> (then)
   n5 = GetOwningPlayer()
   > n6 = Branch(Condition=UseLeftThumbstick!(self:Fort Picker Context Object Reference=Get FortPickerContext(PlayerController=@n5)) -> (ReturnValue:bool))
   > Set Thumbstick(Thumbstick:EControllerAnalogStick Enum=CAS_LeftStick) -> (Output_Get:EControllerAnalogStick Enum=CAS_LeftStick)
   > Set YAxisMultiplier(YAxisMultiplier:double=1) -> (Output_Get:double)
   n7 = GetOwningPlayer()
   > SetIgnoreMoveInput(self=@n7, bNewMoveInput=true)
   > Set bIgnoringMoveInput(bIgnoringMoveInput=true)
   Set Thumbstick(execute=@n6.else, Thumbstick=CAS_RightStick) -> (Output_Get=CAS_LeftStick)
   > n8 = Set YAxisMultiplier(YAxisMultiplier=-1)
   SetIgnoreLookInput(execute=[@n8.then, @n9.then], self=@n7, bNewLookInput=true)
   > Set bIgnoringLookInput(bIgnoringLookInput=true)
   n9 = SetShowMouseCursor(execute=@n4.else, InShowMouseCursor=true)
  ]}
 SetPointerDirection {nodes [
   SetPointerDirection() {is=title K2Node_FunctionEntry}
   n1 = Get CurrentPointerAngle()
   n2 = float * float(A="Sin (Degrees)"(A=@n1), B=-130)
   n3 = float * float(A="Cos (Degrees)"(A=@n1), B=130)
   > SetRenderTransform!(
    execute
    self:Widget Object Reference=Get ImageRadialPointer()
    InTransform:Widget Transform Structure=Make Widget Transform!(
     Translation:Vector 2D Structure=MakeVector2D(X=@n3, Y=@n2)
     Scale:Vector 2D Structure=(X=1,Y=1)
     Shear:Vector 2D Structure=(X=0,Y=0)
     Angle:float=float * float(A=@n1, B=-1)
    ) -> (WidgetTransform:Widget Transform Structure) {is=title K2Node_MakeStruct}
   ) -> (then)
  ]}
 SetShowMouseCursor {nodes [
   n1 = SetShowMouseCursor!() -> (then, InShowMouseCursor:bool=~) {is=title K2Node_FunctionEntry}
   > Branch(Condition="Not Equal (Boolean)"(A=@n1, B=Get bShowingMouseCursor() -> (bShowingMouseCursor:bool)))
   > n2 = Set bShowingMouseCursor(bShowingMouseCursor=@n1)
   n3 = GetOwningPlayer()
   > n4 = Set bShowMouseCursor!(execute, bShowMouseCursor:bool=@n2, self:Player Controller Object Reference=@n3) -> (then, Output_Get:bool)
   Branch(Condition=@n4)
   n5 = BreakVector2D(InVec=vector2d / float(A=GetViewportSize(), B=2))
   > Set Mouse Position(self=@n3, X=Truncate(A=@n5), Y=Truncate(A=@n5.Y))
  ]}
 SetupMobileVisuals {nodes [
   SetupMobileVisuals() {is=title K2Node_FunctionEntry}
   n1 = Get CommonInputSubsystem()
   > Set Mobile(Mobile:bool="Equal (Enum)"(A=GetCurrentInputType(self=@n1), B=Touch)) -> (Output_Get:bool)
   > SetActiveWidget(
    self:Widget Switcher Object Reference=Get PrevPageSwitcher() -> (PrevPageSwitcher:Widget Switcher Object Reference)
    Widget=Select(
     Option 0:Widget Object Reference=Get PrevSizeBox() -> (PrevSizeBox:Size Box Object Reference)
     Option 1:Widget Object Reference=Get SizeBoxPreviousWheelButton() -> (SizeBoxPreviousWheelButton:Size Box Object Reference)
     Index=Get Mobile() -> (Mobile:bool)
    ) -> (ReturnValue:Widget Object Reference)
   )
   > SetActiveWidget(
    self=Get NextPageSwitcher() -> (NextPageSwitcher:Widget Switcher Object Reference)
    Widget=Select(
     Option 0=Get NextSizeBox() -> (NextSizeBox:Size Box Object Reference)
     Option 1=Get SizeBoxNextWheelButton() -> (SizeBoxNextWheelButton:Size Box Object Reference)
     Index=Get Mobile()
    )
   )
   > SetVisibility(
    self=Get ClickCatcher() -> (ClickCatcher:Canvas Panel Object Reference)
    InVisibility=Select(Option 0:ESlateVisibility Enum=SelfHitTestInvisible, Option 1:ESlateVisibility Enum=Visible, Index=Get Mobile()) -> (ReturnValue:ESlateVisibility Enum)
   )
   > n2 = Branch(Condition=Get Mobile())
   > n3 = SetVisibility(self=Get CenterBox() -> (CenterBox:Size Box Object Reference), InVisibility=Collapsed)
   n4 = Get WheelTitleBar()
   SetRenderTranslation!(
    execute=[@n3.then, @n2.else]
    self=@n4
    Translation_X:double=Break Widget Transform!(WidgetTransform:"Widget Transform Structure (by ref)"=Get RenderTransform!(self:Widget Object Reference=@n4) -> (RenderTransform:Widget Transform Structure)) -> (Translation_X:double, Translation_Y:double, Scale:Vector 2D Structure, Shear:Vector 2D Structure, Angle:float=~) {is=title K2Node_BreakStruct}
    Translation_Y:double=Select(Option 0:double=-60, Option 1:double=-100, Index=Get Mobile()) -> (ReturnValue:double)
   ) -> (then)
   > SetHeightOverride(self=[Get SizeBoxNextWheelButton(), Get SizeBoxPreviousWheelButton()], InHeightOverride=Select(Option 0=0, Option 1=90, Index=Get Mobile()))
  ]}
]
variables [
 ActiveOption {type=integer, default=-1}
 bIgnoringLookInput {type=boolean}
 bIgnoringMoveInput {type=boolean}
 bIsUsingTouch {type=boolean}
 bShowingMouseCursor {type=boolean}
 bUsingGamepad {type=boolean}
 CurrentPointerAngle {type=number}
 NumOptions {type=integer}
 Picker Closing {}
 PickerClosing {type=boolean}
 PickerMode {type=EFortPickerMode, default=BuildingCategory}
 PickerOptions {type="array<RadialPickerItem_C>", default []}
 SoundOnAccept {type=SoundBase, default=/Game/Sounds/Fort_UI_Sounds/FrontEnd/New_FrontEnd/Fortnite_UI_MainMenu_Generic_03_Cue}
 SoundOnCancel {type=SoundBase, default=/Game/Sounds/Fort_UI_Sounds/FrontEnd/fort_ui_dismiss_03_Cue}
 SoundOnSelect {type=SoundBase, default=/Game/Sounds/Fort_UI_Sounds/FrontEnd/DefaultSounds/PressedButtons/MSS_UI_Button_Hover_Default}
 Thumbstick {type=EControllerAnalogStick, default=CAS_LeftStick}
 YAxisMultiplier {type=number}
]
widgets {
 info {inheritedWidgetCount=0, parentClass=/Script/FortniteUI.FortRadialPicker, rootWidgetClass=&CanvasPanel, widgetCount=40}
 widgets [
  ClickCatcher {widgetClassPath=&CanvasPanel, bIsVariable=true, widget=ClickCatcher}
  Overlay_0 {
   widgetClassPath=&Overlay
   parent=ClickCatcher
   widget=Overlay_0
   properties {visibility=Visible, renderTransformPivot=<0 0.5>}
   slot {
    class=CanvasPanelSlot
    properties {bAutoSize=true, layoutData {alignment=<0.5 0.5>, anchors {maximum=<0.5 0.5>, minimum=<0.5 0.5>}, offsets=<ltrb 0 0 1909.9219 1073.0811>}}
   }
  }
  Image_1 {
   widgetClassPath=&Image
   parent=Overlay_0
   widget=Image_1
   properties {
    visibility=HitTestInvisible
    brush {
     bIsDynamicallyLoaded=false
     drawAs=Image
     imageType=NoImage
     mirroring=NoMirror
     resourceName=None
     resourceObject=/Game/UI/Items/Materials/M_UI_OutlineCircle_Inst
     tiling=NoTile
     imageSize=<244 244>
     margin=<ltrb 0 0 0 0>
     outlineSettings {
      bUseBrushTransparency=false
      roundingType=HalfHeightRadius
      width=0
      color {colorUseRule=UseColor_Specified, specifiedColor=<rgba 0 0 0 0>}
      cornerRadii=<0 0 0 0>
     }
     tintColor {colorUseRule=UseColor_Specified, specifiedColor=<rgba 0 0.0164 0.0833 1>}
     uVRegion {bIsValid=false, max=<0 0>, min=<0 0>}
    }
   }
   slot {class=OverlaySlot, properties {horizontalAlignment=HAlign_Center, verticalAlignment=VAlign_Center}}
  }
  Image_Radial {
   widgetClassPath=&Image
   parent=Overlay_0
   widget=Image_Radial
   properties {
    visibility=SelfHitTestInvisible
    brush {
     bIsDynamicallyLoaded=false
     drawAs=Image
     imageType=NoImage
     mirroring=NoMirror
     resourceName=None
     resourceObject=/Game/UI/Items/Materials/MI-PieMenu
     tiling=NoTile
     imageSize=<128 128>
     margin=<ltrb 0 0 0 0>
     outlineSettings {
      bUseBrushTransparency=false
      roundingType=HalfHeightRadius
      width=0
      color {colorUseRule=UseColor_Specified, specifiedColor=<rgba 0 0 0 0>}
      cornerRadii=<0 0 0 0>
     }
     tintColor {colorUseRule=UseColor_Specified, specifiedColor=<rgba 1 1 1 1>}
     uVRegion {bIsValid=false, max=<0 0>, min=<0 0>}
    }
    colorAndOpacity=<rgba 1 1 1 0.85>
   }
   slot {class=OverlaySlot, properties {horizontalAlignment=HAlign_Fill, verticalAlignment=VAlign_Fill}}
  }
  ImageRadialPointer {
   widgetClassPath=&Image
   bIsVariable=true
   parent=Overlay_0
   widget=ImageRadialPointer
   properties {
    visibility=Hidden
    brush {
     bIsDynamicallyLoaded=false
     drawAs=Image
     imageType=NoImage
     mirroring=NoMirror
     resourceName=None
     resourceObject=/Game/UI/CodeReferenced/Radial_Pointer
     tiling=NoTile
     imageSize=<32 32>
     margin=<ltrb 0 0 0 0>
     outlineSettings {
      bUseBrushTransparency=false
      roundingType=HalfHeightRadius
      width=0
      color {colorUseRule=UseColor_Specified, specifiedColor=<rgba 0 0 0 0>}
      cornerRadii=<0 0 0 0>
     }
     tintColor {colorUseRule=UseColor_Specified, specifiedColor=<rgba 1 1 1 1>}
     uVRegion {bIsValid=false, max=<0 0>, min=<0 0>}
    }
    renderTransform {angle=-90, scale=<1 1>, shear=<0 0>, translation=<0 0>}
   }
   slot {class=OverlaySlot, properties {horizontalAlignment=HAlign_Center, verticalAlignment=VAlign_Center}}
  }
  CommonBorder_0 {
   widgetClassPath=&CommonBorder
   parent=Overlay_0
   widget=CommonBorder_0
   properties {
    background {
     bIsDynamicallyLoaded=false
     drawAs=NoDrawType
     imageType=NoImage
     mirroring=NoMirror
     resourceName=None
     resourceObject=/FortUILibrary/Styles/Border/Materials/M-Border-MainL-Black
     tiling=NoTile
     imageSize=<128 32>
     margin=<ltrb 0.1 0.45 0.9 0.5>
     outlineSettings {
      bUseBrushTransparency=false
      roundingType=HalfHeightRadius
      width=0
      color {colorUseRule=UseColor_Specified, specifiedColor=<rgba 0 0 0 0>}
      cornerRadii=<0 0 0 0>
     }
     tintColor {colorUseRule=UseColor_Specified, specifiedColor=<rgba 1 1 1 1>}
     uVRegion {bIsValid=false, max=<0 0>, min=<0 0>}
    }
    padding=<ltrb 0 0 0 0>
   }
   slot {class=OverlaySlot, properties {horizontalAlignment=HAlign_Fill, verticalAlignment=VAlign_Fill}}
  }
  SizeBox_0 {
   widgetClassPath=&SizeBox
   parent=CommonBorder_0
   widget=SizeBox_0
   properties {bOverride_HeightOverride=true, bOverride_WidthOverride=true, heightOverride=700, widthOverride=700}
   slot {class=BorderSlot, properties {padding=<ltrb 0 0 0 0>}}
  }
  CanvasPanel_0 {widgetClassPath=&CanvasPanel, bIsVariable=true, parent=SizeBox_0, widget=CanvasPanel_0, slot {class=SizeBoxSlot}}
  PickerOption0 {
   widgetClassPath=/Game/UI/InGame/HUD/RadialPickerItem
   bIsVariable=true
   parent=CanvasPanel_0
   widget=PickerOption0
   slot {
    class=CanvasPanelSlot
    properties {bAutoSize=true, layoutData {alignment=<0.5 0.5>, anchors {maximum=<0.5 0.5>, minimum=<0.5 0.5>}, offsets=<ltrb 0 -100 90 134.6667>}}
   }
  }
  PickerOption1 {
   widgetClassPath=&RadialPickerItem
   bIsVariable=true
   parent=CanvasPanel_0
   widget=PickerOption1
   slot {
    class=CanvasPanelSlot
    properties {bAutoSize=true, layoutData {alignment=<0.5 0.5>, anchors {maximum=<0.5 0.5>, minimum=<0.5 0.5>}, offsets=<ltrb 100 -100 100 100>}}
   }
  }
  PickerOption2 {
   widgetClassPath=&RadialPickerItem
   bIsVariable=true
   parent=CanvasPanel_0
   widget=PickerOption2
   slot {
    class=CanvasPanelSlot
    properties {bAutoSize=true, layoutData {alignment=<0.5 0.5>, anchors {maximum=<0.5 0.5>, minimum=<0.5 0.5>}, offsets=<ltrb 100 0 100 30>}}
   }
  }
  PickerOption3 {
   widgetClassPath=&RadialPickerItem
   bIsVariable=true
   parent=CanvasPanel_0
   widget=PickerOption3
   slot {
    class=CanvasPanelSlot
    properties {bAutoSize=true, layoutData {alignment=<0.5 0.5>, anchors {maximum=<0.5 0.5>, minimum=<0.5 0.5>}, offsets=<ltrb 100 100 100 30>}}
   }
  }
  PickerOption4 {
   widgetClassPath=&RadialPickerItem
   bIsVariable=true
   parent=CanvasPanel_0
   widget=PickerOption4
   slot {
    class=CanvasPanelSlot
    properties {bAutoSize=true, layoutData {alignment=<0.5 0.5>, anchors {maximum=<0.5 0.5>, minimum=<0.5 0.5>}, offsets=<ltrb 0 100 100 30>}}
   }
  }
  PickerOption5 {
   widgetClassPath=&RadialPickerItem
   bIsVariable=true
   parent=CanvasPanel_0
   widget=PickerOption5
   slot {
    class=CanvasPanelSlot
    properties {bAutoSize=true, layoutData {alignment=<0.5 0.5>, anchors {maximum=<0.5 0.5>, minimum=<0.5 0.5>}, offsets=<ltrb -100 100 100 30>}}
   }
  }
  PickerOption6 {
   widgetClassPath=&RadialPickerItem
   bIsVariable=true
   parent=CanvasPanel_0
   widget=PickerOption6
   slot {
    class=CanvasPanelSlot
    properties {bAutoSize=true, layoutData {alignment=<0.5 0.5>, anchors {maximum=<0.5 0.5>, minimum=<0.5 0.5>}, offsets=<ltrb -100 0 100 30>}}
   }
  }
  PickerOption7 {
   widgetClassPath=&RadialPickerItem
   bIsVariable=true
   parent=CanvasPanel_0
   widget=PickerOption7
   slot {
    class=CanvasPanelSlot
    properties {bAutoSize=true, layoutData {alignment=<0.5 0.5>, anchors {maximum=<0.5 0.5>, minimum=<0.5 0.5>}, offsets=<ltrb -100 -100 100 100>}}
   }
  }
  CenterBox {
   widgetClassPath=&SizeBox
   bIsVariable=true
   parent=Overlay_0
   widget=CenterBox
   properties {
    bOverride_HeightOverride=true
    bOverride_MinDesiredWidth=true
    heightOverride=190
    minDesiredHeight=230
    minDesiredWidth=230
    widthOverride=230
   }
   slot {class=OverlaySlot, properties {horizontalAlignment=HAlign_Center, verticalAlignment=VAlign_Center}}
  }
  VerticalBox_0 {widgetClassPath=&VerticalBox, parent=CenterBox, widget=VerticalBox_0, slot {class=SizeBoxSlot}}
  SizeBox_1 {
   widgetClassPath=&SizeBox
   parent=VerticalBox_0
   widget=SizeBox_1
   properties {bOverride_HeightOverride=true, heightOverride=40, widthOverride=30}
   slot {class=VerticalBoxSlot}
  }
  KeybindPickerConfirm {
   widgetClassPath=/FortUILibrary/Input/KeybindWidget
   bIsVariable=true
   parent=SizeBox_1
   widget=KeybindPickerConfirm
   slot {class=SizeBoxSlot}
  }
  Text_OK {
   widgetClassPath=&CommonTextBlock
   bIsVariable=true
   parent=VerticalBox_0
   widget=Text_OK
   properties {
    autoWrapText=true
    justification=Center
    style=/FortUILibrary/Styles/Typography/Foundation/TextStyle-DecoBase-M.TextStyle-DecoBase-M_C
    text=OK
    font {
     bForceMonospaced=false
     bMaterialIsStencil=false
     fontMaterial=null
     fontObject=&BurbankSmall
     letterSpacing=0
     monospacedWidth=1
     size=24
     skewAmount=0
     typefaceFontName=Bold
     outlineSettings {
      bApplyOutlineToDropShadows=false
      bMiteredCorners=false
      bSeparateFillAlpha=false
      outlineMaterial=null
      outlineSize=0
      outlineColor=<rgba 0 0 0 1>
     }
    }
    shadowOffset=<0 0>
   }
   slot {class=VerticalBoxSlot, properties {verticalAlignment=VAlign_Center}}
  }
  Spacer_0 {widgetClassPath=&Spacer, parent=VerticalBox_0, widget=Spacer_0, properties {size=<1 5>}, slot {class=VerticalBoxSlot}}
  Image_0 {
   widgetClassPath=&Image
   bIsVariable=true
   parent=VerticalBox_0
   widget=Image_0
   properties {brush {
     bIsDynamicallyLoaded=false
     drawAs=Image
     imageType=NoImage
     mirroring=NoMirror
     resourceName=None
     resourceObject=/Game/UI/Foundation/Textures/Rule/Rule-Dotted-HUD
     tiling=Horizontal
     imageSize=<8 10>
     margin=<ltrb 0 0 0 0>
     outlineSettings {
      bUseBrushTransparency=false
      roundingType=HalfHeightRadius
      width=0
      color {colorUseRule=UseColor_Specified, specifiedColor=<rgba 0 0 0 0>}
      cornerRadii=<0 0 0 0>
     }
     tintColor {colorUseRule=UseColor_Specified, specifiedColor=<rgba 1 1 1 1>}
     uVRegion {bIsValid=false, max=<0 0>, min=<0 0>}
    }}
   slot {class=VerticalBoxSlot, properties {verticalAlignment=VAlign_Center, padding=<ltrb 10 0 10 0>}}
  }
  Spacer_1 {
   widgetClassPath=&Spacer
   parent=VerticalBox_0
   widget=Spacer_1
   properties {size=<1 5>}
   slot {class=VerticalBoxSlot, properties {size {sizeRule=Fill, value=1}}}
  }
  KeybindPickerCancel {
   widgetClassPath=&CommonActionWidget
   parent=VerticalBox_0
   widget=KeybindPickerCancel
   properties {inputActions [{dataTable=&InputActionDataTable, rowName=FrontEndPickerCancel}]}
   slot {class=VerticalBoxSlot, properties {verticalAlignment=VAlign_Center}}
  }
  CommonTextBlock_1 {
   widgetClassPath=&CommonTextBlock
   parent=VerticalBox_0
   widget=CommonTextBlock_1
   properties {
    justification=Center
    style=&TextStyle-DecoBase-M_C
    text=Cancel
    font {
     bForceMonospaced=false
     bMaterialIsStencil=false
     fontMaterial=null
     fontObject=&BurbankSmall
     letterSpacing=0
     monospacedWidth=1
     size=24
     skewAmount=0
     typefaceFontName=Bold
     outlineSettings {
      bApplyOutlineToDropShadows=false
      bMiteredCorners=false
      bSeparateFillAlpha=false
      outlineMaterial=null
      outlineSize=0
      outlineColor=<rgba 0 0 0 1>
     }
    }
    shadowOffset=<0 0>
   }
   slot {class=VerticalBoxSlot, properties {verticalAlignment=VAlign_Center}}
  }
  WheelTitleBar {
   widgetClassPath=&CommonBorder
   bIsVariable=true
   parent=Overlay_0
   widget=WheelTitleBar
   properties {
    horizontalAlignment=HAlign_Center
    verticalAlignment=VAlign_Bottom
    background {
     bIsDynamicallyLoaded=false
     drawAs=Image
     imageType=NoImage
     mirroring=NoMirror
     resourceName=None
     resourceObject=null
     tiling=NoTile
     imageSize=<32 32>
     margin=<ltrb 0 0 0 0>
     outlineSettings {
      bUseBrushTransparency=false
      roundingType=HalfHeightRadius
      width=0
      color {colorUseRule=UseColor_Specified, specifiedColor=<rgba 0 0 0 0>}
      cornerRadii=<0 0 0 0>
     }
     tintColor {colorUseRule=UseColor_Specified, specifiedColor=<rgba 0.0018 0.0027 0.0116 0.819>}
     uVRegion {bIsValid=false, max=<0 0>, min=<0 0>}
    }
    padding=<ltrb 8 8 8 8>
    renderTransform {angle=0, scale=<1 1>, shear=<0 0>, translation=<0 -60>}
    renderTransformPivot=<0.5 1>
   }
   slot {class=OverlaySlot, properties {horizontalAlignment=HAlign_Fill}}
  }
  HorizontalBox_1 {
   widgetClassPath=&HorizontalBox
   parent=WheelTitleBar
   widget=HorizontalBox_1
   slot {class=BorderSlot, properties {horizontalAlignment=HAlign_Center, verticalAlignment=VAlign_Bottom, padding=<ltrb 8 8 8 8>}}
  }
  PrevPageSwitcher {
   widgetClassPath=&WidgetSwitcher
   bIsVariable=true
   parent=HorizontalBox_1
   widget=PrevPageSwitcher
   slot {class=HorizontalBoxSlot, properties {padding=<ltrb 0 0 32 0>}}
  }
  PrevSizeBox {
   widgetClassPath=&SizeBox
   bIsVariable=true
   parent=PrevPageSwitcher
   widget=PrevSizeBox
   properties {bOverride_HeightOverride=true, heightOverride=40, widthOverride=30}
   slot {class=WidgetSwitcherSlot}
  }
  PrevPageKeyBind {
   widgetClassPath=&KeybindWidget
   bIsVariable=true
   parent=PrevSizeBox
   widget=PrevPageKeyBind
   properties {boundAction=FrontEndPickerPreviousWheel}
   slot {class=SizeBoxSlot}
  }
  SizeBoxPreviousWheelButton {
   widgetClassPath=&SizeBox
   bIsVariable=true
   parent=PrevPageSwitcher
   widget=SizeBoxPreviousWheelButton
   properties {bOverride_HeightOverride=true, heightOverride=52.5}
   slot {class=WidgetSwitcherSlot}
  }
  PreviousWheelButton {
   widgetClassPath=&IconTextButton
   bIsVariable=true
   parent=SizeBoxPreviousWheelButton
   widget=PreviousWheelButton
   properties {
    buttonText=Previous
    clickMethod=PreciseClick
    contentAlignment=HAlign_Center
    onButtonBaseClicked=(&self:WidgetTree.PreviousWheelButton.HandleOnButtonClicked)
    onButtonBaseDoubleClicked=(&self:WidgetTree.PreviousWheelButton.HandleOnButtonDoubleClicked)
    onButtonBaseHovered=(&self:WidgetTree.PreviousWheelButton.HandleOnButtonHovered)
    onButtonBaseUnhovered=(&self:WidgetTree.PreviousWheelButton.HandleOnButtonUnhovered)
    onSelectedChangedBase=(&self:WidgetTree.PreviousWheelButton.HandleOnSelectedChanged)
   }
   slot {class=SizeBoxSlot}
  }
  SizeBox_5 {
   widgetClassPath=&SizeBox
   parent=HorizontalBox_1
   widget=SizeBox_5
   properties {bOverride_MinDesiredWidth=true, minDesiredWidth=250}
   slot {class=HorizontalBoxSlot, properties {horizontalAlignment=HAlign_Center, size {sizeRule=Fill, value=1}}}
  }
  CurrentWheelDisplayLabel {
   widgetClassPath=&CommonTextBlock
   bIsVariable=true
   parent=SizeBox_5
   widget=CurrentWheelDisplayLabel
   properties {
    autoWrapText=true
    justification=Center
    style=&TextStyle-DecoBase-M_C
    text=Wheel Name
    font {
     bForceMonospaced=false
     bMaterialIsStencil=false
     fontMaterial=null
     fontObject=&BurbankSmall
     letterSpacing=0
     monospacedWidth=1
     size=24
     skewAmount=0
     typefaceFontName=Bold
     outlineSettings {
      bApplyOutlineToDropShadows=false
      bMiteredCorners=false
      bSeparateFillAlpha=false
      outlineMaterial=null
      outlineSize=0
      outlineColor=<rgba 0 0 0 1>
     }
    }
    shadowOffset=<0 0>
   }
   slot {class=SizeBoxSlot, properties {verticalAlignment=VAlign_Center, padding=<ltrb 0 0 0 -3>}}
  }
  NextPageSwitcher {
   widgetClassPath=&WidgetSwitcher
   bIsVariable=true
   parent=HorizontalBox_1
   widget=NextPageSwitcher
   slot {class=HorizontalBoxSlot, properties {verticalAlignment=VAlign_Center, padding=<ltrb 32 0 0 0>}}
  }
  NextSizeBox {
   widgetClassPath=&SizeBox
   bIsVariable=true
   parent=NextPageSwitcher
   widget=NextSizeBox
   properties {bOverride_HeightOverride=true, heightOverride=40, widthOverride=30}
   slot {class=WidgetSwitcherSlot}
  }
  NextPageKeyBind {
   widgetClassPath=&KeybindWidget
   bIsVariable=true
   parent=NextSizeBox
   widget=NextPageKeyBind
   properties {boundAction=FrontEndPickerNextWheel}
   slot {class=SizeBoxSlot}
  }
  SizeBoxNextWheelButton {
   widgetClassPath=&SizeBox
   bIsVariable=true
   parent=NextPageSwitcher
   widget=SizeBoxNextWheelButton
   slot {class=WidgetSwitcherSlot}
  }
  NextWheelButton {
   widgetClassPath=&IconTextButton
   bIsVariable=true
   parent=SizeBoxNextWheelButton
   widget=NextWheelButton
   properties {
    buttonText=Next
    clickMethod=PreciseClick
    contentAlignment=HAlign_Center
    onButtonBaseClicked=(&self:WidgetTree.NextWheelButton.HandleOnButtonClicked)
    onButtonBaseDoubleClicked=(&self:WidgetTree.NextWheelButton.HandleOnButtonDoubleClicked)
    onButtonBaseHovered=(&self:WidgetTree.NextWheelButton.HandleOnButtonHovered)
    onButtonBaseUnhovered=(&self:WidgetTree.NextWheelButton.HandleOnButtonUnhovered)
    onSelectedChangedBase=(&self:WidgetTree.NextWheelButton.HandleOnSelectedChanged)
   }
   slot {class=SizeBoxSlot}
  }
 ]
}
