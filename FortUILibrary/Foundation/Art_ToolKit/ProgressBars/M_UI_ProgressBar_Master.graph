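# Translucent UI master material for progress bars, built on MF_UI_Simple_ProgressBar, which is called once in color mode and once in non-color mode.
# The non-color result passes through a static switch, "Use Mask for skew", that multiplies it by a sharp-edged rectangular UV mask with adjustable inset.
# With that switch off, an optional screen-space treatment instead fades it toward the left and right screen edges using a parabolic, screen-aligned falloff.
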Material /FortUILibrary/Foundation/Art_ToolKit/ProgressBars/M_UI_ProgressBar_Master
graph {
 properties {
  bUsedWithStaticMesh=false
  blendMode=BLEND_Translucent
  materialDomain=MD_UI
  pixelDepthOffsetMode=PDOM_Legacy
  usedShadingModels=Default Lit
 }
 nodes [
  MaterialFunctionCall!(IsColor=Constant() {r=1}) {materialFunction=/FortUILibrary/Foundation/Art_ToolKit/ProgressBars/MaterialFunctions/MF_UI_Simple_ProgressBar}
  n1 = MaterialFunctionCall(IsColor=Constant()) {materialFunction=&MF_UI_Simple_ProgressBar}
  n2 = TextureCoordinate()
  n3 = ScalarParameter() {defaultValue=0.1, group=Mask, parameterName=Mask Value}
  n4 = Add(A=@n2, B=Multiply(A=@n3) {constB=-1})
  n5 = ScalarParameter() {defaultValue=64, group=Mask, parameterName=Mask Sharpness}
  n6 = Saturate(None=Multiply(A=ComponentMask(None=@n4) {r=true}, B=@n5))
  n7 = ComponentMask(None=Add(A=@n2, B=@n3)) {r=true}
  n8 = Saturate(None=Multiply(A=OneMinus(None=@n7), B=@n5))
  n9 = Add(A=@n2, B=Multiply(A=@n3) {constB=-1})
  n10 = Saturate(None=Multiply(A=ComponentMask(None=@n9) {g=true}, B=@n5))
  n11 = ComponentMask(None=Add(A=@n2, B=@n3)) {g=true}
  n12 = Saturate(None=Multiply(A=OneMinus(None=@n11), B=@n5))
  n13 = Saturate(None=DotProduct(A=DotProduct(A=@n6, B=@n8), B=DotProduct(A=@n10, B=@n12)))
  n14 = ComponentMask(None=MaterialFunctionCall(Texture Dimensions=@n2) {materialFunction=&ScreenAlignedUVs}."X 100%, Y 100%") {r=true}
  n15 = Multiply(A=DotProduct(A=OneMinus(None=@n14), B=@n14), B=ScalarParameter() {defaultValue=16, group=ScreenSpace, parameterName=FallOffPoint})
  n16 = Saturate(None=Power(Base=Saturate(None=@n15), Exp=ScalarParameter() {defaultValue=1, group=ScreenSpace, parameterName=FallOffRate}))
  StaticSwitchParameter(
   True=Saturate(None=Multiply(A=@n1.Result, B=@n13))
   False=StaticSwitchParameter(True=DotProduct(A=@n1.Result, B=@n16), False=@n1.Result) {group=ScreenSpace, parameterName=HasScreenSpaceTreatment}
  ) {group=Mask, parameterName=Use Mask for skew}
 ]
}
