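# AthenaTemperature is a HUD widget, derived from AthenaTemperatureBase, that shows a right-aligned temperature text label inside a fixed-width SizeBox.
# On Construct it casts the owning player to FortPlayerController, binds to On Show Temperature Change, and applies the current bShowTemperature value immediately.
# ShowTemperatureChanged toggles the widget between SelfHitTestInvisible and Collapsed from that flag.
# The text block is inherited, and this graph never sets its content.
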WidgetBlueprint /FortniteCoreUI/Athena/HUD/AthenaTemperature
graphs [
 EventGraph {nodes [
   Event Construct()
   > n1 = Cast To FortPlayerController(Object=GetOwningPlayer())
   > Bind Event to On Show Temperature Change(self:Fort Player Controller Object Reference=@n1, Delegate=Create Event()) {is=title K2Node_AddDelegate}
   > ShowTemperatureChanged!(execute, self, bShowTemperature:bool=Get bShowTemperature!(self:Fort Player Controller Object Reference=@n1) -> (bShowTemperature:bool)) -> (then)
  ]}
 ShowTemperatureChanged {nodes [
   n1 = ShowTemperatureChanged!() -> (then, bShowTemperature:bool=~) {is=title K2Node_FunctionEntry}
   > SetVisibility(InVisibility=Select(Option 0=Collapsed, Option 1=SelfHitTestInvisible, Index=@n1))
  ]}
]
variables []
widgets {
 info {inheritedWidgetCount=1, parentClass=/Script/FortniteUI.AthenaTemperatureBase, rootWidgetClass=&SizeBox, widgetCount=2}
 widgets [
  SizeBox_0 {widgetClassPath=&SizeBox, widget=SizeBox_0, properties {bOverride_WidthOverride=true, widthOverride=160}}
  Temperature {
   widgetClassPath=&TextBlock
   bInherited=true
   bIsVariable=true
   parent=SizeBox_0
   widget=Temperature
   properties {
    justification=Right
    text=Good
    font {
     bForceMonospaced=false
     bMaterialIsStencil=false
     fontMaterial=null
     fontObject=&NotoSans
     letterSpacing=0
     monospacedWidth=1
     size=15
     skewAmount=0
     typefaceFontName=NotoSans-Bold
     outlineSettings {
      bApplyOutlineToDropShadows=false
      bMiteredCorners=false
      bSeparateFillAlpha=false
      outlineMaterial=null
      outlineSize=1
      outlineColor=<rgba 0.0052 0.0052 0.0056 0.15>
     }
    }
   }
   slot {class=SizeBoxSlot}
  }
 ]
}
