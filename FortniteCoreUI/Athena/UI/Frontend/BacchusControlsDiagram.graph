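# Widget Blueprint on FortBacchusControlsDiagram that shows annotated touch-control diagrams for General, Building and Combat controls.
# A visibility switcher holds the three screens, each a tutorial texture with text callouts, with a row of pips in an overlay.
# UpdateNumberOfScreens collapses the third pip in No Build mode unless all modes are shown.
# ShowScreen forwards mode and index to SetControlSchemeDisplay, while Update_Pip resets all pips and lights the current one.
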WidgetBlueprint /FortniteCoreUI/Athena/UI/Frontend/BacchusControlsDiagram
graphs [
 EventGraph {nodes [
   n1 = Event UpdateNumberOfScreens!() -> (OutputDelegate, then, bIsNoBuildMode:bool) {is=title K2Node_Event}
   n2 = Get bShowAllModes() -> (bShowAllModes:bool)
   > n3 = Branch(Condition=AND Boolean(A=@n1.bIsNoBuildMode, B=NOT Boolean(A=@n2)))
   > n4 = SetVisibility(self=Get Pip_2() -> (Pip_2:Image Object Reference), InVisibility=Collapsed)
   Update_Pip!(execute=[@n4.then, @n3.else], self, ICurrentScreen:int=0) -> (then)
   n5 = Event UpdatePip!() -> (OutputDelegate, then, ScreenIndex:int=0) {is=title K2Node_Event}
   > Update_Pip(ICurrentScreen=@n5.ScreenIndex)
  ]}
 ShowScreen {nodes [
   n1 = ShowScreen!() -> (then, NoBuildMode:bool=~, iSwitcherIndex:int) {is=title K2Node_FunctionEntry}
   > SetControlSchemeDisplay!(execute, self:Fort Bacchus Controls Diagram Object Reference, bIsNoBuildMode:bool=@n1, ScreenIndex:int=@n1.iSwitcherIndex) -> (then)
  ]}
 Update_Pip {nodes [
   n1 = Update_Pip!() -> (then, ICurrentScreen:int) {is=title K2Node_FunctionEntry}
   > SetBrushFromMaterial(
    self=[Get Pip_1() -> (Pip_1:Image Object Reference), Get Pip_2(), Get Pip_0() -> (Pip_0:Image Object Reference)]
    Material=/FortniteCoreUI/Frontend/MI_Pip_Entry_Off
   )
   > SetBrushFromMaterial(
    self=Select!(
     Option 0:Image Object Reference=Get Pip_0()
     Option 1:Image Object Reference=Get Pip_1()
     Option 2:Image Object Reference=Get Pip_2()
     Index:int=@n1
    ) -> (ReturnValue:Image Object Reference)
    Material=/FortniteCoreUI/Frontend/MI_Pip_Entry_On
   )
  ]}
]
variables []
widgets {
 info {inheritedWidgetCount=4, parentClass=/Script/FortniteUI.FortBacchusControlsDiagram, rootWidgetClass=&Overlay, widgetCount=62}
 widgets [
  Overlay_3 {widgetClassPath=&Overlay, widget=Overlay_3}
  Switcher_Controls {
   widgetClassPath=&CommonVisibilitySwitcher
   bInherited=true
   bIsVariable=true
   parent=Overlay_3
   widget=Switcher_Controls
   slot {class=OverlaySlot, properties {horizontalAlignment=HAlign_Fill, verticalAlignment=VAlign_Fill, padding=<ltrb 0 30 0 30>}}
  }
  ScaleBox_GeneralMode {
   widgetClassPath=&ScaleBox
   bInherited=true
   bIsVariable=true
   parent=Switcher_Controls
   widget=ScaleBox_GeneralMode
   properties {stretch=None}
   slot {class=CommonVisibilitySwitcherSlot}
  }
  SBGeneralMode {
   widgetClassPath=&SizeBox
   bIsVariable=true
   parent=ScaleBox_GeneralMode
   widget=SBGeneralMode
   properties {bOverride_HeightOverride=true, bOverride_WidthOverride=true, heightOverride=850, widthOverride=1680}
   slot {class=ScaleBoxSlot}
  }
  CP_GeneralMode {widgetClassPath=&CanvasPanel, parent=SBGeneralMode, widget=CP_GeneralMode, slot {class=SizeBoxSlot}}
  IMG_DropShadow {
   widgetClassPath=&Image
   bIsVariable=true
   parent=CP_GeneralMode
   widget=IMG_DropShadow
   properties {
    visibility=SelfHitTestInvisible
    brush {
     bIsDynamicallyLoaded=false
     drawAs=Image
     imageType=FullColor
     mirroring=NoMirror
     resourceName=None
     resourceObject=/FortUILibrary/MaterialInstances/MI_UI_LinearGradientTop
     tiling=NoTile
     imageSize=<32 32>
     margin=<ltrb 0 0 0 0>
     outlineSettings {
      bUseBrushTransparency=false
      roundingType=HalfHeightRadius
      width=0
      color {colorUseRule=UseColor_Specified, specifiedColor=<rgba 0 0 0 0>}
      cornerRadii=<0 0 0 0>
     }
     tintColor {colorUseRule=UseColor_Specified, specifiedColor=<rgba 0 0 0 0.5>}
     uVRegion {bIsValid=false, max=<0 0>, min=<0 0>}
    }
   }
   slot {
    class=CanvasPanelSlot
    properties {layoutData {alignment=<0.5 0>, anchors {maximum=<0.5 0.5>, minimum=<0.5 0.5>}, offsets=<ltrb 0 264.1305 907 15>}}
   }
  }
  Image {
   widgetClassPath=&Image
   parent=CP_GeneralMode
   widget=Image
   properties {
    visibility=HitTestInvisible
    brush {
     bIsDynamicallyLoaded=false
     drawAs=Image
     imageType=FullColor
     mirroring=NoMirror
     resourceName=None
     resourceObject=/Game/UI/Foundation/Textures/LoadingScreens/BacchusTutorial/T_Bacchus_Tuto_GeneralControls
     tiling=NoTile
     imageSize=<1024 512>
     margin=<ltrb 0 0 0 0>
     outlineSettings {
      bUseBrushTransparency=false
      roundingType=HalfHeightRadius
      width=0
      color {colorUseRule=UseColor_Specified, specifiedColor=<rgba 0 0 0 0>}
      cornerRadii=<0 0 0 0>
     }
     tintColor {colorUseRule=UseColor_Specified, specifiedColor=<rgba 1 1 1 1>}
     uVRegion {bIsValid=false, max=<0 0>, min=<0 0>}
    }
   }
   slot {
    class=CanvasPanelSlot
    properties {bAutoSize=true, layoutData {alignment=<0.5 0.5>, anchors {maximum=<0.5 0.5>, minimum=<0.5 0.5>}, offsets=<ltrb 0 13 100 30>}}
   }
  }
  CommonTextBlock_1 {
   widgetClassPath=&CommonTextBlock
   parent=CP_GeneralMode
   widget=CommonTextBlock_1
   properties {
    justification=Center
    lineHeightPercentage=0.9
    text=General Controls
    textTransformPolicy=ToUpper
    font {
     bForceMonospaced=false
     bMaterialIsStencil=false
     fontMaterial=null
     fontObject=&HeadingNow
     letterSpacing=0
     monospacedWidth=1
     size=33.75
     skewAmount=0
     typefaceFontName=76 Bold Italic
     outlineSettings {
      bApplyOutlineToDropShadows=false
      bMiteredCorners=false
      bSeparateFillAlpha=false
      outlineMaterial=null
      outlineSize=0
      outlineColor=<rgba 0 0 0 1>
     }
    }
    shadowOffset=<0 0>
   }
   slot {
    class=CanvasPanelSlot
    properties {bAutoSize=true, layoutData {alignment=<0.5 0.5>, anchors {maximum=<0.4956 0.1222>, minimum=<0.4956 0.1222>}, offsets=<ltrb 7 -16 100 30>}}
   }
  }
  CommonTextBlock_5 {
   widgetClassPath=&CommonTextBlock
   parent=CP_GeneralMode
   widget=CommonTextBlock_5
   properties {
    justification=Center
    lineHeightPercentage=0.9
    text=Health and shield
    wrapTextAt=250
    font {
     bForceMonospaced=false
     bMaterialIsStencil=false
     fontMaterial=null
     fontObject=&HeadingNow
     letterSpacing=0
     monospacedWidth=1
     size=18
     skewAmount=0
     typefaceFontName=75 Medium
     outlineSettings {
      bApplyOutlineToDropShadows=false
      bMiteredCorners=false
      bSeparateFillAlpha=false
      outlineMaterial=null
      outlineSize=0
      outlineColor=<rgba 0 0 0 1>
     }
    }
    shadowOffset=<0 0>
   }
   slot {
    class=CanvasPanelSlot
    properties {bAutoSize=true, layoutData {alignment=<0.5 0>, anchors {maximum=<0.5 0.5>, minimum=<0.5 0.5>}, offsets=<ltrb -360 329.9999 750 30>}}
   }
  }
  CommonTextBlock_23 {
   widgetClassPath=&CommonTextBlock
   parent=CP_GeneralMode
   widget=CommonTextBlock_23
   properties {
    justification=Right
    lineHeightPercentage=0.9
    text=Minimap
    wrapTextAt=250
    font {
     bForceMonospaced=false
     bMaterialIsStencil=false
     fontMaterial=null
     fontObject=&HeadingNow
     letterSpacing=0
     monospacedWidth=1
     size=18
     skewAmount=0
     typefaceFontName=75 Medium
     outlineSettings {
      bApplyOutlineToDropShadows=false
      bMiteredCorners=false
      bSeparateFillAlpha=false
      outlineMaterial=null
      outlineSize=0
      outlineColor=<rgba 0 0 0 1>
     }
    }
    shadowOffset=<0 0>
   }
   slot {
    class=CanvasPanelSlot
    properties {bAutoSize=true, layoutData {alignment=<0.5 0>, anchors {maximum=<0.5 0.5>, minimum=<0.5 0.5>}, offsets=<ltrb -600 -174 750 30>}}
   }
  }
  CommonTextBlock_24 {
   widgetClassPath=&CommonTextBlock
   parent=CP_GeneralMode
   widget=CommonTextBlock_24
   properties {
    justification=Right
    lineHeightPercentage=0.9
    text=Drag here to autorun
    wrapTextAt=200
    font {
     bForceMonospaced=false
     bMaterialIsStencil=false
     fontMaterial=null
     fontObject=&HeadingNow
     letterSpacing=0
     monospacedWidth=1
     size=18
     skewAmount=0
     typefaceFontName=75 Medium
     outlineSettings {
      bApplyOutlineToDropShadows=false
      bMiteredCorners=false
      bSeparateFillAlpha=false
      outlineMaterial=null
      outlineSize=0
      outlineColor=<rgba 0 0 0 1>
     }
    }
    shadowOffset=<0 0>
   }
   slot {
    class=CanvasPanelSlot
    properties {layoutData {alignment=<0.5 0>, anchors {maximum=<0.5 0.5>, minimum=<0.5 0.5>}, offsets=<ltrb -628 8 174.4329 51.6216>}}
   }
  }
  CommonTextBlock_27 {
   widgetClassPath=&CommonTextBlock
   parent=CP_GeneralMode
   widget=CommonTextBlock_27
   properties {
    justification=Right
    lineHeightPercentage=0.9
    text="Drag to move (left side of screen)"
    wrapTextAt=200
    font {
     bForceMonospaced=false
     bMaterialIsStencil=false
     fontMaterial=null
     fontObject=&HeadingNow
     letterSpacing=0
     monospacedWidth=1
     size=18
     skewAmount=0
     typefaceFontName=75 Medium
     outlineSettings {
      bApplyOutlineToDropShadows=false
      bMiteredCorners=false
      bSeparateFillAlpha=false
      outlineMaterial=null
      outlineSize=0
      outlineColor=<rgba 0 0 0 1>
     }
    }
    shadowOffset=<0 0>
   }
   slot {
    class=CanvasPanelSlot
    properties {bAutoSize=true, layoutData {alignment=<1 0.5>, anchors {maximum=<0.5 0.5>, minimum=<0.5 0.5>}, offsets=<ltrb -543 143 190.6434 74.4444>}}
   }
  }
  CommonTextBlock_9 {
   widgetClassPath=&CommonTextBlock
   parent=CP_GeneralMode
   widget=CommonTextBlock_9
   properties {
    justification=Center
    lineHeightPercentage=0.9
    text=Toggle crouch
    wrapTextAt=300
    font {
     bForceMonospaced=false
     bMaterialIsStencil=false
     fontMaterial=null
     fontObject=&HeadingNow
     letterSpacing=0
     monospacedWidth=1
     size=18
     skewAmount=0
     typefaceFontName=75 Medium
     outlineSettings {
      bApplyOutlineToDropShadows=false
      bMiteredCorners=false
      bSeparateFillAlpha=false
      outlineMaterial=null
      outlineSize=0
      outlineColor=<rgba 0 0 0 1>
     }
    }
    shadowOffset=<0 0>
   }
   slot {
    class=CanvasPanelSlot
    properties {bAutoSize=true, layoutData {alignment=<0.5 0>, anchors {maximum=<0.5 0.5>, minimum=<0.5 0.5>}, offsets=<ltrb 276 330 600 30>}}
   }
  }
  CommonTextBlock_11 {
   widgetClassPath=&CommonTextBlock
   parent=CP_GeneralMode
   widget=CommonTextBlock_11
   properties {
    justification=Center
    lineHeightPercentage=0.9
    text=Inventory
    wrapTextAt=250
    font {
     bForceMonospaced=false
     bMaterialIsStencil=false
     fontMaterial=null
     fontObject=&HeadingNow
     letterSpacing=0
     monospacedWidth=1
     size=18
     skewAmount=0
     typefaceFontName=75 Medium
     outlineSettings {
      bApplyOutlineToDropShadows=false
      bMiteredCorners=false
      bSeparateFillAlpha=false
      outlineMaterial=null
      outlineSize=0
      outlineColor=<rgba 0 0 0 1>
     }
    }
    shadowOffset=<0 0>
   }
   slot {
    class=CanvasPanelSlot
    properties {bAutoSize=true, layoutData {alignment=<0 1>, anchors {maximum=<0.5 0.5>, minimum=<0.5 0.5>}, offsets=<ltrb -151 356 262.916 90>}}
   }
  }
  SizeBox_1 {
   widgetClassPath=&SizeBox
   parent=CP_GeneralMode
   widget=SizeBox_1
   properties {bOverride_MaxDesiredHeight=true, maxDesiredHeight=100}
   slot {
    class=CanvasPanelSlot
    properties {layoutData {alignment=<0 0>, anchors {maximum=<0.5 0.5>, minimum=<0.5 0.5>}, offsets=<ltrb 545 -95 200 100>}}
   }
  }
  CommonTextBlock_21 {
   widgetClassPath=&CommonTextBlock
   parent=SizeBox_1
   widget=CommonTextBlock_21
   properties {
    lineHeightPercentage=0.9
    scrollOrientation=Orient_Vertical
    scrollStyle=&TextScrollStyle-Base_C
    text="Drag to aim\r\n(right half of screen)"
    wrapTextAt=208
    font {
     bForceMonospaced=false
     bMaterialIsStencil=false
     fontMaterial=null
     fontObject=&HeadingNow
     letterSpacing=0
     monospacedWidth=1
     size=18
     skewAmount=0
     typefaceFontName=75 Medium
     outlineSettings {
      bApplyOutlineToDropShadows=false
      bMiteredCorners=false
      bSeparateFillAlpha=false
      outlineMaterial=null
      outlineSize=0
      outlineColor=<rgba 0 0 0 1>
     }
    }
    shadowOffset=<0 0>
   }
   slot {class=SizeBoxSlot}
  }
  SizeBox_0 {
   widgetClassPath=&SizeBox
   parent=CP_GeneralMode
   widget=SizeBox_0
   properties {bOverride_MaxDesiredHeight=true, maxDesiredHeight=70, minDesiredHeight=200}
   slot {
    class=CanvasPanelSlot
    properties {layoutData {alignment=<0 0.5>, anchors {maximum=<0.5 0.5>, minimum=<0.5 0.5>}, offsets=<ltrb 545 -149 200 70>}}
   }
  }
  CommonTextBlock_28 {
   widgetClassPath=&CommonTextBlock
   parent=SizeBox_0
   widget=CommonTextBlock_28
   properties {
    lineHeightPercentage=0.9
    scrollOrientation=Orient_Vertical
    scrollStyle=&TextScrollStyle-Base_C
    text="Place marker (tap or drag)"
    wrapTextAt=205
    font {
     bForceMonospaced=false
     bMaterialIsStencil=false
     fontMaterial=null
     fontObject=&HeadingNow
     letterSpacing=0
     monospacedWidth=1
     size=18
     skewAmount=0
     typefaceFontName=75 Medium
     outlineSettings {
      bApplyOutlineToDropShadows=false
      bMiteredCorners=false
      bSeparateFillAlpha=false
      outlineMaterial=null
      outlineSize=0
      outlineColor=<rgba 0 0 0 1>
     }
    }
    shadowOffset=<0 0>
   }
   slot {class=SizeBoxSlot}
  }
  CommonTextBlock_12 {
   widgetClassPath=&CommonTextBlock
   parent=CP_GeneralMode
   widget=CommonTextBlock_12
   properties {
    lineHeightPercentage=0.9
    text=Emote
    wrapTextAt=260
    font {
     bForceMonospaced=false
     bMaterialIsStencil=false
     fontMaterial=null
     fontObject=&HeadingNow
     letterSpacing=0
     monospacedWidth=1
     size=18
     skewAmount=0
     typefaceFontName=75 Medium
     outlineSettings {
      bApplyOutlineToDropShadows=false
      bMiteredCorners=false
      bSeparateFillAlpha=false
      outlineMaterial=null
      outlineSize=0
      outlineColor=<rgba 0 0 0 1>
     }
    }
    shadowOffset=<0 0>
   }
   slot {
    class=CanvasPanelSlot
    properties {bAutoSize=true, layoutData {alignment=<0 0.5>, anchors {maximum=<0.5 0.5>, minimum=<0.5 0.5>}, offsets=<ltrb 545 -216 449.9602 30>}}
   }
  }
  CommonTextBlock_14 {
   widgetClassPath=&CommonTextBlock
   parent=CP_GeneralMode
   widget=CommonTextBlock_14
   properties {
    lineHeightPercentage=0.9
    text=Jump
    wrapTextAt=200
    font {
     bForceMonospaced=false
     bMaterialIsStencil=false
     fontMaterial=null
     fontObject=&HeadingNow
     letterSpacing=0
     monospacedWidth=1
     size=18
     skewAmount=0
     typefaceFontName=75 Medium
     outlineSettings {
      bApplyOutlineToDropShadows=false
      bMiteredCorners=false
      bSeparateFillAlpha=false
      outlineMaterial=null
      outlineSize=0
      outlineColor=<rgba 0 0 0 1>
     }
    }
    shadowOffset=<0 0>
   }
   slot {
    class=CanvasPanelSlot
    properties {bAutoSize=true, layoutData {alignment=<0 0.5>, anchors {maximum=<0.5 0.5>, minimum=<0.5 0.5>}, offsets=<ltrb 545 47 200 60>}}
   }
  }
  CommonTextBlock_16 {
   widgetClassPath=&CommonTextBlock
   parent=CP_GeneralMode
   widget=CommonTextBlock_16
   properties {
    lineHeightPercentage=0.9
    text=Sprint
    wrapTextAt=200
    font {
     bForceMonospaced=false
     bMaterialIsStencil=false
     fontMaterial=null
     fontObject=&HeadingNow
     letterSpacing=0
     monospacedWidth=1
     size=18
     skewAmount=0
     typefaceFontName=75 Medium
     outlineSettings {
      bApplyOutlineToDropShadows=false
      bMiteredCorners=false
      bSeparateFillAlpha=false
      outlineMaterial=null
      outlineSize=0
      outlineColor=<rgba 0 0 0 1>
     }
    }
    shadowOffset=<0 0>
   }
   slot {
    class=CanvasPanelSlot
    properties {bAutoSize=true, layoutData {alignment=<0 0.5>, anchors {maximum=<0.5 0.5>, minimum=<0.5 0.5>}, offsets=<ltrb 545 94 260 60>}}
   }
  }
  CommonTextBlock_17 {
   widgetClassPath=&CommonTextBlock
   parent=CP_GeneralMode
   widget=CommonTextBlock_17
   properties {
    lineHeightPercentage=0.9
    text=Use pickaxe
    wrapTextAt=200
    font {
     bForceMonospaced=false
     bMaterialIsStencil=false
     fontMaterial=null
     fontObject=&HeadingNow
     letterSpacing=0
     monospacedWidth=1
     size=18
     skewAmount=0
     typefaceFontName=75 Medium
     outlineSettings {
      bApplyOutlineToDropShadows=false
      bMiteredCorners=false
      bSeparateFillAlpha=false
      outlineMaterial=null
      outlineSize=0
      outlineColor=<rgba 0 0 0 1>
     }
    }
    shadowOffset=<0 0>
   }
   slot {
    class=CanvasPanelSlot
    properties {bAutoSize=true, layoutData {alignment=<0 0.5>, anchors {maximum=<0.5 0.5>, minimum=<0.5 0.5>}, offsets=<ltrb 545 208 200 29.4482>}}
   }
  }
  B_Line {
   widgetClassPath=&Border
   parent=CP_GeneralMode
   widget=B_Line
   properties {visibility=SelfHitTestInvisible, padding=<ltrb 0 0 0 0>}
   slot {
    class=CanvasPanelSlot
    properties {layoutData {alignment=<0 0>, anchors {maximum=<0.5 0.5>, minimum=<0.5 0.5>}, offsets=<ltrb -362 249.0837 3 63.2156>}}
   }
  }
  B_Line_1 {
   widgetClassPath=&Border
   parent=CP_GeneralMode
   widget=B_Line_1
   properties {visibility=SelfHitTestInvisible, padding=<ltrb 0 0 0 0>}
   slot {
    class=CanvasPanelSlot
    properties {layoutData {alignment=<0 0>, anchors {maximum=<0.5 0.5>, minimum=<0.5 0.5>}, offsets=<ltrb -88.8769 264.7087 3 47.5906>}}
   }
  }
  B_Line_2 {
   widgetClassPath=&Border
   parent=CP_GeneralMode
   widget=B_Line_2
   properties {visibility=SelfHitTestInvisible, padding=<ltrb 0 0 0 0>}
   slot {
    class=CanvasPanelSlot
    properties {layoutData {alignment=<0 0>, anchors {maximum=<0.5 0.5>, minimum=<0.5 0.5>}, offsets=<ltrb 274 263.3337 3 48.9656>}}
   }
  }
  ScaleBox_BuildingMode {
   widgetClassPath=&ScaleBox
   bInherited=true
   bIsVariable=true
   parent=Switcher_Controls
   widget=ScaleBox_BuildingMode
   properties {stretch=None}
   slot {class=CommonVisibilitySwitcherSlot}
  }
  SBBuilding {
   widgetClassPath=&SizeBox
   bIsVariable=true
   parent=ScaleBox_BuildingMode
   widget=SBBuilding
   properties {bOverride_HeightOverride=true, bOverride_WidthOverride=true, heightOverride=850, widthOverride=1680}
   slot {class=ScaleBoxSlot}
  }
  CanvasPanel_0 {widgetClassPath=&CanvasPanel, parent=SBBuilding, widget=CanvasPanel_0, slot {class=SizeBoxSlot}}
  IMG_DropShadow_1 {
   widgetClassPath=&Image
   bIsVariable=true
   parent=CanvasPanel_0
   widget=IMG_DropShadow_1
   properties {
    visibility=SelfHitTestInvisible
    brush {
     bIsDynamicallyLoaded=false
     drawAs=Image
     imageType=FullColor
     mirroring=NoMirror
     resourceName=None
     resourceObject=&MI_UI_LinearGradientTop
     tiling=NoTile
     imageSize=<32 32>
     margin=<ltrb 0 0 0 0>
     outlineSettings {
      bUseBrushTransparency=false
      roundingType=HalfHeightRadius
      width=0
      color {colorUseRule=UseColor_Specified, specifiedColor=<rgba 0 0 0 0>}
      cornerRadii=<0 0 0 0>
     }
     tintColor {colorUseRule=UseColor_Specified, specifiedColor=<rgba 0 0 0 0.5>}
     uVRegion {bIsValid=false, max=<0 0>, min=<0 0>}
    }
   }
   slot {
    class=CanvasPanelSlot
    properties {layoutData {alignment=<0.5 0>, anchors {maximum=<0.5 0.5>, minimum=<0.5 0.5>}, offsets=<ltrb 0 264.1305 907 15>}}
   }
  }
  Image_0 {
   widgetClassPath=&Image
   parent=CanvasPanel_0
   widget=Image_0
   properties {
    visibility=SelfHitTestInvisible
    brush {
     bIsDynamicallyLoaded=false
     drawAs=Image
     imageType=NoImage
     mirroring=NoMirror
     resourceName=None
     resourceObject=/Game/UI/Foundation/Textures/LoadingScreens/BacchusTutorial/T_Bacchus_Tuto_Builderpro_Build
     tiling=NoTile
     imageSize=<1024 512>
     margin=<ltrb 0 0 0 0>
     outlineSettings {
      bUseBrushTransparency=false
      roundingType=HalfHeightRadius
      width=0
      color {colorUseRule=UseColor_Specified, specifiedColor=<rgba 0 0 0 0>}
      cornerRadii=<0 0 0 0>
     }
     tintColor {colorUseRule=UseColor_Specified, specifiedColor=<rgba 1 1 1 1>}
     uVRegion {bIsValid=false, max=<0 0>, min=<0 0>}
    }
   }
   slot {
    class=CanvasPanelSlot
    properties {bAutoSize=true, layoutData {alignment=<0.5 0.5>, anchors {maximum=<0.5 0.5>, minimum=<0.5 0.5>}, offsets=<ltrb 0 14 100 30>}}
   }
  }
  CommonTextBlock_13 {
   widgetClassPath=&CommonTextBlock
   parent=CanvasPanel_0
   widget=CommonTextBlock_13
   properties {
    justification=Center
    lineHeightPercentage=0.9
    text=Building Controls
    textTransformPolicy=ToUpper
    font {
     bForceMonospaced=false
     bMaterialIsStencil=false
     fontMaterial=null
     fontObject=&HeadingNow
     letterSpacing=0
     monospacedWidth=1
     size=33.75
     skewAmount=0
     typefaceFontName=76 Bold Italic
     outlineSettings {
      bApplyOutlineToDropShadows=false
      bMiteredCorners=false
      bSeparateFillAlpha=false
      outlineMaterial=null
      outlineSize=0
      outlineColor=<rgba 0 0 0 1>
     }
    }
    shadowOffset=<0 0>
   }
   slot {
    class=CanvasPanelSlot
    properties {bAutoSize=true, layoutData {alignment=<0.5 0.5>, anchors {maximum=<0.4956 0.1222>, minimum=<0.4956 0.1222>}, offsets=<ltrb 0 -16 100 30>}}
   }
  }
  CommonTextBlock_10 {
   widgetClassPath=&CommonTextBlock
   parent=CanvasPanel_0
   widget=CommonTextBlock_10
   properties {
    justification=Center
    lineHeightPercentage=0.9
    text=Object select
    wrapTextAt=220
    font {
     bForceMonospaced=false
     bMaterialIsStencil=false
     fontMaterial=null
     fontObject=&HeadingNow
     letterSpacing=0
     monospacedWidth=1
     size=18
     skewAmount=0
     typefaceFontName=75 Medium
     outlineSettings {
      bApplyOutlineToDropShadows=false
      bMiteredCorners=false
      bSeparateFillAlpha=false
      outlineMaterial=null
      outlineSize=0
      outlineColor=<rgba 0 0 0 1>
     }
    }
    shadowOffset=<0 0>
   }
   slot {
    class=CanvasPanelSlot
    properties {
     bAutoSize=true
     layoutData {alignment=<0.5 0>, anchors {maximum=<0.5 0.5>, minimum=<0.5 0.5>}, offsets=<ltrb -117.4396 330 349.5888 30.4481>}
    }
   }
  }
  CommonTextBlock_6 {
   widgetClassPath=&CommonTextBlock
   parent=CanvasPanel_0
   widget=CommonTextBlock_6
   properties {
    justification=Center
    lineHeightPercentage=0.9
    text=Switch to Combat Mode
    wrapTextAt=330
    font {
     bForceMonospaced=false
     bMaterialIsStencil=false
     fontMaterial=null
     fontObject=&HeadingNow
     letterSpacing=0
     monospacedWidth=1
     size=18
     skewAmount=0
     typefaceFontName=75 Medium
     outlineSettings {
      bApplyOutlineToDropShadows=false
      bMiteredCorners=false
      bSeparateFillAlpha=false
      outlineMaterial=null
      outlineSize=0
      outlineColor=<rgba 0 0 0 1>
     }
    }
    shadowOffset=<0 0>
   }
   slot {
    class=CanvasPanelSlot
    properties {bAutoSize=true, layoutData {alignment=<0.5 0>, anchors {maximum=<0.5 0.5>, minimum=<0.5 0.5>}, offsets=<ltrb 208 330 331.1846 61.5112>}}
   }
  }
  CommonTextBlock_7 {
   widgetClassPath=&CommonTextBlock
   parent=CanvasPanel_0
   widget=CommonTextBlock_7
   properties {
    lineHeightPercentage=0.9
    text=Change materials
    wrapTextAt=270
    font {
     bForceMonospaced=false
     bMaterialIsStencil=false
     fontMaterial=null
     fontObject=&HeadingNow
     letterSpacing=0
     monospacedWidth=1
     size=18
     skewAmount=0
     typefaceFontName=75 Medium
     outlineSettings {
      bApplyOutlineToDropShadows=false
      bMiteredCorners=false
      bSeparateFillAlpha=false
      outlineMaterial=null
      outlineSize=0
      outlineColor=<rgba 0 0 0 1>
     }
    }
    shadowOffset=<0 0>
   }
   slot {
    class=CanvasPanelSlot
    properties {bAutoSize=true, layoutData {alignment=<0 0.5>, anchors {maximum=<0.5 0.5>, minimum=<0.5 0.5>}, offsets=<ltrb 545 -159 100 0>}}
   }
  }
  CommonTextBlock_8 {
   widgetClassPath=&CommonTextBlock
   parent=CanvasPanel_0
   widget=CommonTextBlock_8
   properties {
    lineHeightPercentage=0.9
    text=Edit object
    wrapTextAt=260
    font {
     bForceMonospaced=false
     bMaterialIsStencil=false
     fontMaterial=null
     fontObject=&HeadingNow
     letterSpacing=0
     monospacedWidth=1
     size=18
     skewAmount=0
     typefaceFontName=75 Medium
     outlineSettings {
      bApplyOutlineToDropShadows=false
      bMiteredCorners=false
      bSeparateFillAlpha=false
      outlineMaterial=null
      outlineSize=0
      outlineColor=<rgba 0 0 0 1>
     }
    }
    shadowOffset=<0 0>
   }
   slot {
    class=CanvasPanelSlot
    properties {bAutoSize=true, layoutData {alignment=<0 0.4>, anchors {maximum=<0.5 0.5>, minimum=<0.5 0.5>}, offsets=<ltrb 545 -108 194.3755 64.2177>}}
   }
  }
  CommonTextBlock_25 {
   widgetClassPath=&CommonTextBlock
   parent=CanvasPanel_0
   widget=CommonTextBlock_25
   properties {
    lineHeightPercentage=0.9
    text=Rotate object
    wrapTextAt=260
    font {
     bForceMonospaced=false
     bMaterialIsStencil=false
     fontMaterial=null
     fontObject=&HeadingNow
     letterSpacing=0
     monospacedWidth=1
     size=18
     skewAmount=0
     typefaceFontName=75 Medium
     outlineSettings {
      bApplyOutlineToDropShadows=false
      bMiteredCorners=false
      bSeparateFillAlpha=false
      outlineMaterial=null
      outlineSize=0
      outlineColor=<rgba 0 0 0 1>
     }
    }
    shadowOffset=<0 0>
   }
   slot {
    class=CanvasPanelSlot
    properties {bAutoSize=true, layoutData {alignment=<0 0.4>, anchors {maximum=<0.5 0.5>, minimum=<0.5 0.5>}, offsets=<ltrb 545 93 194.3755 60>}}
   }
  }
  CommonTextBlock_26 {
   widgetClassPath=&CommonTextBlock
   parent=CanvasPanel_0
   widget=CommonTextBlock_26
   properties {
    lineHeightPercentage=0.9
    text=Place object
    wrapTextAt=260
    font {
     bForceMonospaced=false
     bMaterialIsStencil=false
     fontMaterial=null
     fontObject=&HeadingNow
     letterSpacing=0
     monospacedWidth=1
     size=18
     skewAmount=0
     typefaceFontName=75 Medium
     outlineSettings {
      bApplyOutlineToDropShadows=false
      bMiteredCorners=false
      bSeparateFillAlpha=false
      outlineMaterial=null
      outlineSize=0
      outlineColor=<rgba 0 0 0 1>
     }
    }
    shadowOffset=<0 0>
   }
   slot {
    class=CanvasPanelSlot
    properties {bAutoSize=true, layoutData {alignment=<0 0.4>, anchors {maximum=<0.5 0.5>, minimum=<0.5 0.5>}, offsets=<ltrb 545 146 194.3755 64.2177>}}
   }
  }
  B_Line_3 {
   widgetClassPath=&Border
   parent=CanvasPanel_0
   widget=B_Line_3
   properties {visibility=SelfHitTestInvisible, padding=<ltrb 0 0 0 0>}
   slot {
    class=CanvasPanelSlot
    properties {layoutData {alignment=<0 0>, anchors {maximum=<0.5 0.5>, minimum=<0.5 0.5>}, offsets=<ltrb -120 249.0837 3 63.2156>}}
   }
  }
  B_Line_4 {
   widgetClassPath=&Border
   parent=CanvasPanel_0
   widget=B_Line_4
   properties {visibility=SelfHitTestInvisible, padding=<ltrb 0 0 0 0>}
   slot {
    class=CanvasPanelSlot
    properties {layoutData {alignment=<0 0>, anchors {maximum=<0.5 0.5>, minimum=<0.5 0.5>}, offsets=<ltrb 205 261.3754 3 50.9239>}}
   }
  }
  ScaleBox_CombatMode {
   widgetClassPath=&ScaleBox
   bInherited=true
   bIsVariable=true
   parent=Switcher_Controls
   widget=ScaleBox_CombatMode
   properties {stretch=None}
   slot {class=CommonVisibilitySwitcherSlot}
  }
  SBCombatFireButton {
   widgetClassPath=&SizeBox
   bIsVariable=true
   parent=ScaleBox_CombatMode
   widget=SBCombatFireButton
   properties {bOverride_HeightOverride=true, bOverride_WidthOverride=true, heightOverride=850, widthOverride=1680}
   slot {class=ScaleBoxSlot}
  }
  CP_CombatFireButton {widgetClassPath=&CanvasPanel, parent=SBCombatFireButton, widget=CP_CombatFireButton, slot {class=SizeBoxSlot}}
  IMG_DropShadow_2 {
   widgetClassPath=&Image
   bIsVariable=true
   parent=CP_CombatFireButton
   widget=IMG_DropShadow_2
   properties {
    visibility=SelfHitTestInvisible
    brush {
     bIsDynamicallyLoaded=false
     drawAs=Image
     imageType=FullColor
     mirroring=NoMirror
     resourceName=None
     resourceObject=&MI_UI_LinearGradientTop
     tiling=NoTile
     imageSize=<32 32>
     margin=<ltrb 0 0 0 0>
     outlineSettings {
      bUseBrushTransparency=false
      roundingType=HalfHeightRadius
      width=0
      color {colorUseRule=UseColor_Specified, specifiedColor=<rgba 0 0 0 0>}
      cornerRadii=<0 0 0 0>
     }
     tintColor {colorUseRule=UseColor_Specified, specifiedColor=<rgba 0 0 0 0.5>}
     uVRegion {bIsValid=false, max=<0 0>, min=<0 0>}
    }
   }
   slot {
    class=CanvasPanelSlot
    properties {layoutData {alignment=<0.5 0>, anchors {maximum=<0.5 0.5>, minimum=<0.5 0.5>}, offsets=<ltrb 0 264.1305 907 15>}}
   }
  }
  Image_2 {
   widgetClassPath=&Image
   parent=CP_CombatFireButton
   widget=Image_2
   properties {
    visibility=HitTestInvisible
    brush {
     bIsDynamicallyLoaded=false
     drawAs=Image
     imageType=NoImage
     mirroring=NoMirror
     resourceName=None
     resourceObject=/Game/UI/Foundation/Textures/LoadingScreens/BacchusTutorial/T_Bacchus_Tuto_Builderpro_Combat_DB
     tiling=NoTile
     imageSize=<1024 512>
     margin=<ltrb 0 0 0 0>
     outlineSettings {
      bUseBrushTransparency=false
      roundingType=HalfHeightRadius
      width=0
      color {colorUseRule=UseColor_Specified, specifiedColor=<rgba 0 0 0 0>}
      cornerRadii=<0 0 0 0>
     }
     tintColor {colorUseRule=UseColor_Specified, specifiedColor=<rgba 1 1 1 1>}
     uVRegion {bIsValid=false, max=<0 0>, min=<0 0>}
    }
    renderTransform {angle=0, scale=<1 1>, shear=<0 0>, translation=<1 0>}
   }
   slot {
    class=CanvasPanelSlot
    properties {bAutoSize=true, layoutData {alignment=<0.5 0.5>, anchors {maximum=<0.5 0.5>, minimum=<0.5 0.5>}, offsets=<ltrb 0 14 100 30>}}
   }
  }
  CommonTextBlock_15 {
   widgetClassPath=&CommonTextBlock
   parent=CP_CombatFireButton
   widget=CommonTextBlock_15
   properties {
    justification=Center
    lineHeightPercentage=0.9
    text=Combat Controls
    textTransformPolicy=ToUpper
    font {
     bForceMonospaced=false
     bMaterialIsStencil=false
     fontMaterial=null
     fontObject=&HeadingNow
     letterSpacing=0
     monospacedWidth=1
     size=33.75
     skewAmount=0
     typefaceFontName=76 Bold Italic
     outlineSettings {
      bApplyOutlineToDropShadows=false
      bMiteredCorners=false
      bSeparateFillAlpha=false
      outlineMaterial=null
      outlineSize=0
      outlineColor=<rgba 0 0 0 1>
     }
    }
    shadowOffset=<0 0>
   }
   slot {
    class=CanvasPanelSlot
    properties {bAutoSize=true, layoutData {alignment=<0.5 0.5>, anchors {maximum=<0.4956 0.1222>, minimum=<0.4956 0.1222>}, offsets=<ltrb 7 -16 100 30>}}
   }
  }
  CommonTextBlock_20 {
   widgetClassPath=&CommonTextBlock
   parent=CP_CombatFireButton
   widget=CommonTextBlock_20
   properties {
    justification=Right
    lineHeightPercentage=0.9
    text=Item select
    wrapTextAt=250
    font {
     bForceMonospaced=false
     bMaterialIsStencil=false
     fontMaterial=null
     fontObject=&HeadingNow
     letterSpacing=0
     monospacedWidth=1
     size=18
     skewAmount=0
     typefaceFontName=75 Medium
     outlineSettings {
      bApplyOutlineToDropShadows=false
      bMiteredCorners=false
      bSeparateFillAlpha=false
      outlineMaterial=null
      outlineSize=0
      outlineColor=<rgba 0 0 0 1>
     }
    }
    shadowOffset=<0 0>
   }
   slot {
    class=CanvasPanelSlot
    properties {bAutoSize=true, layoutData {alignment=<0.5 0>, anchors {maximum=<0.5 0.5>, minimum=<0.5 0.5>}, offsets=<ltrb -26 330 750 30>}}
   }
  }
  CommonTextBlock {
   widgetClassPath=&CommonTextBlock
   parent=CP_CombatFireButton
   widget=CommonTextBlock
   properties {
    justification=Center
    lineHeightPercentage=0.9
    text=Switch to Build Mode
    wrapTextAt=300
    font {
     bForceMonospaced=false
     bMaterialIsStencil=false
     fontMaterial=null
     fontObject=&HeadingNow
     letterSpacing=0
     monospacedWidth=1
     size=18
     skewAmount=0
     typefaceFontName=75 Medium
     outlineSettings {
      bApplyOutlineToDropShadows=false
      bMiteredCorners=false
      bSeparateFillAlpha=false
      outlineMaterial=null
      outlineSize=0
      outlineColor=<rgba 0 0 0 1>
     }
    }
    shadowOffset=<0 0>
   }
   slot {
    class=CanvasPanelSlot
    properties {bAutoSize=true, layoutData {alignment=<0.5 0>, anchors {maximum=<0.5 0.5>, minimum=<0.5 0.5>}, offsets=<ltrb 247 330 600 30>}}
   }
  }
  SizeBox_5 {
   widgetClassPath=&SizeBox
   parent=CP_CombatFireButton
   widget=SizeBox_5
   properties {bOverride_MaxDesiredWidth=true, maxDesiredWidth=210}
   slot {
    class=CanvasPanelSlot
    properties {layoutData {alignment=<0 1>, anchors {maximum=<0.5 0.5>, minimum=<0.5 0.5>}, offsets=<ltrb 545 -123 210 30>}}
   }
  }
  CommonTextBlock_4 {
   widgetClassPath=&CommonTextBlock
   parent=SizeBox_5
   widget=CommonTextBlock_4
   properties {
    lineHeightPercentage=0.9
    scrollStyle=&TextScrollStyle-Base_C
    text=Toggle zoom
    font {
     bForceMonospaced=false
     bMaterialIsStencil=false
     fontMaterial=null
     fontObject=&HeadingNow
     letterSpacing=0
     monospacedWidth=1
     size=18
     skewAmount=0
     typefaceFontName=75 Medium
     outlineSettings {
      bApplyOutlineToDropShadows=false
      bMiteredCorners=false
      bSeparateFillAlpha=false
      outlineMaterial=null
      outlineSize=0
      outlineColor=<rgba 0 0 0 1>
     }
    }
    shadowOffset=<0 0>
   }
   slot {class=SizeBoxSlot}
  }
  SizeBox_4 {
   widgetClassPath=&SizeBox
   parent=CP_CombatFireButton
   widget=SizeBox_4
   properties {bOverride_MaxDesiredWidth=true, maxDesiredWidth=210}
   slot {
    class=CanvasPanelSlot
    properties {layoutData {alignment=<0 0.5>, anchors {maximum=<0.5 0.5>, minimum=<0.5 0.5>}, offsets=<ltrb 545 -101 210 30>}}
   }
  }
  CommonTextBlock_3 {
   widgetClassPath=&CommonTextBlock
   parent=SizeBox_4
   widget=CommonTextBlock_3
   properties {
    lineHeightPercentage=0.9
    scrollStyle=&TextScrollStyle-Base_C
    text=Toggle autofire
    font {
     bForceMonospaced=false
     bMaterialIsStencil=false
     fontMaterial=null
     fontObject=&HeadingNow
     letterSpacing=0
     monospacedWidth=1
     size=18
     skewAmount=0
     typefaceFontName=75 Medium
     outlineSettings {
      bApplyOutlineToDropShadows=false
      bMiteredCorners=false
      bSeparateFillAlpha=false
      outlineMaterial=null
      outlineSize=0
      outlineColor=<rgba 0 0 0 1>
     }
    }
    shadowOffset=<0 0>
   }
   slot {class=SizeBoxSlot}
  }
  SizeBox_3 {
   widgetClassPath=&SizeBox
   parent=CP_CombatFireButton
   widget=SizeBox_3
   properties {bOverride_MaxDesiredHeight=true, maxDesiredHeight=70}
   slot {
    class=CanvasPanelSlot
    properties {layoutData {alignment=<0 1>, anchors {maximum=<0.5 0.5>, minimum=<0.5 0.5>}, offsets=<ltrb 545 33 260 30>}}
   }
  }
  CommonTextBlock_2 {
   widgetClassPath=&CommonTextBlock
   parent=SizeBox_3
   widget=CommonTextBlock_2
   properties {
    lineHeightPercentage=0.9
    scrollOrientation=Orient_Vertical
    scrollStyle=&TextScrollStyle-Base_C
    text=Tap to fire
    wrapTextAt=200
    font {
     bForceMonospaced=false
     bMaterialIsStencil=false
     fontMaterial=null
     fontObject=&HeadingNow
     letterSpacing=0
     monospacedWidth=1
     size=18
     skewAmount=0
     typefaceFontName=75 Medium
     outlineSettings {
      bApplyOutlineToDropShadows=false
      bMiteredCorners=false
      bSeparateFillAlpha=false
      outlineMaterial=null
      outlineSize=0
      outlineColor=<rgba 0 0 0 1>
     }
    }
    shadowOffset=<0 0>
   }
   slot {class=SizeBoxSlot}
  }
  SizeBox_2 {
   widgetClassPath=&SizeBox
   parent=CP_CombatFireButton
   widget=SizeBox_2
   properties {bOverride_MaxDesiredHeight=true, maxDesiredHeight=70}
   slot {
    class=CanvasPanelSlot
    properties {layoutData {alignment=<0 0>, anchors {maximum=<0.5 0.5>, minimum=<0.5 0.5>}, offsets=<ltrb 545 67 260 60>}}
   }
  }
  CommonTextBlock_18 {
   widgetClassPath=&CommonTextBlock
   parent=SizeBox_2
   widget=CommonTextBlock_18
   properties {
    lineHeightPercentage=0.9
    scrollOrientation=Orient_Vertical
    scrollStyle=&TextScrollStyle-Base_C
    text=Hold to keep firing
    wrapTextAt=200
    font {
     bForceMonospaced=false
     bMaterialIsStencil=false
     fontMaterial=null
     fontObject=&HeadingNow
     letterSpacing=0
     monospacedWidth=1
     size=18
     skewAmount=0
     typefaceFontName=75 Medium
     outlineSettings {
      bApplyOutlineToDropShadows=false
      bMiteredCorners=false
      bSeparateFillAlpha=false
      outlineMaterial=null
      outlineSize=0
      outlineColor=<rgba 0 0 0 1>
     }
    }
    shadowOffset=<0 0>
   }
   slot {class=SizeBoxSlot}
  }
  CommonTextBlock_22 {
   widgetClassPath=&CommonTextBlock
   parent=CP_CombatFireButton
   widget=CommonTextBlock_22
   properties {
    lineHeightPercentage=0.9
    text=Reload
    wrapTextAt=260
    font {
     bForceMonospaced=false
     bMaterialIsStencil=false
     fontMaterial=null
     fontObject=&HeadingNow
     letterSpacing=0
     monospacedWidth=1
     size=18
     skewAmount=0
     typefaceFontName=75 Medium
     outlineSettings {
      bApplyOutlineToDropShadows=false
      bMiteredCorners=false
      bSeparateFillAlpha=false
      outlineMaterial=null
      outlineSize=0
      outlineColor=<rgba 0 0 0 1>
     }
    }
    shadowOffset=<0 0>
   }
   slot {
    class=CanvasPanelSlot
    properties {bAutoSize=true, layoutData {alignment=<0 0.5>, anchors {maximum=<0.5 0.5>, minimum=<0.5 0.5>}, offsets=<ltrb 545 171 449.9602 30>}}
   }
  }
  B_Line_5 {
   widgetClassPath=&Border
   parent=CP_CombatFireButton
   widget=B_Line_5
   properties {visibility=SelfHitTestInvisible, padding=<ltrb 0 0 0 0>}
   slot {
    class=CanvasPanelSlot
    properties {layoutData {alignment=<0 0>, anchors {maximum=<0.5 0.5>, minimum=<0.5 0.5>}, offsets=<ltrb -27.941 263.5837 3 48.7156>}}
   }
  }
  B_Line_6 {
   widgetClassPath=&Border
   parent=CP_CombatFireButton
   widget=B_Line_6
   properties {visibility=SelfHitTestInvisible, padding=<ltrb 0 0 0 0>}
   slot {
    class=CanvasPanelSlot
    properties {layoutData {alignment=<0 0>, anchors {maximum=<0.5 0.5>, minimum=<0.5 0.5>}, offsets=<ltrb 244.2144 249.0837 3 63.2156>}}
   }
  }
  Pip_List {
   widgetClassPath=&Overlay
   parent=Overlay_3
   widget=Pip_List
   properties {bOverride_Cursor=true}
   slot {class=OverlaySlot, properties {horizontalAlignment=HAlign_Center, verticalAlignment=VAlign_Center, padding=<ltrb 0 0 0 280>}}
  }
  Pip_0 {
   widgetClassPath=&Image
   bIsVariable=true
   parent=Pip_List
   widget=Pip_0
   properties {brush {
     bIsDynamicallyLoaded=false
     drawAs=Image
     imageType=FullColor
     mirroring=NoMirror
     resourceName=None
     resourceObject=&MI_Pip_Entry_Off
     tiling=NoTile
     imageSize=<15 15>
     margin=<ltrb 0 0 0 0>
     outlineSettings {
      bUseBrushTransparency=false
      roundingType=HalfHeightRadius
      width=0
      color {colorUseRule=UseColor_Specified, specifiedColor=<rgba 0 0 0 0>}
      cornerRadii=<0 0 0 0>
     }
     tintColor {colorUseRule=UseColor_Specified, specifiedColor=<rgba 1 1 1 1>}
     uVRegion {bIsValid=false, max=<0 0>, min=<0 0>}
    }}
   slot {class=OverlaySlot, properties {padding=<ltrb -25 0 0 0>}}
  }
  Pip_1 {
   widgetClassPath=&Image
   bIsVariable=true
   parent=Pip_List
   widget=Pip_1
   properties {brush {
     bIsDynamicallyLoaded=false
     drawAs=Image
     imageType=FullColor
     mirroring=NoMirror
     resourceName=None
     resourceObject=&MI_Pip_Entry_Off
     tiling=NoTile
     imageSize=<15 15>
     margin=<ltrb 0 0 0 0>
     outlineSettings {
      bUseBrushTransparency=false
      roundingType=HalfHeightRadius
      width=0
      color {colorUseRule=UseColor_Specified, specifiedColor=<rgba 0 0 0 0>}
      cornerRadii=<0 0 0 0>
     }
     tintColor {colorUseRule=UseColor_Specified, specifiedColor=<rgba 1 1 1 1>}
     uVRegion {bIsValid=false, max=<0 0>, min=<0 0>}
    }}
   slot {class=OverlaySlot}
  }
  Pip_2 {
   widgetClassPath=&Image
   bIsVariable=true
   parent=Pip_List
   widget=Pip_2
   properties {brush {
     bIsDynamicallyLoaded=false
     drawAs=Image
     imageType=FullColor
     mirroring=NoMirror
     resourceName=None
     resourceObject=&MI_Pip_Entry_Off
     tiling=NoTile
     imageSize=<15 15>
     margin=<ltrb 0 0 0 0>
     outlineSettings {
      bUseBrushTransparency=false
      roundingType=HalfHeightRadius
      width=0
      color {colorUseRule=UseColor_Specified, specifiedColor=<rgba 0 0 0 0>}
      cornerRadii=<0 0 0 0>
     }
     tintColor {colorUseRule=UseColor_Specified, specifiedColor=<rgba 1 1 1 1>}
     uVRegion {bIsValid=false, max=<0 0>, min=<0 0>}
    }}
   slot {class=OverlaySlot, properties {horizontalAlignment=HAlign_Right, padding=<ltrb 0 0 -25 0>}}
  }
 ]
}
