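# Widget for the matchmaking UI that shows the current region in a rounded, bordered pill, with no separate pool element.
# The OnRegionChanged event from MatchmakingRegionAndPoolBase sets RegionText from the front end context's MCP region text.
# Construct clears the size box's width cap on non-mobile games, so only mobile keeps it.
# OnInitialized adjusts the text's mobile font multiplier when the Bacchus top tab navigation runtime option is on.
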WidgetBlueprint /FortniteCoreUI/Athena/UI/Matchmaking/MatchmakingRegionAndPool
graphs [EventGraph {nodes [
   Event Construct()
   > n1 = Branch(Condition=IsMobileGame())
   ClearMaxDesiredWidth(execute=@n1.else, self:Size Box Object Reference=Get SizeBox_Mobile() -> (SizeBox_Mobile:Size Box Object Reference))
   Event OnRegionChanged() {is=title K2Node_Event}
   > "SetText (Text)"(
    self=Get RegionText() -> (RegionText:Common Text Object Reference)
    InText=GetMCPRegionText!(self:Fort Front End Context Object Reference=Get FortFrontEndContext!() -> (ReturnValue:Fort Front End Context Object Reference) {is=title K2Node_GetBlueprintContextEx}) -> (ReturnValue:Text)
   )
   Event OnInitialized()
   > Branch(Condition=GetIsBacchusTabNavOnTop!(self:Fort Runtime Options Object Reference=GetRuntimeOptions()) -> (ReturnValue:bool))
   > SetMobileFontSizeMultiplier!(execute, self:Common Text Object Reference=Get RegionText(), InMobileFontSizeMultiplier:float=1.2) -> (then)
  ]}]
variables []
widgets {
 info {
  inheritedWidgetCount=0
  parentClass=/Script/FortniteUI.MatchmakingRegionAndPoolBase
  rootWidgetClass=&CommonBorder
  widgetCount=4
 }
 widgets [
  Border {
   widgetClassPath=&CommonBorder
   widget=Border
   properties {
    style=/Game/Athena/UI/Matchmaking/Border_Matchmaking_Rectangle.Border_Matchmaking_Rectangle_C
    verticalAlignment=VAlign_Center
    background {
     bIsDynamicallyLoaded=false
     drawAs=Box
     imageType=NoImage
     mirroring=NoMirror
     resourceName=None
     resourceObject=/Game/UI/Foundation/Materials/MI_Notification_Rectangle
     tiling=NoTile
     imageSize=<128 64>
     margin=<ltrb 0.5 0 0.5 0>
     outlineSettings {
      bUseBrushTransparency=false
      roundingType=HalfHeightRadius
      width=0
      color {colorUseRule=UseColor_Specified, specifiedColor=<rgba 0 0 0 0>}
      cornerRadii=<0 0 0 0>
     }
     tintColor {colorUseRule=UseColor_Specified, specifiedColor=<rgba 1 1 1 1>}
     uVRegion {bIsValid=false, max=<0 0>, min=<0 0>}
    }
    padding=<ltrb 24 5 24 5>
   }
  }
  HorizontalBox_0 {
   widgetClassPath=&HorizontalBox
   parent=Border
   widget=HorizontalBox_0
   slot {class=BorderSlot, properties {verticalAlignment=VAlign_Center, padding=<ltrb 24 5 24 5>}}
  }
  SizeBox_Mobile {
   widgetClassPath=&SizeBox
   bIsVariable=true
   parent=HorizontalBox_0
   widget=SizeBox_Mobile
   properties {bOverride_MaxDesiredWidth=true, maxDesiredWidth=180}
   slot {class=HorizontalBoxSlot}
  }
  RegionText {
   widgetClassPath=&CommonTextBlock
   bIsVariable=true
   parent=SizeBox_Mobile
   widget=RegionText
   properties {
    mobileFontSizeMultiplier=1.4
    scrollStyle=&TextScrollStyle-Base_C
    style=/FortUILibrary/Styles/Typography/Foundation/TextStyle-Header-XS.TextStyle-Header-XS_C
    text="{Region}"
    font {
     bForceMonospaced=false
     bMaterialIsStencil=false
     fontMaterial=null
     fontObject=&Burbank
     letterSpacing=0
     monospacedWidth=1
     size=16
     skewAmount=0
     typefaceFontName=Burbank-Black
     outlineSettings {
      bApplyOutlineToDropShadows=false
      bMiteredCorners=false
      bSeparateFillAlpha=false
      outlineMaterial=null
      outlineSize=0
      outlineColor=<rgba 0 0 0 1>
     }
    }
    shadowOffset=<0 0>
   }
   slot {class=SizeBoxSlot}
  }
 ]
}
