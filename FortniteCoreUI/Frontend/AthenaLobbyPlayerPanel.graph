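# Lobby player panel widget on native AthenaLobbyPlayerPanel, showing a player's banner, name, platform icon and Battle Pass tier above social action buttons.
# Parent-class events toggle the Battle Pass row, switch the Mute/Unmute label, and enable and relabel Add Friend by friend-request status.
# A mouse-down handler closes the panel and consumes the click.
# Most widgets with content are inherited bindings, and OnGadgetsClicked is an empty stub.
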WidgetBlueprint /FortniteCoreUI/Frontend/AthenaLobbyPlayerPanel
graphs [
 EventGraph {nodes [
   n1 = Event OnHasBattlePassUpdated!() -> (OutputDelegate, then, bHasBattlePass:bool) {is=title K2Node_Event}
   > SetVisibility(
    self=Get BattlePassRow() -> (BattlePassRow:Horizontal Box Object Reference)
    InVisibility=Select(Option 0=Collapsed, Option 1=HitTestInvisible, Index=@n1.bHasBattlePass)
   )
   n2 = Event OnIsMutedUpdated!() -> (OutputDelegate, then, bIsMuted:bool) {is=title K2Node_Event}
   > n3 = Cast To WBP_UIKit_ButtonCTA_Base!(execute, Object=Get Button_Mute() -> (Button_Mute:Common Button Base Object Reference)) -> (then, CastFailed, AsWBP UIKit Button CTA Base:WBP UIKit Button CTA Base Object Reference) {is=title K2Node_DynamicCast}
   > SetText!(
    execute
    self:Fort CTAButton Object Reference=@n3
    ButtonText:"Text (by ref)"=Select(Option 0:Text="LOCTEXT(\"Mute\")", Option 1:Text="LOCTEXT(\"Unmute\")", Index=@n2.bIsMuted) -> (ReturnValue:Text)
   ) -> (then)
   n4 = Event OnFriendStatusUpdated!() -> (OutputDelegate, then, FriendRequestStatus:EFortFriendRequestStatus Enum=None) {is=title K2Node_Event}
   n5 = Get Button_AddFriend() -> (Button_AddFriend:Common Button Base Object Reference)
   > SetIsEnabled(
    self=@n5
    bInIsEnabled=Select!(
     None:bool=true
     Accepted:bool=false
     PendingReceived:bool=true
     PendingSent:bool=false
     Index:EFortFriendRequestStatus Enum=@n4.FriendRequestStatus
    ) -> (ReturnValue:bool)
   )
   > n6 = Cast To WBP_UIKit_ButtonCTA_Base(Object=@n5)
   > SetText(
    self=@n6
    ButtonText=Select(
     None:"Text (by ref)"=Get AddFriendText() -> (AddFriendText:Text)
     Accepted:"Text (by ref)"=Get AcceptedInviteText() -> (AcceptedInviteText:Text)
     PendingReceived:"Text (by ref)"=Get AcceptInviteText() -> (AcceptInviteText:Text)
     PendingSent:"Text (by ref)"=Get SentInviteText() -> (SentInviteText:Text)
     Index=@n4.FriendRequestStatus
    ) -> (ReturnValue:"Text (by ref)")
   )
  ]}
 OnGadgetsClicked {nodes [OnGadgetsClicked() {is=title K2Node_FunctionEntry}]}
 OnMouseButtonDown_0 {nodes [
   OnMouseButtonDown_0!() -> (then, MyGeometry:Geometry Structure, MouseEvent:"Pointer Event Structure (by ref)") {is=title K2Node_FunctionEntry}
   > ClosePanel(self:Fort Lobby Player Panel Object Reference)
   > Return Node!(execute, ReturnValue:Event Reply Structure=Handled())
  ]}
]
variables [
 AcceptedInviteText {type=string, default=Epic Friend Invite Accepted}
 AcceptInviteText {type=string, default=Accept Epic Friend Invite}
 AddFriendText {type=string, default=Add Epic Friend}
 OnGadgetsClicked {}
 SentInviteText {type=string, default=Epic Friend Invite Sent}
 TeamMemberInfo {
  type=FortTeamMemberInfo
  default {
   bBattlePassPurchased=false
   bHasBoostXp=false
   bHasRestXp=false
   bIsInZone=false
   bPartyLeader=false
   bannerColorId=""
   bannerIconId=""
   battlePassLevel=0
   heroClass=""
   heroItem=null
   heroLevel=""
   heroXP=-1
   numPlayersInParty=0
   playerIndex=-1
   playerName=""
   teamAffiliation=255
   heroIcon {
    bIsDynamicallyLoaded=false
    drawAs=Image
    imageType=NoImage
    mirroring=NoMirror
    resourceName=None
    resourceObject=null
    tiling=NoTile
    imageSize=<32 32>
    margin=<ltrb 0 0 0 0>
    outlineSettings {
     bUseBrushTransparency=false
     roundingType=HalfHeightRadius
     width=0
     color {colorUseRule=UseColor_Specified, specifiedColor=<rgba 0 0 0 0>}
     cornerRadii=<0 0 0 1>
    }
    tintColor {colorUseRule=UseColor_Specified, specifiedColor=<rgba 1 1 1 1>}
    uVRegion {bIsValid=false, max=<0 0>, min=<0 0>}
   }
   homeBaseInfo {name="", bannerColorId="", bannerIconId="", validData=false}
   selectedGadgetItems []
  }
 }
]
widgets {
 info {inheritedWidgetCount=11, parentClass=/Script/FortniteUI.AthenaLobbyPlayerPanel, rootWidgetClass=&Border, widgetCount=22}
 widgets [
  Border_0 {
   widgetClassPath=&Border
   widget=Border_0
   properties {
    horizontalAlignment=HAlign_Center
    verticalAlignment=VAlign_Bottom
    background {
     bIsDynamicallyLoaded=false
     drawAs=NoDrawType
     imageType=NoImage
     mirroring=NoMirror
     resourceName=None
     resourceObject=null
     tiling=NoTile
     imageSize=<32 32>
     margin=<ltrb 0 0 0 0>
     outlineSettings {
      bUseBrushTransparency=false
      roundingType=HalfHeightRadius
      width=0
      color {colorUseRule=UseColor_Specified, specifiedColor=<rgba 0 0 0 0>}
      cornerRadii=<0 0 0 0>
     }
     tintColor {colorUseRule=UseColor_Specified, specifiedColor=<rgba 1 1 1 1>}
     uVRegion {bIsValid=false, max=<0 0>, min=<0 0>}
    }
    brushColor=<rgba 1 1 1 0>
    padding=<ltrb 0 0 0 0>
   }
  }
  SizeBox_1 {
   widgetClassPath=&SizeBox
   parent=Border_0
   widget=SizeBox_1
   properties {bOverride_MaxDesiredWidth=true, bOverride_MinDesiredWidth=true, maxDesiredWidth=720, minDesiredWidth=584}
   slot {class=BorderSlot, properties {horizontalAlignment=HAlign_Center, verticalAlignment=VAlign_Bottom, padding=<ltrb 0 0 0 0>}}
  }
  GridPanel_0 {widgetClassPath=&GridPanel, parent=SizeBox_1, widget=GridPanel_0, slot {class=SizeBoxSlot}}
  UIKit_Backplate {
   widgetClassPath=&WBP_UIKit_Backplate
   parent=GridPanel_0
   widget=UIKit_Backplate
   properties {brightness=Dark}
   slot {class=GridSlot}
  }
  VerticalBox_2 {
   widgetClassPath=&VerticalBox
   parent=GridPanel_0
   widget=VerticalBox_2
   slot {class=GridSlot, properties {layer=1, padding=<ltrb 24 24 24 32>}}
  }
  HorizontalBox_76 {
   widgetClassPath=&HorizontalBox
   parent=VerticalBox_2
   widget=HorizontalBox_76
   slot {class=VerticalBoxSlot, properties {horizontalAlignment=HAlign_Center, padding=<ltrb 0 0 0 8>}}
  }
  Banner_Player {
   widgetClassPath=/Script/FortniteUI.FortPlayerBanner
   bInherited=true
   bIsVariable=true
   parent=HorizontalBox_76
   widget=Banner_Player
   properties {desiredSize=<80 100>}
   slot {class=HorizontalBoxSlot, properties {horizontalAlignment=HAlign_Left, verticalAlignment=VAlign_Top}}
  }
  VerticalBox_86 {
   widgetClassPath=&VerticalBox
   parent=HorizontalBox_76
   widget=VerticalBox_86
   slot {class=HorizontalBoxSlot, properties {verticalAlignment=VAlign_Center, padding=<ltrb 12 0 0 0>, size {sizeRule=Fill, value=1}}}
  }
  HorizontalBox_155 {
   widgetClassPath=&HorizontalBox
   parent=VerticalBox_86
   widget=HorizontalBox_155
   slot {class=VerticalBoxSlot, properties {verticalAlignment=VAlign_Top}}
  }
  RichText_PlayerName {
   widgetClassPath=&CommonRichTextBlock
   bInherited=true
   bIsVariable=true
   parent=HorizontalBox_155
   widget=RichText_PlayerName
   properties {
    bOverrideDefaultStyle=true
    scrollStyle=/FortUILibrary/Styles/Typography/Foundation/TextScrollStyle-NoFade.TextScrollStyle-NoFade_C
    text="{Player Name}"
    defaultTextStyleOverride {
     overflowPolicy=Clip
     transformPolicy=None
     colorAndOpacity {colorUseRule=UseColor_Specified, specifiedColor=<rgba 1 1 1 1>}
     font {
      bForceMonospaced=false
      bMaterialIsStencil=false
      fontMaterial=null
      fontObject=&HeadingNow
      letterSpacing=0
      monospacedWidth=1
      size=22.5
      skewAmount=0
      typefaceFontName=75 Medium
      outlineSettings {
       bApplyOutlineToDropShadows=false
       bMiteredCorners=false
       bSeparateFillAlpha=false
       outlineMaterial=null
       outlineSize=0
       outlineColor=<rgba 0 0 0 1>
      }
     }
     highlightColor {colorUseRule=UseColor_Specified, specifiedColor=<rgba 0 0 0 1>}
     highlightShape {
      bIsDynamicallyLoaded=false
      drawAs=Image
      imageType=NoImage
      mirroring=NoMirror
      resourceName=None
      resourceObject=null
      tiling=NoTile
      imageSize=<32 32>
      margin=<ltrb 0 0 0 0>
      outlineSettings {
       bUseBrushTransparency=false
       roundingType=HalfHeightRadius
       width=0
       color {colorUseRule=UseColor_Specified, specifiedColor=<rgba 0 0 0 0>}
       cornerRadii=<0 0 0 0>
      }
      tintColor {colorUseRule=UseColor_Specified, specifiedColor=<rgba 1 1 1 1>}
      uVRegion {bIsValid=false, max=<0 0>, min=<0 0>}
     }
     selectedBackgroundColor {colorUseRule=UseColor_Foreground, specifiedColor=<rgba 1 0 1 1>}
     shadowColorAndOpacity=<rgba 0 0 0 1>
     shadowOffset=<0 0>
     strikeBrush {
      bIsDynamicallyLoaded=false
      drawAs=NoDrawType
      imageType=NoImage
      mirroring=NoMirror
      resourceName=None
      resourceObject=null
      tiling=NoTile
      imageSize=<32 32>
      margin=<ltrb 0 0 0 0>
      outlineSettings {
       bUseBrushTransparency=false
       roundingType=HalfHeightRadius
       width=0
       color {colorUseRule=UseColor_Specified, specifiedColor=<rgba 0 0 0 0>}
       cornerRadii=<0 0 0 0>
      }
      tintColor {colorUseRule=UseColor_Specified, specifiedColor=<rgba 1 1 1 1>}
      uVRegion {bIsValid=false, max=<0 0>, min=<0 0>}
     }
     underlineBrush {
      bIsDynamicallyLoaded=false
      drawAs=Image
      imageType=NoImage
      mirroring=NoMirror
      resourceName=None
      resourceObject=null
      tiling=NoTile
      imageSize=<32 32>
      margin=<ltrb 0 0 0 0>
      outlineSettings {
       bUseBrushTransparency=false
       roundingType=HalfHeightRadius
       width=0
       color {colorUseRule=UseColor_Specified, specifiedColor=<rgba 0 0 0 0>}
       cornerRadii=<0 0 0 0>
      }
      tintColor {colorUseRule=UseColor_Specified, specifiedColor=<rgba 1 1 1 1>}
      uVRegion {bIsValid=false, max=<0 0>, min=<0 0>}
     }
    }
   }
   slot {class=HorizontalBoxSlot, properties {horizontalAlignment=HAlign_Left, verticalAlignment=VAlign_Center, size {sizeRule=Fill, value=1}}}
  }
  Image_PlatformIcon {
   widgetClassPath=&Image
   bInherited=true
   bIsVariable=true
   parent=HorizontalBox_155
   widget=Image_PlatformIcon
   properties {
    visibility=Collapsed
    brush {
     bIsDynamicallyLoaded=false
     drawAs=Image
     imageType=NoImage
     mirroring=NoMirror
     resourceName=None
     resourceObject=null
     tiling=NoTile
     imageSize=<36 36>
     margin=<ltrb 0 0 0 0>
     outlineSettings {
      bUseBrushTransparency=false
      roundingType=HalfHeightRadius
      width=0
      color {colorUseRule=UseColor_Specified, specifiedColor=<rgba 0 0 0 0>}
      cornerRadii=<0 0 0 0>
     }
     tintColor {colorUseRule=UseColor_Specified, specifiedColor=<rgba 1 1 1 1>}
     uVRegion {bIsValid=false, max=<0 0>, min=<0 0>}
    }
   }
   slot {
    class=HorizontalBoxSlot
    properties {horizontalAlignment=HAlign_Right, verticalAlignment=VAlign_Center, padding=<ltrb 10 0 10 0>, size {sizeRule=Automatic, value=0.2}}
   }
  }
  Border_1 {
   widgetClassPath=&Border
   parent=VerticalBox_86
   widget=Border_1
   properties {
    background {
     bIsDynamicallyLoaded=false
     drawAs=NoDrawType
     imageType=NoImage
     mirroring=NoMirror
     resourceName=None
     resourceObject=null
     tiling=NoTile
     imageSize=<32 32>
     margin=<ltrb 0 0 0 0>
     outlineSettings {
      bUseBrushTransparency=false
      roundingType=HalfHeightRadius
      width=0
      color {colorUseRule=UseColor_Specified, specifiedColor=<rgba 0 0 0 0>}
      cornerRadii=<0 0 0 0>
     }
     tintColor {colorUseRule=UseColor_Specified, specifiedColor=<rgba 1 1 1 1>}
     uVRegion {bIsValid=false, max=<0 0>, min=<0 0>}
    }
    contentColorAndOpacity=<rgba 0.9301 1 0.0097 1>
    padding=<ltrb 0 0 0 0>
   }
   slot {class=VerticalBoxSlot, properties {verticalAlignment=VAlign_Top, padding=<ltrb 4 0 0 0>}}
  }
  Text_InviteSent {
   widgetClassPath=&CommonTextBlock
   bInherited=true
   parent=Border_1
   widget=Text_InviteSent
   properties {
    lineHeightPercentage=1.29
    style=/FortUILibrary/Styles/Typography/_TextStyle_Body1._TextStyle_Body1_C
    text=Invite Sent
    textTransformPolicy=ToUpper
    font {
     bForceMonospaced=false
     bMaterialIsStencil=false
     fontMaterial=null
     fontObject=&BurbankSmall
     letterSpacing=0
     monospacedWidth=1
     size=18
     skewAmount=0
     typefaceFontName=Bold
     outlineSettings {
      bApplyOutlineToDropShadows=false
      bMiteredCorners=false
      bSeparateFillAlpha=false
      outlineMaterial=null
      outlineSize=0
      outlineColor=<rgba 0 0 0 1>
     }
    }
    shadowOffset=<0 0>
    strikeBrush {
     bIsDynamicallyLoaded=false
     drawAs=NoDrawType
     imageType=NoImage
     mirroring=NoMirror
     resourceName=None
     resourceObject=null
     tiling=NoTile
     imageSize=<32 32>
     margin=<ltrb 0 0 0 0>
     outlineSettings {
      bUseBrushTransparency=false
      roundingType=HalfHeightRadius
      width=0
      color {colorUseRule=UseColor_Specified, specifiedColor=<rgba 0 0 0 0>}
      cornerRadii=<0 0 0 0>
     }
     tintColor {colorUseRule=UseColor_Specified, specifiedColor=<rgba 1 1 1 1>}
     uVRegion {bIsValid=false, max=<0 0>, min=<0 0>}
    }
   }
   slot {class=BorderSlot, properties {padding=<ltrb 0 0 0 0>}}
  }
  BattlePassRow {
   widgetClassPath=&HorizontalBox
   bIsVariable=true
   parent=VerticalBox_86
   widget=BattlePassRow
   slot {class=VerticalBoxSlot, properties {verticalAlignment=VAlign_Top, padding=<ltrb 4 0 0 24>}}
  }
  Text_BattlePassTier {
   widgetClassPath=&CommonTextBlock
   bInherited=true
   parent=BattlePassRow
   widget=Text_BattlePassTier
   properties {
    lineHeightPercentage=1.29
    style=&_TextStyle_Body1_C
    text=Battle Pass Tier
    textTransformPolicy=ToUpper
    font {
     bForceMonospaced=false
     bMaterialIsStencil=false
     fontMaterial=null
     fontObject=&BurbankSmall
     letterSpacing=0
     monospacedWidth=1
     size=18
     skewAmount=0
     typefaceFontName=Bold
     outlineSettings {
      bApplyOutlineToDropShadows=false
      bMiteredCorners=false
      bSeparateFillAlpha=false
      outlineMaterial=null
      outlineSize=0
      outlineColor=<rgba 0 0 0 1>
     }
    }
    shadowOffset=<0 0>
    strikeBrush {
     bIsDynamicallyLoaded=false
     drawAs=NoDrawType
     imageType=NoImage
     mirroring=NoMirror
     resourceName=None
     resourceObject=null
     tiling=NoTile
     imageSize=<32 32>
     margin=<ltrb 0 0 0 0>
     outlineSettings {
      bUseBrushTransparency=false
      roundingType=HalfHeightRadius
      width=0
      color {colorUseRule=UseColor_Specified, specifiedColor=<rgba 0 0 0 0>}
      cornerRadii=<0 0 0 0>
     }
     tintColor {colorUseRule=UseColor_Specified, specifiedColor=<rgba 1 1 1 1>}
     uVRegion {bIsValid=false, max=<0 0>, min=<0 0>}
    }
   }
   slot {class=HorizontalBoxSlot, properties {verticalAlignment=VAlign_Top, padding=<ltrb 0 -4 0 0>, size {sizeRule=Fill, value=1}}}
  }
  VerticalBox_4 {
   widgetClassPath=&VerticalBox
   parent=VerticalBox_2
   widget=VerticalBox_4
   slot {class=VerticalBoxSlot, properties {horizontalAlignment=HAlign_Center}}
  }
  Button_Manage {
   widgetClassPath=&WBP_UIKit_Button_Regular
   bInherited=true
   bIsVariable=true
   parent=VerticalBox_4
   widget=Button_Manage
   properties {minWidth=416, text=Manage, visibility=Collapsed}
   slot {class=VerticalBoxSlot, properties {padding=<ltrb 0 0 0 24>}}
  }
  Button_AddFriend {
   widgetClassPath=&WBP_UIKit_Button_Regular
   bInherited=true
   bIsVariable=true
   parent=VerticalBox_4
   widget=Button_AddFriend
   properties {minWidth=416, text="<DynamicText:Add Epic Friend>"}
   slot {class=VerticalBoxSlot, properties {padding=<ltrb 0 0 0 24>}}
  }
  Button_InviteParty {
   widgetClassPath=&WBP_UIKit_Button_Regular
   bInherited=true
   bIsVariable=true
   parent=VerticalBox_4
   widget=Button_InviteParty
   properties {minWidth=416, text=Invite to Party}
   slot {class=VerticalBoxSlot, properties {verticalAlignment=VAlign_Top, padding=<ltrb 0 0 0 24>}}
  }
  Button_ViewProfile {
   widgetClassPath=&WBP_UIKit_Button_Regular
   bInherited=true
   bIsVariable=true
   parent=VerticalBox_4
   widget=Button_ViewProfile
   properties {minWidth=416, text=View Profile}
   slot {class=VerticalBoxSlot, properties {padding=<ltrb 0 0 0 24>}}
  }
  Button_Mute {
   widgetClassPath=&WBP_UIKit_Button_Regular
   bInherited=true
   bIsVariable=true
   parent=VerticalBox_4
   widget=Button_Mute
   properties {minWidth=416, text=Toggle Mute}
   slot {class=VerticalBoxSlot, properties {padding=<ltrb 0 0 0 24>}}
  }
  Button_Close {
   widgetClassPath=&WBP_UIKit_Button_Regular
   bInherited=true
   bIsVariable=true
   parent=VerticalBox_4
   widget=Button_Close
   properties {minWidth=416, priority=1, text=Close, triggeringInputAction {dataTable=&InputActionDataTable, rowName=Close}}
   slot {class=VerticalBoxSlot}
  }
 ]
}
