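# Legacy frontend button that warns of a pending account merge with an ETA, hidden when no transfer is pending.
# On Construct it queries the MCP context and sets the ETA text once from estimated completion minus now.
# Input-method changes swap the hint text, justification and gamepad action icon.
# Hover darkens or lightens the backing and plays a scale animation; clicking launches the more-info URL from runtime options.
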WidgetBlueprint /FortniteCoreUI/Frontend/PendingAccountMergeNotification_Legacy
animations [Hover {
  binding_count=1
  length_seconds=0.15
  object=&self:Hover
  sequence {bindings [PendingAccountMergeNotification_C_15 {
     id=#1
     possessed_class=PendingAccountMergeNotification_Legacy_C
     tracks [{
       type=MovieScene2DTransformTrack
       sections [{
         type=MovieScene2DTransformSection
         end_frame=3
         start_frame=0
         channels [
          Translation.X {type=MovieSceneScriptingFloatChannel}
          Translation.Y {type=MovieSceneScriptingFloatChannel}
          Angle {type=MovieSceneScriptingFloatChannel}
          Scale.X {type=MovieSceneScriptingFloatChannel, keys [[0,1],[2.9997,1.03]]}
          Scale.Y {type=MovieSceneScriptingFloatChannel, keys [[0,1],[2.9997,1.03]]}
          Shear.X {type=MovieSceneScriptingFloatChannel}
          Shear.Y {type=MovieSceneScriptingFloatChannel}
         ]
        }]
      }]
    }]}
 }]
graphs [EventGraph {nodes [
   n1 = Hovered!() -> (OutputDelegate, then, IsHovered:bool=~) {is=title K2Node_CustomEvent}
   > SetBrushColor(
    self=Get Border_1() -> (Border_1:Border Object Reference)
    InBrushColor=Select(Option 0:Linear Color Structure=(R=0,G=0,B=0,A=0.3), Option 1:Linear Color Structure=(R=0,G=0,B=0,A=0.2), Index=@n1.IsHovered) -> (ReturnValue:Linear Color Structure)
   )
   n2 = OnInputChanged!() -> (OutputDelegate, then, bNewInputType:ECommonInputType Enum) {is=title K2Node_CustomEvent}
   > "SetText (Text)"(
    self=Get TextBlock_42() -> (TextBlock_42:Text Object Reference)
    InText=Select!(
     MouseAndKeyboard:Text="LOCTEXT(\"CLICK HERE FOR MORE INFO\")"
     Gamepad:Text="LOCTEXT(\"OPEN SETTINGS FOR MORE INFO\")"
     Touch:Text="LOCTEXT(\"TAP HERE FOR MORE INFO\")"
     Count:Text
     Index:ECommonInputType Enum=@n2.bNewInputType
    ) -> (ReturnValue:Text)
   )
   n3 = "Equal (Enum)"(A=@n2.bNewInputType, B=Gamepad)
   > SetVisibility(
    self=Get CommonActionWidget_48() -> (CommonActionWidget_48:Common Action Widget Object Reference)
    InVisibility=Select!(Option 0:ESlateVisibility Enum=Collapsed, Option 1:ESlateVisibility Enum=SelfHitTestInvisible, Index:bool=@n3) -> (ReturnValue:ESlateVisibility Enum)
   )
   > SetJustification!(
    execute
    self:Text Layout Widget Object Reference=Get TextBlock_42()
    InJustification:ETextJustify Enum=Select(Option 0:ETextJustify Enum=Center, Option 1:ETextJustify Enum=Left, Index=@n3) -> (ReturnValue:ETextJustify Enum)
   ) -> (then)
   Event Construct()
   n4 = IsPendingAccountTransfer!(self:Fort Mcp Context Object Reference=Get FortMcpContext!() -> (ReturnValue:Fort Mcp Context Object Reference) {is=title K2Node_GetBlueprintContextEx}) -> (EstimatedCompletionMax:Date Time Structure, ReturnValue:bool)
   > SetVisibility(InVisibility=Select(Option 0:ESlateVisibility Enum=Collapsed, Option 1:ESlateVisibility Enum=Visible, Index=@n4.ReturnValue) -> (ReturnValue:ESlateVisibility Enum))
   n5 = Now!() -> (ReturnValue:Date Time Structure)
   > "SetText (Text)"(
    self=Get TextBlock_45() -> (TextBlock_45:Text Object Reference)
    InText=GetTimespanAsText!(
     Timespan:"Timespan Structure (by ref)"=DateTime - DateTime!(A:Date Time Structure=@n4, B:Date Time Structure=@n5) -> (ReturnValue:Timespan Structure)
     bShowHoursWhenZero:bool=true
     bShowSeconds:bool=false
     FormatType:ETimespanAsTextFormat Enum=DaysHoursMinutesSeconds
     UnitDisplayType:ETimespanUnitDisplayFormat Enum=Full
     bUseSentenceCase:bool=false
    ) -> (ReturnValue:Text)
   )
   > OnInputChanged!(execute, self, bNewInputType:ECommonInputType Enum=GetCurrentInputType(self=Get CommonInputSubsystem())) -> (then) {is=title K2Node_CallFunction}
   > Bind Event to On Input Method Changed(self=Get CommonInputSubsystem(), Delegate=Create Event())
   Event On Hovered()
   > Hovered!(execute, self, IsHovered:bool=true) -> (then) {is=title K2Node_CallFunction}
   n6 = Get Hover() -> (Hover:Widget Animation Object Reference)
   > PlayAnimationForward(InAnimation=@n6)
   Event On Unhovered()
   > Hovered(IsHovered=false)
   > PlayAnimationReverse(InAnimation=@n6)
   Event On Clicked()
   > LaunchURL!(
    execute
    URL:String=Get AccountMergeMoreInfoURL!(self:Fort Runtime Options Object Reference=GetRuntimeOptions()) -> (AccountMergeMoreInfoURL:String)
   ) -> (then)
   Event Destruct()
   > Unbind Event from On Input Method Changed(self=Get CommonInputSubsystem(), Delegate=Create Event())
   n7 = Now()
   DateTime - DateTime(
    A=DateTime + Timespan!(
     A:Date Time Structure=@n7
     B:Timespan Structure=MakeTimespan!(Days:int=3, Hours:int=2, Minutes:int=5, Seconds:int=45, Milliseconds:int=3) -> (ReturnValue:Timespan Structure)
    ) -> (ReturnValue:Date Time Structure)
    B=@n7
   )
   SetColorAndOpacity(
    self:User Widget Object Reference
    InColorAndOpacity:Linear Color Structure=Select(Option 0:Linear Color Structure=(R=1,G=1,B=1,A=1), Option 1:Linear Color Structure=(R=0.3229,G=0.3229,B=0.3229,A=1), Index=@n1.IsHovered) -> (ReturnValue:Linear Color Structure)
   )
   Event PreConstruct()
   Event Tick()
  ]}]
variables []
widgets {
 info {inheritedWidgetCount=1, parentClass=&CommonButtonLegacy, rootWidgetClass=&VerticalBox, widgetCount=13}
 widgets [
  VerticalBox_0 {widgetClassPath=&VerticalBox, widget=VerticalBox_0}
  Border_0 {
   widgetClassPath=&Border
   parent=VerticalBox_0
   widget=Border_0
   properties {
    horizontalAlignment=HAlign_Center
    background {
     bIsDynamicallyLoaded=false
     drawAs=Box
     imageType=NoImage
     mirroring=NoMirror
     resourceName=None
     resourceObject=/Game/UI/Foundation/Materials/M-TopBar-RectangleFill
     tiling=NoTile
     imageSize=<40 40>
     margin=<ltrb 0.5 0 0.5 0>
     outlineSettings {
      bUseBrushTransparency=false
      roundingType=HalfHeightRadius
      width=0
      color {colorUseRule=UseColor_Specified, specifiedColor=<rgba 0 0 0 0>}
      cornerRadii=<0 0 0 0>
     }
     tintColor {colorUseRule=UseColor_Specified, specifiedColor=<rgba 1 1 1 1>}
     uVRegion {bIsValid=false, max=<0 0>, min=<0 0>}
    }
    padding=<ltrb 24 0 24 0>
   }
   slot {class=VerticalBoxSlot, properties {horizontalAlignment=HAlign_Center}}
  }
  SizeBox_2 {
   widgetClassPath=&SizeBox
   parent=Border_0
   widget=SizeBox_2
   properties {bOverride_HeightOverride=true, bOverride_MinDesiredWidth=true, heightOverride=40, minDesiredWidth=192}
   slot {class=BorderSlot, properties {horizontalAlignment=HAlign_Center, padding=<ltrb 24 0 24 0>}}
  }
  TextBlock_43 {
   widgetClassPath=&TextBlock
   parent=SizeBox_2
   widget=TextBlock_43
   properties {
    text="ACCOUNT MERGE PENDING!"
    colorAndOpacity {colorUseRule=UseColor_Specified, specifiedColor=<rgba 0.9301 0.117 0.162 1>}
    font {
     bForceMonospaced=false
     bMaterialIsStencil=false
     fontMaterial=null
     fontObject=&BurbankSmall
     letterSpacing=0
     monospacedWidth=1
     size=15
     skewAmount=0
     typefaceFontName=Bold
     outlineSettings {
      bApplyOutlineToDropShadows=false
      bMiteredCorners=false
      bSeparateFillAlpha=false
      outlineMaterial=null
      outlineSize=0
      outlineColor=<rgba 0 0 0 1>
     }
    }
   }
   slot {class=SizeBoxSlot, properties {horizontalAlignment=HAlign_Center, verticalAlignment=VAlign_Center, padding=<ltrb 0 0 0 -2>}}
  }
  Border_1 {
   widgetClassPath=&Border
   bIsVariable=true
   parent=VerticalBox_0
   widget=Border_1
   properties {
    background {
     bIsDynamicallyLoaded=false
     drawAs=Box
     imageType=NoImage
     mirroring=NoMirror
     resourceName=None
     resourceObject=/Game/UI/Foundation/Materials/M-TopBar-RectangleSolidFill
     tiling=NoTile
     imageSize=<64 64>
     margin=<ltrb 0.5 0 0.5 0>
     outlineSettings {
      bUseBrushTransparency=false
      roundingType=HalfHeightRadius
      width=0
      color {colorUseRule=UseColor_Specified, specifiedColor=<rgba 0 0 0 0>}
      cornerRadii=<0 0 0 0>
     }
     tintColor {colorUseRule=UseColor_Specified, specifiedColor=<rgba 1 1 1 1>}
     uVRegion {bIsValid=false, max=<0 0>, min=<0 0>}
    }
    brushColor=<rgba 0 0 0 0.3>
    padding=<ltrb 0 8 0 8>
   }
   slot {class=VerticalBoxSlot, properties {padding=<ltrb 12 0 16 0>}}
  }
  VerticalBox_1 {widgetClassPath=&VerticalBox, parent=Border_1, widget=VerticalBox_1, slot {class=BorderSlot, properties {padding=<ltrb 0 8 0 8>}}}
  HorizontalBox_1 {
   widgetClassPath=&HorizontalBox
   parent=VerticalBox_1
   widget=HorizontalBox_1
   slot {class=VerticalBoxSlot, properties {horizontalAlignment=HAlign_Center}}
  }
  TextBlock_44 {
   widgetClassPath=&TextBlock
   parent=HorizontalBox_1
   widget=TextBlock_44
   properties {
    text=ETA:
    colorAndOpacity {colorUseRule=UseColor_Specified, specifiedColor=<rgba 0.9301 0.117 0.162 1>}
    font {
     bForceMonospaced=false
     bMaterialIsStencil=false
     fontMaterial=null
     fontObject=&BurbankSmall
     letterSpacing=0
     monospacedWidth=1
     size=15
     skewAmount=0
     typefaceFontName=Bold
     outlineSettings {
      bApplyOutlineToDropShadows=false
      bMiteredCorners=false
      bSeparateFillAlpha=false
      outlineMaterial=null
      outlineSize=0
      outlineColor=<rgba 0 0 0 1>
     }
    }
   }
   slot {class=HorizontalBoxSlot, properties {verticalAlignment=VAlign_Center, padding=<ltrb 0 1 5 0>}}
  }
  TextBlock_45 {
   widgetClassPath=&TextBlock
   bIsVariable=true
   parent=HorizontalBox_1
   widget=TextBlock_45
   properties {
    text=Text Block
    font {
     bForceMonospaced=false
     bMaterialIsStencil=false
     fontMaterial=null
     fontObject=&BurbankSmall
     letterSpacing=0
     monospacedWidth=1
     size=20
     skewAmount=0
     typefaceFontName=Bold
     outlineSettings {
      bApplyOutlineToDropShadows=false
      bMiteredCorners=false
      bSeparateFillAlpha=false
      outlineMaterial=null
      outlineSize=0
      outlineColor=<rgba 0 0 0 1>
     }
    }
   }
   slot {class=HorizontalBoxSlot}
  }
  HorizontalBox_0 {
   widgetClassPath=&HorizontalBox
   parent=VerticalBox_1
   widget=HorizontalBox_0
   slot {class=VerticalBoxSlot, properties {horizontalAlignment=HAlign_Center}}
  }
  CommonActionWidget_48 {
   widgetClassPath=&CommonActionWidget
   bIsVariable=true
   parent=HorizontalBox_0
   widget=CommonActionWidget_48
   properties {inputActions [{dataTable=&InputActionDataTable, rowName=MainMenu}]}
   slot {class=HorizontalBoxSlot, properties {padding=<ltrb 5 5 5 5>}}
  }
  TextBlock_42 {
   widgetClassPath=&TextBlock
   bIsVariable=true
   parent=HorizontalBox_0
   widget=TextBlock_42
   properties {
    justification=Center
    text="OPEN  FOR MORE INFO"
    wrapTextAt=176
    colorAndOpacity {colorUseRule=UseColor_Specified, specifiedColor=<rgba 1 1 1 0.5>}
    font {
     bForceMonospaced=false
     bMaterialIsStencil=false
     fontMaterial=null
     fontObject=&BurbankSmall
     letterSpacing=0
     monospacedWidth=1
     size=15
     skewAmount=0
     typefaceFontName=Bold
     outlineSettings {
      bApplyOutlineToDropShadows=false
      bMiteredCorners=false
      bSeparateFillAlpha=false
      outlineMaterial=null
      outlineSize=0
      outlineColor=<rgba 0 0 0 1>
     }
    }
   }
   slot {class=HorizontalBoxSlot, properties {horizontalAlignment=HAlign_Center, verticalAlignment=VAlign_Center}}
  }
  InputActionWidget {widgetClassPath=&CommonActionWidget, bInherited=true, widget=null}
 ]
}
