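# Widget Blueprint for the Showdown tournament lobby panel, showing an event's title, pin score progress, highest score, timing and warnings.
# It starts collapsed; Refresh fills it from tournament data, reveals it for valid events and switches warnings for scheduled matchmaking, 2FA, match cap or crossplay.
# EventColorize applies tournament colors, while BrandingToggle swaps in a logo layout for World Cup qualifiers.
# Construct plays a material-parameter reveal intro, then triggers the pin animations.
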WidgetBlueprint /FortniteCoreUI/Frontend/Showdown/ShowdownLobbyPanel
animations [
 intro {
  binding_count=5
  display_label=""
  length_seconds=1.853
  object=&self:intro
  sequence {bindings [
    Image_scorezag2 {
     id=#1
     possessed_class=Image
     tracks [{
       type=MovieSceneWidgetMaterialTrack
       sections [{
         type=MovieSceneParameterSection
         end_frame=37
         start_frame=0
         channels [RevealRadius {type=MovieSceneScriptingFloatChannel, keys [[0,0],[3.027,0],[22.6613,3]]}]
        }]
      }]
    }
    Image_scorezag {
     id=#2
     possessed_class=Image
     tracks [{
       type=MovieSceneWidgetMaterialTrack
       sections [{
         type=MovieSceneParameterSection
         end_frame=37
         start_frame=0
         channels [RevealRadius {type=MovieSceneScriptingFloatChannel, keys [[0,0],[3.027,0],[22.6613,3]]}]
        }]
      }]
    }
    Image_Zag2 {
     id=#3
     possessed_class=Image
     tracks [{
       type=MovieSceneWidgetMaterialTrack
       sections [{
         type=MovieSceneParameterSection
         end_frame=37
         start_frame=0
         channels [RevealRadius {type=MovieSceneScriptingFloatChannel, keys [[0,0],[3.027,0],[22.6613,3]]}]
        }]
      }]
    }
    Image_Zag {
     id=#4
     possessed_class=Image
     tracks [{
       type=MovieSceneWidgetMaterialTrack
       sections [{
         type=MovieSceneParameterSection
         end_frame=37
         start_frame=0
         channels [RevealRadius {type=MovieSceneScriptingFloatChannel, keys [[0,0],[3.027,0],[22.6613,3]]}]
        }]
      }]
    }
    SLP_Size_Box {
     id=#5
     possessed_class=SizeBox
     tracks [
      {
       type=MovieScene2DTransformTrack
       sections [{
         type=MovieScene2DTransformSection
         end_frame=36
         start_frame=0
         channels [
          Translation.X {type=MovieSceneScriptingFloatChannel}
          Translation.Y {type=MovieSceneScriptingFloatChannel}
          Angle {type=MovieSceneScriptingFloatChannel}
          Scale.X {type=MovieSceneScriptingFloatChannel, keys [[0,1.1],[7.0467,0.9944],[8.7,1.0041],[9.6777,1]]}
          Scale.Y {type=MovieSceneScriptingFloatChannel, keys [[0,1.1],[7.0467,0.9944],[8.7,1.0041],[9.6777,1]]}
          Shear.X {type=MovieSceneScriptingFloatChannel}
          Shear.Y {type=MovieSceneScriptingFloatChannel}
         ]
        }]
      }
      {
       type=MovieSceneFloatTrack
       display_name=Render Opacity
       sections [{type=MovieSceneFloatSection, end_frame=36, start_frame=0, channels [None {type=MovieSceneScriptingFloatChannel, keys [[0,0],[9.6777,1]]}]}]
      }
     ]
    }
   ]}
 }
 EventButtonHover {
  binding_count=1
  display_label=""
  length_seconds=0.7595
  object=&self:EventButtonHover
  sequence {bindings [Image_DetailsButtonBG {
     id=#6
     possessed_class=Image
     tracks [{
       type=MovieSceneWidgetMaterialTrack
       sections [{
         type=MovieSceneParameterSection
         end_frame=14
         start_frame=-1
         channels [Hover {type=MovieSceneScriptingFloatChannel, keys [[-0.07,0],[14.8303,1]]}]
        }]
      }]
    }]}
 }
]
graphs [
 EventGraph {nodes [
   "On Button Hovered (Button_EventDetails)"!() -> (OutputDelegate, then, Button:Common Button Legacy Object Reference) {is=title K2Node_ComponentBoundEvent}
   n1 = Self-Reference()
   n2 = Get EventButtonHover() -> (EventButtonHover:Widget Animation Object Reference)
   > Anim Transition!(Exec:Exec, In:bool=true, Animation:Widget Animation Object Reference=@n2, PlaySpeed:double=2, Widget:User Widget Object Reference=@n1) -> (ExecOut:Exec) {is=title K2Node_MacroInstance}
   "On Button Unhovered (Button_EventDetails)"!() -> (OutputDelegate, then, Button:Common Button Legacy Object Reference) {is=title K2Node_ComponentBoundEvent}
   > Anim Transition(Animation=@n2, PlaySpeed=1, Widget=@n1)
   EventColorize() {is=title K2Node_CustomEvent}
   > n3 = GetDynamicMaterial(self=Get Image_DetailsButtonBG() -> (Image_DetailsButtonBG:Image Object Reference))
   n4 = Break Fort Tournament Display Info!(FortTournamentDisplayInfo:"Fort Tournament Display Info Structure (by ref)"=Get TournamentDisplayInfo() -> (TournamentDisplayInfo:Fort Tournament Display Info Structure)) -> (
    TitleLine1:Text
    TitleLine2:Text
    ScheduleInfo:Text
    FlavorDescription:Text
    DetailsDescription:Text
    ShortFormatTitle:Text
    LongFormatTitle:Text
    BackgroundTitle:Text
    PinScoreRequirement:int
    PinEarnedText:Text
    BaseColor:Linear Color Structure
    PrimaryColor:Linear Color Structure
    SecondaryColor:Linear Color Structure
    HighlightColor:Linear Color Structure
    TitleColor:Linear Color Structure
    ShadowColor:Linear Color Structure
    BackgroundLeftColor:Linear Color Structure
    BackgroundRightColor:Linear Color Structure
    BackgroundTextColor:Linear Color Structure
    PosterFadeColor:Linear Color Structure
    AlertText:Text
    SeriesPointLeaderboardName:Text
    AlertType:EFortTournamentAlertType Enum
    RoundDisplayNames:Array of Texts
   ) {is=title K2Node_BreakStruct}
   > SetVectorParameterValue(self=@n3, ParameterName=Color1, Value=@n4.PrimaryColor)
   > SetVectorParameterValue(self=@n3, ParameterName=Color2, Value=@n4.PrimaryColor)
   > SetVectorParameterValue(self=@n3, ParameterName=HoverColor, Value=(R=1,G=1,B=1,A=1))
   > SetBrushColor(self=Get Border_TimeRemaining() -> (Border_TimeRemaining:Border Object Reference), InBrushColor=@n4.ShadowColor)
   > SetColorAndOpacity(
    self=Get CommonTextBlock_EarnAPin() -> (CommonTextBlock_EarnAPin:Common Text Object Reference)
    InColorAndOpacity=Make Slate Color(SpecifiedColor=@n4.HighlightColor)
   )
   n5 = BrandingToggle!() -> (OutputDelegate, then, hasBranding:bool=~) {is=title K2Node_CustomEvent}
   > SetActiveWidget(
    self:Widget Switcher Object Reference=Get CommonWidgetSwitcher_Score() -> (CommonWidgetSwitcher_Score:Common Widget Switcher Legacy Object Reference)
    Widget=Select(
     Option 0:Widget Object Reference=Get Overlay_ScoreInfo() -> (Overlay_ScoreInfo:Overlay Object Reference)
     Option 1:Widget Object Reference=Get TournamentLogo() -> (TournamentLogo:Fort Lazy Image Object Reference)
     Index=@n5.hasBranding
    ) -> (ReturnValue:Widget Object Reference)
   )
   > Set Padding!(
    execute
    Padding:Margin Structure=Make Margin(Left=0, Top=Select(Option 0:double=-40, Option 1:double=0, Index=@n5.hasBranding) -> (ReturnValue:double), Right=0, Bottom=0)
    self:Grid Slot Object Reference=SlotAsGridSlot(Widget=Get GP_PanelContent() -> (GP_PanelContent:Grid Panel Object Reference))
   ) -> (then, Output_Get:Margin Structure)
   > SetVisibility(
    self=Get HB_SpecialScore() -> (HB_SpecialScore:Horizontal Box Object Reference)
    InVisibility=Select(Option 0:ESlateVisibility Enum=Collapsed, Option 1:ESlateVisibility Enum=SelfHitTestInvisible, Index=@n5.hasBranding) -> (ReturnValue:ESlateVisibility Enum)
   )
   Event Construct()
   > SetVisibility(InVisibility=Collapsed)
   > Bind Event to On Crossplay Preference Changed(
    self:Fort Party Context Object Reference=Get FortPartyContext!() -> (ReturnValue:Fort Party Context Object Reference) {is=title K2Node_GetBlueprintContextEx}
    Delegate=Create Event()
   ) {is=title K2Node_AddDelegate}
   > Bind Event to On Input Method Changed(
    self=Get CommonInputSubsystem()
    Delegate=EventSwitchUI!() -> (OutputDelegate, then, bNewInputType:ECommonInputType Enum) {is=title K2Node_CustomEvent}
   )
   > PlayAnimationForward(InAnimation=Get intro() -> (intro:Widget Animation Object Reference))
   > Delay(Duration=0.75)
   > EventRevealAnim(self:Showdown Pin Object Reference=[
     Get ShowdownPin() -> (ShowdownPin:Showdown Pin Object Reference)
     Get ShowdownPin_SpecialScore() -> (ShowdownPin_SpecialScore:Showdown Pin Object Reference)
    ])
   n6 = Event PreConstruct()
   > Branch(Condition=@n6.IsDesignTime)
   > BrandingToggle!(execute, self, hasBranding:bool=Get DEVhasBranding() -> (DEVhasBranding:bool)) -> (then) {is=title K2Node_CallFunction}
   Event Refresh Data() {is=title K2Node_Event}
   > Refresh()
  ]}
 HandleCrossplayChanged {nodes [
   HandleCrossplayChanged() {is=title K2Node_FunctionEntry}
   > Refresh()
  ]}
 InputTypeChanged {nodes [InputTypeChanged!() -> (then, isGamepad:bool=~) {is=title K2Node_FunctionEntry}]}
 Refresh {nodes [
   Refresh() {is=title K2Node_FunctionEntry}
   > SetActiveWidget(
    self=Get Switcher_TimeInfo() -> (Switcher_TimeInfo:Common Widget Switcher Legacy Object Reference)
    Widget=Select(
     Option 0:Widget Object Reference=Get VB_Content_TimeRemaining() -> (VB_Content_TimeRemaining:Vertical Box Object Reference)
     Option 1:Widget Object Reference=Get VB_Content_ScheduledTime() -> (VB_Content_ScheduledTime:Vertical Box Object Reference)
     Index=UsesScheduledMatchmaking!(self:Fort Tournament Lobby Panel Object Reference) -> (ReturnValue:bool)
    ) -> (ReturnValue:Widget Object Reference)
   )
   > "SetText (Text)"(self=Get Text_DetailsButton() -> (Text_DetailsButton:Common Text Object Reference), InText="LOCTEXT(\"DETAILS\")")
   > SetVisibility(self=Get Border_TimeRemaining(), InVisibility=SelfHitTestInvisible)
   n1 = Break Showdown Tournament Entry!(ShowdownTournamentEntry:"Showdown Tournament Entry Structure (by ref)"=Get CMSTournamentData() -> (CMSTournamentData:Showdown Tournament Entry Structure)) -> (
    TournamentDisplayId:String
    TitleLine1:String
    TitleLine2:String
    ScheduleInfo:String
    PosterFrontImage:String
    PosterBackImage:String
    FlavorDescription:String
    DetailsDescription:String
    ShortFormatTitle:String
    LongFormatTitle:String
    BackgroundTitle:String
    PinScoreRequirement:int
    PinEarnedText:String
    BaseColor:String
    PrimaryColor:String
    SecondaryColor:String
    HighlightColor:String
    TitleColor:String
    ShadowColor:String
    BackgroundLeftColor:String
    BackgroundRightColor:String
    BackgroundTextColor:String
    PosterFadeColor:String
    PlaylistTileImage:String
    LoadingScreenImage:String
    TournamentViewBackgroundImage:String
    TournamentLandingPageImage:String
    SquarePosterImage:String
    StyleInfoId:String
    AlertText:String
    SeriesPointLeaderboardName:String
    PlaylistDescription:String
    AlertType:EFortTournamentAlertType Enum
    RoundDisplayNames:Array of Texts
   ) {is=title K2Node_BreakStruct}
   > "SetText (Text)"(
    self=Get TournamentTitle() -> (TournamentTitle:Common Text Object Reference)
    InText=TextToUpper!(InText:"Text (by ref)"="To Text (String)"(InString=@n1.TitleLine1)) -> (ReturnValue:Text)
   )
   > "SetText (Text)"(
    self=Get TournamentTitle2() -> (TournamentTitle2:Common Text Object Reference)
    InText=TextToUpper(InText="To Text (String)"(InString=@n1.TitleLine2))
   )
   n2 = GetEventId!(self:Fort Showdown Detail View Object Reference) -> (ReturnValue:String)
   > SetTournament!(
    execute
    self:Fort Showdown Detail View Object Reference=[Get ShowdownPin(), Get ShowdownPin_SpecialScore()]
    TournamentId:String=GetTournamentSeriesId!(self:Fort Showdown Detail View Object Reference) -> (ReturnValue:String)
    EventId:String=@n2
    bForceRefreshUI:bool=true
   ) -> (then)
   n3 = GetBestEventScore!(self:Fort Showdown Detail View Object Reference) -> (ReturnValue:int=0)
   > "SetText (Text)"(
    self=Get HighestScoreText() -> (HighestScoreText:Common Text Object Reference)
    InText=TextToUpper(InText=Format Text!(Format="LOCTEXT(\"Highest Score: {Score}\")", Score:int=@n3) -> (Result))
   )
   n4 = GetCurrentEventScore!(self:Fort Showdown Detail View Object Reference) -> (ReturnValue:int=0)
   > SetVisibility(
    self=[Get VB_HighestScore() -> (VB_HighestScore:Vertical Box Object Reference), Get Bg_HighestScore() -> (Bg_HighestScore:Image Object Reference)]
    InVisibility=Select(
     Option 0:ESlateVisibility Enum=Collapsed
     Option 1:ESlateVisibility Enum=HitTestInvisible
     Index="integer > integer"(A=@n3, B=@n4) {is=title K2Node_CallFunction}
    ) -> (ReturnValue:ESlateVisibility Enum)
   )
   > "SetText (Text)"(
    self=Get EarnPinScoreText() -> (EarnPinScoreText:Common Text Object Reference)
    InText="To Text (Integer)"(Value=GetTournamentPinUnlockScore!(self:Fort Showdown Detail View Object Reference) -> (ReturnValue:int=0))
   )
   n5 = GetTournamentPinUnlockScore() -> (ReturnValue=0)
   > SetVisibility(
    self=Get EarnPinContent() -> (EarnPinContent:Vertical Box Object Reference)
    InVisibility=Select(Option 0=Collapsed, Option 1=SelfHitTestInvisible, Index="integer > integer"(A=@n5, B=0))
   )
   > SetVisibility(self=Get CentralContentSwitcher() -> (CentralContentSwitcher:Common Widget Switcher Legacy Object Reference), InVisibility=HitTestInvisible)
   n6 = GetTournamentPinUnlockScore() -> (ReturnValue=0)
   n7 = "integer >= integer"!(A:int=@n6, B:int=GetCurrentEventScore() -> (ReturnValue=0)) -> (ReturnValue:bool) {is=title K2Node_CallFunction}
   n8 = "integer >= integer"(A=@n6, B=0)
   > SetActiveWidget(
    self=Get CentralContentSwitcher()
    Widget=Select(
     Option 0:Widget Object Reference=Get NoMessage() -> (NoMessage:Overlay Object Reference)
     Option 1:Widget Object Reference=Get EarnPinMessage() -> (EarnPinMessage:Overlay Object Reference)
     Index=AND Boolean(A=@n8, B=@n7)
    ) -> (ReturnValue:Widget Object Reference)
   )
   > "SetText (Text)"(self=Get HighestScoreTeammatesText() -> (HighestScoreTeammatesText:Common Text Object Reference))
   n9 = "GetText (Text)"!(self:Text Object Reference=Get HighestScoreTeammatesText()) -> (ReturnValue:Text)
   > SetVisibility(
    self=Get HighestScoreTeammatesText()
    InVisibility=Select(Option 0:ESlateVisibility Enum=SelfHitTestInvisible, Option 1:ESlateVisibility Enum=Collapsed, Index=TextIsEmpty(InText=@n9)) -> (ReturnValue:ESlateVisibility Enum)
   )
   > EventColorize() {is=title K2Node_CallFunction}
   n10 = MakeLiteralText(Value="LOCTEXT(\"UNTIL START\")")
   n11 = MakeLiteralText(Value="LOCTEXT(\"REMAINING\")")
   > "SetText (Text)"(
    self=Get Text_Remaining() -> (Text_Remaining:Common Text Object Reference)
    InText=Select!(
     Unknown:Text
     FutureTBD:Text=@n10
     FutureScheduled:Text=@n10
     FutureNext:Text=@n10
     Live:Text=@n11
     LiveParticipating:Text
     LiveNotParticipating:Text
     Completed:Text=@n11
     CompletedParticipated:Text
     CompletedNotPartipated:Text
     Cancelled:Text
     Index:EFortShowdownEventState Enum=GetEventTimeState!(self:Fort Showdown Detail View Object Reference) -> (ReturnValue:EFortShowdownEventState Enum=Unknown)
    ) -> (ReturnValue:Text)
   )
   > BrandingToggle(hasBranding=IsWorldCupQualifierEvent!(self:Fort Showdown Detail View Object Reference) -> (ReturnValue:bool))
   n12 = ShouldDisplayDetailView!(self:Fort Showdown Detail View Object Reference) -> (ReturnValue:bool)
   > n13 = Branch(Condition=OR Boolean(A="Equal Exactly (String)"(A=@n2), B=NOT Boolean(A=@n12)))
   SetVisibility(execute=@n13.else, InVisibility=SelfHitTestInvisible)
   n14 = DoCalloutMFARequirement!(self:Fort Showdown Detail View Object Reference) -> (ReturnValue:bool)
   n15 = GetTournamentMatchCap!(self:Fort Showdown Detail View Object Reference) -> (ReturnValue:int=0)
   n16 = AND Boolean(
    A="integer >= integer"(
     A=GetCurrentEventTeamNumMatchesPlayed!(self:Fort Showdown Detail View Object Reference) -> (ReturnValue:int=0)
     B=int - int(A=@n15, B=5) {is=title K2Node_CallFunction}
    )
    B="integer > integer"(A=GetTournamentMatchCap() -> (ReturnValue=0), B=0)
   )
   > Branch(Condition=OR Boolean(A=OR Boolean(A=@n14, B=@n16), B=UsesScheduledMatchmaking()))
   > SetActiveWidget(self=Get CentralContentSwitcher(), Widget=Get WarningMessage() -> (WarningMessage:Overlay Object Reference))
   > n17 = Branch(Condition=UsesScheduledMatchmaking())
   > SetActiveWidget(
    self=Get CommonWidgetSwitcher_TournamentWarnings() -> (CommonWidgetSwitcher_TournamentWarnings:Common Widget Switcher Legacy Object Reference)
    Widget=Get Text_ScheduledMMInfo() -> (Text_ScheduledMMInfo:Common Text Object Reference)
   )
   n18 = Branch(execute=@n17.else, Condition=DoCalloutMFARequirement())
   > SetActiveWidget(self=Get CommonWidgetSwitcher_TournamentWarnings(), Widget=Get MFA_WarningText() -> (MFA_WarningText:Common Text Object Reference))
   SetActiveWidget(
    execute=@n18.else
    self=Get CommonWidgetSwitcher_TournamentWarnings()
    Widget=Get Text_MatchCapWarning() -> (Text_MatchCapWarning:Common Text Object Reference)
   )
   n19 = GetTournamentMatchCap() -> (ReturnValue=0)
   n20 = GetCurrentEventTeamNumMatchesPlayed() -> (ReturnValue=0)
   > "SetText (Text)"(
    self=Get Text_MatchCapWarning()
    InText=Format Text!(Format="LOCTEXT(\"{NumMatches} {NumMatches}|plural(one=match,other=matches) remaining\")", NumMatches:int=int - int(A=@n19, B=@n20)) -> (Result)
   )
   n21 = Branch(Condition=ShouldShowCrossplayError!(self:Fort Showdown Detail View Object Reference) -> (ReturnValue:bool))
   > SetActiveWidget(self=Get CentralContentSwitcher(), Widget=Get WarningMessage())
   > SetActiveWidget(
    self=Get CommonWidgetSwitcher_TournamentWarnings()
    Widget=Get Text_CrossplayWarning() -> (Text_CrossplayWarning:Common Text Object Reference)
   )
   Reroute Node(InputPin=@n21.else)
   "SetText (Text)"(InText="LOCTEXT(\"REMAINING\")")
   n22 = GetTournamentPinUnlockScore() -> (ReturnValue=0)
   AND Boolean(
    A="integer >= integer"(A=@n22, B=0)
    B="integer <= integer"!(A:int=@n22, B:int=GetCurrentEventScore() -> (ReturnValue=0)) -> (ReturnValue:bool) {is=title K2Node_CallFunction}
   )
  ]}
 UpdateColors {nodes [UpdateColors() {is=title K2Node_FunctionEntry}]}
]
variables [
 DEVhasBranding {type=boolean}
 HighlightColor {type=SlateColor, default {colorUseRule=UseColor_Specified, specifiedColor=<rgba 1 1 1 1>}}
 PrimaryColor {type=SlateColor, default {colorUseRule=UseColor_Specified, specifiedColor=<rgba 1 1 1 1>}}
 ShadowColor {type=SlateColor, default {colorUseRule=UseColor_Specified, specifiedColor=<rgba 1 1 1 1>}}
]
widgets {
 info {inheritedWidgetCount=10, parentClass=/Script/FortniteUI.FortTournamentLobbyPanel, rootWidgetClass=&SizeBox, widgetCount=61}
 widgets [
  SLP_Size_Box {widgetClassPath=&SizeBox, widget=SLP_Size_Box, properties {bOverride_WidthOverride=true, widthOverride=368}}
  SLP_Grid_Panel {
   widgetClassPath=&GridPanel
   parent=SLP_Size_Box
   widget=SLP_Grid_Panel
   properties {columnFill [1], rowFill [1]}
   slot {class=SizeBoxSlot}
  }
  UIKit_Backplate {
   widgetClassPath=&WBP_UIKit_Backplate
   parent=SLP_Grid_Panel
   widget=UIKit_Backplate
   properties {brightness=Dark, cornerRadius=Interactive Component}
   slot {class=GridSlot}
  }
  GP_Content {
   widgetClassPath=&GridPanel
   parent=SLP_Grid_Panel
   widget=GP_Content
   properties {columnFill [1], rowFill [0,0]}
   slot {class=GridSlot, properties {layer=10}}
  }
  CommonWidgetSwitcher_Score {
   widgetClassPath=&CommonWidgetSwitcherLegacy
   bIsVariable=true
   parent=GP_Content
   widget=CommonWidgetSwitcher_Score
   slot {class=GridSlot, properties {horizontalAlignment=HAlign_Center, verticalAlignment=VAlign_Top, padding=<ltrb 0 -48 0 0>}}
  }
  Overlay_ScoreInfo {
   widgetClassPath=&Overlay
   bIsVariable=true
   parent=CommonWidgetSwitcher_Score
   widget=Overlay_ScoreInfo
   slot {class=WidgetSwitcherSlot}
  }
  ShowdownPin {
   widgetClassPath=&/ShowdownPin
   bIsVariable=true
   parent=Overlay_ScoreInfo
   widget=ShowdownPin
   properties {bTryUsingAlternativeScore=true, colorSilhouette=true, dEVPinScore=0}
   slot {class=OverlaySlot, properties {horizontalAlignment=HAlign_Fill, verticalAlignment=VAlign_Fill}}
  }
  Text_DebugLBId {
   widgetClassPath=&TextBlock
   bInherited=true
   parent=Overlay_ScoreInfo
   widget=Text_DebugLBId
   properties {
    autoWrapText=true
    justification=Center
    text="[DEBUG] LB Key:<None>"
    visibility=Collapsed
    wrapTextAt=500
    wrappingPolicy=AllowPerCharacterWrapping
    font {
     bForceMonospaced=false
     bMaterialIsStencil=false
     fontMaterial=null
     fontObject=&Roboto
     letterSpacing=0
     monospacedWidth=1
     size=22
     skewAmount=0
     typefaceFontName=Bold
     outlineSettings {
      bApplyOutlineToDropShadows=false
      bMiteredCorners=false
      bSeparateFillAlpha=false
      outlineMaterial=null
      outlineSize=3
      outlineColor=<rgba 0 0 0 1>
     }
    }
   }
   slot {class=OverlaySlot, properties {horizontalAlignment=HAlign_Fill, verticalAlignment=VAlign_Center}}
  }
  TournamentLogo {
   widgetClassPath=&FortLazyImage
   bIsVariable=true
   parent=CommonWidgetSwitcher_Score
   widget=TournamentLogo
   properties {
    brush {
     bIsDynamicallyLoaded=false
     drawAs=Image
     imageType=NoImage
     mirroring=NoMirror
     resourceName=None
     resourceObject=/FortniteCoreUI/Frontend/Showdown/TournamentAssets/FWC_OnlineOpen_EN
     tiling=NoTile
     imageSize=<300 300>
     margin=<ltrb 0 0 0 0>
     outlineSettings {
      bUseBrushTransparency=false
      roundingType=HalfHeightRadius
      width=0
      color {colorUseRule=UseColor_Specified, specifiedColor=<rgba 0 0 0 0>}
      cornerRadii=<0 0 0 0>
     }
     tintColor {colorUseRule=UseColor_Specified, specifiedColor=<rgba 1 1 1 1>}
     uVRegion {bIsValid=false, max=<0 0>, min=<0 0>}
    }
    loadingBackgroundBrush {
     bIsDynamicallyLoaded=false
     drawAs=NoDrawType
     imageType=NoImage
     mirroring=NoMirror
     resourceName=None
     resourceObject=null
     tiling=NoTile
     imageSize=<512 512>
     margin=<ltrb 0 0 0 0>
     outlineSettings {
      bUseBrushTransparency=false
      roundingType=HalfHeightRadius
      width=0
      color {colorUseRule=UseColor_Specified, specifiedColor=<rgba 0 0 0 0>}
      cornerRadii=<0 0 0 0>
     }
     tintColor {colorUseRule=UseColor_Specified, specifiedColor=<rgba 1 1 1 1>}
     uVRegion {bIsValid=false, max=<0 0>, min=<0 0>}
    }
   }
   slot {class=WidgetSwitcherSlot}
  }
  GP_PanelContent {
   widgetClassPath=&GridPanel
   bIsVariable=true
   parent=GP_Content
   widget=GP_PanelContent
   properties {columnFill [1], rowFill [0,0,0,0,0]}
   slot {class=GridSlot, properties {layer=1, row=1, padding=<ltrb 0 -40 0 0>}}
  }
  VerticalBox_3 {
   widgetClassPath=&VerticalBox
   parent=GP_PanelContent
   widget=VerticalBox_3
   slot {class=GridSlot, properties {verticalAlignment=VAlign_Top, padding=<ltrb 24 0 24 0>}}
  }
  TournamentTitle {
   widgetClassPath=&CommonTextBlock
   bIsVariable=true
   parent=VerticalBox_3
   widget=TournamentTitle
   properties {
    justification=Center
    lineHeightPercentage=0.9
    scrollStyle=&TextScrollStyle-Base_C
    text=SOLO
    font {
     bForceMonospaced=false
     bMaterialIsStencil=false
     fontMaterial=null
     fontObject=&Burbank
     letterSpacing=0
     monospacedWidth=1
     size=48
     skewAmount=0
     typefaceFontName=Burbank-Black
     outlineSettings {
      bApplyOutlineToDropShadows=false
      bMiteredCorners=false
      bSeparateFillAlpha=false
      outlineMaterial=null
      outlineSize=0
      outlineColor=<rgba 0 0 0 1>
     }
    }
    renderTransformPivot=<0 0.5>
   }
   slot {class=VerticalBoxSlot, properties {verticalAlignment=VAlign_Top}}
  }
  TournamentTitle2 {
   widgetClassPath=&CommonTextBlock
   bIsVariable=true
   parent=VerticalBox_3
   widget=TournamentTitle2
   properties {
    justification=Center
    lineHeightPercentage=0.9
    scrollStyle=&TextScrollStyle-Base_C
    text=WEEK 1
    font {
     bForceMonospaced=false
     bMaterialIsStencil=false
     fontMaterial=null
     fontObject=&Burbank
     letterSpacing=0
     monospacedWidth=1
     size=48
     skewAmount=0
     typefaceFontName=Burbank-Black
     outlineSettings {
      bApplyOutlineToDropShadows=false
      bMiteredCorners=false
      bSeparateFillAlpha=false
      outlineMaterial=null
      outlineSize=0
      outlineColor=<rgba 0 0 0 1>
     }
    }
    renderTransformPivot=<0 0.5>
   }
   slot {class=VerticalBoxSlot, properties {verticalAlignment=VAlign_Top}}
  }
  HB_SpecialScore {
   widgetClassPath=&HorizontalBox
   bIsVariable=true
   parent=GP_PanelContent
   widget=HB_SpecialScore
   slot {class=GridSlot, properties {row=1, padding=<ltrb 0 30 0 10>}}
  }
  Image_scorezag {
   widgetClassPath=&Image
   parent=HB_SpecialScore
   widget=Image_scorezag
   properties {
    brush {
     bIsDynamicallyLoaded=false
     drawAs=Box
     imageType=NoImage
     mirroring=NoMirror
     resourceName=None
     resourceObject=/FortniteCoreUI/Frontend/Showdown/Materials/M_UI_HZag_Inst
     tiling=NoTile
     imageSize=<128 12>
     margin=<ltrb 0 0 0.4771 0>
     outlineSettings {
      bUseBrushTransparency=false
      roundingType=HalfHeightRadius
      width=0
      color {colorUseRule=UseColor_Specified, specifiedColor=<rgba 0 0 0 0>}
      cornerRadii=<0 0 0 0>
     }
     tintColor {colorUseRule=UseColor_Specified, specifiedColor=<rgba 1 1 1 1>}
     uVRegion {bIsValid=false, max=<0 0>, min=<0 0>}
    }
    renderTransform {angle=0, scale=<-1 1>, shear=<0 0>, translation=<0 0>}
   }
   slot {class=HorizontalBoxSlot, properties {verticalAlignment=VAlign_Center, padding=<ltrb 0 -12 0 0>, size {sizeRule=Fill, value=1}}}
  }
  ShowdownPin_SpecialScore {
   widgetClassPath=&ShowdownPin
   bIsVariable=true
   parent=HB_SpecialScore
   widget=ShowdownPin_SpecialScore
   properties {colorSilhouette=true, dEVPinScore=19, hideBackplate=true}
   slot {class=HorizontalBoxSlot, properties {horizontalAlignment=HAlign_Center, padding=<ltrb 16 0 16 0>}}
  }
  Image_scorezag2 {
   widgetClassPath=&Image
   parent=HB_SpecialScore
   widget=Image_scorezag2
   properties {brush {
     bIsDynamicallyLoaded=false
     drawAs=Box
     imageType=NoImage
     mirroring=NoMirror
     resourceName=None
     resourceObject=&M_UI_HZag_Inst
     tiling=NoTile
     imageSize=<128 12>
     margin=<ltrb 0 0 1 0>
     outlineSettings {
      bUseBrushTransparency=false
      roundingType=HalfHeightRadius
      width=0
      color {colorUseRule=UseColor_Specified, specifiedColor=<rgba 0 0 0 0>}
      cornerRadii=<0 0 0 0>
     }
     tintColor {colorUseRule=UseColor_Specified, specifiedColor=<rgba 1 1 1 1>}
     uVRegion {bIsValid=false, max=<0 0>, min=<0 0>}
    }}
   slot {class=HorizontalBoxSlot, properties {verticalAlignment=VAlign_Center, padding=<ltrb 0 -12 0 0>, size {sizeRule=Fill, value=1}}}
  }
  CentralContentSwitcher {
   widgetClassPath=&CommonWidgetSwitcherLegacy
   bIsVariable=true
   parent=GP_PanelContent
   widget=CentralContentSwitcher
   properties {visibility=Collapsed}
   slot {class=GridSlot, properties {row=2, verticalAlignment=VAlign_Top, padding=<ltrb 0 16 0 0>}}
  }
  EarnPinMessage {widgetClassPath=&Overlay, bIsVariable=true, parent=CentralContentSwitcher, widget=EarnPinMessage, slot {class=WidgetSwitcherSlot}}
  EarnPinContent {
   widgetClassPath=&VerticalBox
   bIsVariable=true
   parent=EarnPinMessage
   widget=EarnPinContent
   slot {class=OverlaySlot, properties {horizontalAlignment=HAlign_Fill, verticalAlignment=VAlign_Fill}}
  }
  HorizontalBox_4 {widgetClassPath=&HorizontalBox, parent=EarnPinContent, widget=HorizontalBox_4, slot {class=VerticalBoxSlot}}
  Image_Zag {
   widgetClassPath=&Image
   parent=HorizontalBox_4
   widget=Image_Zag
   properties {
    brush {
     bIsDynamicallyLoaded=false
     drawAs=Box
     imageType=NoImage
     mirroring=NoMirror
     resourceName=None
     resourceObject=&M_UI_HZag_Inst
     tiling=NoTile
     imageSize=<128 12>
     margin=<ltrb 0 0 0.4771 0>
     outlineSettings {
      bUseBrushTransparency=false
      roundingType=HalfHeightRadius
      width=0
      color {colorUseRule=UseColor_Specified, specifiedColor=<rgba 0 0 0 0>}
      cornerRadii=<0 0 0 0>
     }
     tintColor {colorUseRule=UseColor_Specified, specifiedColor=<rgba 1 1 1 1>}
     uVRegion {bIsValid=false, max=<0 0>, min=<0 0>}
    }
    renderTransform {angle=0, scale=<-1 1>, shear=<0 0>, translation=<0 0>}
   }
   slot {class=HorizontalBoxSlot, properties {verticalAlignment=VAlign_Center, padding=<ltrb 0 -5 0 0>, size {sizeRule=Fill, value=1}}}
  }
  EarnPinScoreText {
   widgetClassPath=&CommonTextBlock
   bIsVariable=true
   parent=HorizontalBox_4
   widget=EarnPinScoreText
   properties {
    justification=Center
    text=40
    font {
     bForceMonospaced=false
     bMaterialIsStencil=false
     fontMaterial=null
     fontObject=&BurbankSmall
     letterSpacing=0
     monospacedWidth=1
     size=34
     skewAmount=0
     typefaceFontName=Black
     outlineSettings {
      bApplyOutlineToDropShadows=false
      bMiteredCorners=false
      bSeparateFillAlpha=false
      outlineMaterial=null
      outlineSize=0
      outlineColor=<rgba 0 0 0 1>
     }
    }
   }
   slot {class=HorizontalBoxSlot, properties {horizontalAlignment=HAlign_Center, padding=<ltrb 16 0 16 0>}}
  }
  Image_Zag2 {
   widgetClassPath=&Image
   parent=HorizontalBox_4
   widget=Image_Zag2
   properties {brush {
     bIsDynamicallyLoaded=false
     drawAs=Box
     imageType=NoImage
     mirroring=NoMirror
     resourceName=None
     resourceObject=&M_UI_HZag_Inst
     tiling=NoTile
     imageSize=<128 12>
     margin=<ltrb 0 0 1 0>
     outlineSettings {
      bUseBrushTransparency=false
      roundingType=HalfHeightRadius
      width=0
      color {colorUseRule=UseColor_Specified, specifiedColor=<rgba 0 0 0 0>}
      cornerRadii=<0 0 0 0>
     }
     tintColor {colorUseRule=UseColor_Specified, specifiedColor=<rgba 1 1 1 1>}
     uVRegion {bIsValid=false, max=<0 0>, min=<0 0>}
    }}
   slot {class=HorizontalBoxSlot, properties {verticalAlignment=VAlign_Center, padding=<ltrb 0 -5 0 0>, size {sizeRule=Fill, value=1}}}
  }
  CommonTextBlock_EarnAPin {
   widgetClassPath=&CommonTextBlock
   bIsVariable=true
   parent=EarnPinContent
   widget=CommonTextBlock_EarnAPin
   properties {
    autoWrapText=true
    justification=Center
    text=POINTS TO EARN A PIN
    colorAndOpacity {colorUseRule=UseColor_Specified, specifiedColor=<rgba 0 1 0.9823 1>}
    font {
     bForceMonospaced=false
     bMaterialIsStencil=false
     fontMaterial=null
     fontObject=&BurbankSmall
     letterSpacing=0
     monospacedWidth=1
     size=18
     skewAmount=0
     typefaceFontName=Black
     outlineSettings {
      bApplyOutlineToDropShadows=false
      bMiteredCorners=false
      bSeparateFillAlpha=false
      outlineMaterial=null
      outlineSize=0
      outlineColor=<rgba 0 0 0 1>
     }
    }
   }
   slot {class=VerticalBoxSlot, properties {padding=<ltrb 16 4 16 0>}}
  }
  WarningMessage {
   widgetClassPath=&Overlay
   bIsVariable=true
   parent=CentralContentSwitcher
   widget=WarningMessage
   slot {class=WidgetSwitcherSlot, properties {verticalAlignment=VAlign_Top, padding=<ltrb -60 0 -60 0>}}
  }
  Image_ErrorBG {
   widgetClassPath=&Image
   bIsVariable=true
   parent=WarningMessage
   widget=Image_ErrorBG
   properties {
    visibility=SelfHitTestInvisible
    brush {
     bIsDynamicallyLoaded=false
     drawAs=Image
     imageType=NoImage
     mirroring=NoMirror
     resourceName=None
     resourceObject=/FortniteCoreUI/Frontend/Showdown/Materials/M_UI_SkewBox_TournamentPanelErrorBG
     tiling=NoTile
     imageSize=<32 60.8663>
     margin=<ltrb 0 0 0 0>
     outlineSettings {
      bUseBrushTransparency=false
      roundingType=HalfHeightRadius
      width=0
      color {colorUseRule=UseColor_Specified, specifiedColor=<rgba 0 0 0 0>}
      cornerRadii=<0 0 0 0>
     }
     tintColor {colorUseRule=UseColor_Specified, specifiedColor=<rgba 1 1 1 1>}
     uVRegion {bIsValid=false, max=<0 0>, min=<0 0>}
    }
    colorAndOpacity=<rgba 0 1 0.9823 1>
    renderTransformPivot=<0 1>
   }
   slot {class=OverlaySlot, properties {horizontalAlignment=HAlign_Fill, verticalAlignment=VAlign_Fill, padding=<ltrb 0 -15 -15 -5>}}
  }
  CommonWidgetSwitcher_TournamentWarnings {
   widgetClassPath=&CommonWidgetSwitcherLegacy
   bIsVariable=true
   parent=WarningMessage
   widget=CommonWidgetSwitcher_TournamentWarnings
   slot {class=OverlaySlot, properties {horizontalAlignment=HAlign_Fill, padding=<ltrb 40 10 40 25>}}
  }
  Text_CrossplayWarning {
   widgetClassPath=&CommonTextBlock
   bIsVariable=true
   parent=CommonWidgetSwitcher_TournamentWarnings
   widget=Text_CrossplayWarning
   properties {
    autoWrapText=true
    justification=Center
    lineHeightPercentage=0.9
    text=You must opt-in for crossplay to participate in competitive events
    wrapTextAt=366
    colorAndOpacity {colorUseRule=UseColor_Specified, specifiedColor=<rgba 0.0044 0.0091 0.0497 1>}
    font {
     bForceMonospaced=false
     bMaterialIsStencil=false
     fontMaterial=null
     fontObject=&Burbank
     letterSpacing=0
     monospacedWidth=1
     size=22
     skewAmount=0
     typefaceFontName=BurbankRegular-Black
     outlineSettings {
      bApplyOutlineToDropShadows=false
      bMiteredCorners=false
      bSeparateFillAlpha=false
      outlineMaterial=null
      outlineSize=0
      outlineColor=<rgba 0 0 0 1>
     }
    }
   }
   slot {class=WidgetSwitcherSlot}
  }
  CommonTextBlock_0 {
   widgetClassPath=&CommonTextBlock
   parent=CommonWidgetSwitcher_TournamentWarnings
   widget=CommonTextBlock_0
   properties {
    autoWrapText=true
    justification=Center
    lineHeightPercentage=0.9
    text="You need more \r\nplayers to compete"
    wrapTextAt=366
    colorAndOpacity {colorUseRule=UseColor_Specified, specifiedColor=<rgba 0.0044 0.0091 0.0497 1>}
    font {
     bForceMonospaced=false
     bMaterialIsStencil=false
     fontMaterial=null
     fontObject=&Burbank
     letterSpacing=0
     monospacedWidth=1
     size=22
     skewAmount=0
     typefaceFontName=BurbankRegular-Black
     outlineSettings {
      bApplyOutlineToDropShadows=false
      bMiteredCorners=false
      bSeparateFillAlpha=false
      outlineMaterial=null
      outlineSize=0
      outlineColor=<rgba 0 0 0 1>
     }
    }
    renderTransform {angle=0, scale=<1 1>, shear=<-7 0>, translation=<0 0>}
   }
   slot {class=WidgetSwitcherSlot}
  }
  Text_MatchCapWarning {
   widgetClassPath=&CommonTextBlock
   bIsVariable=true
   parent=CommonWidgetSwitcher_TournamentWarnings
   widget=Text_MatchCapWarning
   properties {
    autoWrapText=true
    justification=Center
    lineHeightPercentage=0.9
    text=5 matches remaining
    wrapTextAt=366
    colorAndOpacity {colorUseRule=UseColor_Specified, specifiedColor=<rgba 0.0044 0.0091 0.0497 1>}
    font {
     bForceMonospaced=false
     bMaterialIsStencil=false
     fontMaterial=null
     fontObject=&Burbank
     letterSpacing=0
     monospacedWidth=1
     size=22
     skewAmount=0
     typefaceFontName=BurbankRegular-Black
     outlineSettings {
      bApplyOutlineToDropShadows=false
      bMiteredCorners=false
      bSeparateFillAlpha=false
      outlineMaterial=null
      outlineSize=0
      outlineColor=<rgba 0 0 0 1>
     }
    }
    renderTransform {angle=0, scale=<1 1>, shear=<-7 0>, translation=<0 0>}
   }
   slot {class=WidgetSwitcherSlot}
  }
  MFA_WarningText {
   widgetClassPath=&CommonTextBlock
   bIsVariable=true
   parent=CommonWidgetSwitcher_TournamentWarnings
   widget=MFA_WarningText
   properties {
    autoWrapText=true
    justification=Center
    lineHeightPercentage=0.9
    text="Two-Factor Authentication (2FA) is required to compete for cash prizes! See Event Details"
    wrapTextAt=366
    colorAndOpacity {colorUseRule=UseColor_Specified, specifiedColor=<rgba 0.0044 0.0091 0.0497 1>}
    font {
     bForceMonospaced=false
     bMaterialIsStencil=false
     fontMaterial=null
     fontObject=&Burbank
     letterSpacing=0
     monospacedWidth=1
     size=22
     skewAmount=0
     typefaceFontName=BurbankRegular-Black
     outlineSettings {
      bApplyOutlineToDropShadows=false
      bMiteredCorners=false
      bSeparateFillAlpha=false
      outlineMaterial=null
      outlineSize=0
      outlineColor=<rgba 0 0 0 1>
     }
    }
    renderTransform {angle=0, scale=<1 1>, shear=<-7 0>, translation=<0 0>}
   }
   slot {class=WidgetSwitcherSlot}
  }
  Text_ScheduledMMInfo {
   widgetClassPath=&CommonTextBlock
   bIsVariable=true
   parent=CommonWidgetSwitcher_TournamentWarnings
   widget=Text_ScheduledMMInfo
   properties {
    autoWrapText=true
    justification=Center
    lineHeightPercentage=0.9
    text="Remain in matchmaking to avoid missing next match start!"
    wrapTextAt=366
    colorAndOpacity {colorUseRule=UseColor_Specified, specifiedColor=<rgba 0.0044 0.0091 0.0497 1>}
    font {
     bForceMonospaced=false
     bMaterialIsStencil=false
     fontMaterial=null
     fontObject=&Burbank
     letterSpacing=0
     monospacedWidth=1
     size=22
     skewAmount=0
     typefaceFontName=BurbankRegular-Black
     outlineSettings {
      bApplyOutlineToDropShadows=false
      bMiteredCorners=false
      bSeparateFillAlpha=false
      outlineMaterial=null
      outlineSize=0
      outlineColor=<rgba 0 0 0 1>
     }
    }
    renderTransform {angle=0, scale=<1 1>, shear=<-7 0>, translation=<0 0>}
   }
   slot {class=WidgetSwitcherSlot}
  }
  NoMessage {widgetClassPath=&Overlay, bIsVariable=true, parent=CentralContentSwitcher, widget=NoMessage, slot {class=WidgetSwitcherSlot}}
  Bg_HighestScore {
   widgetClassPath=&Image
   bIsVariable=true
   parent=GP_PanelContent
   widget=Bg_HighestScore
   properties {renderOpacity=0.05, visibility=SelfHitTestInvisible}
   slot {class=GridSlot, properties {layer=-10, row=3, rowSpan=2, padding=<ltrb 0 24 0 0>}}
  }
  VB_HighestScore {
   widgetClassPath=&VerticalBox
   bIsVariable=true
   parent=GP_PanelContent
   widget=VB_HighestScore
   slot {class=GridSlot, properties {row=3, verticalAlignment=VAlign_Bottom, padding=<ltrb 16 48 16 0>}}
  }
  HighestScoreText {
   widgetClassPath=&CommonTextBlock
   bIsVariable=true
   parent=VB_HighestScore
   widget=HighestScoreText
   properties {
    autoWrapText=true
    justification=Center
    text="HIGHEST SCORE: {Score}"
    font {
     bForceMonospaced=false
     bMaterialIsStencil=false
     fontMaterial=null
     fontObject=&BurbankSmall
     letterSpacing=0
     monospacedWidth=1
     size=28
     skewAmount=0
     typefaceFontName=Black
     outlineSettings {
      bApplyOutlineToDropShadows=false
      bMiteredCorners=false
      bSeparateFillAlpha=false
      outlineMaterial=null
      outlineSize=0
      outlineColor=<rgba 0 0 0 1>
     }
    }
   }
   slot {class=VerticalBoxSlot}
  }
  HighestScoreTeammatesText {
   widgetClassPath=&CommonTextBlock
   bIsVariable=true
   parent=VB_HighestScore
   widget=HighestScoreTeammatesText
   properties {
    autoWrapText=true
    justification=Center
    text="LOREMIPSUM, DOLOR AMIT, SEQUITOR"
    colorAndOpacity {colorUseRule=UseColor_Specified, specifiedColor=<rgba 0 1 0.9823 1>}
    font {
     bForceMonospaced=false
     bMaterialIsStencil=false
     fontMaterial=null
     fontObject=&BurbankSmall
     letterSpacing=0
     monospacedWidth=1
     size=24
     skewAmount=0
     typefaceFontName=NotoSans-Regular
     outlineSettings {
      bApplyOutlineToDropShadows=false
      bMiteredCorners=false
      bSeparateFillAlpha=false
      outlineMaterial=null
      outlineSize=0
      outlineColor=<rgba 0 0 0 1>
     }
    }
   }
   slot {class=VerticalBoxSlot, properties {padding=<ltrb 0 4 0 0>}}
  }
  VerticalBox_0 {
   widgetClassPath=&VerticalBox
   parent=GP_PanelContent
   widget=VerticalBox_0
   slot {class=GridSlot, properties {row=4, padding=<ltrb 0 16 0 0>}}
  }
  Border_TimeRemaining {
   widgetClassPath=&Border
   bIsVariable=true
   parent=VerticalBox_0
   widget=Border_TimeRemaining
   properties {
    background {
     bIsDynamicallyLoaded=false
     drawAs=Image
     imageType=NoImage
     mirroring=NoMirror
     resourceName=None
     resourceObject=/FortniteCoreUI/Frontend/Showdown/Materials/M_UI_SkewBox_TournamentPanelTimerBG
     tiling=NoTile
     imageSize=<32 32>
     margin=<ltrb 0 0 0 0>
     outlineSettings {
      bUseBrushTransparency=false
      roundingType=HalfHeightRadius
      width=0
      color {colorUseRule=UseColor_Specified, specifiedColor=<rgba 0 0 0 0>}
      cornerRadii=<0 0 0 0>
     }
     tintColor {colorUseRule=UseColor_Specified, specifiedColor=<rgba 1 1 1 1>}
     uVRegion {bIsValid=false, max=<0 0>, min=<0 0>}
    }
    brushColor=<rgba 0.0018 0.0075 0.0762 1>
    padding=<ltrb 16 24 16 16>
   }
   slot {class=VerticalBoxSlot}
  }
  Switcher_TimeInfo {
   widgetClassPath=&CommonWidgetSwitcherLegacy
   bIsVariable=true
   parent=Border_TimeRemaining
   widget=Switcher_TimeInfo
   properties {clipping=ClipToBounds}
   slot {class=BorderSlot, properties {padding=<ltrb 16 24 16 16>}}
  }
  VB_Content_TimeRemaining {
   widgetClassPath=&VerticalBox
   bIsVariable=true
   parent=Switcher_TimeInfo
   widget=VB_Content_TimeRemaining
   slot {class=WidgetSwitcherSlot}
  }
  HorizontalBox_0 {
   widgetClassPath=&HorizontalBox
   parent=VB_Content_TimeRemaining
   widget=HorizontalBox_0
   slot {class=VerticalBoxSlot, properties {horizontalAlignment=HAlign_Center}}
  }
  Image_35 {
   widgetClassPath=&Image
   bIsVariable=true
   parent=HorizontalBox_0
   widget=Image_35
   properties {brush {
     bIsDynamicallyLoaded=false
     drawAs=Image
     imageType=NoImage
     mirroring=NoMirror
     resourceName=None
     resourceObject=/Game/UI/Foundation/Textures/Icons/HUD/T-Icon-Clock-128
     tiling=NoTile
     imageSize=<40 40>
     margin=<ltrb 0 0 0 0>
     outlineSettings {
      bUseBrushTransparency=false
      roundingType=HalfHeightRadius
      width=0
      color {colorUseRule=UseColor_Specified, specifiedColor=<rgba 0 0 0 0>}
      cornerRadii=<0 0 0 0>
     }
     tintColor {colorUseRule=UseColor_Specified, specifiedColor=<rgba 1 1 1 1>}
     uVRegion {bIsValid=false, max=<0 0>, min=<0 0>}
    }}
   slot {class=HorizontalBoxSlot, properties {horizontalAlignment=HAlign_Left, verticalAlignment=VAlign_Center}}
  }
  Text_TimeRemaining {
   widgetClassPath=&CommonTextBlock
   bInherited=true
   bIsVariable=true
   parent=HorizontalBox_0
   widget=Text_TimeRemaining
   properties {
    justification=Center
    scrollStyle=&TextScrollStyle-Base_C
    text=1H 2M
    font {
     bForceMonospaced=false
     bMaterialIsStencil=false
     fontMaterial=null
     fontObject=&BurbankSmall
     letterSpacing=0
     monospacedWidth=1
     size=30
     skewAmount=0
     typefaceFontName=Black
     outlineSettings {
      bApplyOutlineToDropShadows=false
      bMiteredCorners=false
      bSeparateFillAlpha=false
      outlineMaterial=null
      outlineSize=0
      outlineColor=<rgba 0 0 0 1>
     }
    }
   }
   slot {class=HorizontalBoxSlot, properties {verticalAlignment=VAlign_Center, padding=<ltrb 8 0 0 0>, size {sizeRule=Fill, value=1}}}
  }
  Text_Remaining {
   widgetClassPath=&CommonTextBlock
   bIsVariable=true
   parent=VB_Content_TimeRemaining
   widget=Text_Remaining
   properties {
    justification=Center
    text=REMAINING
    font {
     bForceMonospaced=false
     bMaterialIsStencil=false
     fontMaterial=null
     fontObject=&BurbankSmall
     letterSpacing=0
     monospacedWidth=1
     size=18
     skewAmount=0
     typefaceFontName=NotoSans-Regular
     outlineSettings {
      bApplyOutlineToDropShadows=false
      bMiteredCorners=false
      bSeparateFillAlpha=false
      outlineMaterial=null
      outlineSize=0
      outlineColor=<rgba 0 0 0 1>
     }
    }
   }
   slot {class=VerticalBoxSlot, properties {verticalAlignment=VAlign_Center, padding=<ltrb 0 4 0 0>}}
  }
  VB_Content_ScheduledTime {
   widgetClassPath=&VerticalBox
   bIsVariable=true
   parent=Switcher_TimeInfo
   widget=VB_Content_ScheduledTime
   slot {class=WidgetSwitcherSlot}
  }
  HorizontalBox {
   widgetClassPath=&HorizontalBox
   parent=VB_Content_ScheduledTime
   widget=HorizontalBox
   slot {class=VerticalBoxSlot, properties {horizontalAlignment=HAlign_Center}}
  }
  Image {
   widgetClassPath=&Image
   bIsVariable=true
   parent=HorizontalBox
   widget=Image
   properties {brush {
     bIsDynamicallyLoaded=false
     drawAs=Image
     imageType=NoImage
     mirroring=NoMirror
     resourceName=None
     resourceObject=&T-Icon-Clock-128
     tiling=NoTile
     imageSize=<40 40>
     margin=<ltrb 0 0 0 0>
     outlineSettings {
      bUseBrushTransparency=false
      roundingType=HalfHeightRadius
      width=0
      color {colorUseRule=UseColor_Specified, specifiedColor=<rgba 0 0 0 0>}
      cornerRadii=<0 0 0 0>
     }
     tintColor {colorUseRule=UseColor_Specified, specifiedColor=<rgba 1 1 1 1>}
     uVRegion {bIsValid=false, max=<0 0>, min=<0 0>}
    }}
   slot {class=HorizontalBoxSlot, properties {horizontalAlignment=HAlign_Left, verticalAlignment=VAlign_Center}}
  }
  Text_TimeRemainingScheduledMatch {
   widgetClassPath=&CommonTextBlock
   bInherited=true
   bIsVariable=true
   parent=HorizontalBox
   widget=Text_TimeRemainingScheduledMatch
   properties {
    justification=Center
    scrollStyle=&TextScrollStyle-Base_C
    text=1H 2M
    font {
     bForceMonospaced=false
     bMaterialIsStencil=false
     fontMaterial=null
     fontObject=&BurbankSmall
     letterSpacing=0
     monospacedWidth=1
     size=30
     skewAmount=0
     typefaceFontName=Black
     outlineSettings {
      bApplyOutlineToDropShadows=false
      bMiteredCorners=false
      bSeparateFillAlpha=false
      outlineMaterial=null
      outlineSize=0
      outlineColor=<rgba 0 0 0 1>
     }
    }
   }
   slot {class=HorizontalBoxSlot, properties {verticalAlignment=VAlign_Center, padding=<ltrb 8 0 0 0>, size {sizeRule=Fill, value=1}}}
  }
  Text_ScheduledMatchInfo {
   widgetClassPath=&CommonTextBlock
   bInherited=true
   bIsVariable=true
   parent=VB_Content_ScheduledTime
   widget=Text_ScheduledMatchInfo
   properties {
    justification=Center
    text=REMAINING
    font {
     bForceMonospaced=false
     bMaterialIsStencil=false
     fontMaterial=null
     fontObject=&BurbankSmall
     letterSpacing=0
     monospacedWidth=1
     size=26
     skewAmount=0
     typefaceFontName=NotoSans-Regular
     outlineSettings {
      bApplyOutlineToDropShadows=false
      bMiteredCorners=false
      bSeparateFillAlpha=false
      outlineMaterial=null
      outlineSize=0
      outlineColor=<rgba 0 0 0 1>
     }
    }
   }
   slot {class=VerticalBoxSlot, properties {verticalAlignment=VAlign_Center, padding=<ltrb 0 4 0 0>}}
  }
  O_EventDetails {widgetClassPath=&Overlay, parent=VerticalBox_0, widget=O_EventDetails, slot {class=VerticalBoxSlot}}
  Spacer_86 {
   widgetClassPath=&Spacer
   parent=O_EventDetails
   widget=Spacer_86
   properties {size=<1 48>}
   slot {class=OverlaySlot, properties {horizontalAlignment=HAlign_Fill, verticalAlignment=VAlign_Fill}}
  }
  Image_DetailsButtonBG {
   widgetClassPath=&Image
   bIsVariable=true
   parent=O_EventDetails
   widget=Image_DetailsButtonBG
   properties {
    visibility=SelfHitTestInvisible
    brush {
     bIsDynamicallyLoaded=false
     drawAs=Image
     imageType=NoImage
     mirroring=NoMirror
     resourceName=None
     resourceObject=/FortniteCoreUI/Frontend/Showdown/Materials/MI_UI_Showdown_Panel_ButtonBg
     tiling=NoTile
     imageSize=<32 32>
     margin=<ltrb 0 0 0 0>
     outlineSettings {
      bUseBrushTransparency=false
      roundingType=HalfHeightRadius
      width=0
      color {colorUseRule=UseColor_Specified, specifiedColor=<rgba 0 0 0 0>}
      cornerRadii=<0 0 0 0>
     }
     tintColor {colorUseRule=UseColor_Specified, specifiedColor=<rgba 1 1 1 1>}
     uVRegion {bIsValid=false, max=<0 0>, min=<0 0>}
    }
    renderTransformPivot=<0 1>
   }
   slot {class=OverlaySlot, properties {horizontalAlignment=HAlign_Fill, verticalAlignment=VAlign_Fill}}
  }
  Button_EventDetails {
   widgetClassPath=&IconTextButton
   bInherited=true
   bIsVariable=true
   parent=O_EventDetails
   widget=Button_EventDetails
   properties {
    buttonText=""
    contentAlignment=HAlign_Right
    onButtonBaseClicked=(&self:WidgetTree.Button_EventDetails.HandleOnButtonClicked)
    onButtonBaseDoubleClicked=(&self:WidgetTree.Button_EventDetails.HandleOnButtonDoubleClicked)
    onButtonBaseHovered=(&self:WidgetTree.Button_EventDetails.HandleOnButtonHovered)
    onButtonBaseUnhovered=(&self:WidgetTree.Button_EventDetails.HandleOnButtonUnhovered)
    onSelectedChangedBase=(&self:WidgetTree.Button_EventDetails.HandleOnSelectedChanged)
    style=/FortUILibrary/Foundation/Buttons/ButtonStyles/ButtonStyle-TextOnlyBase_Empty.ButtonStyle-TextOnlyBase_Empty_C
   }
   slot {class=OverlaySlot, properties {horizontalAlignment=HAlign_Fill, verticalAlignment=VAlign_Fill}}
  }
  Text_DetailsButton {
   widgetClassPath=&CommonTextBlock
   bIsVariable=true
   parent=O_EventDetails
   widget=Text_DetailsButton
   properties {
    justification=Center
    text=EVENT DETAILS
    colorAndOpacity {colorUseRule=UseColor_Specified, specifiedColor=<rgba 0.0116 0.016 0.0395 1>}
    font {
     bForceMonospaced=false
     bMaterialIsStencil=false
     fontMaterial=null
     fontObject=&BurbankSmall
     letterSpacing=0
     monospacedWidth=1
     size=18
     skewAmount=0
     typefaceFontName=Black
     outlineSettings {
      bApplyOutlineToDropShadows=false
      bMiteredCorners=false
      bSeparateFillAlpha=false
      outlineMaterial=null
      outlineSize=0
      outlineColor=<rgba 0 0 0 1>
     }
    }
   }
   slot {class=OverlaySlot, properties {horizontalAlignment=HAlign_Fill, verticalAlignment=VAlign_Center, padding=<ltrb 48 0 48 0>}}
  }
  Text_DebugId {
   widgetClassPath=&TextBlock
   bInherited=true
   parent=SLP_Grid_Panel
   widget=Text_DebugId
   properties {
    autoWrapText=true
    justification=Center
    text=EventId: Event EventWindowId: EventWindow
    visibility=Collapsed
    wrapTextAt=500
    wrappingPolicy=AllowPerCharacterWrapping
    font {
     bForceMonospaced=false
     bMaterialIsStencil=false
     fontMaterial=null
     fontObject=&Roboto
     letterSpacing=0
     monospacedWidth=1
     size=22
     skewAmount=0
     typefaceFontName=Bold
     outlineSettings {
      bApplyOutlineToDropShadows=false
      bMiteredCorners=false
      bSeparateFillAlpha=false
      outlineMaterial=null
      outlineSize=3
      outlineColor=<rgba 0 0 0 1>
     }
    }
   }
   slot {class=GridSlot, properties {layer=40, verticalAlignment=VAlign_Top, padding=<ltrb 0 350 0 0>}}
  }
  Image_PosterFront {widgetClassPath=/Script/EpicCMSUIFramework.EpicCMSImage, bInherited=true, widget=null}
  Image_PosterBack {widgetClassPath=&EpicCMSImage, bInherited=true, widget=null}
  Image_LoadingScreen {widgetClassPath=&EpicCMSImage, bInherited=true, widget=null}
  Image_PlaylistTile {widgetClassPath=&EpicCMSImage, bInherited=true, widget=null}
 ]
}
