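# Emergency notice widget for Battle Royale, derived from EmergencyNoticeBase, showing an uppercase title, attention icon and scrolling body on a dark backplate.
# ShowNotice fills both texts and reveals the container, enlarging or shrinking the icon on mobile; HideNotice collapses it.
# On construct, compact UI narrows body wrapping and lowers its height cap.
# A separate Show Notice function, not called in the event graph, plays the opacity fade-in Intro.
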WidgetBlueprint /FortniteCoreUI/UI/Foundation/EmergencyNotice/EmergencyNoticeWidgetBR
animations [Intro {
  binding_count=1
  length_seconds=0.4
  object=&self:Intro
  sequence {
   bindings [EmergencyNoticeWidgetBR_C_85 {
     id=#1
     possessed_class=EmergencyNoticeWidgetBR_C
     tracks [{
       type=MovieSceneFloatTrack
       display_name=Render Opacity
       sections [{type=MovieSceneFloatSection, channels [None {type=MovieSceneScriptingFloatChannel, keys [[0,0],[24,1]]}]}]
      }]
    }]
   display_rate {numerator=60}
  }
 }]
graphs [
 EventGraph {nodes [
   Event Construct()
   > SetMobileLayout()
   n1 = Event ShowNotice!() -> (OutputDelegate, then, Title:"Text (by ref)", Body:"Text (by ref)") {is=title K2Node_Event}
   > "SetText (Text)"(
    self=Get TitleText() -> (TitleText:Common Text Object Reference)
    InText=TextToUpper!(InText:"Text (by ref)"=@n1.Title) -> (ReturnValue:Text)
   )
   > "SetText (Text)"(self=Get BodyText() -> (BodyText:Common Text Object Reference), InText=@n1.Body)
   > SetVisibility(self=Get Notice_Container() -> (Notice_Container:Overlay Object Reference), InVisibility=HitTestInvisible)
   > Branch(Condition=IsMobileGame())
   > Set IconSize!(
    execute
    IconSize:E_UI_StatusIndicator_Size Enum=NewEnumerator3
    self:WBP UIKit Status Indicator Base Object Reference=Get WBP_UIKit_StatusIndicator() -> (WBP_UIKit_StatusIndicator:WBP UIKit Status Indicator Object Reference)
   ) -> (then, Output_Get:E_UI_StatusIndicator_Size Enum=NewEnumerator2)
   Event HideNotice() {is=title K2Node_Event}
   > SetVisibility(self=Get Notice_Container(), InVisibility=Collapsed)
  ]}
 SetMobileLayout {nodes [
   SetMobileLayout() {is=title K2Node_FunctionEntry}
   > Branch(Condition=IsUsingCompactUI())
   > SetWrapTextWidth!(execute, self:Common Text Object Reference=Get BodyText(), InWrapTextAt:int=388) -> (then)
   > SetMaxDesiredHeight!(execute, self:Size Box Object Reference=Get SB_Content() -> (SB_Content:Size Box Object Reference), InMaxDesiredHeight:float=120) -> (then)
  ]}
 Show Notice {nodes [
   Show Notice() {is=title K2Node_FunctionEntry}
   > PlayAnimation(InAnimation=Get Intro())
  ]}
]
variables []
widgets {
 info {inheritedWidgetCount=1, parentClass=/Script/FortniteUI.EmergencyNoticeBase, rootWidgetClass=&Overlay, widgetCount=11}
 widgets [
  Notice_Container {widgetClassPath=&Overlay, bIsVariable=true, widget=Notice_Container, properties {clipping=ClipToBounds}}
  WBP_CaptureForPostBufferUpdate {
   widgetClassPath=&WBP_CaptureForPostBufferUpdate
   bIsVariable=true
   parent=Notice_Container
   slot {class=OverlaySlot, properties {horizontalAlignment=HAlign_Fill, verticalAlignment=VAlign_Fill}}
  }
  UIKit_Backplate {
   widgetClassPath=&WBP_UIKit_Backplate
   parent=Notice_Container
   widget=UIKit_Backplate
   properties {brightness=Dark}
   slot {class=OverlaySlot, properties {horizontalAlignment=HAlign_Fill, verticalAlignment=VAlign_Fill}}
  }
  VisibilitySB {
   widgetClassPath=&SizeBox
   parent=Notice_Container
   widget=VisibilitySB
   properties {bOverride_MaxDesiredWidth=true, maxDesiredWidth=500}
   slot {class=OverlaySlot, properties {horizontalAlignment=HAlign_Fill, verticalAlignment=VAlign_Fill, padding=<ltrb 16 16 16 16>}}
  }
  GridPanel_0 {
   widgetClassPath=&GridPanel
   parent=VisibilitySB
   widget=GridPanel_0
   properties {columnFill [1], rowFill [0,1]}
   slot {class=SizeBoxSlot}
  }
  HB_Content {
   widgetClassPath=&HorizontalBox
   parent=GridPanel_0
   widget=HB_Content
   slot {class=GridSlot, properties {verticalAlignment=VAlign_Top, padding=<ltrb 0 0 0 8>}}
  }
  TitleText {
   widgetClassPath=&CommonTextBlock
   bIsVariable=true
   parent=HB_Content
   widget=TitleText
   properties {
    mobileFontSizeMultiplier=1.2
    scrollStyle=/FortUILibrary/Styles/Typography/Foundation/TextScrollStyle-NoFade.TextScrollStyle-NoFade_C
    style=&TextStyle-Base_C
    text=Test Emergency Notice Title
    textTransformPolicy=ToUpper
    font {
     bForceMonospaced=false
     bMaterialIsStencil=false
     fontMaterial=null
     fontObject=&NotoSans
     letterSpacing=0
     monospacedWidth=1
     size=24
     skewAmount=0
     typefaceFontName=NotoSans-Regular
     outlineSettings {
      bApplyOutlineToDropShadows=false
      bMiteredCorners=false
      bSeparateFillAlpha=false
      outlineMaterial=null
      outlineSize=0
      outlineColor=<rgba 0 0 0 1>
     }
    }
    shadowOffset=<0 0>
   }
   slot {class=HorizontalBoxSlot, properties {verticalAlignment=VAlign_Top, padding=<ltrb 0 0 8 0>, size {sizeRule=Fill, value=1}}}
  }
  WBP_UIKit_StatusIndicator {
   widgetClassPath=/FortUILibrary/UIKit/StatusIndicator/WBP_UIKit_StatusIndicator
   bIsVariable=true
   parent=HB_Content
   widget=WBP_UIKit_StatusIndicator
   properties {type=Attention, bangType=Custom, iconSize=Compact, overrideStatusIndicatorVisibility=true}
   slot {class=HorizontalBoxSlot, properties {horizontalAlignment=HAlign_Right, verticalAlignment=VAlign_Top}}
  }
  SB_Content {
   widgetClassPath=&SizeBox
   bIsVariable=true
   parent=GridPanel_0
   widget=SB_Content
   properties {bOverride_MaxDesiredHeight=true, maxDesiredHeight=200}
   slot {class=GridSlot, properties {row=1, verticalAlignment=VAlign_Top}}
  }
  BodyText {
   widgetClassPath=&CommonTextBlock
   bIsVariable=true
   parent=SB_Content
   widget=BodyText
   properties {
    applyLineHeightToBottomLine=false
    mobileFontSizeMultiplier=1.4
    scrollOrientation=Orient_Vertical
    scrollStyle=&TextScrollStyle-Base_C
    text=Test Emergency Notice about having issues with the service that servces things. We now support more than three lines of text.
    wrapTextAt=448
    colorAndOpacity {colorUseRule=UseColor_Specified, specifiedColor=<rgba 0.5776 0.5776 0.5776 1>}
    font {
     bForceMonospaced=false
     bMaterialIsStencil=false
     fontMaterial=null
     fontObject=&HeadingNow
     letterSpacing=10
     monospacedWidth=1
     size=15.75
     skewAmount=0
     typefaceFontName=74 Regular
     outlineSettings {
      bApplyOutlineToDropShadows=false
      bMiteredCorners=false
      bSeparateFillAlpha=false
      outlineMaterial=null
      outlineSize=0
      outlineColor=<rgba 0 0 0 1>
     }
    }
    shadowOffset=<0 0>
   }
   slot {class=SizeBoxSlot}
  }
  Border_TestBG {widgetClassPath=&Border, bInherited=true, widget=null, properties=~}
 ]
}
