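# Fancy toast notification widget that shows a header, body and optional image in a special or regular layout chosen by toast type.
# ShowToast fills uppercased text, then async-loads the image into a portrait material parameter before playing Intro.
# Intro's end starts a dismiss timer unless disabled; Outro's end collapses the widget and broadcasts OnFinishedToast.
# Mouse capture turns a left-button press and release into OnToastClicked.
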WidgetBlueprint /FortniteCoreUI/UI/Foundation/Toasts/FancyToastWidget
animations [
 Intro {
  binding_count=7
  display_label=""
  length_seconds=0.6
  object=&self:Intro
  sequence {
   bindings [
    SmallPreviewImageSB {
     id=#1
     possessed_class=SizeBox
     tracks [
      {
       type=MovieSceneFloatTrack
       display_name=Render Opacity
       sections [{
         type=MovieSceneFloatSection
         end_frame=60
         start_frame=0
         channels [None {type=MovieSceneScriptingFloatChannel, keys [[0,0],[15,1],[59.9983,1]]}]
        }]
      }
      {
       type=MovieScene2DTransformTrack
       sections [{
         type=MovieScene2DTransformSection
         end_frame=40
         start_frame=0
         channels [
          Translation.X {type=MovieSceneScriptingFloatChannel, const [-72,1]}
          Translation.Y {type=MovieSceneScriptingFloatChannel, keys [[0,64],[15,-64],[23,10],[25,-16],[29,-8]]}
          Angle {type=MovieSceneScriptingFloatChannel, keys [[0,-15],[15,15],[23,-5],[25,0],[29,0]]}
          Scale.X {type=MovieSceneScriptingFloatChannel, keys [[0,0.75],[15,1.375],[25,1]]}
          Scale.Y {type=MovieSceneScriptingFloatChannel, keys [[0,0.75],[15,1.375],[25,1]]}
          Shear.X {type=MovieSceneScriptingFloatChannel}
          Shear.Y {type=MovieSceneScriptingFloatChannel}
         ]
        }]
      }
     ]
    }
    TextBackgroundRegular {
     id=#2
     possessed_class=Image
     tracks [{
       type=MovieSceneWidgetMaterialTrack
       sections [{
         type=MovieSceneParameterSection
         end_frame=60
         start_frame=0
         channels [RGBOverrideSwitch {type=MovieSceneScriptingFloatChannel, keys [[0,0],[30,0],[36,1],[45,0],[59.9983,0]]}]
        }]
      }]
    }
    RegularToast {
     id=#3
     possessed_class=Overlay
     tracks [
      {
       type=MovieSceneFloatTrack
       display_name=Render Opacity
       sections [{
         type=MovieSceneFloatSection
         end_frame=60
         start_frame=0
         channels [None {type=MovieSceneScriptingFloatChannel, keys [[0,0],[20,0],[35,1],[59.9983,1]]}]
        }]
      }
      {
       type=MovieScene2DTransformTrack
       sections [{
         type=MovieScene2DTransformSection
         end_frame=60
         start_frame=0
         channels [
          Translation.X {type=MovieSceneScriptingFloatChannel}
          Translation.Y {type=MovieSceneScriptingFloatChannel, keys [[20,128],[35,-64],[45,32],[48,0]]}
          Angle {type=MovieSceneScriptingFloatChannel}
          Scale.X {type=MovieSceneScriptingFloatChannel, keys [[20,1.125],[35,0.875],[45,1]]}
          Scale.Y {type=MovieSceneScriptingFloatChannel, keys [[20,0.75],[35,1.125],[45,1]]}
          Shear.X {type=MovieSceneScriptingFloatChannel}
          Shear.Y {type=MovieSceneScriptingFloatChannel}
         ]
        }]
      }
     ]
    }
    TextBoxSB {
     id=#4
     possessed_class=SizeBox
     tracks [
      {
       type=MovieSceneFloatTrack
       display_name=Render Opacity
       sections [{
         type=MovieSceneFloatSection
         end_frame=60
         start_frame=0
         channels [None {type=MovieSceneScriptingFloatChannel, keys [[0,0],[20,0],[35,1],[59.9983,1]]}]
        }]
      }
      {
       type=MovieScene2DTransformTrack
       sections [{
         type=MovieScene2DTransformSection
         end_frame=60
         start_frame=0
         channels [
          Translation.X {type=MovieSceneScriptingFloatChannel}
          Translation.Y {type=MovieSceneScriptingFloatChannel, keys [[20,128],[35,-64],[45,32],[48,0]]}
          Angle {type=MovieSceneScriptingFloatChannel}
          Scale.X {type=MovieSceneScriptingFloatChannel, keys [[20,1.125],[35,0.875],[45,1]]}
          Scale.Y {type=MovieSceneScriptingFloatChannel, keys [[20,0.75],[35,1.125],[45,1]]}
          Shear.X {type=MovieSceneScriptingFloatChannel}
          Shear.Y {type=MovieSceneScriptingFloatChannel}
         ]
        }]
      }
     ]
    }
    TextBackGround {
     id=#5
     possessed_class=Image
     tracks [{
       type=MovieSceneWidgetMaterialTrack
       sections [{
         type=MovieSceneParameterSection
         end_frame=60
         start_frame=0
         channels [RGBOverrideSwitch {type=MovieSceneScriptingFloatChannel, keys [[0,0],[30,0],[36,1],[45,0],[59.9983,0]]}]
        }]
      }]
    }
    SmallPreviewPortrait {
     id=#6
     possessed_class=Image
     tracks [{
       type=MovieSceneWidgetMaterialTrack
       sections [{
         type=MovieSceneParameterSection
         end_frame=60
         start_frame=0
         channels [RGBOverrideSwitch {type=MovieSceneScriptingFloatChannel, keys [[0,0],[20,0],[23,0.75],[27,0.75],[30.0017,0],[60,0]]}]
        }]
      }]
    }
    FlashingBang {
     id=#7
     possessed_class=Overlay
     tracks [
      {
       type=MovieSceneFloatTrack
       display_name=Render Opacity
       sections [{type=MovieSceneFloatSection, end_frame=21, start_frame=0, channels [None {type=MovieSceneScriptingFloatChannel, keys [[10,0],[20,1]]}]}]
      }
      {
       type=MovieScene2DTransformTrack
       sections [{
         type=MovieScene2DTransformSection
         end_frame=60
         start_frame=0
         channels [
          Translation.X {type=MovieSceneScriptingFloatChannel}
          Translation.Y {type=MovieSceneScriptingFloatChannel, keys [[10,0],[20,-40],[25,-4],[30,-32],[35,-12],[40,-20],[60,-20]]}
          Angle {type=MovieSceneScriptingFloatChannel}
          Scale.X {type=MovieSceneScriptingFloatChannel, keys [[10,1],[20,1.5],[30,1]]}
          Scale.Y {type=MovieSceneScriptingFloatChannel, keys [[10,1],[20,1.5],[30,1]]}
          Shear.X {type=MovieSceneScriptingFloatChannel}
          Shear.Y {type=MovieSceneScriptingFloatChannel}
         ]
        }]
      }
     ]
    }
   ]
   display_rate {numerator=100}
  }
 }
 Outro {
  binding_count=4
  display_label=""
  length_seconds=0.65
  object=&self:Outro
  sequence {bindings [
    TextBoxSB {
     id=#8
     possessed_class=SizeBox
     tracks [
      {
       type=MovieScene2DTransformTrack
       sections [{
         type=MovieScene2DTransformSection
         end_frame=13
         start_frame=0
         channels [
          Translation.X {type=MovieSceneScriptingFloatChannel, const [0,1]}
          Translation.Y {type=MovieSceneScriptingFloatChannel, keys [[0,0],[2,-16],[5,8],[10,-128]]}
          Angle {type=MovieSceneScriptingFloatChannel}
          Scale.X {type=MovieSceneScriptingFloatChannel, keys [[0,1],[5,1],[10,0.75]]}
          Scale.Y {type=MovieSceneScriptingFloatChannel, keys [[0,1],[5,1],[10,1.25]]}
          Shear.X {type=MovieSceneScriptingFloatChannel}
          Shear.Y {type=MovieSceneScriptingFloatChannel}
         ]
        }]
      }
      {
       type=MovieSceneFloatTrack
       display_name=Render Opacity
       sections [{type=MovieSceneFloatSection, end_frame=13, start_frame=0, channels [None {type=MovieSceneScriptingFloatChannel, keys [[0,1],[5,1],[10,0]]}]}]
      }
     ]
    }
    FlashingBang {
     id=#9
     possessed_class=Overlay
     tracks [{
       type=MovieScene2DTransformTrack
       sections [{
         type=MovieScene2DTransformSection
         end_frame=5
         start_frame=0
         channels [
          Translation.X {type=MovieSceneScriptingFloatChannel}
          Translation.Y {type=MovieSceneScriptingFloatChannel, keys [[0,-20],[2,-32],[4,-20]]}
          Angle {type=MovieSceneScriptingFloatChannel, const [0,1]}
          Scale.X {type=MovieSceneScriptingFloatChannel, keys [[0,1],[2,1.5],[4,1.5]]}
          Scale.Y {type=MovieSceneScriptingFloatChannel, keys [[0,1],[2,1.5],[4,1.5]]}
          Shear.X {type=MovieSceneScriptingFloatChannel}
          Shear.Y {type=MovieSceneScriptingFloatChannel}
         ]
        }]
      }]
    }
    RegularToast {
     id=#10
     possessed_class=Overlay
     tracks [
      {
       type=MovieScene2DTransformTrack
       sections [{
         type=MovieScene2DTransformSection
         end_frame=13
         start_frame=0
         channels [
          Translation.X {type=MovieSceneScriptingFloatChannel, const [0,1]}
          Translation.Y {type=MovieSceneScriptingFloatChannel, keys [[0,0],[2,-16],[5,8],[10,-128]]}
          Angle {type=MovieSceneScriptingFloatChannel}
          Scale.X {type=MovieSceneScriptingFloatChannel, keys [[0,1],[5,1],[10,0.75]]}
          Scale.Y {type=MovieSceneScriptingFloatChannel, keys [[0,1],[5,1],[10,1.25]]}
          Shear.X {type=MovieSceneScriptingFloatChannel}
          Shear.Y {type=MovieSceneScriptingFloatChannel}
         ]
        }]
      }
      {
       type=MovieSceneFloatTrack
       display_name=Render Opacity
       sections [{type=MovieSceneFloatSection, end_frame=13, start_frame=0, channels [None {type=MovieSceneScriptingFloatChannel, keys [[0,1],[5,1],[10,0]]}]}]
      }
     ]
    }
    SmallPreviewImageSB {
     id=#11
     possessed_class=SizeBox
     tracks [
      {
       type=MovieScene2DTransformTrack
       sections [{
         type=MovieScene2DTransformSection
         end_frame=13
         start_frame=0
         channels [
          Translation.X {type=MovieSceneScriptingFloatChannel, const [-72,1]}
          Translation.Y {type=MovieSceneScriptingFloatChannel, keys [[0,-8],[2,-32],[4,-8],[8,-8],[11,-96],[11.9997,-96]]}
          Angle {type=MovieSceneScriptingFloatChannel}
          Scale.X {type=MovieSceneScriptingFloatChannel, keys [[8,1],[11,0.875]]}
          Scale.Y {type=MovieSceneScriptingFloatChannel, keys [[8,1],[11,1.25]]}
          Shear.X {type=MovieSceneScriptingFloatChannel}
          Shear.Y {type=MovieSceneScriptingFloatChannel}
         ]
        }]
      }
      {
       type=MovieSceneFloatTrack
       display_name=Render Opacity
       sections [{type=MovieSceneFloatSection, end_frame=13, start_frame=0, channels [None {type=MovieSceneScriptingFloatChannel, keys [[0,1],[8,1],[11,0]]}]}]
      }
     ]
    }
   ]}
 }
]
graphs [
 EventGraph {nodes [
   n1 = OnShowToast!() -> (OutputDelegate, then, DispalyInfo:Fort Toast Display Info Structure) {is=title K2Node_CustomEvent}
   n2 = Break Fort Toast Display Info!(FortToastDisplayInfo:"Fort Toast Display Info Structure (by ref)"=@n1.DispalyInfo) -> (Header:Text, Body:Text, Image:Texture 2D Soft Object Reference, Type:EFortToastType Enum) {is=title K2Node_BreakStruct}
   > n3 = Branch(Condition=IsAssetNull!(Asset:Soft Object Reference=@n2.Image) -> (ReturnValue:bool))
   > Start Toast Animation()
   n4 = Async Load Asset(execute=@n3.else, Asset=@n2.Image)
   n5 = Branch(execute=@n4.Completed, Condition="Equal (Enum)"(A:EFortToastType Enum=@n2.Type, B:EFortToastType Enum=Lock))
   > n6 = GetDynamicMaterial(self=Get SmallPreviewPortraitRB() -> (SmallPreviewPortraitRB:Image Object Reference))
   n7 = GetDynamicMaterial(execute=@n5.else, self=Get SmallPreviewPortrait() -> (SmallPreviewPortrait:Image Object Reference))
   SetTextureParameterValue(
    execute=[@n6.then, @n7.then]
    self=[@n6, @n7]
    ParameterName=SmallPreviewImage
    Value=Cast To Texture2D!(Object=@n4) -> (AsTexture 2D, bSuccess:bool=~)
   )
   > Start Toast Animation()
   Start Toast Animation() {is=title K2Node_CustomEvent}
   > n8 = Branch(Condition=IsAnimationPlaying(InAnimation=Get Intro()))
   > n9 = StopAnimation(InAnimation=Get Intro())
   PlayAnimation(execute=[@n9.then, @n8.else], InAnimation=Get Intro(), PlaybackSpeed=0.675)
   SimulateAthenaStoreToast() {is=title K2Node_CustomEvent}
   > ShowToast!(
    execute
    self
    DisplayInfo:Fort Toast Display Info Structure=Make Fort Toast Display Info!(
     Header:Text="LOCTEXT(\"Hey Listen!\")"
     Body:Text="LOCTEXT(\"Store updated\")"
     Image:Texture 2D Soft Object Reference=Get Dev_TestImage() -> (Dev_TestImage:Texture 2D Soft Object Reference)
     Type:EFortToastType Enum=Impactful
    ) -> (FortToastDisplayInfo:Fort Toast Display Info Structure) {is=title K2Node_MakeStruct}
   ) -> (then)
   On Idle Finished() {is=title K2Node_CustomEvent}
   > PlayAnimation(InAnimation=Get Outro() -> (Outro:Widget Animation Object Reference))
   CompleteToast() {is=title K2Node_CustomEvent}
   > StopAnimationsAndLatentActions(self:User Widget Object Reference)
   > SetVisibility(InVisibility=Collapsed)
   > Call On Finished Toast!(execute:Exec, self:Fancy Toast Widget Object Reference) -> (then:Exec) {is=title K2Node_CallDelegate}
   n10 = RemoveWidget() {is=title K2Node_CustomEvent}
   CompleteToast(execute=[@n10.then, @n12.then]) {is=title K2Node_CallFunction}
   Clicked() {is=title K2Node_CustomEvent}
   > Call On Toast Clicked!(execute:Exec, self:Fancy Toast Widget Object Reference) -> (then:Exec) {is=title K2Node_CallDelegate}
   "Animation Finished (Intro)"!() -> (OutputDelegate:Delegate, then:Exec) {is=title K2Node_WidgetAnimationEvent}
   > n11 = Branch(Condition=Get DoNotUseTimerToDismiss() -> (DoNotUseTimerToDismiss:bool))
   Set Timer by Event(execute=@n11.else, Delegate=Create Event(), Time=4, bLooping=false)
   "Animation Started (Intro)"!() -> (OutputDelegate:Delegate, then:Exec) {is=title K2Node_WidgetAnimationEvent}
   > SetVisibility(InVisibility=Visible)
   n12 = "Animation Finished (Outro)"!() -> (OutputDelegate:Delegate, then:Exec) {is=title K2Node_WidgetAnimationEvent}
  ]}
 OnFinishedToast {nodes [OnFinishedToast() {is=title K2Node_FunctionEntry}]}
 OnMouseButtonDown {nodes [
   n1 = OnMouseButtonDown!() -> (then, MyGeometry:Geometry Structure, MouseEvent:"Pointer Event Structure (by ref)") {is=title K2Node_FunctionEntry}
   > n2 = Branch(Condition="Equal (Key)"!(
     A:Key Structure=Get Effecting Button!(Input:"Pointer Event Structure (by ref)"=@n1.MouseEvent) -> (ReturnValue:Key Structure) {is=title K2Node_CallFunction}
     B:Key Structure=LeftMouseButton
    ) -> (ReturnValue:bool))
   > Return Node!(
    execute
    ReturnValue:Event Reply Structure=CaptureMouse!(Reply:"Event Reply Structure (by ref)"=Handled(), CapturingWidget:Widget Object Reference=Self-Reference()) -> (ReturnValue:Event Reply Structure)
   )
   Return Node(execute=@n2.else, ReturnValue=Unhandled())
  ]}
 OnMouseButtonUp {nodes [
   n1 = OnMouseButtonUp!() -> (then, MyGeometry:Geometry Structure, MouseEvent:"Pointer Event Structure (by ref)") {is=title K2Node_FunctionEntry}
   > n2 = Branch(Condition="Equal (Key)"(A=Get Effecting Button(Input=@n1.MouseEvent), B=LeftMouseButton))
   > n3 = Branch(Condition=HasMouseCaptureByUser!(
     self:Widget Object Reference
     UserIndex:int=Get User Index!(Input:"Pointer Event Structure (by ref)"=@n1.MouseEvent) -> (ReturnValue:int=0) {is=title K2Node_CallFunction}
     PointerIndex:int=Get Pointer Index!(Input:"Pointer Event Structure (by ref)"=@n1.MouseEvent) -> (ReturnValue:int=0) {is=title K2Node_CallFunction}
    ) -> (ReturnValue:bool))
   > Clicked() {is=title K2Node_CallFunction}
   > Return Node(ReturnValue=ReleaseMouseCapture!(Reply:"Event Reply Structure (by ref)"=Handled()) -> (ReturnValue:Event Reply Structure))
   Return Node(execute=[@n2.else, @n3.else], ReturnValue=Unhandled())
  ]}
 OnToastClicked {nodes [OnToastClicked() {is=title K2Node_FunctionEntry}]}
 SetAndShowText {nodes [
   n1 = SetAndShowText!() -> (then, Text:Text, TextBlock:Common Text Object Reference) {is=title K2Node_FunctionEntry}
   > Set LocalText(LocalText:Text=@n1) -> (Output_Get:Text)
   > Set LocalTextBlock(LocalTextBlock:Common Text Object Reference=@n1.TextBlock) -> (Output_Get:Common Text Object Reference)
   > n2 = Branch(Condition=TextIsEmpty(InText=Get LocalText() -> (LocalText:Text)))
   > SetVisibility(self=Get LocalTextBlock() -> (LocalTextBlock:Common Text Object Reference), InVisibility=Collapsed)
   n3 = Get LocalTextBlock()
   "SetText (Text)"(execute=@n2.else, self=@n3, InText=TextToUpper!(InText:"Text (by ref)"=Get LocalText()) -> (ReturnValue:Text))
   > SetVisibility(self=@n3, InVisibility=SelfHitTestInvisible)
  ]}
 ShowToast {nodes [
   n1 = ShowToast!() -> (then, DisplayInfo:Fort Toast Display Info Structure) {is=title K2Node_FunctionEntry}
   n2 = Break Fort Toast Display Info(FortToastDisplayInfo=@n1)
   n3 = Get SpecialToast() -> (SpecialToast:Overlay Object Reference)
   n4 = Get RegularToast() -> (RegularToast:Overlay Object Reference)
   > SetActiveWidget(
    self:Widget Switcher Object Reference=Get TypeWidgetSwitcher() -> (TypeWidgetSwitcher:Widget Switcher Object Reference)
    Widget=Select!(
     Option 0:Widget Object Reference=@n3
     Option 1:Widget Object Reference=@n4
     Option 2:Widget Object Reference=@n3
     Option 3:Widget Object Reference=@n4
     Index:Byte=@n2.Type
    ) -> (ReturnValue:Widget Object Reference)
   )
   > SetAndShowText!(execute, self, Text:Text=@n2, TextBlock:Common Text Object Reference=Get Title() -> (Title:Common Text Object Reference)) -> (then) {is=title K2Node_CallFunction}
   > SetAndShowText(Text=@n2, TextBlock=Get TitleRegular() -> (TitleRegular:Common Text Object Reference))
   > SetAndShowText(Text=@n2.Body, TextBlock=Get Description() -> (Description:Common Text Object Reference))
   > SetAndShowText(Text=@n2.Body, TextBlock=Get DescriptionRegular() -> (DescriptionRegular:Common Text Object Reference))
   > OnShowToast!(execute, self, DispalyInfo:Fort Toast Display Info Structure=@n1) -> (then) {is=title K2Node_CallFunction}
  ]}
]
variables [
 Dev_TestImage {type=Texture2D, default=/Game/UI/Foundation/Textures/Icons/Heroes/Athena/Soldier/T-Soldier-HID-014-Athena-Commando-F-L}
 DoNotUseTimerToDismiss {type=boolean}
 OnFinishedToast {}
 OnToastClicked {}
]
widgets {
 info {inheritedWidgetCount=0, parentClass=&CommonUserWidget, rootWidgetClass=&WidgetSwitcher, widgetCount=33}
 widgets [
  TypeWidgetSwitcher {widgetClassPath=&WidgetSwitcher, bIsVariable=true, widget=TypeWidgetSwitcher, properties {visibility=Visible}}
  Overlay_1 {
   widgetClassPath=&Overlay
   parent=TypeWidgetSwitcher
   widget=Overlay_1
   properties {visibility=Visible}
   slot {class=WidgetSwitcherSlot, properties {horizontalAlignment=HAlign_Left}}
  }
  SpecialToast {
   widgetClassPath=&Overlay
   bIsVariable=true
   parent=Overlay_1
   widget=SpecialToast
   properties {visibility=Visible}
   slot {class=OverlaySlot, properties {horizontalAlignment=HAlign_Fill, verticalAlignment=VAlign_Fill, padding=<ltrb 70 0 0 0>}}
  }
  TextBoxSB {
   widgetClassPath=&SizeBox
   parent=SpecialToast
   widget=TextBoxSB
   properties {bOverride_MaxDesiredWidth=true, bOverride_MinDesiredWidth=true, maxDesiredWidth=448, minDesiredWidth=256, visibility=Visible}
   slot {class=OverlaySlot}
  }
  Overlay_4 {widgetClassPath=&Overlay, parent=TextBoxSB, widget=Overlay_4, properties {visibility=Visible}, slot {class=SizeBoxSlot}}
  TextBackGround {
   widgetClassPath=&Image
   parent=Overlay_4
   widget=TextBackGround
   properties {brush {
     bIsDynamicallyLoaded=false
     drawAs=Image
     imageType=NoImage
     mirroring=NoMirror
     resourceName=None
     resourceObject=/FortUILibrary/Styles/Border/Materials/M_Toast_SkewBoxMain
     tiling=NoTile
     imageSize=<32 32>
     margin=<ltrb 0 0 0 0>
     outlineSettings {
      bUseBrushTransparency=false
      roundingType=HalfHeightRadius
      width=0
      color {colorUseRule=UseColor_Specified, specifiedColor=<rgba 0 0 0 0>}
      cornerRadii=<0 0 0 0>
     }
     tintColor {colorUseRule=UseColor_Specified, specifiedColor=<rgba 1 1 1 1>}
     uVRegion {bIsValid=false, max=<0 0>, min=<0 0>}
    }}
   slot {class=OverlaySlot, properties {horizontalAlignment=HAlign_Fill, verticalAlignment=VAlign_Fill}}
  }
  HorizontalBox_0 {
   widgetClassPath=&HorizontalBox
   parent=Overlay_4
   widget=HorizontalBox_0
   properties {visibility=Visible}
   slot {class=OverlaySlot}
  }
  PreviewImageSpacer {
   widgetClassPath=&Spacer
   parent=HorizontalBox_0
   widget=PreviewImageSpacer
   properties {visibility=Visible, size=<32 1>}
   slot {class=HorizontalBoxSlot}
  }
  VerticalBox_0 {
   widgetClassPath=&VerticalBox
   parent=HorizontalBox_0
   widget=VerticalBox_0
   properties {visibility=Visible}
   slot {class=HorizontalBoxSlot, properties {padding=<ltrb 24 28 44 16>, size {sizeRule=Fill, value=1}}}
  }
  ScaleBox_0 {
   widgetClassPath=&ScaleBox
   parent=VerticalBox_0
   widget=ScaleBox_0
   properties {visibility=Visible, renderTransform {angle=-2, scale=<1 1>, shear=<0 0>, translation=<0 0>}}
   slot {class=VerticalBoxSlot}
  }
  Title {
   widgetClassPath=&CommonTextBlock
   bIsVariable=true
   parent=ScaleBox_0
   widget=Title
   properties {
    style=/FortUILibrary/Styles/Typography/Foundation/TextStyle-Header-M-YellowWhitePulse.TextStyle-Header-M-YellowWhitePulse_C
    text="AVAILABLE NOW!"
    visibility=Visible
    font {
     bForceMonospaced=false
     bMaterialIsStencil=false
     fontMaterial=/FortUILibrary/Materials/Foundation/M_Font_SimplePulse_YellowWhite
     fontObject=&Burbank
     letterSpacing=0
     monospacedWidth=1
     size=32
     skewAmount=0
     typefaceFontName=Burbank-Black
     outlineSettings {
      bApplyOutlineToDropShadows=false
      bMiteredCorners=false
      bSeparateFillAlpha=false
      outlineMaterial=null
      outlineSize=0
      outlineColor=<rgba 0 0 0 1>
     }
    }
    renderTransform {angle=0, scale=<1 1>, shear=<-16 0>, translation=<0 0>}
    shadowOffset=<0 0>
   }
   slot {class=ScaleBoxSlot}
  }
  Description {
   widgetClassPath=&CommonTextBlock
   bIsVariable=true
   parent=VerticalBox_0
   widget=Description
   properties {
    autoWrapText=true
    style=/FortUILibrary/Styles/Typography/Foundation/TextStyle-Base-XS-B.TextStyle-Base-XS-B_C
    text=IN THE ITEM SHOP FOR A LIMITED TIME.
    visibility=Visible
    font {
     bForceMonospaced=false
     bMaterialIsStencil=false
     fontMaterial=null
     fontObject=&NotoSans
     letterSpacing=0
     monospacedWidth=1
     size=16
     skewAmount=0
     typefaceFontName=NotoSans-Bold
     outlineSettings {
      bApplyOutlineToDropShadows=false
      bMiteredCorners=false
      bSeparateFillAlpha=false
      outlineMaterial=null
      outlineSize=0
      outlineColor=<rgba 0 0 0 1>
     }
    }
    renderTransform {angle=0, scale=<1 1>, shear=<-8 0>, translation=<0 0>}
    shadowOffset=<0 0>
   }
   slot {class=VerticalBoxSlot}
  }
  SmallPreviewImageSB {
   widgetClassPath=&SizeBox
   parent=SpecialToast
   widget=SmallPreviewImageSB
   properties {
    bOverride_HeightOverride=true
    bOverride_WidthOverride=true
    heightOverride=128
    visibility=Visible
    widthOverride=128
    renderTransform {angle=0, scale=<1 1>, shear=<0 0>, translation=<-72 -8>}
   }
   slot {class=OverlaySlot, properties {padding=<ltrb -10 0 0 0>}}
  }
  Overlay_5 {widgetClassPath=&Overlay, parent=SmallPreviewImageSB, widget=Overlay_5, properties {visibility=Visible}, slot {class=SizeBoxSlot}}
  SmallPreviewPortrait {
   widgetClassPath=&Image
   bIsVariable=true
   parent=Overlay_5
   widget=SmallPreviewPortrait
   properties {brush {
     bIsDynamicallyLoaded=false
     drawAs=Image
     imageType=NoImage
     mirroring=NoMirror
     resourceName=None
     resourceObject=/FortUILibrary/Styles/Border/Materials/M_Toast_SkewBoxPortrait
     tiling=NoTile
     imageSize=<128 128>
     margin=<ltrb 0 0 0 0>
     outlineSettings {
      bUseBrushTransparency=false
      roundingType=HalfHeightRadius
      width=0
      color {colorUseRule=UseColor_Specified, specifiedColor=<rgba 0 0 0 0>}
      cornerRadii=<0 0 0 0>
     }
     tintColor {colorUseRule=UseColor_Specified, specifiedColor=<rgba 1 1 1 1>}
     uVRegion {bIsValid=false, max=<0 0>, min=<0 0>}
    }}
   slot {class=OverlaySlot, properties {horizontalAlignment=HAlign_Fill, verticalAlignment=VAlign_Fill}}
  }
  FlashingBang {
   widgetClassPath=&Overlay
   parent=Overlay_5
   widget=FlashingBang
   properties {visibility=Visible, renderTransform {angle=0, scale=<1 1>, shear=<0 0>, translation=<14 -20>}}
   slot {class=OverlaySlot, properties {horizontalAlignment=HAlign_Right}}
  }
  Image_2 {
   widgetClassPath=&Image
   parent=FlashingBang
   widget=Image_2
   properties {brush {
     bIsDynamicallyLoaded=false
     drawAs=Image
     imageType=NoImage
     mirroring=NoMirror
     resourceName=None
     resourceObject=/FortUILibrary/Styles/Border/Materials/M_FlashingBang
     tiling=NoTile
     imageSize=<40 40>
     margin=<ltrb 0 0 0 0>
     outlineSettings {
      bUseBrushTransparency=false
      roundingType=HalfHeightRadius
      width=0
      color {colorUseRule=UseColor_Specified, specifiedColor=<rgba 0 0 0 0>}
      cornerRadii=<0 0 0 0>
     }
     tintColor {colorUseRule=UseColor_Specified, specifiedColor=<rgba 1 1 1 1>}
     uVRegion {bIsValid=false, max=<0 0>, min=<0 0>}
    }}
   slot {class=OverlaySlot}
  }
  CommonTextBlock_0 {
   widgetClassPath=&CommonTextBlock
   parent=FlashingBang
   widget=CommonTextBlock_0
   properties {
    style=&TextStyle-Header-M-YellowWhitePulse_C
    text="!"
    visibility=Visible
    font {
     bForceMonospaced=false
     bMaterialIsStencil=false
     fontMaterial=&M_Font_SimplePulse_YellowWhite
     fontObject=&Burbank
     letterSpacing=0
     monospacedWidth=1
     size=32
     skewAmount=0
     typefaceFontName=Burbank-Black
     outlineSettings {
      bApplyOutlineToDropShadows=false
      bMiteredCorners=false
      bSeparateFillAlpha=false
      outlineMaterial=null
      outlineSize=0
      outlineColor=<rgba 0 0 0 1>
     }
    }
    renderTransform {angle=5, scale=<0.875 0.75>, shear=<-8 0>, translation=<0 0>}
    shadowOffset=<0 0>
   }
   slot {class=OverlaySlot, properties {horizontalAlignment=HAlign_Center, verticalAlignment=VAlign_Center, padding=<ltrb 0 -8 0 -8>}}
  }
  RegularToast {
   widgetClassPath=&Overlay
   bIsVariable=true
   parent=TypeWidgetSwitcher
   widget=RegularToast
   properties {visibility=Visible}
   slot {class=WidgetSwitcherSlot}
  }
  SizeBox_3 {
   widgetClassPath=&SizeBox
   parent=RegularToast
   widget=SizeBox_3
   properties {bOverride_MaxDesiredWidth=true, bOverride_MinDesiredWidth=true, maxDesiredWidth=448, minDesiredWidth=256, visibility=Visible}
   slot {class=OverlaySlot}
  }
  Overlay_10 {widgetClassPath=&Overlay, parent=SizeBox_3, widget=Overlay_10, properties {visibility=Visible}, slot {class=SizeBoxSlot}}
  TextBackgroundRegular {
   widgetClassPath=&Image
   parent=Overlay_10
   widget=TextBackgroundRegular
   properties {brush {
     bIsDynamicallyLoaded=false
     drawAs=Image
     imageType=NoImage
     mirroring=NoMirror
     resourceName=None
     resourceObject=/FortUILibrary/Styles/Border/Materials/M_Toast_SkewBoxPlain
     tiling=NoTile
     imageSize=<32 32>
     margin=<ltrb 0 0 0 0>
     outlineSettings {
      bUseBrushTransparency=false
      roundingType=HalfHeightRadius
      width=0
      color {colorUseRule=UseColor_Specified, specifiedColor=<rgba 0 0 0 0>}
      cornerRadii=<0 0 0 0>
     }
     tintColor {colorUseRule=UseColor_Specified, specifiedColor=<rgba 1 1 1 1>}
     uVRegion {bIsValid=false, max=<0 0>, min=<0 0>}
    }}
   slot {class=OverlaySlot, properties {horizontalAlignment=HAlign_Fill, verticalAlignment=VAlign_Fill}}
  }
  HorizontalBox_3 {
   widgetClassPath=&HorizontalBox
   parent=Overlay_10
   widget=HorizontalBox_3
   properties {visibility=Visible}
   slot {class=OverlaySlot}
  }
  VerticalBox_2 {
   widgetClassPath=&VerticalBox
   parent=HorizontalBox_3
   widget=VerticalBox_2
   properties {visibility=Visible}
   slot {class=HorizontalBoxSlot, properties {padding=<ltrb 40 24 44 16>, size {sizeRule=Fill, value=1}}}
  }
  ScaleBox_1 {
   widgetClassPath=&ScaleBox
   parent=VerticalBox_2
   widget=ScaleBox_1
   properties {visibility=Visible, renderTransform {angle=-2, scale=<1 1>, shear=<0 0>, translation=<0 0>}}
   slot {class=VerticalBoxSlot}
  }
  TitleRegular {
   widgetClassPath=&CommonTextBlock
   bIsVariable=true
   parent=ScaleBox_1
   widget=TitleRegular
   properties {
    style=/FortUILibrary/Styles/Typography/Foundation/TextStyle-Header-M-BrightBlue.TextStyle-Header-M-BrightBlue_C
    text="AVAILABLE NOW!"
    visibility=Visible
    colorAndOpacity {colorUseRule=UseColor_Specified, specifiedColor=<rgba 0.0097 0.2623 1 1>}
    font {
     bForceMonospaced=false
     bMaterialIsStencil=false
     fontMaterial=null
     fontObject=&Burbank
     letterSpacing=0
     monospacedWidth=1
     size=32
     skewAmount=0
     typefaceFontName=Burbank-Black
     outlineSettings {
      bApplyOutlineToDropShadows=false
      bMiteredCorners=false
      bSeparateFillAlpha=false
      outlineMaterial=null
      outlineSize=0
      outlineColor=<rgba 0 0 0 1>
     }
    }
    renderTransform {angle=0, scale=<1 1>, shear=<-16 0>, translation=<0 0>}
    shadowOffset=<0 0>
   }
   slot {class=ScaleBoxSlot}
  }
  DescriptionRegular {
   widgetClassPath=&CommonTextBlock
   bIsVariable=true
   parent=VerticalBox_2
   widget=DescriptionRegular
   properties {
    autoWrapText=true
    style=&TextStyle-Base-XS-B_C
    text=IN THE ITEM SHOP FOR A LIMITED TIME.
    visibility=Visible
    font {
     bForceMonospaced=false
     bMaterialIsStencil=false
     fontMaterial=null
     fontObject=&NotoSans
     letterSpacing=0
     monospacedWidth=1
     size=16
     skewAmount=0
     typefaceFontName=NotoSans-Bold
     outlineSettings {
      bApplyOutlineToDropShadows=false
      bMiteredCorners=false
      bSeparateFillAlpha=false
      outlineMaterial=null
      outlineSize=0
      outlineColor=<rgba 0 0 0 1>
     }
    }
    renderTransform {angle=0, scale=<1 1>, shear=<-8 0>, translation=<0 0>}
    shadowOffset=<0 0>
   }
   slot {class=VerticalBoxSlot}
  }
  SmallPreviewImageRB {
   widgetClassPath=&SizeBox
   parent=RegularToast
   widget=SmallPreviewImageRB
   properties {
    bOverride_HeightOverride=true
    bOverride_WidthOverride=true
    heightOverride=128
    visibility=Visible
    widthOverride=128
    renderTransform {angle=0, scale=<1 1>, shear=<0 0>, translation=<-72 -8>}
   }
   slot {class=OverlaySlot, properties {padding=<ltrb -30 0 0 0>}}
  }
  Overlay {widgetClassPath=&Overlay, parent=SmallPreviewImageRB, widget=Overlay, properties {visibility=Visible}, slot {class=SizeBoxSlot}}
  SmallPreviewPortraitRB {
   widgetClassPath=&Image
   bIsVariable=true
   parent=Overlay
   widget=SmallPreviewPortraitRB
   properties {brush {
     bIsDynamicallyLoaded=false
     drawAs=Image
     imageType=NoImage
     mirroring=NoMirror
     resourceName=None
     resourceObject=&M_Toast_SkewBoxPortrait
     tiling=NoTile
     imageSize=<128 128>
     margin=<ltrb 0 0 0 0>
     outlineSettings {
      bUseBrushTransparency=false
      roundingType=HalfHeightRadius
      width=0
      color {colorUseRule=UseColor_Specified, specifiedColor=<rgba 0 0 0 0>}
      cornerRadii=<0 0 0 0>
     }
     tintColor {colorUseRule=UseColor_Specified, specifiedColor=<rgba 1 1 1 1>}
     uVRegion {bIsValid=false, max=<0 0>, min=<0 0>}
    }}
   slot {class=OverlaySlot, properties {horizontalAlignment=HAlign_Fill, verticalAlignment=VAlign_Fill}}
  }
  FlashingBangRB {
   widgetClassPath=&Overlay
   parent=Overlay
   widget=FlashingBangRB
   properties {visibility=Visible, renderTransform {angle=0, scale=<1 1>, shear=<0 0>, translation=<14 -20>}}
   slot {class=OverlaySlot, properties {horizontalAlignment=HAlign_Right}}
  }
  Image {
   widgetClassPath=&Image
   parent=FlashingBangRB
   widget=Image
   properties {brush {
     bIsDynamicallyLoaded=false
     drawAs=Image
     imageType=NoImage
     mirroring=NoMirror
     resourceName=None
     resourceObject=&M_FlashingBang
     tiling=NoTile
     imageSize=<40 40>
     margin=<ltrb 0 0 0 0>
     outlineSettings {
      bUseBrushTransparency=false
      roundingType=HalfHeightRadius
      width=0
      color {colorUseRule=UseColor_Specified, specifiedColor=<rgba 0 0 0 0>}
      cornerRadii=<0 0 0 0>
     }
     tintColor {colorUseRule=UseColor_Specified, specifiedColor=<rgba 1 1 1 1>}
     uVRegion {bIsValid=false, max=<0 0>, min=<0 0>}
    }}
   slot {class=OverlaySlot}
  }
  CommonTextBlock {
   widgetClassPath=&CommonTextBlock
   parent=FlashingBangRB
   widget=CommonTextBlock
   properties {
    style=&TextStyle-Header-M-YellowWhitePulse_C
    text="!"
    visibility=Visible
    font {
     bForceMonospaced=false
     bMaterialIsStencil=false
     fontMaterial=&M_Font_SimplePulse_YellowWhite
     fontObject=&Burbank
     letterSpacing=0
     monospacedWidth=1
     size=32
     skewAmount=0
     typefaceFontName=Burbank-Black
     outlineSettings {
      bApplyOutlineToDropShadows=false
      bMiteredCorners=false
      bSeparateFillAlpha=false
      outlineMaterial=null
      outlineSize=0
      outlineColor=<rgba 0 0 0 1>
     }
    }
    renderTransform {angle=5, scale=<0.875 0.75>, shear=<-8 0>, translation=<0 0>}
    shadowOffset=<0 0>
   }
   slot {class=OverlaySlot, properties {horizontalAlignment=HAlign_Center, verticalAlignment=VAlign_Center, padding=<ltrb 0 -8 0 -8>}}
  }
 ]
}
