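# Lobby widget that switches between the built-in WBP_Expandable_Level and a SeasonPass plugin-injected expandable level widget.
# On initialization, the UseNewSeasonLevelWidgetInFrontend config flag, defaulting to true, selects the switcher's active child.
# OnShowExpandableLevel stores the desired visibility and applies it to the widget the flag selects.
# If the injected widget is not ready, it rebinds to the slot's injection delegate and applies the stored visibility on load.
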WidgetBlueprint /FortniteCoreUI/UI/Frontend/Lobby/Habanero/WBP_Expandable_Level_Switcher
graphs [EventGraph {nodes [
   n1 = OnShowExpandableLevel!() -> (OutputDelegate, then, ShouldShow:bool=~) {is=title K2Node_CustomEvent}
   > Set DesiredChildVisibility(DesiredChildVisibility:ESlateVisibility Enum=Select(Option 0=Collapsed, Option 1=SelfHitTestInvisible, Index=@n1.ShouldShow)) -> (Output_Get:ESlateVisibility Enum=Visible)
   > n2 = Branch(Condition=GetConfigDataBoolWithDefault!(EntryName:String=UseNewSeasonLevelWidgetInFrontend, bDefaultValue:bool=true) -> (ReturnValue:bool))
   n3 = Get FortPluginWidgetInjectionSlot() -> (FortPluginWidgetInjectionSlot:Fort Plugin Widget Injection Slot Object Reference)
   > n4 = Branch(Condition=IsInjectedWidgetReady!(self:Fort Plugin Widget Injection Slot Object Reference=@n3) -> (ReturnValue:bool))
   SetVisibility(
    execute=[@n5.then, @n4.then]
    self=GetInjectedWidget!(self:Fort Plugin Widget Injection Slot Object Reference=Get FortPluginWidgetInjectionSlot()) -> (ReturnValue:Widget Object Reference)
    InVisibility=Get DesiredChildVisibility() -> (DesiredChildVisibility:ESlateVisibility Enum=Visible)
   )
   n5 = OnWidgetInjectedLoaded!() -> (OutputDelegate, then, Widget:Widget Object Reference) {is=title K2Node_CustomEvent}
   Unbind Event from On Widget Injected(execute=@n4.else, self:Fort Plugin Widget Injection Slot Object Reference=@n3, Delegate=@n5) {is=title K2Node_RemoveDelegate}
   > Assign On Widget Injected!(execute:Exec, self:Fort Plugin Widget Injection Slot Object Reference=@n3, Delegate:"Delegate (by ref)"=@n5) -> (then:Exec) {is=title K2Node_AssignDelegate}
   SetVisibility(
    execute=@n2.else
    self=Get WBP_Expandable_Level() -> (WBP_Expandable_Level:WBP Expandable Level Object Reference)
    InVisibility=Get DesiredChildVisibility() -> (DesiredChildVisibility=Visible)
   )
   Event OnInitialized()
   > SetActiveWidget(
    self=Get Switcher_ExpandableLevel() -> (Switcher_ExpandableLevel:Common Visibility Switcher Object Reference)
    Widget=Select(
     Option 0:Widget Object Reference=Get WBP_Expandable_Level()
     Option 1:Widget Object Reference=Get FortPluginWidgetInjectionSlot()
     Index=GetConfigDataBoolWithDefault(EntryName=UseNewSeasonLevelWidgetInFrontend, bDefaultValue=true)
    ) -> (ReturnValue:Widget Object Reference)
   )
  ]}]
variables [DesiredChildVisibility {type=ESlateVisibility, default=Collapsed}]
widgets {
 info {inheritedWidgetCount=0, parentClass=&UserWidget, rootWidgetClass=&Overlay, widgetCount=4}
 widgets [
  Overlay_28 {widgetClassPath=&Overlay, widget=Overlay_28}
  Switcher_ExpandableLevel {
   widgetClassPath=&CommonVisibilitySwitcher
   bIsVariable=true
   parent=Overlay_28
   widget=Switcher_ExpandableLevel
   slot {class=OverlaySlot}
  }
  WBP_Expandable_Level {
   widgetClassPath=&/WBP_Expandable_Level
   bIsVariable=true
   parent=Switcher_ExpandableLevel
   widget=WBP_Expandable_Level
   properties {holdData=&CommonUI_HoldData_Base_C, visibility=Collapsed}
   slot {class=CommonVisibilitySwitcherSlot, properties {horizontalAlignment=HAlign_Right, padding=<ltrb 0 16 0 0>}}
  }
  FortPluginWidgetInjectionSlot {
   widgetClassPath=/Game/UI/Plugin/FortPluginWidgetInjectionSlot
   bIsVariable=true
   parent=Switcher_ExpandableLevel
   widget=FortPluginWidgetInjectionSlot
   properties {pluginId=SeasonPass, pluginWidgetId=SeasonPassExandableLevel}
   slot {class=CommonVisibilitySwitcherSlot}
  }
 ]
}
