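# Expandable lobby widget, built on WBP_Expandable_Base, showing the player's Ranked (Habanero) badge, name, progress and protection shield; clicking pushes a mode info modal.
# Setters combine ranked view model values with display-data overrides to resolve the tier, falling back to an unranked state.
# Tier category drives icon size, progress ring and text, with Unreal showing leaderboard position instead of percentage.
# Division pips come from the Roman numeral parsed from the rank name.
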WidgetBlueprint /FortniteCoreUI/UI/Frontend/Lobby/Habanero/WBP_Expandable_Rank
graphs [
 EventGraph {nodes [
   n1 = Event PreConstruct()
   > Parent: PreConstruct(IsDesignTime=@n1.IsDesignTime)
   Event Construct()
   > Parent: Construct()
   > InitializeRankedViewModal(self:Ranked View Model Object Reference=Get FortRankedViewModel() -> (FortRankedViewModel:Ranked View Model Object Reference))
   > Setup Rank Visual()
   Event On Clicked()
   > Parent: On Clicked() {is=title K2Node_CallParentFunction}
   > Fire Event Element Interacted(Element=Self-Reference())
   n2 = GetRankedWidgetInfoModalOverride!(self:Fort Global UIContext Object Reference=Get FortGlobalUIContext() -> (ReturnValue:Fort Global UIContext Object Reference) {is=title K2Node_GetSubsystem}) -> (ReturnValue:Common Activatable Widget Soft Class Reference)
   > n3 = PushPrimaryContentSoftClassAsync!(
    execute:Exec
    SoftWidgetClass:Common Activatable Widget Soft Class Reference=Select(
     Option 0:Common Activatable Widget Soft Class Reference=/Game/UI/Frontend/Lobby/Habanero/ModeInfo/WBP_ModeInfo_Modal.WBP_ModeInfo_Modal_C
     Option 1:Common Activatable Widget Soft Class Reference=@n2
     Index=IsValidSoftClassReference!(SoftClassReference:"Object Soft Class Reference (by ref)"=@n2) -> (ReturnValue:bool)
    ) -> (ReturnValue:Common Activatable Widget Soft Class Reference)
   ) -> (then:Exec, OnPrimaryContentPushed:Exec, ActivatableWidget:Common Activatable Widget Object Reference) {is=title K2Node_AsyncAction}
   n4 = Cast To FortHabaneroInfoModal!(execute=@n3.OnPrimaryContentPushed, Object=@n3) -> (then, CastFailed, AsFort Habanero Info Modal:Fort Habanero Info Modal Object Reference) {is=title K2Node_DynamicCast}
   > SetGameType!(execute, self:Fort Habanero Info Modal Object Reference=@n4, GameMode:String=Get Rank Type() -> (Rank Type:String)) -> (then)
  ]}
 RefreshDisplayedTextToNextRank {nodes [
   RefreshDisplayedTextToNextRank() {is=title K2Node_FunctionEntry}
   n1 = NOT Boolean(A=TextIsEmpty(InText=Get Value_TextToNextRankPercent() -> (Value_TextToNextRankPercent:Text)))
   n2 = Get Has Rank Protection() -> (Has Rank Protection:bool)
   > SetVisibility(
    self=Get Text_RankProtection() -> (Text_RankProtection:Common Text Object Reference)
    InVisibility=Select(Option 0:ESlateVisibility Enum=Collapsed, Option 1:ESlateVisibility Enum=SelfHitTestInvisible, Index=AND Boolean(A=@n1, B=@n2)) -> (ReturnValue:ESlateVisibility Enum)
   )
   n3 = Get Text_ToNextRankPercent() -> (Text_ToNextRankPercent:Common Text Object Reference)
   n4 = Get Value_TextToNextRankPercent()
   > "SetText (Text)"(self=@n3, InText=@n4)
   n5 = TextIsEmpty(InText=@n4)
   > SetVisibility(self=@n3, InVisibility=Select(Option 0=Collapsed, Option 1=SelfHitTestInvisible, Index=NOT Boolean(A=@n5)))
  ]}
 Set Current Player Rank {nodes [
   n1 = Set Current Player Rank!() -> (then, CurrentRankIndex:int) {is=title K2Node_FunctionEntry}
   > n2 = Branch(Condition=Get Is Player Ranked() -> (Is Player Ranked:bool))
   n3 = Find(
    TargetMap:Map of Strings to Fort Habanero Mode List Structures=Get EntryData!(self:Fort Habanero Display Data Object Reference=Get RankInfo() -> (RankInfo:Fort Habanero Display Data Object Reference)) -> (EntryData:Map of Strings to Fort Habanero Mode List Structures)
    Key=Get Rank Type()
   ) -> (Value:Fort Habanero Mode List Structure)
   > Branch(Condition=@n3.ReturnValue)
   n4 = Break Fort Habanero Mode List!(FortHabaneroModeList:"Fort Habanero Mode List Structure (by ref)"=@n3) -> (TierList:Array of Fort Habanero Tier Structures, bShowTopLeaderboardIcon:bool=~, TopTierTopLeaderboardPositions:Fort Habanero Tier Structure) {is=title K2Node_BreakStruct}
   > Branch(Condition=Is Valid Index!(TargetArray:Array of Fort Habanero Tier Structures=@n4, IndexToTest:int=@n1) -> (ReturnValue:bool))
   n5 = Get FortRankedViewModel()
   n6 = Get RankType!(self:Ranked View Model Object Reference=@n5) -> (RankType:String)
   n7 = GetHabaneroDisplayDataOverride!(
    InDisplayData:Fort Habanero Display Data Object Reference=Get RankInfo()
    InOverrideData:Fort Compete UIHabanero Tier Override List Object Reference=Get RankInfoOverrides() -> (RankInfoOverrides:Fort Compete UIHabanero Tier Override List Object Reference)
    InKey:"Fort Compete UIHabanero Tier Override Key Structure (by ref)"=Make Fort Compete UIHabanero Tier Override Key!(
     RankedTrack:Name=String To Name(InString=@n6)
     Season:String
     Rank:int=Get CurrentRank!(self:Ranked View Model Object Reference=@n5) -> (CurrentRank:int=0)
     MatchmakingRating:int=Get MatchmakingRating!(self:Ranked View Model Object Reference=@n5) -> (MatchmakingRating:int=0)
     CurrentPosition:int=Get CurrentPlayerPosition!(self:Ranked View Model Object Reference=@n5) -> (CurrentPlayerPosition:int=0)
    ) -> (FortCompeteUIHabaneroTierOverrideKey:Fort Compete UIHabanero Tier Override Key Structure) {is=title K2Node_MakeStruct}
   ) -> (
    OutData_ForegroundIcon:Material Instance Soft Object Reference
    OutData_ForegroundIconSmall:Material Instance Soft Object Reference
    OutData_BackgroundIcon:Material Instance Soft Object Reference
    OutData_BackgroundIconSmall:Material Instance Soft Object Reference
    OutData_WhiteIcon:Texture 2D Soft Object Reference
    OutData_SimpleIconSmall:Texture 2D Soft Object Reference
    OutData_Name:Text
    OutData_TierType:Fort Habanero Tier Type Data Object Reference
    OutData_AdditionalLeftIcon:Material Instance Soft Object Reference
    OutData_AdditionalRightIcon:Material Instance Soft Object Reference
    ReturnValue:bool
   )
   > Branch(Condition=IsValid(Object=@n7.OutData_TierType))
   > Set CurrentCategoryData(CurrentCategoryData:Fort Habanero Tier Type Data Object Reference=@n7.OutData_TierType) -> (Output_Get:Fort Habanero Tier Type Data Object Reference)
   > Set CurrentTierCategory(CurrentTierCategory:EHabaneroTierType Enum=Get Category!(self:Fort Habanero Tier Type Data Object Reference=@n7.OutData_TierType) -> (Category:EHabaneroTierType Enum=HabaneroSauce0)) -> (Output_Get:EHabaneroTierType Enum=HabaneroSauce0)
   > "SetText (Text)"(self=Get Text_RankName() -> (Text_RankName:Common Text Object Reference), InText=@n7.OutData_Name)
   n8 = Get FortRankedViewModel()
   > DisplayBadge!(
    execute
    self:Fort Habanero Tier Widget Object Reference=Get WBP_Habanero_Icon() -> (WBP_Habanero_Icon:WBP Habanero Icon Object Reference)
    HabaneroType:String=Get RankType(self=@n8)
    CurrentHabanero:int=Get CurrentRank(self=@n8) -> (CurrentRank=0)
    SeasonName:String=-1
    MatchmakingRating:int=Get MatchmakingRating(self=@n8) -> (MatchmakingRating=0)
    CurrentPlayerPosition:int=Get CurrentPlayerPosition(self=@n8) -> (CurrentPlayerPosition=0)
   ) -> (then)
   > Set bIsUnranked(bIsUnranked:bool="Equal (Enum)"(
     A:EHabaneroTierType Enum=Get CurrentTierCategory() -> (CurrentTierCategory:EHabaneroTierType Enum=HabaneroSauce0)
     B:EHabaneroTierType Enum=HabaneroSauce0
    )) -> (Output_Get:bool)
   > Set bIsUnreal(bIsUnreal:bool="Equal (Enum)"(A=Get CurrentTierCategory() -> (CurrentTierCategory=HabaneroSauce0), B=HabaneroSauce8)) -> (Output_Get:bool)
   > SetActiveWidget(
    self=Get RankedUnranked_Switcher() -> (RankedUnranked_Switcher:Common Visibility Switcher Object Reference)
    Widget=Select(
     Option 0:Widget Object Reference=Get VerticalBox_RankInfo() -> (VerticalBox_RankInfo:Vertical Box Object Reference)
     Option 1:Widget Object Reference=Get Text_Unranked() -> (Text_Unranked:Common Text Object Reference)
     Index=Get bIsUnranked() -> (bIsUnranked:bool)
    ) -> (ReturnValue:Widget Object Reference)
   )
   > Set Rank Rectangle!(execute, self, RankName:Text=@n7.OutData_Name) -> (then) {is=title K2Node_CallFunction}
   > Setup Rank Visual()
   Set Default Player UnRank(execute=@n2.else) {is=title K2Node_CallFunction}
   SetRankText()
   "Get (a copy)"(Array:Array of Fort Habanero Tier Structures=@n4, Dimension 1=@n1) -> (Output:Fort Habanero Tier Structure)
   Reroute Node(InputPin:Material Instance Soft Object Reference=@n7.OutData_ForegroundIconSmall) -> (OutputPin:Material Instance Soft Object Reference)
   Reroute Node(InputPin=@n7.OutData_BackgroundIconSmall)
  ]}
 Set Current Player Rank Position {nodes [
   n1 = Set Current Player Rank Position!() -> (then, CurrentPlayerRankPosition:int) {is=title K2Node_FunctionEntry}
   > Set Played Rank Position(Played Rank Position:int=@n1) -> (Output_Get:int=0)
  ]}
 Set Default Player UnRank {nodes [
   Set Default Player UnRank() {is=title K2Node_FunctionEntry}
   > UpdateDetails!(
    execute
    self:WBP Habanero Icon Object Reference=Get WBP_Habanero_Icon()
    ForegroundIcon:Material Instance Soft Object Reference=/Game/UI/Frontend/Lobby/Habanero/Habanero_Materials/MaterialInstances/Icons/Foreground/Small/MI_UI_Habanero_Foreground_00_Small
    BackgroundIcon:Material Instance Soft Object Reference=/Game/UI/Frontend/Lobby/Habanero/Habanero_Materials/MaterialInstances/Icons/Background/MI_UI_Habanero_Background_00
   ) -> (then)
   n1 = Get Text_Unranked()
   > SetActiveWidget(self=Get RankedUnranked_Switcher(), Widget=@n1)
   > "SetText (Text)"(self=@n1, InText="LOCTABLE(\"/Game/UI/Habanero/HabaneroStrings\", \"infomodal_unrankedtext\")")
   > Set CurrentTierCategory(CurrentTierCategory=HabaneroSauce0) -> (Output_Get=HabaneroSauce0)
   > Setup Rank Visual()
  ]}
 Set Has Rank Protection {nodes [
   n1 = Set Has Rank Protection!() -> (then, Has Protection:bool=~) {is=title K2Node_FunctionEntry}
   > n2 = Set Has Rank Protection(Has Rank Protection:bool=@n1) -> (Output_Get:bool) {is=title K2Node_VariableSet}
   > RefreshDisplayedTextToNextRank() {is=title K2Node_CallFunction}
   > n3 = Branch(Condition=@n2)
   n4 = Get WBP_Habanero_ShieldSmall() -> (WBP_Habanero_ShieldSmall:WBP Habanero Shield Small Object Reference)
   > SetVisibility(self=@n4, InVisibility=HitTestInvisible)
   > PlayShieldResetedAnimation(self:WBP Habanero Shield Small Object Reference=@n4)
   SetVisibility(execute=@n3.else, self=@n4, InVisibility=Collapsed)
  ]}
 Set Is Island Ranked {nodes [
   n1 = Set Is Island Ranked!() -> (then, IsIslandRanked:bool=~) {is=title K2Node_FunctionEntry}
   > n2 = Set Is Island Ranked(Is Island Ranked:bool=@n1) -> (Output_Get:bool) {is=title K2Node_VariableSet}
   n3 = Get bEnableLobbyRankWidget!(self:Fort Runtime Options Object Reference=GetRuntimeOptions()) -> (bEnableLobbyRankWidget:bool)
   > Set Visibility!(
    execute
    Visibility=Select(Option 0:ESlateVisibility Enum=Collapsed, Option 1:ESlateVisibility Enum=SelfHitTestInvisible, Index=AND Boolean(A=@n2, B=@n3)) -> (ReturnValue:ESlateVisibility Enum)
   ) -> (then, Output_Get)
  ]}
 Set Is Player Ranked {nodes [
   n1 = Set Is Player Ranked!() -> (then, IsPlayerRanked:bool=~) {is=title K2Node_FunctionEntry}
   > Set Is Player Ranked(Is Player Ranked:bool=@n1) -> (Output_Get:bool) {is=title K2Node_VariableSet}
  ]}
 Set Rank Circle Progress {nodes [
   n1 = Set Rank Circle Progress!() -> (then, Progress:double=~) {is=title K2Node_FunctionEntry}
   > Set ProgressValue(ProgressValue:double=@n1) -> (Output_Get:double)
   > n2 = GetDynamicMaterial(self=Get RankProgressBar() -> (RankProgressBar:Image Object Reference))
   > SetScalarParameterValue(self=@n2, ParameterName=Get ProgressParameter() -> (ProgressParameter:Name=None), Value=Get ProgressValue() -> (ProgressValue:double))
   > SetRankText()
  ]}
 Set Rank Rectangle {nodes [
   n1 = Set Rank Rectangle!() -> (then, RankName:Text) {is=title K2Node_FunctionEntry}
   > Setup Rank Rectangle!(execute, self, ArabicNum:String=Split(SourceString="To String (Text)"(InText=@n1), InStr=" ", SearchDir=FromStart).RightS) -> (then)
  ]}
 SetRankText {nodes [
   SetRankText() {is=title K2Node_FunctionEntry}
   n1 = Get CurrentTierCategory() -> (CurrentTierCategory=HabaneroSauce0)
   n2 = "Equal (Enum)"(A=@n1, B=HabaneroSauce8)
   n3 = float * float(A=Get ProgressValue(), B=100) {is=title K2Node_PromotableOperator}
   n4 = Format Text!(
    Format="LOCTEXT(\"{number}%\")"
    number:int="Clamp (Integer)"(Value=Round(A=@n3) {is=title K2Node_CallFunction}, Min=1, Max=99) -> (ReturnValue=0)
   ) -> (Result)
   n5 = Get Played Rank Position() -> (Played Rank Position:int=0)
   n6 = Select(
    Option 0:Text=Format Text!(Format="LOCTEXT(\"##{rank}\")", rank:Text="To Text (Integer)"(Value=@n5)) -> (Result)
    Option 1:Text=Get UnrealRankCapText() -> (UnrealRankCapText:Text)
    Index="integer >= integer"(A=@n5, B=100000)
   ) -> (ReturnValue:Text)
   > Set Value_TextToNextRankPercent(Value_TextToNextRankPercent:Text=Select(Option 0=Select(Option 0=@n4, Option 1=@n6, Index=@n2), Index="Equal (Enum)"(A=@n1, B=HabaneroSauce0))) -> (Output_Get:Text)
   > RefreshDisplayedTextToNextRank()
  ]}
 SetRankType {nodes [
   n1 = SetRankType!() -> (then, RankType:String) {is=title K2Node_FunctionEntry}
   > Set Rank Type(Rank Type:String=@n1) -> (Output_Get:String)
   > n2 = Get!(execute:Exec) -> (then:Exec, else:Exec, FortRankedViewModel:Ranked View Model Object Reference)
   > Set Current Player Rank!(execute, self, CurrentRankIndex:int=Get CurrentRank(self=@n2) -> (CurrentRank=0)) -> (then) {is=title K2Node_CallFunction}
  ]}
 Setup Rank Rectangle {nodes [
   n1 = Setup Rank Rectangle!() -> (then, ArabicNum:String) {is=title K2Node_FunctionEntry}
   > Clear(TargetArray:Array of WBP Hero Rank Rectangle Object References=Get RectangleArray() -> (RectangleArray:Array of WBP Hero Rank Rectangle Object References))
   > Reset!(
    execute
    self:Dynamic Entry Box Object Reference=Get DynamicEntryBox() -> (DynamicEntryBox:Dynamic Entry Box Object Reference)
    bDeleteWidgets:bool=true
   ) -> (then)
   n2 = Get RectangleMap() -> (RectangleMap:Map of Strings to Integers)
   > For Loop(LastIndex=int - int(A=Find(TargetMap:Map of Strings to Integers=@n2, Key=@n1) -> (Value:int)))
   > n3 = Create Entry!(execute, self:Dynamic Entry Box Object Reference=Get DynamicEntryBox()) -> (then, ReturnValue:User Widget Object Reference)
   > n4 = Cast To WBP_HeroRank_Rectangle!(execute, Object=@n3) -> (then, CastFailed, AsWBP Hero Rank Rectangle:WBP Hero Rank Rectangle Object Reference) {is=title K2Node_DynamicCast}
   > Add(
    TargetArray:Array of WBP Hero Rank Rectangle Object References=Get RectangleArray()
    NewItem:"WBP Hero Rank Rectangle Object Reference (by ref)"=@n4
   ) {is=title K2Node_CallArrayFunction}
   > n5 = Cast To BP_UI_Habanero_TierTypeDataBase!(execute, Object=Get CurrentCategoryData() -> (CurrentCategoryData:Fort Habanero Tier Type Data Object Reference)) -> (then, CastFailed, AsBP UI Habanero Tier Type Data Base:BP UI Habanero Tier Type Data Base Object Reference) {is=title K2Node_DynamicCast}
   > Set Rectangle Color!(
    execute
    self:WBP Hero Rank Rectangle Object Reference=@n4
    Color:Linear Color Structure="To LinearColor (Color)"!(InColor:Color Structure=Get Text_Color!(self:BP UI Habanero Tier Type Data Base Object Reference=@n5) -> (Text_Color:Color Structure)) -> (ReturnValue:Linear Color Structure)
   ) -> (then) {is=title K2Node_CallFunction}
   Get Tier Foreground Color!(self:WBP Habanero Icon Object Reference=Get WBP_Habanero_Icon()) -> (Color:Linear Color Structure) {is=title K2Node_CallFunction}
  ]}
 Setup Rank Visual {nodes [
   Setup Rank Visual() {is=title K2Node_FunctionEntry}
   n1 = Get SmallSize() -> (SmallSize:Vector 2D Structure)
   n2 = Get MediumSize() -> (MediumSize:Vector 2D Structure)
   > SetSize!(
    execute
    self:WBP Habanero Icon Object Reference=Get WBP_Habanero_Icon()
    IconSize:Vector 2D Structure=Select!(
     HabaneroSauce0:Vector 2D Structure=@n2
     HabaneroSauce1:Vector 2D Structure=@n1
     HabaneroSauce2:Vector 2D Structure=@n1
     HabaneroSauce3:Vector 2D Structure=@n1
     HabaneroSauce4:Vector 2D Structure=@n1
     HabaneroSauce5:Vector 2D Structure=@n1
     HabaneroSauce6:Vector 2D Structure=@n2
     HabaneroSauce7:Vector 2D Structure=@n2
     HabaneroSauce8:Vector 2D Structure=Get LargeSize() -> (LargeSize:Vector 2D Structure)
     Index:EHabaneroTierType Enum=Get CurrentTierCategory() -> (CurrentTierCategory=HabaneroSauce0)
    ) -> (ReturnValue:Vector 2D Structure)
   ) -> (then)
   n3 = Get CurrentTierCategory() -> (CurrentTierCategory=HabaneroSauce0)
   > SetVisibility(
    self=Get DynamicEntryBox()
    InVisibility=Select!(
     Option 0:ESlateVisibility Enum=SelfHitTestInvisible
     Option 1:ESlateVisibility Enum=Collapsed
     Index:bool=Select!(
      HabaneroSauce0:bool=true
      HabaneroSauce1:bool=false
      HabaneroSauce2:bool=false
      HabaneroSauce3:bool=false
      HabaneroSauce4:bool=false
      HabaneroSauce5:bool=false
      HabaneroSauce6:bool=true
      HabaneroSauce7:bool=true
      HabaneroSauce8:bool=true
      Index:EHabaneroTierType Enum=@n3
     ) -> (ReturnValue:bool)
    ) -> (ReturnValue:ESlateVisibility Enum)
   )
   n4 = Get CurrentTierCategory() -> (CurrentTierCategory=HabaneroSauce0)
   > SetVisibility(
    self=Get RankProgressBar()
    InVisibility=Select!(
     Option 0:ESlateVisibility Enum=SelfHitTestInvisible
     Option 1:ESlateVisibility Enum=Collapsed
     Index:bool=Select!(
      HabaneroSauce0:bool=false
      HabaneroSauce1:bool=false
      HabaneroSauce2:bool=false
      HabaneroSauce3:bool=false
      HabaneroSauce4:bool=false
      HabaneroSauce5:bool=false
      HabaneroSauce6:bool=false
      HabaneroSauce7:bool=false
      HabaneroSauce8:bool=true
      Index:EHabaneroTierType Enum=@n4
     ) -> (ReturnValue:bool)
    ) -> (ReturnValue:ESlateVisibility Enum)
   )
  ]}
]
variables [
 bIsUnranked {type=boolean}
 bIsUnreal {type=boolean}
 CurrentCategoryData {type=FortHabaneroTierTypeData, default=null}
 CurrentTierCategory {type=EHabaneroTierType, default=HabaneroSauce0}
 Has Rank Protection {type=boolean}
 Is Island Ranked {type=boolean}
 Is Player Ranked {type=boolean}
 Played Rank Position {type=integer}
 ProgressParameter {type=string, default=Progress}
 ProgressValue {type=number}
 Rank Type {type=string, default=""}
 RankInfoOverrides {type=FortCompeteUIHabaneroTierOverrideList, default=/Game/UI/Habanero/DA_RankDisplayDataOverride}
 RectangleArray {type="array<WBP_HeroRank_Rectangle_C>", default []}
 RectangleMap {type="map<integer>", default {I=1, II=2, III=3, IV=4, V=5}}
 UnrealRankCapText {type=string, default="##100k+"}
 Value_TextToNextRankPercent {type=string, default=""}
]
widgets {
 info {
  inheritedWidgetCount=0
  namedSlotCount=2
  parentClass=/Game/UI/Frontend/Lobby/Habanero/WBP_Expandable_Base.WBP_Expandable_Base_C
  rootWidgetClass=null
  widgetCount=16
 }
 widgets [
  GridPanel_LevelBox {widgetClassPath=&GridPanel, widget=GridPanel_LevelBox, properties {columnFill [1], rowFill [1]}}
  FortVisualAttachment_0 {
   widgetClassPath=&FortVisualAttachment
   parent=GridPanel_LevelBox
   widget=FortVisualAttachment_0
   properties {bOverride_HeightOverride=true, bOverride_WidthOverride=true, heightOverride=144, widthOverride=144, contentAnchor=<0.5 0.5>}
   slot {class=GridSlot, properties {horizontalAlignment=HAlign_Center, verticalAlignment=VAlign_Center}}
  }
  RankProgressBar {
   widgetClassPath=&Image
   bIsVariable=true
   parent=FortVisualAttachment_0
   widget=RankProgressBar
   properties {brush {
     bIsDynamicallyLoaded=false
     drawAs=Image
     imageType=NoImage
     mirroring=NoMirror
     resourceName=None
     resourceObject=/Game/UI/Frontend/Lobby/Habanero/Habanero_Materials/MaterialInstances/ProgressBars/MI_UI_Habanero_ProgressBar_RankLevel
     tiling=NoTile
     imageSize=<32 32>
     margin=<ltrb 0 0 0 0>
     outlineSettings {
      bUseBrushTransparency=false
      roundingType=HalfHeightRadius
      width=0
      color {colorUseRule=UseColor_Specified, specifiedColor=<rgba 0 0 0 0>}
      cornerRadii=<0 0 0 0>
     }
     tintColor {colorUseRule=UseColor_Specified, specifiedColor=<rgba 1 1 1 1>}
     uVRegion {bIsValid=false, max=<0 0>, min=<0 0>}
    }}
   slot {class=SizeBoxSlot}
  }
  RadialDropShadow {
   widgetClassPath=&Image
   bIsVariable=true
   parent=GridPanel_LevelBox
   widget=RadialDropShadow
   properties {brush {
     bIsDynamicallyLoaded=false
     drawAs=Image
     imageType=NoImage
     mirroring=NoMirror
     resourceName=None
     resourceObject=/Game/UI/Frontend/Lobby/Habanero/Habanero_Materials/MaterialInstances/Background/MI_UI_Habanero_RadialDropShadow_Black40
     tiling=NoTile
     imageSize=<32 32>
     margin=<ltrb 0 0 0 0>
     outlineSettings {
      bUseBrushTransparency=false
      roundingType=HalfHeightRadius
      width=0
      color {colorUseRule=UseColor_Specified, specifiedColor=<rgba 0 0 0 0>}
      cornerRadii=<0 0 0 0>
     }
     tintColor {colorUseRule=UseColor_Specified, specifiedColor=<rgba 1 1 1 1>}
     uVRegion {bIsValid=false, max=<0 0>, min=<0 0>}
    }}
   slot {class=GridSlot}
  }
  VerticalBox_Level {
   widgetClassPath=&VerticalBox
   parent=GridPanel_LevelBox
   widget=VerticalBox_Level
   slot {class=GridSlot, properties {horizontalAlignment=HAlign_Center, layer=1, verticalAlignment=VAlign_Center}}
  }
  WBP_Habanero_Icon {
   widgetClassPath=/Game/UI/Frontend/Lobby/Habanero/WBP_Habanero_Icon
   bIsVariable=true
   parent=VerticalBox_Level
   widget=WBP_Habanero_Icon
   properties {iconMinSize=<64 64>}
   slot {class=VerticalBoxSlot}
  }
  DynamicEntryBox {
   widgetClassPath=&DynamicEntryBox
   bIsVariable=true
   parent=VerticalBox_Level
   widget=DynamicEntryBox
   properties {
    entryWidgetClass=/Game/UI/Frontend/Lobby/Habanero/WBP_HeroRank_Rectangle.WBP_HeroRank_Rectangle_C
    entrySpacing=<4 0>
    renderTransform {angle=0, scale=<1 1>, shear=<0 0>, translation=<0 -2>}
   }
   slot {class=VerticalBoxSlot, properties {horizontalAlignment=HAlign_Center, verticalAlignment=VAlign_Top, padding=<ltrb 0 0 0 4>}}
  }
  WBP_Habanero_ShieldSmall {
   widgetClassPath=/Game/UI/Frontend/Lobby/Habanero/WBP_Habanero_ShieldSmall
   bIsVariable=true
   parent=GridPanel_LevelBox
   widget=WBP_Habanero_ShieldSmall
   properties {visibility=Collapsed, renderTransform {angle=0, scale=<1 1>, shear=<0 0>, translation=<28 22>}}
   slot {class=GridSlot, properties {horizontalAlignment=HAlign_Right, layer=2, verticalAlignment=VAlign_Bottom}}
  }
  GridPanel_40 {widgetClassPath=&GridPanel, widget=GridPanel_40, properties {columnFill [1], rowFill [1]}}
  RankedUnranked_Switcher {
   widgetClassPath=&CommonVisibilitySwitcher
   bIsVariable=true
   parent=GridPanel_40
   widget=RankedUnranked_Switcher
   slot {class=GridSlot, properties {padding=<ltrb 16 0 0 0>}}
  }
  VerticalBox_RankInfo {
   widgetClassPath=&VerticalBox
   bIsVariable=true
   parent=RankedUnranked_Switcher
   widget=VerticalBox_RankInfo
   slot {class=CommonVisibilitySwitcherSlot, properties {verticalAlignment=VAlign_Center}}
  }
  HorizontalBox_0 {
   widgetClassPath=&HorizontalBox
   parent=VerticalBox_RankInfo
   widget=HorizontalBox_0
   properties {flowDirectionPreference=Culture}
   slot {class=VerticalBoxSlot, properties {verticalAlignment=VAlign_Center, size {sizeRule=Fill, value=1}}}
  }
  Text_RankName {
   widgetClassPath=&CommonTextBlock
   bIsVariable=true
   parent=HorizontalBox_0
   widget=Text_RankName
   properties {
    bIsScrollingEnabled=false
    mobileFontSizeMultiplier=1.25
    text=Unranked
    textTransformPolicy=ToUpper
    colorAndOpacity {colorUseRule=UseColor_Specified, specifiedColor=<rgba 0.0048 0.004 0.0116 0.7>}
    font {
     bForceMonospaced=false
     bMaterialIsStencil=false
     fontMaterial=null
     fontObject=&HeadingNow
     letterSpacing=0
     monospacedWidth=1
     size=18
     skewAmount=0
     typefaceFontName=75 Medium
     outlineSettings {
      bApplyOutlineToDropShadows=false
      bMiteredCorners=false
      bSeparateFillAlpha=false
      outlineMaterial=null
      outlineSize=0
      outlineColor=<rgba 0 0 0 1>
     }
    }
    shadowOffset=<0 0>
    shapedTextOptions {bOverride_TextFlowDirection=true, bOverride_TextShapingMethod=false, textFlowDirection=Culture, textShapingMethod=Auto}
   }
   slot {class=HorizontalBoxSlot}
  }
  Text_ToNextRankPercent {
   widgetClassPath=&CommonTextBlock
   bIsVariable=true
   parent=HorizontalBox_0
   widget=Text_ToNextRankPercent
   properties {
    bIsScrollingEnabled=false
    mobileFontSizeMultiplier=1.25
    text=00%
    textTransformPolicy=ToUpper
    colorAndOpacity {colorUseRule=UseColor_Specified, specifiedColor=<rgba 0.0048 0.004 0.0116 0.5>}
    font {
     bForceMonospaced=false
     bMaterialIsStencil=false
     fontMaterial=null
     fontObject=&HeadingNow
     letterSpacing=0
     monospacedWidth=1
     size=18
     skewAmount=0
     typefaceFontName=75 Medium
     outlineSettings {
      bApplyOutlineToDropShadows=false
      bMiteredCorners=false
      bSeparateFillAlpha=false
      outlineMaterial=null
      outlineSize=0
      outlineColor=<rgba 0 0 0 1>
     }
    }
    shadowOffset=<0 0>
    shapedTextOptions {bOverride_TextFlowDirection=true, bOverride_TextShapingMethod=false, textFlowDirection=Culture, textShapingMethod=Auto}
   }
   slot {class=HorizontalBoxSlot, properties {verticalAlignment=VAlign_Top, padding=<ltrb 10 0 10 0>, size {sizeRule=Fill, value=1}}}
  }
  Text_RankProtection {
   widgetClassPath=&CommonTextBlock
   bIsVariable=true
   parent=VerticalBox_RankInfo
   widget=Text_RankProtection
   properties {
    bIsScrollingEnabled=false
    mobileFontSizeMultiplier=1.25
    text=First Match Protected
    textTransformPolicy=ToUpper
    colorAndOpacity {colorUseRule=UseColor_Specified, specifiedColor=<rgba 0.0048 0.004 0.0116 0.5>}
    font {
     bForceMonospaced=false
     bMaterialIsStencil=false
     fontMaterial=null
     fontObject=&HeadingNow
     letterSpacing=0
     monospacedWidth=1
     size=15.75
     skewAmount=0
     typefaceFontName=74 Regular
     outlineSettings {
      bApplyOutlineToDropShadows=false
      bMiteredCorners=false
      bSeparateFillAlpha=false
      outlineMaterial=null
      outlineSize=0
      outlineColor=<rgba 0 0 0 1>
     }
    }
    shadowOffset=<0 0>
    shapedTextOptions {bOverride_TextFlowDirection=true, bOverride_TextShapingMethod=false, textFlowDirection=Culture, textShapingMethod=Auto}
   }
   slot {class=VerticalBoxSlot, properties {verticalAlignment=VAlign_Top, padding=<ltrb 0 0 10 0>, size {sizeRule=Fill, value=1}}}
  }
  Text_Unranked {
   widgetClassPath=&CommonTextBlock
   bIsVariable=true
   parent=RankedUnranked_Switcher
   widget=Text_Unranked
   properties {
    bIsScrollingEnabled=false
    mobileFontSizeMultiplier=1.25
    text="Play one match to \r\nreveal your rank"
    colorAndOpacity {colorUseRule=UseColor_Specified, specifiedColor=<rgba 0.0048 0.004 0.0116 0.702>}
    font {
     bForceMonospaced=false
     bMaterialIsStencil=false
     fontMaterial=null
     fontObject=&HeadingNow
     letterSpacing=0
     monospacedWidth=1
     size=18
     skewAmount=0
     typefaceFontName=75 Medium
     outlineSettings {
      bApplyOutlineToDropShadows=false
      bMiteredCorners=false
      bSeparateFillAlpha=false
      outlineMaterial=null
      outlineSize=0
      outlineColor=<rgba 0 0 0 1>
     }
    }
    shadowOffset=<0 0>
    shapedTextOptions {bOverride_TextFlowDirection=true, bOverride_TextShapingMethod=false, textFlowDirection=Culture, textShapingMethod=Auto}
   }
   slot {class=CommonVisibilitySwitcherSlot, properties {verticalAlignment=VAlign_Center}}
  }
 ]
}
