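# Lobby coach mark that introduces Habanero rank protection as a first-time tip.
# DisplayMessageIfFirstTime requires a visible container, protection, a ranked island and an unseen FTUE flag, then activates after a one-second delay.
# Construct initializes the ranked view model; dismissal records the FTUE as seen there, deactivates and broadcasts the click.
# Focus loss or HideMessage only deactivates, and the dismiss button takes focus.
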WidgetBlueprint /FortniteCoreUI/UI/Frontend/Lobby/Habanero/WBP_HabaneroRankProtection_CoachMark
graphs [
 BP_GetDesiredFocusTarget {nodes [
   Get Desired Focus Target()
   > Return Node!(execute, ReturnValue:Widget Object Reference=Get DismissButton() -> (DismissButton:WBP UIKit Button Regular Object Reference))
  ]}
 DisplayMessageIfFirstTime {nodes [
   n1 = DisplayMessageIfFirstTime!() -> (then, IsContainerVisible:bool=~) {is=title K2Node_FunctionEntry}
   > Branch(Condition=@n1)
   n2 = Get CheckHasSeenRankProtectionFTUE() -> (CheckHasSeenRankProtectionFTUE:bool)
   > n3 = Branch(Condition=AND Boolean!(
     A=NOT Boolean(A=@n2)
     B=Get CheckHasRankProtection() -> (CheckHasRankProtection:bool)
     C:bool=Get CheckIsIslandRanked() -> (CheckIsIslandRanked:bool)
    ) -> (ReturnValue))
   > ActivateAfterDelay()
   > Return Node!(execute, MessageDisplayed:bool=true)
   Return Node(execute=@n3.else, MessageDisplayed=false)
  ]}
 EventGraph {nodes [
   ActivateAfterDelay() {is=title K2Node_CustomEvent}
   > Delay(Duration=1)
   > ActivateWidget()
   Event OnRemovedFromFocusPath!() -> (OutputDelegate, then, InFocusEvent:Focus Event Structure) {is=title K2Node_Event}
   > DeactivateWidget()
   Event Construct()
   > Parent: Construct()
   > InitializeRankedViewModal(self:Ranked View Model Object Reference=Get FortRankedViewModel() -> (FortRankedViewModel:Ranked View Model Object Reference))
   Event OnButtonDismissClicked!() -> (OutputDelegate, then, Button:Common Button Base Object Reference) {is=title K2Node_Event}
   > SetHasSeenRankProtectionFTUE!(execute, self:Ranked View Model Object Reference=Get FortRankedViewModel(), bInHasSeenRankProtectionFTUE:bool=true) -> (then)
   > DeactivateWidget()
   > Call Button Dismiss Clicked!(execute:Exec, self:WBP UIKit Coach Mark Base Object Reference) -> (then:Exec) {is=title K2Node_CallDelegate}
  ]}
 HideMessage {nodes [
   HideMessage() {is=title K2Node_FunctionEntry}
   > DeactivateWidget()
  ]}
]
variables [
 CheckHasRankProtection {type=boolean}
 CheckHasSeenRankProtectionFTUE {type=boolean}
 CheckIsIslandRanked {type=boolean}
]
widgets {
 info {
  inheritedWidgetCount=0
  parentClass=/FortUILibrary/Systems/CoachMark/WBP_UIKit_CoachMark_Base.WBP_UIKit_CoachMark_Base_C
  rootWidgetClass=null
  widgetCount=0
 }
 widgets []
}
