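# Widget Blueprint that displays the unit labels (Day, HRS, Min, Sec) for an Athena event countdown timer.
# A horizontal box holds four equal-fill scale boxes, each wrapping a centered, uppercased CommonTextBlock in Burbank Black so the labels resize to fit their column.
# It has no graph logic or variables and shows only static text, with no countdown numbers.
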WidgetBlueprint /Game/Athena/Events/Generic/CountdownTimer/Widgets/LocalizedCountdownDisplay
graphs [EventGraph {links [], nodes []}]
variables []
widgets {
 info {inheritedWidgetCount=0, parentClass=&UserWidget, rootWidgetClass=&HorizontalBox, widgetCount=9}
 widgets [
  HorizontalBox_35 {widgetClassPath=&HorizontalBox, widget=HorizontalBox_35}
  ScaleBox_1 {
   widgetClassPath=&ScaleBox
   parent=HorizontalBox_35
   widget=ScaleBox_1
   slot {class=HorizontalBoxSlot, properties {verticalAlignment=VAlign_Center, size {sizeRule=Fill, value=1}}}
  }
  Text_Days {
   widgetClassPath=&CommonTextBlock
   parent=ScaleBox_1
   widget=Text_Days
   properties {
    justification=Center
    text=Day
    textTransformPolicy=ToUpper
    font {
     bForceMonospaced=false
     bMaterialIsStencil=false
     fontMaterial=null
     fontObject=&Burbank
     letterSpacing=0
     monospacedWidth=1
     size=64
     skewAmount=0
     typefaceFontName=Burbank-Black
     outlineSettings {
      bApplyOutlineToDropShadows=false
      bMiteredCorners=false
      bSeparateFillAlpha=false
      outlineMaterial=null
      outlineSize=0
      outlineColor=<rgba 0 0 0 1>
     }
    }
    shadowOffset=<0 0>
   }
   slot {class=ScaleBoxSlot}
  }
  ScaleBox_0 {
   widgetClassPath=&ScaleBox
   parent=HorizontalBox_35
   widget=ScaleBox_0
   slot {class=HorizontalBoxSlot, properties {verticalAlignment=VAlign_Center, size {sizeRule=Fill, value=1}}}
  }
  Text_Hours {
   widgetClassPath=&CommonTextBlock
   parent=ScaleBox_0
   widget=Text_Hours
   properties {
    justification=Center
    text=HRS
    textTransformPolicy=ToUpper
    font {
     bForceMonospaced=false
     bMaterialIsStencil=false
     fontMaterial=null
     fontObject=&Burbank
     letterSpacing=0
     monospacedWidth=1
     size=64
     skewAmount=0
     typefaceFontName=Burbank-Black
     outlineSettings {
      bApplyOutlineToDropShadows=false
      bMiteredCorners=false
      bSeparateFillAlpha=false
      outlineMaterial=null
      outlineSize=0
      outlineColor=<rgba 0 0 0 1>
     }
    }
    shadowOffset=<0 0>
   }
   slot {class=ScaleBoxSlot}
  }
  ScaleBox_2 {
   widgetClassPath=&ScaleBox
   parent=HorizontalBox_35
   widget=ScaleBox_2
   slot {class=HorizontalBoxSlot, properties {verticalAlignment=VAlign_Center, size {sizeRule=Fill, value=1}}}
  }
  Text_Min {
   widgetClassPath=&CommonTextBlock
   parent=ScaleBox_2
   widget=Text_Min
   properties {
    justification=Center
    text=Min
    textTransformPolicy=ToUpper
    font {
     bForceMonospaced=false
     bMaterialIsStencil=false
     fontMaterial=null
     fontObject=&Burbank
     letterSpacing=0
     monospacedWidth=1
     size=64
     skewAmount=0
     typefaceFontName=Burbank-Black
     outlineSettings {
      bApplyOutlineToDropShadows=false
      bMiteredCorners=false
      bSeparateFillAlpha=false
      outlineMaterial=null
      outlineSize=0
      outlineColor=<rgba 0 0 0 1>
     }
    }
    shadowOffset=<0 0>
   }
   slot {class=ScaleBoxSlot}
  }
  ScaleBox_3 {
   widgetClassPath=&ScaleBox
   parent=HorizontalBox_35
   widget=ScaleBox_3
   slot {class=HorizontalBoxSlot, properties {verticalAlignment=VAlign_Center, size {sizeRule=Fill, value=1}}}
  }
  Text_Secs {
   widgetClassPath=&CommonTextBlock
   parent=ScaleBox_3
   widget=Text_Secs
   properties {
    justification=Center
    text=Sec
    textTransformPolicy=ToUpper
    font {
     bForceMonospaced=false
     bMaterialIsStencil=false
     fontMaterial=null
     fontObject=&Burbank
     letterSpacing=0
     monospacedWidth=1
     size=64
     skewAmount=0
     typefaceFontName=Burbank-Black
     outlineSettings {
      bApplyOutlineToDropShadows=false
      bMiteredCorners=false
      bSeparateFillAlpha=false
      outlineMaterial=null
      outlineSize=0
      outlineColor=<rgba 0 0 0 1>
     }
    }
    shadowOffset=<0 0>
   }
   slot {class=ScaleBoxSlot}
  }
 ]
}
