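# Sound cue for the Showdown panel in the front-end UI, routed to the Fort_UI sound class.
# A mixer layers a dress-up wave, slightly lowered in level, with a pulse wave, and a short fixed delay holds back the combined sound.
# The cue plays at reduced overall volume and overrides concurrency to prevent new instances once the limit is reached.
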SoundCue /Game/Sounds/Fort_UI_Sounds/FrontEnd/Fort_UI_Showdown_Panel_Cue
root=Delay_2
nodes [
 {id=Delay_2, children [Mixer_0], properties {delayMax=0.15, delayMin=0.15}}
 {id=Mixer_0, children [WavePlayer_5, WavePlayer_36], properties {inputVolume [0.9,1]}}
 {id=WavePlayer_36, properties {soundWaveAssetPtr=&/fort_ui_pulse_01}}
 {id=WavePlayer_5, properties {soundWaveAssetPtr=&/fort_ui_dressup_03}}
]
properties {
 bOverrideConcurrency=true
 maxDistance=2097152
 soundClassObject=&Fort_UI
 volumeMultiplier=0.5
 concurrencyOverrides {resolutionRule=PreventNew}
}
