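# Sound cue for the Athena front-end store "new item available" toast opening, routed to the Fort_UI sound class.
# A single wave player reuses the gameplay mission_update_mid UI wave at reduced volume, with no other processing nodes.
# A concurrency override uses the PreventNew rule, so new plays are blocked once the limit is reached instead of cutting off playing instances.
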SoundCue /Game/Sounds/Fort_UI_Sounds/FrontEnd/Store/NewItemAvailable_Toast_Open_Athena_Cue
root=WavePlayer_0
nodes [{id=WavePlayer_0, properties {soundWaveAssetPtr=/Game/Sounds/Fort_UI_Sounds/Gameplay/mission_update_mid}}]
properties {
 bOverrideConcurrency=true
 maxDistance=2097152
 soundClassObject=&Fort_UI
 volumeMultiplier=0.4
 concurrencyOverrides {resolutionRule=PreventNew}
}
