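# Expandable lobby widget, built on WBP_Expandable_Base, that shows the player's Survivor Medal badge, header and current prestige and tier.
# On Construct, it reads prestige and tier from the player controller's survivor tiers component and looks up the prestige name in the seasonal display data.
# It writes both into the progress text, clearing it when display data is invalid.
# Clicking fires an interaction event and opens the quest screen at the BR Survivor category.
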WidgetBlueprint /Game/UI/Frontend/Lobby/SurvivorTier/WBP_Expandable_SurvivorTier
graphs [EventGraph {nodes [
   Event On Clicked()
   > Parent: On Clicked() {is=title K2Node_CallParentFunction}
   > Fire Event Element Interacted(Element=Self-Reference())
   > NavigateToQuestScreen!(
    execute
    Context:Widget Object Reference=Self-Reference()
    StartingCategoryTag:"Gameplay Tag Structure (by ref)"=MakeLiteralGameplayTag!(Value:Gameplay Tag Structure=(TagName="QuestCategory.BR.Survivor")) -> (ReturnValue:Gameplay Tag Structure)
   ) -> (then)
   Event Construct()
   > n1 = Cast To FortPlayerController(Object=GetOwningPlayer())
   n2 = GetSurvivorTierDisplayData!() -> (ReturnValue:Athena Seasonal Survivor Tiers Display Data Object Reference)
   > n3 = Is Valid(InputObject=@n2)
   n4 = GetPrestigeAndTier!(self:Fort Controller Component Survivor Tiers Object Reference=GetComponentByClass!(self:Actor Object Reference=@n1, ComponentClass:Actor Component Class Reference=/Script/FortniteGame.FortControllerComponent_SurvivorTiers) -> (ReturnValue:Fort Controller Component Survivor Tiers Object Reference)) -> (OutSurvivorPrestige:int=0, OutSurvivorTierRaw:int=0, OutSurvivorTierDisplay:int=0)
   n5 = GetPrestigeData!(self:Athena Seasonal Survivor Tiers Display Data Object Reference=@n2, PrestigeLevel:int=@n4) -> (ReturnValue:Survivor Presige Data Structure)
   n6 = Get Text_CurrentProgress() -> (Text_CurrentProgress:Common Text Object Reference)
   > "SetText (Text)"(
    self=@n6
    InText=Format Text!(
     Format="LOCTEXT(\"{Prestige} {Tier}\")"
     Prestige:Text=Break Survivor Presige Data!(SurvivorPresigeData:"Survivor Presige Data Structure (by ref)"=@n5) -> (
      DisplayName:Text
      SmallIconMaterial:Material Interface Soft Object Reference
      LargeIconMaterial:Material Interface Soft Object Reference
      Color1:Linear Color Structure
      Color2:Linear Color Structure
      TextColor:Linear Color Structure
     ) {is=title K2Node_BreakStruct}
     Tier:int=@n4.OutSurvivorTierDisplay
    ) -> (Result)
   )
   "SetText (Text)"(execute=@n3.Is Not Valid, self=@n6)
  ]}]
variables []
widgets {
 info {
  inheritedWidgetCount=0
  namedSlotCount=2
  parentClass=/Game/UI/Frontend/Lobby/Habanero/WBP_Expandable_Base.WBP_Expandable_Base_C
  rootWidgetClass=null
  widgetCount=5
 }
 widgets [
  W_SurvivorTierBadge {widgetClassPath=/Game/UI/SurvivorTiers/W_SurvivorTierBadge, bIsVariable=true, widget=W_SurvivorTierBadge}
  Overlay_86 {widgetClassPath=&Overlay, widget=Overlay_86}
  VerticalBox_87 {
   widgetClassPath=&VerticalBox
   parent=Overlay_86
   widget=VerticalBox_87
   slot {class=OverlaySlot, properties {verticalAlignment=VAlign_Center, padding=<ltrb 16 0 0 0>}}
  }
  Text_Header {
   widgetClassPath=&CommonTextBlock
   parent=VerticalBox_87
   widget=Text_Header
   properties {
    bIsScrollingEnabled=false
    mobileFontSizeMultiplier=1.25
    text=Survivor Medal
    textTransformPolicy=ToUpper
    colorAndOpacity {colorUseRule=UseColor_Specified, specifiedColor=<rgba 0.0048 0.004 0.0116 0.7>}
    font {
     bForceMonospaced=false
     bMaterialIsStencil=false
     fontMaterial=null
     fontObject=&HeadingNow
     letterSpacing=0
     monospacedWidth=1
     size=22.5
     skewAmount=0
     typefaceFontName=75 Medium Italic
     outlineSettings {
      bApplyOutlineToDropShadows=false
      bMiteredCorners=false
      bSeparateFillAlpha=false
      outlineMaterial=null
      outlineSize=0
      outlineColor=<rgba 0 0 0 1>
     }
    }
    shadowOffset=<0 0>
    shapedTextOptions {bOverride_TextFlowDirection=true, bOverride_TextShapingMethod=false, textFlowDirection=Culture, textShapingMethod=Auto}
   }
   slot {class=VerticalBoxSlot, properties {verticalAlignment=VAlign_Center}}
  }
  Text_CurrentProgress {
   widgetClassPath=&CommonTextBlock
   bIsVariable=true
   parent=VerticalBox_87
   widget=Text_CurrentProgress
   properties {
    bIsScrollingEnabled=false
    mobileFontSizeMultiplier=1.25
    text="{Prestige} {Tier}"
    textTransformPolicy=ToUpper
    colorAndOpacity {colorUseRule=UseColor_Specified, specifiedColor=<rgba 0.0048 0.004 0.0116 0.5>}
    font {
     bForceMonospaced=false
     bMaterialIsStencil=false
     fontMaterial=null
     fontObject=&HeadingNow
     letterSpacing=0
     monospacedWidth=1
     size=18
     skewAmount=0
     typefaceFontName=75 Medium
     outlineSettings {
      bApplyOutlineToDropShadows=false
      bMiteredCorners=false
      bSeparateFillAlpha=false
      outlineMaterial=null
      outlineSize=0
      outlineColor=<rgba 0 0 0 1>
     }
    }
    shadowOffset=<0 0>
    shapedTextOptions {bOverride_TextFlowDirection=true, bOverride_TextShapingMethod=false, textFlowDirection=Culture, textShapingMethod=Auto}
   }
   slot {class=VerticalBoxSlot, properties {verticalAlignment=VAlign_Top}}
  }
 ]
}
