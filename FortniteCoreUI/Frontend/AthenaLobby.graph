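# AthenaLobby is the frontend lobby widget, hosting matchmaking, expandable rank/level/survivor-tier widgets, MOTD teaser and store toasts.
# Homespace transitions animate matchmaking and details panels and toggle bottom-bar play/back buttons.
# Lobby data and runtime options drive widget visibility; the MOTD widget loads asynchronously and store toasts retry on timers.
# Focus tracking restores navigation after overlays, while side-menu and text-chat events suspend input preprocessing.
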WidgetBlueprint /FortniteCoreUI/Frontend/AthenaLobby
animations [Anim_ShowHideMatchmaking {
  binding_count=5
  length_seconds=0.2
  object=&self:Anim_ShowHideMatchmaking
  sequence {
   bindings [
    Content_MatchmakingWidget {
     id=#1
     possessed_class=CommonBorder
     tracks [
      {
       type=MovieSceneFloatTrack
       display_name=Render Opacity
       sections [{type=MovieSceneFloatSection, end_frame=12, start_frame=0, channels [None {type=MovieSceneScriptingFloatChannel, keys [[0,1],[12,0]]}]}]
      }
      {
       type=MovieSceneByteTrack
       display_name=Visibility
       sections [{type=MovieSceneByteSection, end_frame=12, start_frame=0, channels [None {type=MovieSceneScriptingByteChannel, keys [[0,4],[12,2]]}]}]
      }
      {
       type=MovieScene2DTransformTrack
       sections [{
         type=MovieScene2DTransformSection
         end_frame=12
         start_frame=0
         channels [
          Translation.X {type=MovieSceneScriptingFloatChannel, const [0,2]}
          Translation.Y {type=MovieSceneScriptingFloatChannel, keys [[0,0],[11.999,-40]]}
          Angle {type=MovieSceneScriptingFloatChannel}
          Scale.X {type=MovieSceneScriptingFloatChannel}
          Scale.Y {type=MovieSceneScriptingFloatChannel}
          Shear.X {type=MovieSceneScriptingFloatChannel}
          Shear.Y {type=MovieSceneScriptingFloatChannel}
         ]
        }]
      }
     ]
    }
    CommonBorder_SqueegeeSlot {
     id=#2
     possessed_class=CommonBorder
     tracks [{
       type=MovieSceneByteTrack
       display_name=Visibility
       sections [{type=MovieSceneByteSection, end_frame=12, start_frame=0, channels [None {type=MovieSceneScriptingByteChannel, keys [[0,4],[12,2]]}]}]
      }]
    }
    SqueegeeSlot_RatingCard {
     id=#3
     possessed_class=SqueegeeInjectionSlot_C
     tracks [
      {
       type=MovieSceneFloatTrack
       display_name=Render Opacity
       sections [{type=MovieSceneFloatSection, end_frame=12, start_frame=0, channels [None {type=MovieSceneScriptingFloatChannel, keys [[0,1],[12,0]]}]}]
      }
      {
       type=MovieScene2DTransformTrack
       sections [{
         type=MovieScene2DTransformSection
         end_frame=12
         start_frame=0
         channels [
          Translation.X {type=MovieSceneScriptingFloatChannel, const [0,1]}
          Translation.Y {type=MovieSceneScriptingFloatChannel, keys [[0,0],[12,-40]]}
          Angle {type=MovieSceneScriptingFloatChannel, const [0,1]}
          Scale.X {type=MovieSceneScriptingFloatChannel, const [1,1]}
          Scale.Y {type=MovieSceneScriptingFloatChannel, const [1,1]}
          Shear.X {type=MovieSceneScriptingFloatChannel, const [0,1]}
          Shear.Y {type=MovieSceneScriptingFloatChannel, const [0,1]}
         ]
        }]
      }
     ]
    }
    Switcher_LobbyDetailsMode {
     id=#4
     possessed_class=CommonWidgetSwitcherLegacy
     tracks [
      {
       type=MovieSceneFloatTrack
       display_name=Render Opacity
       sections [{type=MovieSceneFloatSection, end_frame=12, start_frame=0, channels [None {type=MovieSceneScriptingFloatChannel, keys [[0,1],[12,0]]}]}]
      }
      {
       type=MovieSceneByteTrack
       display_name=Visibility
       sections [{type=MovieSceneByteSection, end_frame=12, start_frame=0, channels [None {type=MovieSceneScriptingByteChannel, keys [[0,4],[12,2]]}]}]
      }
      {
       type=MovieScene2DTransformTrack
       sections [{
         type=MovieScene2DTransformSection
         end_frame=12
         start_frame=0
         channels [
          Translation.X {type=MovieSceneScriptingFloatChannel, const [0,2]}
          Translation.Y {type=MovieSceneScriptingFloatChannel, keys [[0,0],[11.999,-40]]}
          Angle {type=MovieSceneScriptingFloatChannel}
          Scale.X {type=MovieSceneScriptingFloatChannel}
          Scale.Y {type=MovieSceneScriptingFloatChannel}
          Shear.X {type=MovieSceneScriptingFloatChannel}
          Shear.Y {type=MovieSceneScriptingFloatChannel}
         ]
        }]
      }
     ]
    }
    Overlay_AboveMatchmaking_Navigable {
     id=#5
     possessed_class=Overlay
     tracks [
      {
       type=MovieSceneFloatTrack
       display_name=Render Opacity
       sections [{type=MovieSceneFloatSection, end_frame=12, start_frame=0, channels [None {type=MovieSceneScriptingFloatChannel, keys [[0,1],[12,0]]}]}]
      }
      {
       type=MovieScene2DTransformTrack
       sections [{
         type=MovieScene2DTransformSection
         end_frame=12
         start_frame=0
         channels [
          Translation.X {type=MovieSceneScriptingFloatChannel, const [0,2]}
          Translation.Y {type=MovieSceneScriptingFloatChannel, keys [[0,0],[11.999,-40]]}
          Angle {type=MovieSceneScriptingFloatChannel}
          Scale.X {type=MovieSceneScriptingFloatChannel}
          Scale.Y {type=MovieSceneScriptingFloatChannel}
          Shear.X {type=MovieSceneScriptingFloatChannel}
          Shear.Y {type=MovieSceneScriptingFloatChannel}
         ]
        }]
      }
      {
       type=MovieSceneByteTrack
       display_name=Visibility
       sections [{type=MovieSceneByteSection, end_frame=12, start_frame=0, channels [None {type=MovieSceneScriptingByteChannel, keys [[0,4],[12,2]]}]}]
      }
     ]
    }
   ]
   display_rate {numerator=60}
  }
 }]
graphs [
 ActivateMOTDWidget {nodes [
   ActivateMOTDWidget!() -> (then, Activate:bool=~, SkipActivation:bool) {is=title K2Node_FunctionEntry}
   > n1 = Cast To FortDynamicMotdTeaserWidget!(execute, Object=Get MOTDWidgetReference() -> (MOTDWidgetReference:User Widget Object Reference)) -> (then, CastFailed, AsFort Dynamic Motd Teaser Widget:Fort Dynamic Motd Teaser Widget Object Reference) {is=title K2Node_DynamicCast}
   > n2 = Branch(Condition=Get Activate() -> (Activate:bool))
   > n3 = Branch(Condition=IsDiscoverFullscreen!(self:Athena Lobby Base Object Reference) -> (ReturnValue:bool))
   Set bShouldBeActive!(execute=@n3.else, bShouldBeActive:bool=true, self:Fort Dynamic Motd Teaser Widget Object Reference=@n1) -> (then, Output_Get:bool)
   > n4 = Branch(Condition=Get SkipActivation() -> (SkipActivation:bool))
   ResetActiveState(execute=[@n5.else, @n4.else], self:Fort Dynamic Motd Teaser Widget Object Reference=@n1)
   Set bShouldBeActive(execute=@n2.else, bShouldBeActive=false, self=@n1)
   > n5 = Branch(Condition=Get SkipActivation())
  ]}
 ConfigureGenericLinkButton {nodes [
   ConfigureGenericLinkButton() {is=title K2Node_FunctionEntry}
   > n1 = GetLobbyGenericLinkButtonOverrideText!(execute, self:Fort Runtime Options Object Reference=GetRuntimeOptions()) -> (then, ReturnValue:Text)
   > SetTitleText!(
    execute:Exec
    self:Object Reference=Get Button_GenericLink() -> (Button_GenericLink:Common Button Base Object Reference)
    NameText:Text=Select(Option 0:Text=@n1, Option 1:Text="LOCTEXT(\"New Player Guide\")", Index=TextIsEmpty(InText=@n1)) -> (ReturnValue:Text)
   ) -> (then:Exec) {is=title K2Node_Message}
  ]}
 Determine Right Side Focus Widget {nodes [
   n1 = Determine Right Side Focus Widget!() -> (then, FocusTopWidget:bool=~) {is=title K2Node_FunctionEntry}
   n2 = GetActiveWidget!(self:Widget Switcher Object Reference=Get Switcher_LobbyDetailsMode() -> (Switcher_LobbyDetailsMode:Common Widget Switcher Legacy Object Reference)) -> (ReturnValue:Widget Object Reference)
   > n3 = Branch(Condition="Equal (Object)"(A=Get VerticalBox_ExpandableWidget() -> (VerticalBox_ExpandableWidget:Vertical Box Object Reference), B=@n2))
   > n4 = Branch(Condition=@n1)
   n5 = GetAllChildren(self=Get VerticalBox_ExpandableWidget())
   > n6 = For Each Loop with Break!(Exec:Exec, Array:Array of Widget Object References=@n5, Break:Exec) -> (LoopBody:Exec, Array Element:Widget Object Reference, Array Index:int, Completed:Exec)
   > n7 = Cast To UserWidget!(execute, Object=@n6) -> (then, CastFailed, AsUser Widget:User Widget Object Reference) {is=title K2Node_DynamicCast}
   > Branch(Condition=AND Boolean(A=Get bIsFocusable!(self:User Widget Object Reference=@n7) -> (bIsFocusable:bool), B=IsVisible(self=@n7)))
   > SetFocus(self=@n7)
   > n8 = Set Last Focused Widget(Last Focused Widget:Widget Object Reference=@n7) -> (Output_Get:Widget Object Reference)
   Return Node!(execute=[@n8.then, @n13.then], FocusHandled:bool=true)
   n9 = Branch(
    execute=[@n14.else, @n6.Completed, @n11.Completed]
    Condition=IsVisible(self=Get SqueegeeSlot_RatingCard() -> (SqueegeeSlot_RatingCard:Squeegee Injection Slot Object Reference))
   )
   n10 = Get SqueegeeSlot_RatingCard()
   > SetFocus(self=@n10)
   > Set Last Focused Widget(Last Focused Widget=@n10)
   > Return Node(FocusHandled=true)
   Return Node(execute=@n9.else, FocusHandled=false)
   n11 = Reverse for Each Loop!(Exec:Exec=@n4.else, Array:Array of Widget Object References=@n5) -> (LoopBody:Exec, ArrayIndex:int, ArrayElement:Widget Object Reference, Completed:Exec) {is=title K2Node_MacroInstance}
   > n12 = Cast To UserWidget(Object=@n11.ArrayElement)
   > Branch(Condition=AND Boolean(A=Get bIsFocusable(self=@n12), B=IsVisible(self=@n12)))
   > SetFocus(self=@n12)
   > n13 = Set Last Focused Widget(Last Focused Widget=@n12)
   n14 = Branch(
    execute=@n3.else
    Condition="Equal (Object)"(A=Get LobbyDetails_Tournament() -> (LobbyDetails_Tournament:Fort Tournament Lobby Panel Object Reference), B=@n2)
   )
   n15 = Get Button_EventDetails!(self:Fort Tournament Lobby Panel Object Reference=Get LobbyDetails_Tournament()) -> (Button_EventDetails:Common Button Legacy Object Reference)
   > SetFocus(self=@n15)
   > Set Last Focused Widget(Last Focused Widget=@n15)
   > Return Node(FocusHandled=true)
  ]}
 EventGraph {nodes [
   "On Clicked (WBP_BottomBarHack_BackToDiscover)"!() -> (OutputDelegate, then, Button:Common Button Base Object Reference) {is=title K2Node_ComponentBoundEvent}
   > NavigateToDiscover!(execute, self:Athena Lobby Base Object Reference) -> (then, ReturnValue:bool)
   n1 = "On Visibility Changed (WBP_Expandable_Rank)"!() -> (OutputDelegate, then, InVisibility:ESlateVisibility Enum=Visible) {is=title K2Node_ComponentBoundEvent}
   > SetVisibility(self=Get Expandable_Rank_Container() -> (Expandable_Rank_Container:Overlay Object Reference), InVisibility=@n1.InVisibility)
   "On Finished Toast (StoreToastWidget)"() {is=title K2Node_ComponentBoundEvent}
   > Is Valid(InputObject=Get StoreToastRequest() -> (StoreToastRequest:Store Toast Request Object Reference))
   > CompleteRequest(self:Store Toast Request Object Reference=Get StoreToastRequest())
   > Set StoreToastRequest(StoreToastRequest:Store Toast Request Object Reference) -> (Output_Get:Store Toast Request Object Reference)
   "On Clicked (WBP_BottomBarHack_MMPlayOrCancel)"!() -> (OutputDelegate, then, Button:Common Button Base Object Reference) {is=title K2Node_ComponentBoundEvent}
   n2 = Get MVVMSubsystem() -> (ReturnValue:Viewmodel Engine Subsytem Object Reference) {is=title K2Node_GetEngineSubsystem}
   n3 = Get MatchmakingWidget() -> (MatchmakingWidget:Common User Widget Object Reference)
   > n4 = Cast To SelectedIslandViewModel!(
    execute
    Object:Notify Field Value Changed Interface=GetViewModel!(
     self:MVVM View Object Reference=Get View From User Widget!(self:Viewmodel Engine Subsytem Object Reference=@n2, UserWidget:User Widget Object Reference=@n3) -> (ReturnValue:MVVM View Object Reference) {is=title K2Node_CallFunction}
     ViewModelName:Name=FortMatchmakingViewModel
    ) -> (ReturnValue:Notify Field Value Changed Interface)
   ) -> (then, CastFailed, AsSelected Island View Model:Selected Island View Model Object Reference) {is=title K2Node_DynamicCast}
   > n5 = Branch(Condition=IsCancelCurrentAction!(self) -> (ReturnValue:bool))
   > Unready(self:Selected Island View Model Object Reference=@n4)
   ReadyUp(execute=@n5.else, self:Selected Island View Model Object Reference=@n4)
   n6 = OnHomespaceTellLobbyToTransition!() -> (OutputDelegate, then, PlayVisible:bool=~) {is=title K2Node_CustomEvent}
   > n7 = Set is MM Widget on Screen(is MM Widget on Screen:bool=@n6.PlayVisible) -> (Output_Get:bool)
   > Call On MMWidget Display Changed!(execute:Exec, self:Athena Lobby Object Reference, Visible:bool=@n7) -> (then:Exec) {is=title K2Node_CallDelegate}
   > SetVoiceChannelSpeakerContainerVisibility!(
    execute
    self:Athena Frontend Object Reference=Get AthenaFrontend() -> (AthenaFrontend:Athena Frontend Object Reference)
    ShouldBeVisible:bool=@n6.PlayVisible
   ) -> (then)
   > UpdateEvoScrollTransitionAllowed()
   > Branch(Condition=GetIsDiscoverInLobbyEnabled!(self:Fort Runtime Options Object Reference=GetRuntimeOptions(), ContextWidget:Widget Object Reference=Self-Reference()) -> (ReturnValue:bool))
   > n8 = Branch(Condition=@n6.PlayVisible)
   n9 = HandleMatchmakingAnimationFinished() {is=title K2Node_CustomEvent}
   n10 = Get Anim_ShowHideMatchmaking() -> (Anim_ShowHideMatchmaking:Widget Animation Object Reference)
   BindToAnimationFinished(execute=@n8.then, Animation=@n10, Delegate=@n9)
   > PlayAnimationReverse(InAnimation=Get Anim_ShowHideMatchmaking())
   > SetPlayCancelAndBackButtonsVisible!(execute, self, ShouldPlayCancelButtonBeVisible:bool=Get CanPlayCancelButtonBeVisible() -> (CanPlayCancelButtonBeVisible:bool)) -> (then)
   > OnShowExpandableLevel!(
    execute
    self:WBP Expandable Level Switcher Object Reference=Get Expandable_Level_Switcher() -> (Expandable_Level_Switcher:WBP Expandable Level Switcher Object Reference)
    ShouldShow:bool=true
   ) -> (then)
   n11 = ShouldShowHomeBarOnActivated!(execute=@n8.else, self:Athena Lobby Base Object Reference) -> (then, ReturnValue:bool)
   > SetHomeBarShownOnActivated(self:Athena Lobby Base Object Reference)
   n12 = Get Anim_ShowHideMatchmaking()
   > PlayAnimation(InAnimation=@n12, StartAtTime=SelectFloat(A=GetEndTime(self=@n12), B=0, bPickA=@n11))
   > SetBottomButtonVisibility!(
    execute
    self:WBP Bottom Bar Decoy Button Object Reference=Get WBP_BottomBarHack_MMPlayOrCancel() -> (WBP_BottomBarHack_MMPlayOrCancel:WBP Bottom Bar Decoy Button Object Reference)
    ShouldButtonBeVisible:bool=false
   ) -> (then)
   > SetBottomButtonVisibility(
    self=Get WBP_BottomBarHack_BackToDiscover() -> (WBP_BottomBarHack_BackToDiscover:WBP Bottom Bar Decoy Button Object Reference)
    ShouldButtonBeVisible=false
   )
   > OnShowExpandableLevel(self=Get Expandable_Level_Switcher(), ShouldShow=false)
   TryToastAgainNextFrame() {is=title K2Node_CustomEvent}
   > n13 = IsRequestReady!(execute, self:Store Toast Request Object Reference=Get StoreToastRequest()) -> (then, ReturnValue:bool)
   > n14 = Branch(Condition=@n13)
   > n15 = ShouldBeginRequest!(execute, self:Store Toast Request Object Reference=Get StoreToastRequest()) -> (then, ReturnValue:bool)
   > n16 = Branch(Condition=@n15)
   > BeginRequest(self:Store Toast Request Object Reference=Get StoreToastRequest())
   n17 = Get StoreToastRequest()
   > ShowToast!(
    execute
    self:Fancy Toast Widget Object Reference=Get StoreToastWidget() -> (StoreToastWidget:Fancy Toast Widget Object Reference)
    DisplayInfo:Fort Toast Display Info Structure=Break Store Callout!(StoreCallout:"Store Callout Structure (by ref)"=GetToast!(self:Store Toast Request Object Reference=@n17) -> (ReturnValue:Store Callout Structure)) -> (ToastDisplayInfo:Fort Toast Display Info Structure) {is=title K2Node_BreakStruct}
   ) -> (then)
   > PlaySound2D(Sound=/Game/Sounds/Fort_UI_Sounds/FrontEnd/Store/NewItemAvailable_Toast_Open_Athena_Cue)
   SkipRequest(execute=@n16.else, self:Store Toast Request Object Reference=Get StoreToastRequest())
   > Set StoreToastRequest()
   n18 = Set Timer by Event(execute=@n14.else, Delegate=Create Event(), Time=0.1, bLooping=false)
   > Set TryShowToast(TryShowToast:Timer Handle Structure=@n18) -> (Output_Get:Timer Handle Structure)
   n19 = OnTabChange!() -> (OutputDelegate, then, LeavingFeature:EFortUIFeature Enum=ShowHome, EnteringFeature:EFortUIFeature Enum=ShowHome) {is=title K2Node_CustomEvent}
   > Branch(Condition="Equal (Enum)"(A:EFortUIFeature Enum=@n19.LeavingFeature, B:EFortUIFeature Enum=ShowPlay))
   > Set Last Focused Widget(Last Focused Widget=Get MatchmakingWidget())
   MOTDFullscreenWidgetDeactivated() {is=title K2Node_CustomEvent}
   > n20 = Set is MOTD Activated(is MOTD Activated:bool=false) -> (Output_Get:bool)
   DelayUntilNextTick!(execute=[@n77.else, @n74.else, @n20.then]) -> (then)
   > Navigate To Last Focused Widget()
   > SetInputPreprocessorEnabled!(execute, self:Fort Lobby Widget Object Reference, bValue:bool=true) -> (then)
   n21 = SetBackButtonVisibility!() -> (OutputDelegate, then, ShouldButtonBeVisible:bool=~) {is=title K2Node_CustomEvent}
   > SetBottomButtonVisibility(self=Get WBP_BottomBarHack_BackToDiscover(), ShouldButtonBeVisible=@n21.ShouldButtonBeVisible)
   n22 = MOTD Fullscreen Widget Activated() {is=title K2Node_CustomEvent}
   > Set is MOTD Activated(is MOTD Activated=true)
   AttemptToShowToast() {is=title K2Node_CustomEvent}
   > n23 = Branch(Condition=IsActivated!(self:Common Activatable Widget Object Reference) -> (ReturnValue:bool))
   > TryToastAgainNextFrame() {is=title K2Node_CallFunction}
   n24 = Set Timer by Event(execute=@n23.else, Delegate=Create Event(), Time=0.5, bLooping=false)
   > Set TryShowToast(TryShowToast=@n24)
   Show Matchmaking Region() {is=title K2Node_CustomEvent}
   n25 = GetRuntimeOptions()
   > Set Visibility(
    Visibility=Select(
     Option 0:ESlateVisibility Enum=Hidden
     Option 1:ESlateVisibility Enum=SelfHitTestInvisible
     Index=IsShowCurrentRegionInLobbyEnabled!(self:Fort Runtime Options Object Reference=@n25) -> (ReturnValue:bool)
    ) -> (ReturnValue:ESlateVisibility Enum)
    self=Get MatchmakingRegion() -> (MatchmakingRegion:Matchmaking Region And Pool Object Reference)
   )
   n26 = UpdateBottomBarPlayButtonText!() -> (OutputDelegate, then, BottomBarPlayState:Data Table Row Handle Structure) {is=title K2Node_CustomEvent}
   > n27 = Set Bottom Bar Play State(Bottom Bar Play State:Data Table Row Handle Structure=@n26.BottomBarPlayState) -> (Output_Get:Data Table Row Handle Structure)
   n28 = Get WBP_BottomBarHack_MMPlayOrCancel()
   > n29 = Set BottomButtonInputAction!(execute, BottomButtonInputAction:Data Table Row Handle Structure=@n27, self:WBP Bottom Bar Decoy Button Object Reference=@n28) -> (then, Output_Get:Data Table Row Handle Structure)
   > SetTriggeringInputAction!(execute, self:Common Button Base Object Reference=@n28, InputActionRow:"Data Table Row Handle Structure (by ref)"=@n29) -> (then)
   > SetPlayCancelAndBackButtonsVisible(ShouldPlayCancelButtonBeVisible=Get CanPlayCancelButtonBeVisible())
   n30 = UpdateBottomBarPlayButtonVisibility!() -> (OutputDelegate, then, IsVisible:bool=~) {is=title K2Node_CustomEvent}
   > n31 = Set CanPlayCancelButtonBeVisible(CanPlayCancelButtonBeVisible:bool=@n30.IsVisible) -> (Output_Get:bool)
   > SetPlayCancelAndBackButtonsVisible(ShouldPlayCancelButtonBeVisible=@n31)
   Navigate To Last Focused Widget() {is=title K2Node_CustomEvent}
   n32 = Get IsSideMenuActive() -> (IsSideMenuActive:bool)
   n33 = Get IsTextChatActive() -> (IsTextChatActive:bool)
   > Branch(Condition=AND Boolean!(A=Get is MM Widget on Screen() -> (is MM Widget on Screen:bool), B=NOT Boolean(A=@n32), C:bool=NOT Boolean(A=@n33)) -> (ReturnValue))
   > DelayUntilNextTick()
   > n34 = Branch(Condition="Equal (Object)"(A=Get Last Focused Widget() -> (Last Focused Widget:Widget Object Reference), B=Get MatchmakingWidget()))
   > NavigateToMatchmaking(self:Athena Lobby Base Object Reference)
   n35 = Get Last Focused Widget()
   Is Valid(exec=@n34.else, InputObject=@n35)
   > SetFocus(self=@n35)
   n36 = MOTDFullscreenWidgetCreated!() -> (OutputDelegate, then, Widget:Fort Dynamic Motd Fullscreen Widget Object Reference) {is=title K2Node_CustomEvent}
   > n37 = Set FullScreenMOTD(FullScreenMOTD:Fort Dynamic Motd Fullscreen Widget Object Reference=@n36.Widget) -> (Output_Get:Fort Dynamic Motd Fullscreen Widget Object Reference)
   > Assign On Widget Deactivated!(execute:Exec, self:Common Activatable Widget Object Reference=@n37, Delegate:"Delegate (by ref)"=Create Event()) -> (then:Exec) {is=title K2Node_AssignDelegate}
   > Bind Event to On Widget Activated(self:Common Activatable Widget Object Reference=@n37, Delegate=@n22) {is=title K2Node_AddDelegate}
   n38 = OnHomespaceFullScreenStateChanged!() -> (OutputDelegate, then, IsFullScreen:bool=~) {is=title K2Node_CustomEvent}
   > SetVisibility(
    self=Get SafeZone() -> (SafeZone:Safe Zone Object Reference)
    InVisibility=Select(Option 0=SelfHitTestInvisible, Option 1=Hidden, Index=@n38.IsFullScreen)
   )
   Event On Activated()
   > n39 = Sequence!(execute) -> (then_0, then_1, then_2:Exec, then_3:Exec, then_4:Exec, then_5:Exec)
   > ConfigureGenericLinkButton() {is=title K2Node_CallFunction}
   > n40 = SendEnsureClientQuestLogin!(
    execute:Exec
    QuestManager:Fort Quest Manager Object Reference=GetQuestManager!(
     self:Fort Global UIContext Object Reference=Get FortGlobalUIContext!() -> (ReturnValue:Fort Global UIContext Object Reference) {is=title K2Node_GetBlueprintContextEx}
     SubGame:ESubGame Enum=Athena
    ) -> (ReturnValue:Fort Quest Manager Object Reference)
   ) -> (then:Exec, OnSuccess:Exec, OnFailure:Exec) {is=title K2Node_AsyncAction}
   Branch(execute=[@n40.OnFailure, @n40.OnSuccess], Condition=Get bIsActive() -> (bIsActive:bool))
   n41 = Get FortFrontEndContext!(PlayerController:Player Controller Object Reference=GetOwningPlayer()) -> (ReturnValue:Fort Front End Context Object Reference) {is=title K2Node_GetSubsystemFromPC}
   > SetFrontEndCamera!(
    execute
    self:Fort Front End Context Object Reference=@n41
    NewState:EFrontEndCamera Enum=Invalid
    TransitionParams:View Target Transition Params Structure=(BlendTime=0,BlendFunction=VTBlend_Cubic,BlendExp=2,bLockOutgoing=False)
   ) -> (then)
   > SetFrontEndCamera(
    self=@n41
    NewState=Get LastSavedCameraState() -> (LastSavedCameraState:EFrontEndCamera Enum=Invalid)
    TransitionParams=Make View Target Transition Params!(BlendTime:float=1, BlendFunction:EViewTargetBlendFunction Enum=VTBlend_EaseOut, BlendExp:float=2, bLockOutgoing:bool=False) -> (ViewTargetTransitionParams:View Target Transition Params Structure) {is=title K2Node_MakeStruct}
   )
   Show Matchmaking Region(execute=@n39.then_2) {is=title K2Node_CallFunction}
   n42 = NOT Boolean(A=Get Is Play Time Out() -> (Is Play Time Out:bool))
   ActivateMOTDWidget!(execute=@n39.then_3, self, Activate:bool=@n42, SkipActivation:bool=false) -> (then) {is=title K2Node_CallFunction}
   > OnShowExpandableLevel(self=Get Expandable_Level_Switcher(), ShouldShow=@n42)
   Branch(execute=@n39.then_4, Condition=Get is MM Widget on Screen())
   > DelayUntilNextTick()
   > Navigate To Last Focused Widget() {is=title K2Node_CallFunction}
   UpdateEvoScrollTransitionAllowed(execute=@n39.then_5)
   Do Once(execute=@n39.then_1)
   > n43 = CreateWidgetAsync!(
    execute:Exec
    UserWidgetSoftClass:User Widget Soft Class Reference=/Game/Athena/Events/Generic/CountdownTimer/Widgets/LocalizedCountdownDisplay.LocalizedCountdownDisplay_C
    OwningPlayer:Player Controller Object Reference
    bSuspendInputUntilComplete:bool=false
   ) -> (then:Exec, OnComplete:Exec, OnFailure:Exec, UserWidget:User Widget Object Reference) {is=title K2Node_AsyncAction}
   Is Valid(exec=@n43.OnComplete, InputObject=@n43)
   > DrawWidgetToRenderTarget!(
    execute
    Widget:User Widget Object Reference=@n43
    RenderTarget:Texture Render Target 2D Object Reference=/Game/Athena/Events/Generic/CountdownTimer/Textures/RT_LocalizedCountdownText
   ) -> (then)
   n44 = Event OnLobbyDataChanged!() -> (OutputDelegate, then, LobbyData:Frontend Lobby Data Structure) {is=title K2Node_Event}
   n45 = Break Frontend Lobby Data!(FrontendLobbyData:"Frontend Lobby Data Structure (by ref)"=@n44.LobbyData) -> (
    Texture:Texture Soft Object Reference
    Material:Material Instance Soft Object Reference
    PreBlurMaterial:Material Instance Soft Object Reference
    SpawnFXList:Array of Frontend Lobby FXData Structures
    SpawnActorList:Array of Frontend Lobby Actor Structures
    CameraPositionOverrideList:Array of Frontend Lobby Camera Position Override Structures
    FadeGrid:float=~
    FadeFloorToBlack:float=~
    NamePlateHeight:float=~
    SuggestionWidgetHeight:float=~
    bHasRankedWidget:bool=~
    bHasLevelWidget:bool=~
    bHasSurvivorTierWidget:bool=~
   ) {is=title K2Node_BreakStruct}
   n46 = Get bEnableLobbyRankWidget!(self:Fort Runtime Options Object Reference=GetRuntimeOptions()) -> (bEnableLobbyRankWidget:bool)
   > SetVisibility(
    self=Get Expandable_Rank_Container()
    InVisibility=Select(Option 0=Collapsed, Option 1=SelfHitTestInvisible, Index=AND Boolean!(A=@n46, B=@n45.bHasRankedWidget) -> (ReturnValue))
   )
   n47 = Get bEnableLobbyLevelWidget!(self:Fort Runtime Options Object Reference=GetRuntimeOptions()) -> (bEnableLobbyLevelWidget:bool)
   > SetVisibility(
    self=Get Expandable_Level_Switcher()
    InVisibility=Select(Option 0=Collapsed, Option 1=SelfHitTestInvisible, Index=AND Boolean(A=@n47, B=@n45.bHasLevelWidget))
   )
   n48 = Get bEnableLobbySurvivorTierWidget!(self:Fort Runtime Options Object Reference=GetRuntimeOptions()) -> (bEnableLobbySurvivorTierWidget:bool)
   > SetVisibility(
    self=Get WBP_Expandable_SurvivorTier() -> (WBP_Expandable_SurvivorTier:WBP Expandable Survivor Tier Object Reference)
    InVisibility=Select(Option 0=Collapsed, Option 1=SelfHitTestInvisible, Index=AND Boolean(A=@n48, B=@n45.bHasSurvivorTierWidget))
   )
   Event OnInitialized()
   > n49 = Sequence!(execute) -> (then_0, then_1)
   > Branch(Condition=IsMobileGame())
   > SetMobileExpandableWidgetsOrder()
   > Branch(Condition=GetIsBacchusTabNavOnTop!(self:Fort Runtime Options Object Reference=GetRuntimeOptions()) -> (ReturnValue:bool))
   > Set Padding!(
    execute
    Padding:Margin Structure=Make Margin(Left=0, Top=120, Right=0, Bottom=0)
    self:Overlay Slot Object Reference=SlotAsOverlaySlot(Widget=Get MinimalTeamDisplay() -> (MinimalTeamDisplay:Minimal Team Display Object Reference))
   ) -> (then, Output_Get:Margin Structure)
   InjectMusicPlayer(execute=@n49.then_1)
   Event OnLobbyDataCleared() {is=title K2Node_Event}
   n50 = Get bEnableLobbyRankWidget(self=GetRuntimeOptions())
   n51 = Get Is Island Ranked!(self:WBP Expandable Rank Object Reference=Get WBP_Expandable_Rank() -> (WBP_Expandable_Rank:WBP Expandable Rank Object Reference)) -> (Is Island Ranked:bool)
   > SetVisibility(
    self=Get Expandable_Rank_Container()
    InVisibility=Select(Option 0=Collapsed, Option 1=SelfHitTestInvisible, Index=AND Boolean(A=@n50, B=@n51))
   )
   n52 = GetRuntimeOptions()
   > SetVisibility(
    self=Get Expandable_Level_Switcher()
    InVisibility=Select(Option 0=Collapsed, Option 1=SelfHitTestInvisible, Index=Get bEnableLobbyLevelWidget(self=@n52))
   )
   n53 = GetRuntimeOptions()
   > SetVisibility(
    self=Get WBP_Expandable_SurvivorTier()
    InVisibility=Select(Option 0=Collapsed, Option 1=SelfHitTestInvisible, Index=Get bEnableLobbySurvivorTierWidget(self=@n53))
   )
   n54 = Event HandleUnseenQuestsChanged!() -> (OutputDelegate, then, bHasUnseenQuests:bool) {is=title K2Node_Event}
   > n55 = Cast To WBP_UIKit_ButtonCTA_Base!(execute, Object=Get Button_Quests() -> (Button_Quests:Common Button Base Object Reference)) -> (then, CastFailed, AsWBP UIKit Button CTA Base:WBP UIKit Button CTA Base Object Reference) {is=title K2Node_DynamicCast}
   > SetText!(
    execute
    self:Fort CTAButton Object Reference=@n55
    ButtonText:"Text (by ref)"=SelectText(A="LOCTEXT(\"NEW QUESTS!\")", B="LOCTEXT(\"VIEW QUESTS\")", bPickA=@n54.bHasUnseenQuests)
   ) -> (then)
   Event OnMatchmakingWidgetChanged() {is=title K2Node_Event}
   > n56 = Get!(execute:Exec) -> (then:Exec, else:Exec, MatchmakingWidget:Common User Widget Object Reference)
   n57 = IsBROnly!() -> (ReturnValue:bool)
   n58 = "Not Equal (Enum)"!(
    A:ECommonInputType Enum=GetCurrentInputType(self=Get CommonInputSubsystem!() -> (ReturnValue:Common Input Object Reference) {is=title K2Node_GetBlueprintContextEx})
    B:ECommonInputType Enum=Gamepad
   ) -> (ReturnValue:bool)
   > SetPadding!(
    execute
    self:User Widget Object Reference=@n56
    InPadding_Left:float=0
    InPadding_Top:float=0
    InPadding_Right:float=0
    InPadding_Bottom:float=SelectFloat(A=32, B=0, bPickA=AND Boolean(A=@n57, B=@n58))
   ) -> (then)
   n59 = Event DisplayStoreUpdated!() -> (OutputDelegate, then, StoreUpdatedRequest:Store Toast Request Object Reference) {is=title K2Node_Event}
   > Set StoreToastRequest(StoreToastRequest=@n59.StoreUpdatedRequest)
   > AttemptToShowToast() {is=title K2Node_CallFunction}
   Event Construct()
   > n60 = Sequence!(execute) -> (then_0, then_1, then_2:Exec, then_3:Exec, then_4:Exec)
   n61 = ShouldShowHomeBarOnActivated(execute=@n60.then_1)
   > n62 = Branch(Condition=@n61)
   > n63 = SetRenderOpacity(
    self=[
     Get SqueegeeSlot_RatingCard()
     Get Switcher_LobbyDetailsMode()
     Get Content_MatchmakingWidget() -> (Content_MatchmakingWidget:Content Widget Object Reference)
    ]
    InOpacity=0
   )
   SetFortPickerData!(
    execute=[@n63.then, @n67.then]
    self:Fort Picker Context Object Reference=Get FortPickerContext!(PlayerController:Player Controller Object Reference=GetOwningPlayer()) -> (ReturnValue:Fort Picker Context Object Reference) {is=title K2Node_GetSubsystemFromPC}
    Data:Fort Picker Data Object Reference=/Game/UI/InGame/HUD/FortPickers
   ) -> (then)
   > n64 = GetMotdTeaserClass!(execute, self:Athena Lobby Base Object Reference) -> (then, ReturnValue:User Widget Soft Class Reference)
   > Branch(Condition=IsValidSoftClassReference!(SoftClassReference:"Object Soft Class Reference (by ref)"=@n64) -> (ReturnValue:bool))
   > n65 = CreateWidgetAsync(UserWidgetSoftClass=@n64, bSuspendInputUntilComplete=false)
   Is Valid(exec=@n65.OnComplete, InputObject=@n65)
   > Set MOTDWidgetReference(MOTDWidgetReference:User Widget Object Reference=@n65) -> (Output_Get:User Widget Object Reference)
   > ActivateMOTDWidget(Activate=NOT Boolean(A=Get Is Play Time Out()), SkipActivation=true)
   > AddChild(self=Get Content_MOTDHolder() -> (Content_MOTDHolder:Common Border Object Reference), Content=@n65)
   > n66 = Cast To FortDynamicMotdTeaserWidget(Object=@n65)
   > Assign Fullscreen Widget Created!(execute:Exec, self:Fort Dynamic Motd Teaser Widget Object Reference=@n66, Delegate:"Delegate (by ref)"=@n36) -> (then:Exec) {is=title K2Node_AssignDelegate}
   n67 = SetBottomButtonVisibility(execute=@n62.else, self=Get WBP_BottomBarHack_MMPlayOrCancel(), ShouldButtonBeVisible=true)
   n68 = Create Event()
   n69 = Get FortFrontEndContext(PlayerController=GetOwningPlayer())
   Bind Event to On Front End Camera Changed(execute=@n60.then_4, self:Fort Front End Context Object Reference=@n69, Delegate=@n68) {is=title K2Node_AddDelegate}
   n70 = FindParentWidgetOfType!(
    execute=@n60.then_0
    StartingWidget:Widget Object Reference=Self-Reference()
    Type:Widget Class Reference=&/AthenaFrontend.AthenaFrontend_C
   ) -> (then, ReturnValue:Athena Frontend Object Reference)
   > Set AthenaFrontend(AthenaFrontend:Athena Frontend Object Reference=@n70) -> (Output_Get:Athena Frontend Object Reference)
   > Set Last Focused Widget(Last Focused Widget=Get MatchmakingWidget())
   > n71 = Sequence!(execute) -> (then_0, then_1, then_2)
   n72 = StartListeningToEvent!(
    execute:Exec=@n71.then_1
    Target:Gameplay Event Router Component Object Reference=GetGlobalEventRouter()
    EventType:Script Struct Object Reference=/FortniteCoreUI/UI/Social/Chat/TextChat/BPS_TextChatActiveStateChange
    Context:Object Reference=GetOwningPlayer()
    EventBubblingRule:EEventBubblingRule Enum
    None:%=Get GameplayMessageRouter() -> (ReturnValue:Gameplay Message Router Object Reference) {is=title K2Node_GetSubsystem}
   ) -> (
    then:Exec
    AsyncTaskProxy:Async Action Start Listening To Event Object Reference
    OnEventReceived:Exec
    Payload_bActive_4_#6:bool=~
   )
   n73 = Set IsTextChatActive(execute=@n72.OnEventReceived, IsTextChatActive:bool=@n72.Payload_bActive_4_#6) -> (Output_Get:bool)
   > n74 = Branch(Condition=@n73)
   > SetInputPreprocessorEnabled(bValue=false)
   BindToTabChange!(execute=@n71.then_2, Context:Widget Object Reference=Self-Reference(), EventToBind:"Delegate (by ref)"=@n19) -> (then)
   n75 = StartListeningToEvent!(
    execute=@n71.then_0
    Target=GetGlobalEventRouter()
    EventType=/Script/FortniteGame.GameplayMessageDualBladeMenuVisible
    Context=Self-Reference()
    EventBubblingRule
    None:%=Get GameplayMessageRouter()
   ) -> (then, AsyncTaskProxy, OnEventReceived, Payload_bBladeMenuVisible:bool=~)
   n76 = Set IsSideMenuActive(execute=@n75.OnEventReceived, IsSideMenuActive:bool=@n75.Payload_bBladeMenuVisible) -> (Output_Get:bool)
   > n77 = Branch(Condition=@n76)
   > SetInputPreprocessorEnabled(bValue=false)
   Branch(execute=@n60.then_3, Condition=IsMobileGame())
   > SetMobileStyle() {is=title K2Node_CallFunction}
   n78 = GetNumericCVar!(CVar:String=Fort.Rollback.UseDynamicUIFrontend) -> (OutValue:float, ReturnValue:bool)
   Branch(execute=@n60.then_2, Condition=@n78.ReturnValue)
   > n79 = Branch(Condition="float > float"(A=@n78))
   > SetPadding(
    self:Border Slot Object Reference=SlotAsBorderSlot!(Widget:Widget Object Reference=Get Content_MOTDHolder()) -> (ReturnValue:Border Slot Object Reference)
    InPadding_Left=-2
    InPadding_Top=136
    InPadding_Right=0
    InPadding_Bottom=0
   )
   SetPadding(
    execute=@n79.else
    self:Overlay Slot Object Reference=SlotAsOverlaySlot(Widget=Get Overlay_Discover() -> (Overlay_Discover:Overlay Object Reference))
    InPadding_Left=0
    InPadding_Top=-88
    InPadding_Right=0
    InPadding_Bottom=0
   )
   Event On Deactivated()
   > UpdateEvoScrollTransitionAllowed()
   > OnShowExpandableLevel(self=Get Expandable_Level_Switcher(), ShouldShow=false)
   > n80 = Branch(Condition=Get bIsPushedOnToContentPanelStack() -> (bIsPushedOnToContentPanelStack:bool))
   CompleteToast(execute=@n80.else, self:Fancy Toast Widget Object Reference=Get StoreToastWidget())
   > n81 = Get!(execute) -> (then, else, StoreToastRequest:Store Toast Request Object Reference)
   > SkipRequest(self=@n81)
   > Set StoreToastRequest()
   > n82 = Clear and Invalidate Timer by Handle(Handle=Get TryShowToast() -> (TryShowToast:Timer Handle Structure))
   ActivateMOTDWidget(execute=[@n82.then, @n81.else], Activate=false, SkipActivation=false)
   n83 = Event OnPlaylistChanged!() -> (OutputDelegate, then, LobbyType:EFortLobbyType Enum=Default, PlaylistData:Fort Playlist Athena Object Reference) {is=title K2Node_Event}
   > n84 = Switch on EFortLobbyType!(execute, Selection:EFortLobbyType Enum=@n83.LobbyType) -> (Default:Exec, Tournament:Exec, Creative:Exec, Division:Exec) {is=title K2Node_SwitchEnum}
   SetActionHandlerStateFromHandle!(
    execute=@n84.Tournament
    self:Common Activatable Panel Legacy Object Reference
    InputActionRow:Data Table Row Handle Structure=Get InputAction_InspectChallenges() -> (InputAction_InspectChallenges:Data Table Row Handle Structure)
    State:EInputActionState Enum=HiddenAndDisabled
   ) -> (then)
   > PlaySound2D(Sound=/Game/Sounds/Fort_UI_Sounds/FrontEnd/Fort_UI_Showdown_Panel_Cue)
   Event Destruct()
   > Unbind Event from On Front End Camera Changed(self:Fort Front End Context Object Reference=@n69, Delegate=@n68) {is=title K2Node_RemoveDelegate}
   n85 = Get Expandable_Rank_Container()
   DisplayMessageIfFirstTime!(
    execute
    self:WBP Habanero Rank Protection Coach Mark Object Reference=Get WBP_HabaneroRankProtection_CoachMark() -> (WBP_HabaneroRankProtection_CoachMark:WBP Habanero Rank Protection Coach Mark Object Reference)
    IsContainerVisible:bool="Equal (Enum)"(
     A:ESlateVisibility Enum=GetVisibility!(self:Widget Object Reference=@n85) -> (ReturnValue:ESlateVisibility Enum=Visible)
     B:ESlateVisibility Enum=SelfHitTestInvisible
    )
   ) -> (then, MessageDisplayed:bool)
   Navigate To Last Focused Widget(execute=@n86.then)
   Event DisplayGlyphQuestUpdated!() -> (OutputDelegate, then, NewGlyphs:int=0) {is=title K2Node_Event}
   Event OnQuestsButtonVisibilityChanged!() -> (OutputDelegate, then, NewVisibility:ESlateVisibility Enum=Visible) {is=title K2Node_Event}
   Event OnRankedProgressDisplayUpdated!() -> (OutputDelegate, then, bPlayingRanked:bool) {is=title K2Node_Event}
   UnbindAllFromAnimationFinished!(execute=@n9.then, self:User Widget Object Reference, Animation:Widget Animation Object Reference=@n10) -> (then)
   > n86 = Set Last Focused Widget(Last Focused Widget=Get MatchmakingWidget())
  ]}
 GetPlayTimeOut {nodes [
   n1 = GetPlayTimeOut!() -> (then, PlayTimeState:EPlaytimeState Enum=Unblocked) {is=title K2Node_FunctionEntry}
   > Return Node!(execute, IsPlayTimeOut:bool="Equal (Enum)"(A:EPlaytimeState Enum=@n1, B:EPlaytimeState Enum=PlayTime_Blocked))
  ]}
 GetShowGrantedAdditionalPlaytimeModal {nodes [
   n1 = GetShowGrantedAdditionalPlaytimeModal!() -> (then, PlaytimeState:EPlaytimeState Enum=Unblocked, PreviousPlaytimestate:EPlaytimeState Enum=Unblocked) {is=title K2Node_FunctionEntry}
   > Return Node!(
    execute
    ShouldShowDialog:bool=AND Boolean(A="Equal (Enum)"(A=@n1, B=Unblocked), B="Equal (Enum)"(A=@n1.PreviousPlaytimestate, B=PlayTime_Blocked))
   )
  ]}
 GrantedAdditionalPlaytimeConfirmDialog_ButtonClicked {nodes [
   GrantedAdditionalPlaytimeConfirmDialog_ButtonClicked!() -> (then, ButtonIndex:int) {is=title K2Node_FunctionEntry}
   > FireEventElementInteractedByName!(execute, Element Name:String=GrantedAdditionalPlaytimeModal_Continue_Button, WidgetCallingFrom:Widget Object Reference=Self-Reference()) -> (then, Success:bool)
  ]}
 HandleLobbyColorSchemeChange {nodes [
   n1 = HandleLobbyColorSchemeChange!() -> (then, NewColorScheme:Linear Color Structure, UseDefaultScheme:bool=~) {is=title K2Node_FunctionEntry}
   > n2 = GetDynamicMaterial(self:Border Object Reference=Get LobbyDetails_Default() -> (LobbyDetails_Default:Common Border Object Reference))
   > SetVectorParameterValue(
    self=@n2
    ParameterName=BackingColor
    Value=Select(Option 0:Linear Color Structure, Option 1:Linear Color Structure=@n1, Index=@n1.UseDefaultScheme) -> (ReturnValue:Linear Color Structure)
   )
  ]}
 HandleNavigateDownFromRightSide {nodes [
   n1 = HandleNavigateDownFromRightSide!() -> (then, Navigation:EUINavigation Enum) {is=title K2Node_FunctionEntry}
   n2 = Get SqueegeeSlot_RatingCard()
   n3 = HasAnyUserFocus!(self:Widget Object Reference=@n2) -> (ReturnValue:bool)
   > n4 = Branch(Condition=AND Boolean(A=IsVisible(self=@n2), B=NOT Boolean(A=@n3)))
   n5 = Get SqueegeeSlot_RatingCard()
   > SetFocus(self=@n5)
   > Set Last Focused Widget(Last Focused Widget=@n5)
   > Return Node!(execute, ReturnValue:Widget Object Reference)
   n6 = HandleNavigateDownToDiscover!(execute=@n4.else, self, Navigation:EUINavigation Enum=@n1) -> (then, ReturnValue:Widget Object Reference)
   > Return Node(ReturnValue=@n6)
  ]}
 HandleNavigateDownToDiscover {nodes [
   HandleNavigateDownToDiscover!() -> (then, Navigation:EUINavigation Enum) {is=title K2Node_FunctionEntry}
   > NavigateToDiscover()
   > Return Node()
  ]}
 HandleNavigateDownToMM {nodes [
   HandleNavigateDownToMM!() -> (then, Navigation:EUINavigation Enum) {is=title K2Node_FunctionEntry}
   > n1 = Set Last Focused Widget(Last Focused Widget=Get MatchmakingWidget())
   > Return Node(ReturnValue=@n1)
  ]}
 HandleNavigateLeftFromRightContent {nodes [
   HandleNavigateLeftFromRightContent!() -> (then, Navigation:EUINavigation Enum) {is=title K2Node_FunctionEntry}
   > NavigateFromRightContent(self:Athena Lobby Base Object Reference)
   > Set Last Focused Widget()
   > Return Node()
  ]}
 HandleNavigateRightFromLeftContent {nodes [
   HandleNavigateRightFromLeftContent!() -> (then, Navigation:EUINavigation Enum) {is=title K2Node_FunctionEntry}
   > NavigateFromLeftContent(self:Athena Lobby Base Object Reference)
   > Set Last Focused Widget()
   > Return Node()
  ]}
 HandleNavigateUpFromRatingCard {nodes [
   HandleNavigateUpFromRatingCard!() -> (then, Navigation:EUINavigation Enum) {is=title K2Node_FunctionEntry}
   > Determine Right Side Focus Widget!(execute, self, FocusTopWidget:bool=false) -> (then, FocusHandled:bool) {is=title K2Node_CallFunction}
   > Return Node()
  ]}
 HandleNavigateUpToMOTDWidget {nodes [
   HandleNavigateUpToMOTDWidget!() -> (then, Navigation:EUINavigation Enum) {is=title K2Node_FunctionEntry}
   n1 = Get MOTDWidgetReference()
   > n2 = Branch(Condition=AND Boolean(A=IsVisible(self=@n1), B=IsValid(Object=@n1)))
   > n3 = Set Last Focused Widget(Last Focused Widget=@n1)
   Return Node(execute=[@n3.then, @n2.else], ReturnValue=@n3)
  ]}
 InjectMusicPlayer {nodes [
   InjectMusicPlayer() {is=title K2Node_FunctionEntry}
   n1 = GetConfigDataBool!(EntryName:String=music_player_activation) -> (bValueOut:bool, ReturnValue:bool)
   > Branch(Condition=AND Boolean(A=@n1, B=@n1.ReturnValue))
   > TryInjectPluginWidget(self:Fort Plugin Widget Injection Slot Object Reference=Get WidgetMusicPlayer() -> (WidgetMusicPlayer:Fort Plugin Widget Injection Slot Object Reference))
  ]}
 IsCancelCurrentAction {nodes [
   IsCancelCurrentAction() {is=title K2Node_FunctionEntry}
   n1 = Get Bottom Bar Play State() -> (Bottom Bar Play State:Data Table Row Handle Structure)
   > Return Node(ReturnValue:bool="Equal (Name)"(
     A=Cancel
     B=Break Data Table Row Handle!(DataTableRowHandle:"Data Table Row Handle Structure (by ref)"=@n1) -> (DataTable:Data Table Object Reference, RowName:Name) {is=title K2Node_BreakStruct}.RowName
    ))
  ]}
 OnActivityAgeRatingChanged {nodes [
   n1 = OnActivityAgeRatingChanged!() -> (then, HasAgeRating:bool=~) {is=title K2Node_FunctionEntry}
   n2 = Get Is Play Time Out()
   n3 = Get FortActivityViewModel() -> (FortActivityViewModel:Discover Activity View Model Object Reference)
   > n4 = Branch(Condition=AND Boolean!(
     A=@n1
     B=Get bShouldShowIARCRating!(self:Discover Activity View Model Object Reference=@n3) -> (bShouldShowIARCRating:bool)
     C:bool=NOT Boolean(A=@n2)
    ) -> (ReturnValue))
   > ShowActivityDataFromViewModel!(
    execute
    self:Fort Squeegee Widget Injection Slot Object Reference=Get SqueegeeSlot_RatingCard()
    GameActivityViewModel:Fort Discover Link Entry Tile Item VM Object Reference=Get FortActivityViewModel()
   ) -> (then)
   > SetVisibility(self=Get SqueegeeSlot_RatingCard(), InVisibility=SelfHitTestInvisible)
   SetVisibility(execute=@n4.else, self=Get SqueegeeSlot_RatingCard(), InVisibility=Collapsed)
  ]}
 OnFrontendCameraChanged {nodes [
   n1 = OnFrontendCameraChanged!() -> (then, NewCamera:EFrontEndCamera Enum=Invalid, OldCamera:EFrontEndCamera Enum=Invalid) {is=title K2Node_FunctionEntry}
   > n2 = Switch on EFrontEndCamera!(execute, Selection:EFrontEndCamera Enum=@n1) -> (
    Command:Exec
    Command_HeroLoadout:Exec
    LegacyHeroLoadout:Exec
    Cosmetics:Exec
    Expeditions:Exec
    FrontendDefault:Exec
    Heroes:Exec
    HeroSelect:Exec
    HeroLoadout:Exec
    Home:Exec
    HomeBase:Exec
    Login:Exec
    Manage1:Exec
    Manage2:Exec
    Manage3:Exec
    Manage4:Exec
    MissionControl:Exec
    Party:Exec
    Play:Exec
    Research:Exec
    SkillTrees:Exec
    SmallCosmetics:Exec
    SpatialUI:Exec
    SpecialEvent:Exec
    SpecialEvent2:Exec
    SpecialEvent3:Exec
    Store:Exec
    StoreItemInspect:Exec
    StwFrontendDefault:Exec
    SurvivorSquadBuilding1:Exec
    SurvivorSquadBuilding2:Exec
    SurvivorSquadBuilding3:Exec
    SurvivorSquadBuilding4:Exec
    TutorialPhaseOne:Exec
    TutorialPhaseTwo:Exec
    TutorialPhaseThree:Exec
    Upgrades:Exec
    Vault:Exec
    WorldMap:Exec
    LobbyCentered:Exec
    CosmeticDisplay:Exec
    BattlePass:Exec
    Rewards:Exec
    SpecialEventRewards:Exec
    Armory:Exec
    DM_Customization:Exec
    DiscoveryHome:Exec
    MultipleSquads:Exec
    LargeSquad:Exec
    HestiaBeauty:Exec
    Defenders:Exec
    Schematics:Exec
    Workbench:Exec
    Storage:Exec
    CollectionBook:Exec
   ) {is=title K2Node_SwitchEnum}
   Set LastSavedCameraState(
    execute=[@n2.DiscoveryHome, @n2.HeroSelect, @n2.LargeSquad, @n2.LobbyCentered, @n2.MultipleSquads]
    LastSavedCameraState:EFrontEndCamera Enum=@n1
   ) -> (Output_Get:EFrontEndCamera Enum=Invalid)
  ]}
 OnMMWidgetDisplayChanged {nodes [OnMMWidgetDisplayChanged!() -> (then, Visible:bool=~) {is=title K2Node_FunctionEntry}]}
 OnNavigateToLeftContent {nodes [
   OnNavigateToLeftContent() {is=title K2Node_FunctionEntry}
   > n1 = Branch(Condition=Get is MM Widget on Screen())
   n2 = Get MatchmakingWidget()
   > SetFocus(self=@n2)
   > Set Last Focused Widget(Last Focused Widget=@n2)
   > Return Node(ReturnValue=true)
   Return Node(execute=@n1.else, ReturnValue=false)
  ]}
 OnNavigateToRightContent {nodes [
   OnNavigateToRightContent() {is=title K2Node_FunctionEntry}
   > n1 = Branch(Condition=Get Is Play Time Out())
   > Return Node(ReturnValue=false)
   n2 = Determine Right Side Focus Widget(execute=@n1.else, FocusTopWidget=true)
   > Return Node(ReturnValue=@n2)
  ]}
 SetFrontendCameraState {nodes [
   n1 = SetFrontendCameraState!() -> (then, NewState:EFrontEndCamera Enum=Invalid) {is=title K2Node_FunctionEntry}
   > n2 = Set LastSavedCameraState(LastSavedCameraState=@n1) -> (Output_Get=Invalid)
   > SetFrontEndCamera(
    self=Get FortFrontEndContext(PlayerController=GetOwningPlayer())
    NewState=@n2
    TransitionParams=Make View Target Transition Params(BlendTime=1, BlendFunction=VTBlend_EaseOut, BlendExp=2, bLockOutgoing=False)
   )
  ]}
 SetMobileExpandableWidgetsOrder {nodes [
   SetMobileExpandableWidgetsOrder() {is=title K2Node_FunctionEntry}
   > ClearChildren(self=Get VerticalBox_ExpandableWidget())
   > n1 = Sequence!(execute) -> (then_0, then_1, then_2, then_3:Exec, then_4:Exec, then_5:Exec)
   n2 = Get VerticalBox_ExpandableWidget()
   > n3 = AddChildToVerticalBox(self=@n2, Content=Get LobbyDetailsSlottingHostOverlay() -> (LobbyDetailsSlottingHostOverlay:Widget Slotting Host Overlay Object Reference))
   n4 = AddChildToVerticalBox(execute=@n1.then_3, self=@n2, Content=Get Expandable_Level_Switcher())
   n5 = AddChildToVerticalBox(execute=@n1.then_5, self=@n2, Content=Get WidgetMusicPlayer())
   n6 = AddChildToVerticalBox(
    execute=@n1.then_4
    self=@n2
    Content=Get VerticalBox_LobbyRankedDetails_ExtensionHolder() -> (VerticalBox_LobbyRankedDetails_ExtensionHolder:Vertical Box Object Reference)
   )
   n7 = AddChildToVerticalBox(execute=@n1.then_1, self=@n2, Content=Get WBP_Expandable_SurvivorTier())
   n8 = AddChildToVerticalBox(execute=@n1.then_2, self=@n2, Content=Get Expandable_Rank_Container())
   SetHorizontalAlignment(
    execute=[@n4.then, @n5.then, @n6.then, @n3.then, @n7.then, @n8.then]
    self:Vertical Box Slot Object Reference=[@n4, @n5, @n6, @n3, @n7, @n8]
    InHorizontalAlignment=HAlign_Right
   )
   > SetVerticalAlignment(self:Vertical Box Slot Object Reference=[@n4, @n5, @n7, @n8], InVerticalAlignment=VAlign_Fill)
  ]}
 SetMobileStyle {nodes [
   SetMobileStyle() {is=title K2Node_FunctionEntry}
   n1 = Get ContainerPaddings() -> (ContainerPaddings:Common Border Object Reference)
   > SetPadding(
    self:Border Slot Object Reference=SlotAsBorderSlot(Widget=GetContent!(self:Content Widget Object Reference=@n1) -> (ReturnValue:Widget Object Reference))
    InPadding_Left=48
    InPadding_Top=24
    InPadding_Right=48
    InPadding_Bottom=24
   )
   > SetPadding(
    self=SlotAsBorderSlot(Widget=Get Overlay_Main() -> (Overlay_Main:Overlay Object Reference))
    InPadding_Left=48
    InPadding_Top=24
    InPadding_Right=48
    InPadding_Bottom=24
   )
   > SetPadding(
    self:Vertical Box Slot Object Reference=SlotAsVerticalBoxSlot(Widget=Get Switcher_LobbyDetailsMode())
    InPadding_Left=0
    InPadding_Top=0
    InPadding_Right=0
    InPadding_Bottom=0
   )
   > SetPadding(
    self:Widget Switcher Slot Object Reference=SlotAsWidgetSwitcherSlot!(Widget:Widget Object Reference=Get VerticalBox_ExpandableWidget()) -> (ReturnValue:Widget Switcher Slot Object Reference)
    InPadding_Left=0
    InPadding_Top=0
    InPadding_Right=0
    InPadding_Bottom=0
   )
   > SetPadding(
    self:Vertical Box Slot Object Reference=[
     SlotAsVerticalBoxSlot(Widget=Get Expandable_Rank_Container())
     SlotAsVerticalBoxSlot(Widget=Get Expandable_Level_Switcher())
     SlotAsVerticalBoxSlot(Widget=Get VerticalBox_LobbyRankedDetails_ExtensionHolder())
     SlotAsVerticalBoxSlot(Widget=Get WBP_Expandable_SurvivorTier())
    ]
    InPadding_Left=0
    InPadding_Top=24
    InPadding_Right=0
    InPadding_Bottom=0
   )
   n2 = SlotAsWidgetSwitcherSlot(Widget=Get LobbyDetails_Tournament())
   > SetPadding(self:Widget Switcher Slot Object Reference=@n2, InPadding_Left=0, InPadding_Top=0, InPadding_Right=0, InPadding_Bottom=50)
   > SetHorizontalAlignment(self:Widget Switcher Slot Object Reference=@n2, InHorizontalAlignment=HAlign_Center)
   > SetVerticalAlignment(self:Widget Switcher Slot Object Reference=@n2, InVerticalAlignment=VAlign_Center)
   > SetPadding(
    self:Overlay Slot Object Reference=SlotAsOverlaySlot(Widget=Get PlayerPanel_PlayerOptions() -> (PlayerPanel_PlayerOptions:Fort Lobby Player Panel Object Reference))
    InPadding_Left=0
    InPadding_Top=0
    InPadding_Right=0
    InPadding_Bottom=125
   )
   > SetPadding(
    self:Border Slot Object Reference=SlotAsBorderSlot(Widget=Get Content_MOTDHolder())
    InPadding_Left=-2
    InPadding_Top=136
    InPadding_Right=0
    InPadding_Bottom=0
   )
   > SetPadding(
    self:Vertical Box Slot Object Reference=SlotAsVerticalBoxSlot(Widget=Get GridPanel_MatcmakingWidget() -> (GridPanel_MatcmakingWidget:Grid Panel Object Reference))
    InPadding_Left=0
    InPadding_Top=0
    InPadding_Right=0
    InPadding_Bottom=72
   )
   n3 = SlotAsBorderSlot(Widget=GetContent(self=Get Content_MatchmakingWidget()))
   > SetPadding(self:Border Slot Object Reference=@n3, InPadding_Left=0, InPadding_Top=380, InPadding_Right=0, InPadding_Bottom=0)
   > SetHorizontalAlignment(self:Border Slot Object Reference=@n3, InHorizontalAlignment=HAlign_Right)
   > SetVerticalAlignment(self:Border Slot Object Reference=@n3, InVerticalAlignment=VAlign_Bottom)
   n4 = SlotAsOverlaySlot(Widget=Get VerticalBox_Region() -> (VerticalBox_Region:Vertical Box Object Reference))
   > SetPadding(self:Overlay Slot Object Reference=@n4, InPadding_Left=0, InPadding_Top=78, InPadding_Right=0, InPadding_Bottom=0)
   > SetHorizontalAlignment(self:Overlay Slot Object Reference=@n4, InHorizontalAlignment=HAlign_Right)
   > SetVerticalAlignment(self:Overlay Slot Object Reference=@n4, InVerticalAlignment=VAlign_Fill)
   > SetRenderTranslation(self=Get MatchmakingRegion(), Translation=(X=-205,Y=0))
   > SetPadding(self=SlotAsOverlaySlot(Widget=Get StoreToastWidget()), InPadding_Left=0, InPadding_Top=40, InPadding_Right=0, InPadding_Bottom=0)
   > SetPadding(
    self:Vertical Box Slot Object Reference=SlotAsVerticalBoxSlot(Widget=Get CommonBorder_SqueegeeSlot() -> (CommonBorder_SqueegeeSlot:Common Border Object Reference))
    InPadding_Left=0
    InPadding_Top=0
    InPadding_Right=0
    InPadding_Bottom=112
   )
   > SetPadding(
    self:Overlay Slot Object Reference=SlotAsOverlaySlot(Widget=Get Overlay_Discover())
    InPadding_Left=0
    InPadding_Top=0
    InPadding_Right=0
    InPadding_Bottom=0
   )
   n5 = Get Notice_Wrapper() -> (Notice_Wrapper:Overlay Object Reference)
   > AddChild(self=Get ScaleBox_Notice_Wrapper() -> (ScaleBox_Notice_Wrapper:Scale Box Object Reference), Content=@n5)
   > SetHorizontalAlignment(
    self:Scale Box Slot Object Reference=SlotAsScaleBoxSlot!(Widget:Widget Object Reference=@n5) -> (ReturnValue:Scale Box Slot Object Reference)
    InHorizontalAlignment=HAlign_Fill
   )
   > SetVisibility(self=Get SizeBox_Notice_Wrapper_Mobile() -> (SizeBox_Notice_Wrapper_Mobile:Size Box Object Reference), InVisibility=SelfHitTestInvisible)
   > SetSidesToPad!(
    execute
    self:Safe Zone Object Reference=Get SafeZone()
    InPadLeft:bool=true
    InPadRight:bool=true
    InPadTop:bool=true
    InPadBottom:bool=true
   ) -> (then)
   n6 = Get!() -> (Padding_Left:float=~, Padding_Top:float=~, Padding_Right:float=~, Padding_Bottom:float=~)
   > SetPadding(
    self=SlotAsOverlaySlot(Widget=Get SafeZone_RIGHT_LobbyDetailsMode() -> (SafeZone_RIGHT_LobbyDetailsMode:Safe Zone Object Reference))
    InPadding_Left=@n6
    InPadding_Top=@n6.Padding_Top
    InPadding_Right=0
    InPadding_Bottom=@n6.Padding_Bottom
   )
   n7 = Get() -> (Padding_Left=~, Padding_Top=~, Padding_Right=~, Padding_Bottom=~)
   > SetPadding(
    self=SlotAsOverlaySlot(Widget=Get SafeZone_LEFT_MMColumn() -> (SafeZone_LEFT_MMColumn:Safe Zone Object Reference))
    InPadding_Left=0
    InPadding_Top=@n7.Padding_Top
    InPadding_Right=@n7.Padding_Right
    InPadding_Bottom=@n7.Padding_Bottom
   )
  ]}
 SetPlayCancelAndBackButtonsVisible {nodes [
   n1 = SetPlayCancelAndBackButtonsVisible!() -> (then, ShouldPlayCancelButtonBeVisible:bool=~) {is=title K2Node_FunctionEntry}
   > SetBottomButtonVisibility(self=Get WBP_BottomBarHack_MMPlayOrCancel(), ShouldButtonBeVisible=@n1)
   n2 = IsCancelCurrentAction() {is=title K2Node_CallFunction}
   > SetBottomButtonVisibility(
    self=Get WBP_BottomBarHack_BackToDiscover()
    ShouldButtonBeVisible=AND Boolean!(A=NOT Boolean(A=@n2), B=Get is MM Widget on Screen()) -> (ReturnValue)
   )
  ]}
 SetPlayTimeOut {nodes [
   n1 = SetPlayTimeOut!() -> (then, IsPlayTimeOut:bool=~) {is=title K2Node_FunctionEntry}
   > n2 = Set Is Play Time Out(Is Play Time Out:bool=@n1) -> (Output_Get:bool)
   > SetVisibility(
    self=Get VerticalBox_ExpandableWidget()
    InVisibility="To Visibility (Boolean)"!(bIsVisible:bool=@n2, TrueVisibility:ESlateVisibility Enum=Collapsed, FalseVisibility:ESlateVisibility Enum=SelfHitTestInvisible) -> (ReturnValue:ESlateVisibility Enum=Visible)
   )
   > Branch(Condition=Get Is Play Time Out())
   > ActivateMOTDWidget(Activate=false, SkipActivation=false)
   > OnActivityAgeRatingChanged!(execute, self, HasAgeRating:bool=false) -> (then) {is=title K2Node_CallFunction}
  ]}
 ShowGrantedAdditionalPlaytimeModal {nodes [
   n1 = ShowGrantedAdditionalPlaytimeModal!() -> (then, ShouldShowDialog:bool=~) {is=title K2Node_FunctionEntry}
   > Branch(Condition=@n1)
   > n2 = Push Simple UIKit Modal Dialog!(
    execute
    Context:User Widget Object Reference=Self-Reference()
    VM Class:UI Kit Dialog Viewmodel Class Reference=/FortniteCoreUI/Frontend/Lobby/DialogVM_Lobby_GrantedAdditionalActivityTime.DialogVM_Lobby_GrantedAdditionalActivityTime_C
    OnDialogButtonClicked:"Delegate (by ref)"=Create Event()
   ) -> (then, DialogActivatableWidget:WBP UIKit Dialog Object Reference, DialogVM:UI Kit Dialog Viewmodel Object Reference)
   > Fire Event Main View Opened!(execute, View:Widget Object Reference=@n2, AdditionalInfo:Flow Analytics Additional Info Structure, bIsOverlay:bool=true) -> (then)
   n3 = Get PlayerPlaytimeStateVM() -> (PlayerPlaytimeStateVM:Player Playtime State View Model Object Reference)
   n4 = GetLocalizedPlaytimeWindowBounds!(self:Player Playtime State View Model Object Reference=Get PlayerPlaytimeStateVM()) -> (ReturnValue_StartTime:Text, ReturnValue_EndTime:Text, ReturnValue_bAlreadyStarted:bool=~)
   n5 = GetRemainingPlaytime!(self:Player Playtime State View Model Object Reference=Get PlayerPlaytimeStateVM()) -> (ReturnValue:Timespan Structure)
   n6 = Get Minutes!(A:Timespan Structure=@n5) -> (ReturnValue:int=0) {is=title K2Node_CallFunction}
   n7 = Get Hours!(A:Timespan Structure=@n5) -> (ReturnValue:int=0) {is=title K2Node_CallFunction}
   n8 = SelectText(
    A=Format Text!(
     Format="LOCTEXT(\"{Hours} {Hours}|plural(one=hour,other=hours) {Minutes} {Minutes}|plural(one=min,other=mins)\")"
     Hours:int=@n7
     Minutes:int=@n6
    ) -> (Result)
    B=Format Text!(Format="LOCTEXT(\"{Minutes} {Minutes}|plural(one=min,other=mins)\")", Minutes=@n6) -> (Result)
    bPickA="integer > integer"(A=@n7)
   )
   n9 = Format Text!(
    Format="LOCTEXT(\"You have {Time} to use from {StartTime} to {EndTime} today.\")"
    Time:Text=@n8
    StartTime:Text=@n4
    EndTime:Text=@n4.ReturnValue_EndTime
   ) -> (Result)
   n10 = Format Text!(Format="LOCTEXT(\"You have {Time} to use before {EndTime}.\")", Time=@n8, EndTime=@n4.ReturnValue_EndTime) -> (Result)
   > Set DialogInfoText!(
    execute
    DialogInfoText:Text=SelectText(
     A=SelectText(A=@n10, B=@n9, bPickA=@n4.ReturnValue_bAlreadyStarted)
     B=Format Text!(Format="LOCTEXT(\"You have {Time} to use today.\")", Time=@n8) -> (Result)
     bPickA=HasPlaytimeWindowBounds!(self:Player Playtime State View Model Object Reference=@n3) -> (ReturnValue:bool)
    )
    self:UI Kit Dialog Viewmodel Object Reference=@n2.DialogVM
   ) -> (then, Output_Get:Text)
  ]}
 UpdateEvoScrollTransitionAllowed {nodes [
   UpdateEvoScrollTransitionAllowed() {is=title K2Node_FunctionEntry}
   > "Set Player Parameter (Bool)"!(
    execute
    self:UIState Tree Manager Object Reference=Get UIStateTreeManager() -> (ReturnValue:UIState Tree Manager Object Reference) {is=title K2Node_GetSubsystem}
    ParameterName:Name=DoesUniverseAllowScrollTransition
    Value:bool=AND Boolean(A=IsActivated(), B=Get is MM Widget on Screen())
    StateTreeOwner:Player Controller Object Reference=GetFirstLocalFortPlayerController!(bIncludeNonSpectators:bool=true, bIncludeSpectators:bool=false) -> (ReturnValue:Fort Player Controller Object Reference)
   ) -> (then) {is=title K2Node_CallFunction}
  ]}
]
variables [
 AthenaFrontend {type=AthenaFrontend_C, default=null}
 Bottom Bar Play State {type=DataTableRowHandle, default {dataTable=null, rowName=None}}
 CanPlayCancelButtonBeVisible {type=boolean, default=true}
 FullScreenMOTD {type=FortDynamicMotdFullscreenWidget, default=null}
 GetConfigDataBoolWithDefault {type=boolean, default=true}
 InputAction_InspectChallenges {type=DataTableRowHandle, default {dataTable=&InputActionDataTable, rowName=Athena_QuestInspect_Open}}
 is MM Widget on Screen {type=boolean, default=true}
 is MOTD Activated {type=boolean}
 Is Play Time Out {type=boolean}
 IsSideMenuActive {type=boolean}
 IsTextChatActive {type=boolean}
 Last Focused Widget {type=Widget, default=null}
 LastSavedCameraState {type=EFrontEndCamera, default=HeroSelect}
 MOTDWidgetReference {type=UserWidget, default=null}
 OnMMWidgetDisplayChanged {}
 Selected Island View Model {type=SelectedIslandViewModel, default=null}
 StoreToastRequest {type=StoreToastRequest, default=null}
 TryShowToast {type=TimerHandle}
]
widgets {
 info {inheritedWidgetCount=15, parentClass=/Script/FortniteUI.AthenaLobbyBase, rootWidgetClass=&Overlay, widgetCount=53}
 widgets [
  Overlay_1 {widgetClassPath=&Overlay, widget=Overlay_1}
  Overlay_Discover {
   widgetClassPath=&Overlay
   bInherited=true
   parent=Overlay_1
   widget=Overlay_Discover
   slot {class=OverlaySlot, properties {horizontalAlignment=HAlign_Fill, verticalAlignment=VAlign_Fill}}
  }
  SafeZone {
   widgetClassPath=&SafeZone
   bIsVariable=true
   parent=Overlay_1
   widget=SafeZone
   properties {padBottom=false, padTop=false}
   slot {class=OverlaySlot, properties {horizontalAlignment=HAlign_Fill, verticalAlignment=VAlign_Fill}}
  }
  ContainerPaddings {
   widgetClassPath=&CommonBorder
   bIsVariable=true
   parent=SafeZone
   widget=ContainerPaddings
   properties {padding=<ltrb 12 0 12 80>}
   slot {class=SafeZoneSlot}
  }
  Overlay_Main {
   widgetClassPath=&Overlay
   bIsVariable=true
   parent=ContainerPaddings
   widget=Overlay_Main
   properties {navigation=&self:WidgetTree.Overlay_Main.WidgetNavigation_0}
   slot {class=BorderSlot, properties {padding=<ltrb 12 0 12 80>}}
  }
  SafeZone_RIGHT_LobbyDetailsMode {
   widgetClassPath=&SafeZone
   bIsVariable=true
   parent=Overlay_Main
   widget=SafeZone_RIGHT_LobbyDetailsMode
   properties {padBottom=false, padLeft=false, padRight=false, padTop=false}
   slot {class=OverlaySlot, properties {horizontalAlignment=HAlign_Right, verticalAlignment=VAlign_Fill, padding=<ltrb 0 0 68 32>}}
  }
  VerticalBox_LobbyDetailsMode {
   widgetClassPath=&VerticalBox
   bIsVariable=true
   parent=SafeZone_RIGHT_LobbyDetailsMode
   widget=VerticalBox_LobbyDetailsMode
   slot {class=SafeZoneSlot}
  }
  Notice_Wrapper {
   widgetClassPath=&Overlay
   bIsVariable=true
   parent=VerticalBox_LobbyDetailsMode
   widget=Notice_Wrapper
   slot {class=VerticalBoxSlot, properties {horizontalAlignment=HAlign_Right, padding=<ltrb 0 28 0 20>}}
  }
  Notice_Emergency {
   widgetClassPath=/FortniteCoreUI/UI/Foundation/EmergencyNotice/EmergencyNoticeWidgetBR
   bInherited=true
   parent=Notice_Wrapper
   widget=Notice_Emergency
   slot {class=OverlaySlot}
  }
  WidgetSlottingHostOverlay_LobbyDetails_TopRightSlot {
   widgetClassPath=/Script/WidgetSlotting.WidgetSlottingHostOverlay
   parent=VerticalBox_LobbyDetailsMode
   widget=WidgetSlottingHostOverlay_LobbyDetails_TopRightSlot
   properties {
    slottedWidgetDefaultHorizontalAlignment=HAlign_Right
    slottedWidgetDefaultVerticalAlignment=VAlign_Top
    hostTag {tagName=DynamicUI.Frontend.AthenaLobby.LobbyDetails.TopRight.Slot}
   }
   slot {class=VerticalBoxSlot, properties {verticalAlignment=VAlign_Top}}
  }
  SizeBox_Notice_Wrapper_Mobile {
   widgetClassPath=&SizeBox
   bIsVariable=true
   parent=VerticalBox_LobbyDetailsMode
   widget=SizeBox_Notice_Wrapper_Mobile
   properties {bOverride_WidthOverride=true, visibility=Collapsed, widthOverride=420}
   slot {class=VerticalBoxSlot, properties {horizontalAlignment=HAlign_Right, padding=<ltrb 0 122 0 0>, size {sizeRule=Automatic, value=0.5}}}
  }
  ScaleBox_Notice_Wrapper {
   widgetClassPath=&ScaleBox
   bIsVariable=true
   parent=SizeBox_Notice_Wrapper_Mobile
   widget=ScaleBox_Notice_Wrapper
   properties {stretch=ScaleToFitY, stretchDirection=DownOnly}
   slot {class=SizeBoxSlot, properties {verticalAlignment=VAlign_Top}}
  }
  Switcher_LobbyDetailsMode {
   widgetClassPath=&CommonWidgetSwitcherLegacy
   bInherited=true
   bIsVariable=true
   parent=VerticalBox_LobbyDetailsMode
   widget=Switcher_LobbyDetailsMode
   slot {class=VerticalBoxSlot, properties {horizontalAlignment=HAlign_Right, verticalAlignment=VAlign_Bottom, size {sizeRule=Fill, value=1}}}
  }
  VerticalBox_ExpandableWidget {
   widgetClassPath=&VerticalBox
   bIsVariable=true
   parent=Switcher_LobbyDetailsMode
   widget=VerticalBox_ExpandableWidget
   properties {navigation=&self:WidgetTree.VerticalBox_ExpandableWidget.WidgetNavigation_0}
   slot {class=WidgetSwitcherSlot, properties {verticalAlignment=VAlign_Bottom}}
  }
  VerticalBox_LobbyRankedDetails_ExtensionHolder {
   widgetClassPath=&VerticalBox
   bInherited=true
   bIsVariable=true
   parent=VerticalBox_ExpandableWidget
   widget=VerticalBox_LobbyRankedDetails_ExtensionHolder
   slot {class=VerticalBoxSlot, properties {horizontalAlignment=HAlign_Right, verticalAlignment=VAlign_Bottom}}
  }
  LobbyDetailsSlottingHostOverlay {
   widgetClassPath=&WidgetSlottingHostOverlay
   bIsVariable=true
   parent=VerticalBox_ExpandableWidget
   widget=LobbyDetailsSlottingHostOverlay
   properties {hostTag {tagName=Athena.UI.Frontend.AthenaLobbyDetails}}
   slot {class=VerticalBoxSlot, properties {horizontalAlignment=HAlign_Right, verticalAlignment=VAlign_Bottom, padding=<ltrb 0 16 0 0>}}
  }
  LobbySquadSlottingHostOverlay {
   widgetClassPath=&WidgetSlottingHostOverlay
   bIsVariable=true
   parent=VerticalBox_ExpandableWidget
   widget=LobbySquadSlottingHostOverlay
   properties {hostTag {tagName=Athena.UI.Frontend.AthenaSquadDetails}}
   slot {class=VerticalBoxSlot, properties {horizontalAlignment=HAlign_Right, verticalAlignment=VAlign_Bottom, padding=<ltrb 0 16 0 0>}}
  }
  WBP_PluginStub_Expandable {
   widgetClassPath=&/WBP_PluginStub_Expandable
   bIsVariable=true
   parent=VerticalBox_ExpandableWidget
   widget=WBP_PluginStub_Expandable
   slot {class=VerticalBoxSlot, properties {horizontalAlignment=HAlign_Right, padding=<ltrb 0 16 0 0>}}
  }
  Expandable_Rank_Container {
   widgetClassPath=&Overlay
   bIsVariable=true
   parent=VerticalBox_ExpandableWidget
   widget=Expandable_Rank_Container
   slot {class=VerticalBoxSlot, properties {horizontalAlignment=HAlign_Right, padding=<ltrb 0 16 0 0>}}
  }
  WBP_Expandable_Rank {
   widgetClassPath=/FortniteCoreUI/UI/Frontend/Lobby/Habanero/WBP_Expandable_Rank
   bIsVariable=true
   parent=Expandable_Rank_Container
   widget=WBP_Expandable_Rank
   properties {holdData=&CommonUI_HoldData_Base_C}
   slot {class=OverlaySlot}
  }
  HabaneroRankProtection_CoachMark_AnchorContainer {
   widgetClassPath=&FortVisualAttachment
   parent=Expandable_Rank_Container
   widget=HabaneroRankProtection_CoachMark_AnchorContainer
   properties {contentAnchor=<1 1>}
   slot {class=OverlaySlot, properties {horizontalAlignment=HAlign_Right, padding=<ltrb 0 -35 0 0>}}
  }
  WBP_HabaneroRankProtection_CoachMark {
   widgetClassPath=/FortniteCoreUI/UI/Frontend/Lobby/Habanero/WBP_HabaneroRankProtection_CoachMark
   bIsVariable=true
   parent=HabaneroRankProtection_CoachMark_AnchorContainer
   widget=WBP_HabaneroRankProtection_CoachMark
   slot {class=SizeBoxSlot}
  }
  Expandable_Level_Switcher {
   widgetClassPath=/FortniteCoreUI/UI/Frontend/Lobby/Habanero/WBP_Expandable_Level_Switcher
   bIsVariable=true
   parent=VerticalBox_ExpandableWidget
   widget=Expandable_Level_Switcher
   slot {class=VerticalBoxSlot, properties {horizontalAlignment=HAlign_Right, padding=<ltrb 0 16 0 0>}}
  }
  WBP_Expandable_SurvivorTier {
   widgetClassPath=/Game/UI/Frontend/Lobby/SurvivorTier/WBP_Expandable_SurvivorTier
   bIsVariable=true
   parent=VerticalBox_ExpandableWidget
   widget=WBP_Expandable_SurvivorTier
   slot {class=VerticalBoxSlot, properties {horizontalAlignment=HAlign_Right, padding=<ltrb 0 16 0 0>}}
  }
  WidgetMusicPlayer {
   widgetClassPath=/Game/UI/Plugin/FortPluginWidgetInjectionSlot
   bIsVariable=true
   parent=VerticalBox_ExpandableWidget
   widget=WidgetMusicPlayer
   properties {bAutoInjectWidget=false, pluginId=MusicPlaylist, pluginWidgetId=LobbyMusic}
   slot {class=VerticalBoxSlot, properties {horizontalAlignment=HAlign_Right, padding=<ltrb 0 16 0 0>}}
  }
  LobbyDetails_Tournament {
   widgetClassPath=/FortniteCoreUI/Frontend/Showdown/ShowdownLobbyPanel
   bInherited=true
   parent=Switcher_LobbyDetailsMode
   widget=LobbyDetails_Tournament
   properties {eventDetailsAction {dataTable=null, rowName=None}}
   slot {class=WidgetSwitcherSlot, properties {padding=<ltrb 0 0 0 32>}}
  }
  CommonBorder_SqueegeeSlot {
   widgetClassPath=&CommonBorder
   bIsVariable=true
   parent=VerticalBox_LobbyDetailsMode
   widget=CommonBorder_SqueegeeSlot
   properties {horizontalAlignment=HAlign_Right, verticalAlignment=VAlign_Bottom, padding=<ltrb 0 0 0 0>}
   slot {class=VerticalBoxSlot, properties {horizontalAlignment=HAlign_Right, verticalAlignment=VAlign_Bottom}}
  }
  SqueegeeSlot_RatingCard {
   widgetClassPath=/Game/UI/Plugin/Squeegee/SqueegeeInjectionSlot
   bIsVariable=true
   parent=CommonBorder_SqueegeeSlot
   widget=SqueegeeSlot_RatingCard
   properties {bIsFocusable=true, pluginId=Squeegee, pluginWidgetId=RatingCard, visibility=Collapsed, padding=<ltrb 0 16 0 0>}
   slot {class=BorderSlot, properties {horizontalAlignment=HAlign_Right, verticalAlignment=VAlign_Bottom, padding=<ltrb 0 0 0 0>}}
  }
  VerticalBox_2 {
   widgetClassPath=&VerticalBox
   parent=Overlay_Main
   widget=VerticalBox_2
   slot {class=OverlaySlot, properties {horizontalAlignment=HAlign_Right, verticalAlignment=VAlign_Bottom, padding=<ltrb 0 0 0 16>}}
  }
  PendingAccountMergeNotification {
   widgetClassPath=&/PendingAccountMergeNotification_Legacy
   parent=VerticalBox_2
   widget=PendingAccountMergeNotification
   properties {
    onButtonBaseClicked=(&self:WidgetTree.PendingAccountMergeNotification.HandleOnButtonClicked)
    onButtonBaseDoubleClicked=(&self:WidgetTree.PendingAccountMergeNotification.HandleOnButtonDoubleClicked)
    onButtonBaseHovered=(&self:WidgetTree.PendingAccountMergeNotification.HandleOnButtonHovered)
    onButtonBaseUnhovered=(&self:WidgetTree.PendingAccountMergeNotification.HandleOnButtonUnhovered)
    onSelectedChangedBase=(&self:WidgetTree.PendingAccountMergeNotification.HandleOnSelectedChanged)
   }
   slot {class=VerticalBoxSlot}
  }
  PlayerPanel_PlayerOptions {
   widgetClassPath=&/AthenaLobbyPlayerPanel
   bInherited=true
   parent=Overlay_Main
   widget=PlayerPanel_PlayerOptions
   properties {visibility=Hidden}
   slot {class=OverlaySlot, properties {horizontalAlignment=HAlign_Center, verticalAlignment=VAlign_Bottom, padding=<ltrb 0 0 0 134>}}
  }
  MinimalTeamDisplay {
   widgetClassPath=/FortniteCoreUI/Frontend/Parties/SquadManagement/MinimalTeamDisplay
   bIsVariable=true
   parent=Overlay_Main
   widget=MinimalTeamDisplay
   slot {class=OverlaySlot, properties {horizontalAlignment=HAlign_Center, padding=<ltrb 125 12 0 0>}}
  }
  StoreToastWidget {
   widgetClassPath=/FortniteCoreUI/UI/Foundation/Toasts/FancyToastWidget
   bIsVariable=true
   parent=Overlay_Main
   widget=StoreToastWidget
   properties {visibility=Collapsed}
   slot {class=OverlaySlot, properties {horizontalAlignment=HAlign_Center, padding=<ltrb 0 10 0 0>}}
  }
  VerticalBox_Region {
   widgetClassPath=&VerticalBox
   bIsVariable=true
   parent=Overlay_Main
   widget=VerticalBox_Region
   slot {class=OverlaySlot, properties {horizontalAlignment=HAlign_Right}}
  }
  MatchmakingRegion {
   widgetClassPath=/FortniteCoreUI/Athena/UI/Matchmaking/MatchmakingRegionAndPool
   bIsVariable=true
   parent=VerticalBox_Region
   widget=MatchmakingRegion
   properties {visibility=Hidden}
   slot {class=VerticalBoxSlot, properties {horizontalAlignment=HAlign_Right}}
  }
  Button_ShadowPlayHighlights {
   widgetClassPath=&IconTextButton
   bInherited=true
   parent=VerticalBox_Region
   widget=Button_ShadowPlayHighlights
   properties {
    bHideInputActionWithKeyboard=true
    buttonText=""
    onButtonBaseClicked=(&self:WidgetTree.Button_ShadowPlayHighlights.HandleOnButtonClicked)
    onButtonBaseDoubleClicked=(&self:WidgetTree.Button_ShadowPlayHighlights.HandleOnButtonDoubleClicked)
    onButtonBaseHovered=(&self:WidgetTree.Button_ShadowPlayHighlights.HandleOnButtonHovered)
    onButtonBaseUnhovered=(&self:WidgetTree.Button_ShadowPlayHighlights.HandleOnButtonUnhovered)
    onSelectedChangedBase=(&self:WidgetTree.Button_ShadowPlayHighlights.HandleOnSelectedChanged)
    style=/FortUILibrary/Foundation/Buttons/ButtonStyles/Primary/ButtonStyle-Primary-L.ButtonStyle-Primary-L_C
    visibility=Collapsed
    iconBrush {
     bIsDynamicallyLoaded=false
     drawAs=Image
     imageType=NoImage
     mirroring=NoMirror
     resourceName=None
     resourceObject=/Game/Athena/HUD/Art/ShadowPlayHighlightsButton
     tiling=NoTile
     imageSize=<225 32>
     margin=<ltrb 0 0 0 0>
     outlineSettings {
      bUseBrushTransparency=false
      roundingType=HalfHeightRadius
      width=0
      color {colorUseRule=UseColor_Specified, specifiedColor=<rgba 0 0 0 0>}
      cornerRadii=<0 0 0 1>
     }
     tintColor {colorUseRule=UseColor_Specified, specifiedColor=<rgba 0 0 0 1>}
     uVRegion {bIsValid=false, max=<0 0>, min=<0 0>}
    }
    pressedSlateSoundOverride {resourceObject=/Game/Sounds/Fort_UI_Sounds/FrontEnd/New_FrontEnd/Fortnite_UI_MainMenu_Generic_02_Cue}
   }
   slot {class=VerticalBoxSlot, properties {horizontalAlignment=HAlign_Center, verticalAlignment=VAlign_Center, padding=<ltrb 0 0 0 5>}}
  }
  Button_GenericLink {
   widgetClassPath=&IconTextButton
   bInherited=true
   parent=VerticalBox_Region
   widget=Button_GenericLink
   properties {
    buttonText=New Player Guide
    onButtonBaseClicked=(&self:WidgetTree.Button_GenericLink.HandleOnButtonClicked)
    onButtonBaseDoubleClicked=(&self:WidgetTree.Button_GenericLink.HandleOnButtonDoubleClicked)
    onButtonBaseHovered=(&self:WidgetTree.Button_GenericLink.HandleOnButtonHovered)
    onButtonBaseUnhovered=(&self:WidgetTree.Button_GenericLink.HandleOnButtonUnhovered)
    onSelectedChangedBase=(&self:WidgetTree.Button_GenericLink.HandleOnSelectedChanged)
   }
   slot {class=VerticalBoxSlot}
  }
  SafeZone_LEFT_MMColumn {
   widgetClassPath=&SafeZone
   bIsVariable=true
   parent=Overlay_Main
   widget=SafeZone_LEFT_MMColumn
   properties {padBottom=false, padLeft=false, padRight=false, padTop=false}
   slot {class=OverlaySlot, properties {verticalAlignment=VAlign_Fill, padding=<ltrb 68 0 0 32>}}
  }
  VerticalBox_MotdMatchmakingBox {
   widgetClassPath=&VerticalBox
   bIsVariable=true
   parent=SafeZone_LEFT_MMColumn
   widget=VerticalBox_MotdMatchmakingBox
   slot {class=SafeZoneSlot}
  }
  Content_MotdTeaserWidget {
   widgetClassPath=&CommonBorder
   bIsVariable=true
   parent=VerticalBox_MotdMatchmakingBox
   widget=Content_MotdTeaserWidget
   properties {navigation=&self:WidgetTree.Content_MotdTeaserWidget.WidgetNavigation_0, padding=<ltrb -2 38 0 0>}
   slot {class=VerticalBoxSlot, properties {horizontalAlignment=HAlign_Left, verticalAlignment=VAlign_Top, padding=<ltrb 0 0 0 24>}}
  }
  Content_MOTDHolder {
   widgetClassPath=&CommonBorder
   bIsVariable=true
   parent=Content_MotdTeaserWidget
   widget=Content_MOTDHolder
   properties {navigation=&self:WidgetTree.Content_MOTDHolder.WidgetNavigation_1}
   slot {class=BorderSlot, properties {padding=<ltrb -2 38 0 0>}}
  }
  Overlay_AboveMatchmaking {
   widgetClassPath=&Overlay
   bIsVariable=true
   parent=VerticalBox_MotdMatchmakingBox
   widget=Overlay_AboveMatchmaking
   slot {class=VerticalBoxSlot, properties {verticalAlignment=VAlign_Bottom, size {sizeRule=Fill, value=1}}}
  }
  GridPanel_MatcmakingWidget {
   widgetClassPath=&GridPanel
   bIsVariable=true
   parent=VerticalBox_MotdMatchmakingBox
   widget=GridPanel_MatcmakingWidget
   properties {navigation=&self:WidgetTree.GridPanel_MatcmakingWidget.WidgetNavigation_0}
   slot {class=VerticalBoxSlot, properties {verticalAlignment=VAlign_Bottom}}
  }
  Overlay_AboveMatchmaking_Navigable {
   widgetClassPath=&Overlay
   bIsVariable=true
   parent=GridPanel_MatcmakingWidget
   widget=Overlay_AboveMatchmaking_Navigable
   slot {class=GridSlot}
  }
  Content_MatchmakingWidget {
   widgetClassPath=&CommonBorder
   bInherited=true
   parent=GridPanel_MatcmakingWidget
   widget=Content_MatchmakingWidget
   properties {bReducePaddingBySafezone=true, padding=<ltrb 0 0 0 0>}
   slot {class=GridSlot, properties {row=1}}
  }
  Overlay_DynamicExtensions {
   widgetClassPath=&Overlay
   bInherited=true
   parent=Overlay_Main
   widget=Overlay_DynamicExtensions
   slot {class=OverlaySlot, properties {horizontalAlignment=HAlign_Fill, verticalAlignment=VAlign_Fill}}
  }
  WBP_BottomBarHack_BackToDiscover {
   widgetClassPath=/Game/UI/Frontend/BottomBar/WBP_BottomBarDecoyButton
   bIsVariable=true
   parent=Overlay_Main
   widget=WBP_BottomBarHack_BackToDiscover
   properties {triggeringInputAction {dataTable=&InputActionDataTable, rowName=Back_ToHomebar}}
   slot {class=OverlaySlot, properties {horizontalAlignment=HAlign_Right, verticalAlignment=VAlign_Bottom}}
  }
  WBP_BottomBarHack_MMPlayOrCancel {
   widgetClassPath=&WBP_BottomBarDecoyButton
   bIsVariable=true
   parent=Overlay_Main
   widget=WBP_BottomBarHack_MMPlayOrCancel
   properties {bDisplayInActionBar=false, triggeringInputAction {dataTable=&InputActionDataTable, rowName=PlayMatchmaking}}
   slot {class=OverlaySlot, properties {horizontalAlignment=HAlign_Right, verticalAlignment=VAlign_Bottom, padding=<ltrb 0 0 450 0>}}
  }
  LobbyDetails_Default {widgetClassPath=&CommonBorder, bInherited=true, widget=null}
  Button_Quests {widgetClassPath=&CommonButtonBase, bInherited=true, widget=null}
  VerticalBox_LobbyDetails {widgetClassPath=&VerticalBox, bInherited=true, widget=null}
  VerticalBox_LobbyDetails_ExtensionHolder {widgetClassPath=&VerticalBox, bInherited=true, widget=null}
  HabaneroDisplayWidget {widgetClassPath=/Script/FortniteUI.FortHabaneroLobbyWidget, bInherited=true, widget=null}
 ]
}
