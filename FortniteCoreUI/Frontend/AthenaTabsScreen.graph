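# AthenaTabsScreen is a frontend tab screen widget, based on FortAthenaTabsScreenBase, that hosts the Athena lobby and the V-Bucks store screen.
# Both screens sit in the inherited TabContentSwitcher, which has zero transition duration and fills an overlay inside a safe zone with padding disabled on every side.
# On Construct, an IsMobileGame check leads to SetCompactUIStyle, which clears the lobby slot's padding, removing its default top inset.
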WidgetBlueprint /FortniteCoreUI/Frontend/AthenaTabsScreen
graphs [
 EventGraph {nodes [
   Event Construct()
   > Branch(Condition=IsMobileGame())
   > SetCompactUIStyle() {is=title K2Node_CallFunction}
  ]}
 SetCompactUIStyle {nodes [
   SetCompactUIStyle() {is=title K2Node_FunctionEntry}
   > SetPadding!(
    execute
    self:Widget Switcher Slot Object Reference=SlotAsWidgetSwitcherSlot!(Widget:Widget Object Reference=Get AthenaLobby() -> (AthenaLobby:Athena Lobby Object Reference)) -> (ReturnValue:Widget Switcher Slot Object Reference)
    InPadding_Left:float=0
    InPadding_Top:float=0
    InPadding_Right:float=0
    InPadding_Bottom:float=0
   ) -> (then)
  ]}
]
variables []
widgets {
 info {inheritedWidgetCount=1, parentClass=/Script/FortniteUI.FortAthenaTabsScreenBase, rootWidgetClass=&SafeZone, widgetCount=5}
 widgets [
  SafeZoneContent {
   widgetClassPath=&SafeZone
   bIsVariable=true
   widget=SafeZoneContent
   properties {padBottom=false, padLeft=false, padRight=false, padTop=false}
  }
  TEMP_DONOTCHANGEORDER {widgetClassPath=&Overlay, parent=SafeZoneContent, widget=TEMP_DONOTCHANGEORDER, slot {class=SafeZoneSlot}}
  TabContentSwitcher {
   widgetClassPath=&CommonWidgetSwitcherLegacy
   bInherited=true
   bIsVariable=true
   parent=TEMP_DONOTCHANGEORDER
   widget=TabContentSwitcher
   properties {transitionDuration=0}
   slot {class=OverlaySlot, properties {horizontalAlignment=HAlign_Fill, verticalAlignment=VAlign_Fill}}
  }
  AthenaLobby {
   widgetClassPath=&/AthenaLobby
   bIsVariable=true
   parent=TabContentSwitcher
   widget=AthenaLobby
   slot {class=WidgetSwitcherSlot, properties {padding=<ltrb 0 88 0 0>}}
  }
  StoreMain_Root {
   widgetClassPath=/FortniteCoreUI/Frontend/StoreMain/WBP_VbucksStore_Screen
   bIsVariable=true
   parent=TabContentSwitcher
   widget=StoreMain_Root
   slot {class=WidgetSwitcherSlot}
  }
 ]
}
