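# DualBladeMenu is the frontend main-menu overlay hosting the sidebar, XP bar, backboard and window controls.
# Activation plays IntroAnim with a sound, making the injection slot volatile during it when in zone, and closing reverses it before deactivating.
# When VM-driven sidebar is enabled, the injected plugin sidebar is activated, deferred until injection if needed.
# Orientation changes realign sidebar and XP bar, gated by a portrait CVar and compact UI.
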WidgetBlueprint /FortniteCoreUI/Frontend/MainMenu/DualBladeMenu
animations [IntroAnim {
  binding_count=3
  length_seconds=0.25
  object=&self:IntroAnim
  sequence {
   bindings [
    Overlay_Sidebar {
     id=#1
     possessed_class=Overlay
     tracks [
      {
       type=MovieScene2DTransformTrack
       sections [{
         type=MovieScene2DTransformSection
         end_frame=15
         start_frame=0
         channels [Translation.X {type=MovieSceneScriptingFloatChannel, keys [[0,200],[15,0]]}]
        }]
      }
      {
       type=MovieSceneFloatTrack
       display_name=Render Opacity
       sections [{type=MovieSceneFloatSection, end_frame=15, start_frame=0, channels [None {type=MovieSceneScriptingFloatChannel, keys [[0,0],[7,1]]}]}]
      }
     ]
    }
    Overlay_xp {
     id=#2
     possessed_class=Overlay
     tracks [{
       type=MovieSceneFloatTrack
       display_name=Render Opacity
       sections [{type=MovieSceneFloatSection, end_frame=15, start_frame=0, channels [None {type=MovieSceneScriptingFloatChannel, keys [[0,0],[15,1]]}]}]
      }]
    }
    Button_BackBoard {
     id=#3
     possessed_class=IconTextButton_C
     tracks [{
       type=MovieSceneFloatTrack
       display_name=Render Opacity
       sections [{type=MovieSceneFloatSection, end_frame=15, start_frame=0, channels [None {type=MovieSceneScriptingFloatChannel, keys [[0,0],[15,1]]}]}]
      }]
    }
   ]
   display_rate {numerator=60}
  }
 }]
graphs [
 EventGraph {nodes [
   n1 = OnWidgetInjected!() -> (OutputDelegate, then, Widget:Widget Object Reference) {is=title K2Node_CustomEvent}
   > n2 = Cast To CommonActivatableWidget!(execute, Object=@n1.Widget) -> (then, CastFailed, AsCommon Activatable Widget:Common Activatable Widget Object Reference) {is=title K2Node_DynamicCast}
   > ActivateWidget(self=@n2)
   n3 = ForceVolatileOffOnAnimationFinished() {is=title K2Node_CustomEvent}
   > ForceVolatile!(
    execute
    self:Widget Object Reference=Get NewSidebarInjectionSlot() -> (NewSidebarInjectionSlot:Fort Plugin Widget Injection Slot Object Reference)
    bForce:bool=false
   ) -> (then)
   > UnbindAllFromAnimationFinished!(
    execute
    self:User Widget Object Reference
    Animation:Widget Animation Object Reference=Get IntroAnim() -> (IntroAnim:Widget Animation Object Reference)
   ) -> (then)
   n4 = Orientation Changed!() -> (OutputDelegate, then, inScreenOrientation:EScreenOrientation Enum=Unknown) {is=title K2Node_CustomEvent}
   > HandleOrientationChanged!(execute, self, ScreenOrientation:EScreenOrientation Enum=@n4.inScreenOrientation) -> (then)
   n5 = Event OnFullScreenWidgetCreated!() -> (OutputDelegate, then, NewWidget:Widget Object Reference) {is=title K2Node_Event}
   > n6 = AddChildToOverlay!(
    execute
    self:Overlay Object Reference=Get MainOverlay() -> (MainOverlay:Overlay Object Reference)
    Content:Widget Object Reference=@n5.NewWidget
   ) -> (then, ReturnValue:Overlay Slot Object Reference)
   > SetHorizontalAlignment(self=@n6, InHorizontalAlignment=HAlign_Fill)
   > SetVerticalAlignment(self=@n6, InVerticalAlignment=VAlign_Fill)
   Event On Activated()
   > PlayAnimationForward(InAnimation=Get IntroAnim())
   > n7 = Branch(Condition=IsInZone!() -> (ReturnValue:bool))
   > BindToAnimationFinished(Animation=Get IntroAnim(), Delegate=@n3)
   > n8 = ForceVolatile(self=Get NewSidebarInjectionSlot(), bForce=true)
   PlaySound2D(execute=[@n8.then, @n7.else], Sound=Get SidebarOpenSound() -> (SidebarOpenSound:Sound Base Object Reference))
   > HandleOrientationChanged(ScreenOrientation=GetDeviceOrientation!() -> (ReturnValue:EScreenOrientation Enum=Unknown))
   > Branch(Condition=Get bEnableVMDrivenSidebar!(self:Fort Runtime Options Object Reference=GetRuntimeOptions()) -> (bEnableVMDrivenSidebar:bool))
   > n9 = Cast To CommonActivatableWidget(Object=GetInjectedWidget!(self:Fort Plugin Widget Injection Slot Object Reference=Get NewSidebarInjectionSlot()) -> (ReturnValue:Widget Object Reference))
   > ActivateWidget(self=@n9)
   Unbind all Events from On Widget Injected!(execute:Exec=@n9.CastFailed, self:Fort Plugin Widget Injection Slot Object Reference=Get NewSidebarInjectionSlot()) -> (then:Exec) {is=title K2Node_ClearDelegate}
   > Bind Event to On Widget Injected(self:Fort Plugin Widget Injection Slot Object Reference=Get NewSidebarInjectionSlot(), Delegate=@n1) {is=title K2Node_AddDelegate}
   Event Destruct()
   > n10 = Cast To PlatformGameInstance!(execute, Object=GetGameInstance!() -> (ReturnValue:Game Instance Object Reference)) -> (then, CastFailed, AsPlatform Game Instance:Platform Game Instance Object Reference) {is=title K2Node_DynamicCast}
   n11 = Create Event()
   > Unbind Event from Application Received Screen Orientation Changed Notification Delegate(self:Platform Game Instance Object Reference=@n10, Delegate=@n11) {is=title K2Node_RemoveDelegate}
   Event OnCloseDualBladeMenu() {is=title K2Node_Event}
   > Branch(Condition=IsActivated!(self:Common Activatable Widget Object Reference) -> (ReturnValue:bool))
   > n12 = "Play Animation with Finished event (legacy)"!(
    execute:Exec
    Widget:User Widget Object Reference=Self-Reference()
    InAnimation:Widget Animation Object Reference=Get IntroAnim()
    StartAtTime:float=0
    NumLoopsToPlay:int=1
    PlayMode:EUMGSequencePlayMode Enum=Reverse
    PlaybackSpeed:float=2
   ) -> (then:Exec, Result:UMGSequence Player Object Reference, Finished:Exec) {is=title K2Node_PlayAnimation}
   > PlaySound2D(Sound=Get SidebarCloseSound() -> (SidebarCloseSound:Sound Base Object Reference))
   n13 = Sequence(execute=@n12.Finished)
   > Branch(Condition=Get bEnableVMDrivenSidebar(self=GetRuntimeOptions()))
   n14 = Get NewSidebarInjectionSlot()
   > Unbind all Events from On Widget Injected(self=@n14)
   > n15 = Cast To CommonActivatableWidget(Object=GetInjectedWidget(self=@n14))
   > DeactivateWidget(self=@n15)
   DeactivateWidget(execute=@n13.then_1)
   Event Construct()
   n16 = Get IntroAnim()
   > PlayAnimation(InAnimation=@n16, StartAtTime=GetEndTime(self=@n16), PlayMode=Reverse)
   > n17 = Cast To PlatformGameInstance(Object=GetGameInstance())
   > Assign Application Received Screen Orientation Changed Notification Delegate!(execute:Exec, self:Platform Game Instance Object Reference=@n17, Delegate:"Delegate (by ref)"=@n11) -> (then:Exec) {is=title K2Node_AssignDelegate}
   > HandleOrientationChanged(ScreenOrientation=GetDeviceOrientation() -> (ReturnValue=Unknown))
   Event PreConstruct()
   > SetCompactUIStyle() {is=title K2Node_CallFunction}
  ]}
 HandleOrientationChanged {nodes [
   HandleOrientationChanged!() -> (then, ScreenOrientation:EScreenOrientation Enum=Unknown) {is=title K2Node_FunctionEntry}
   > n1 = GetConsoleVariableBoolValue!(execute, VariableName:String=PilgrimInverse.UsePortraitMode) -> (then, ReturnValue:bool)
   > Branch(Condition=AND Boolean(A=@n1, B=IsUsingCompactUI()))
   n2 = DisplayOrientationFromScreenOrientation!(ScreenOrientation:EScreenOrientation Enum=Get ScreenOrientation() -> (ScreenOrientation:EScreenOrientation Enum=Unknown)) -> (DisplayOrientation:EDisplayOrientationType Enum=Landscape)
   > Branch(Condition="Not Equal (Enum)"!(
     A:EDisplayOrientationType Enum=Get CachedDisplayOrientation() -> (CachedDisplayOrientation:EDisplayOrientationType Enum=Landscape)
     B:EDisplayOrientationType Enum=@n2
    ) -> (ReturnValue:bool))
   > n3 = Set CachedDisplayOrientation(CachedDisplayOrientation:EDisplayOrientationType Enum=@n2) -> (Output_Get:EDisplayOrientationType Enum=Landscape)
   > n4 = Branch(Condition="Equal (Enum)"(A:EDisplayOrientationType Enum=@n3, B:EDisplayOrientationType Enum=Portrait))
   > n5 = Branch(Condition=CompareToScreenAspectRatio!(Comparison:EAspectRatioComparison Enum=GreaterThan, AspectRatio:EAspectRatio Enum=Ratio_23_16) -> (ReturnValue:bool))
   > UpdateLayoutForOrientation!(execute, self, NewDisplayOrientation:EDisplayOrientationType Enum=Portrait) -> (then)
   UpdateLayoutForOrientation(execute=[@n5.else, @n4.else], NewDisplayOrientation=Landscape)
  ]}
 SetCompactUIStyle {nodes [
   SetCompactUIStyle() {is=title K2Node_FunctionEntry}
   > Branch(Condition=IsUsingCompactUI())
   > Set XpWidgetRightPadding(XpWidgetRightPadding:double=384) -> (Output_Get:double)
   n1 = SlotAsOverlaySlot(Widget=Get SZ_XpBar() -> (SZ_XpBar:Safe Zone Object Reference))
   n2 = Get!(self:Overlay Slot Object Reference=@n1) -> (Padding_Left:float=~, Padding_Top:float=~, Padding_Right:float=~, Padding_Bottom:float=~)
   > SetPadding!(
    execute
    self:Overlay Slot Object Reference=@n1
    InPadding_Left:float=@n2
    InPadding_Top:float=@n2.Padding_Top
    InPadding_Right:float=Get XpWidgetRightPadding() -> (XpWidgetRightPadding:double)
    InPadding_Bottom:float=@n2.Padding_Bottom
   ) -> (then)
  ]}
 UpdateLayoutForOrientation {nodes [
   n1 = UpdateLayoutForOrientation!() -> (then, NewDisplayOrientation:EDisplayOrientationType Enum=Landscape) {is=title K2Node_FunctionEntry}
   > SetHorizontalAlignment(
    self=[
     SlotAsOverlaySlot(Widget=Get Overlay_Sidebar() -> (Overlay_Sidebar:Overlay Object Reference))
     SlotAsOverlaySlot(Widget=Get NewSidebarInjectionSlot())
    ]
    InHorizontalAlignment=Select!(Landscape:EHorizontalAlignment Enum=HAlign_Right, Portrait:EHorizontalAlignment Enum=HAlign_Fill, Index:EDisplayOrientationType Enum=@n1) -> (ReturnValue:EHorizontalAlignment Enum)
   )
   n2 = SlotAsOverlaySlot(Widget=Get SZ_XpBar())
   > SetVerticalAlignment(
    self=@n2
    InVerticalAlignment=Select(Landscape:EVerticalAlignment Enum=VAlign_Bottom, Portrait:EVerticalAlignment Enum=VAlign_Top, Index=@n1) -> (ReturnValue:EVerticalAlignment Enum)
   )
   n3 = Get XpWidgetRightPadding()
   > SetPadding!(
    execute
    self=@n2
    InPadding:Margin Structure=Select(
     Landscape:Margin Structure=Make Margin(Left=0, Top=0, Right=@n3, Bottom=65)
     Portrait:Margin Structure=Make Margin(Left=0, Top=180, Right=0, Bottom=0)
     Index=@n1
    ) -> (ReturnValue:Margin Structure)
   ) -> (then)
  ]}
]
variables [
 CachedDisplayOrientation {type=EDisplayOrientationType, default=Landscape}
 SidebarCloseSound {type=SoundBase, default=/Game/Sounds/Fort_UI_Sounds/FrontEnd/SC_UI_Sidebar_Close_New}
 SidebarOpenSound {type=SoundBase, default=/Game/Sounds/Fort_UI_Sounds/FrontEnd/MSS_UI_Sidebar_Open}
 XpWidgetRightPadding {type=number}
]
widgets {
 info {inheritedWidgetCount=5, parentClass=/Script/FortniteUI.FortDualBladeMenu, rootWidgetClass=&Overlay, widgetCount=9}
 widgets [
  MainOverlay {widgetClassPath=&Overlay, bIsVariable=true, widget=MainOverlay}
  Button_BackBoard {
   widgetClassPath=&IconTextButton
   bInherited=true
   bIsVariable=true
   parent=MainOverlay
   widget=Button_BackBoard
   properties {
    bHideInputAction=true
    bIsFocusable=false
    buttonText=""
    clickMethod=MouseDown
    onButtonBaseClicked=(&self:WidgetTree.Button_BackBoard.HandleOnButtonClicked)
    onButtonBaseDoubleClicked=(&self:WidgetTree.Button_BackBoard.HandleOnButtonDoubleClicked)
    onButtonBaseHovered=(&self:WidgetTree.Button_BackBoard.HandleOnButtonHovered)
    onButtonBaseUnhovered=(&self:WidgetTree.Button_BackBoard.HandleOnButtonUnhovered)
    onSelectedChangedBase=(&self:WidgetTree.Button_BackBoard.HandleOnSelectedChanged)
    pressMethod=ButtonPress
    style=/FortUILibrary/Foundation/Buttons/ButtonStyles/ButtonStyle-Lightbox.ButtonStyle-Lightbox_C
    touchMethod=Down
    visibility=Collapsed
    hoveredSlateSoundOverride {resourceObject=null}
   }
   slot {class=OverlaySlot, properties {horizontalAlignment=HAlign_Fill, verticalAlignment=VAlign_Fill}}
  }
  SZ_XpBar {
   widgetClassPath=&SafeZone
   bIsVariable=true
   parent=MainOverlay
   widget=SZ_XpBar
   properties {padLeft=false, padRight=false}
   slot {class=OverlaySlot, properties {horizontalAlignment=HAlign_Center, verticalAlignment=VAlign_Bottom, padding=<ltrb 0 0 0 50>}}
  }
  Overlay_xp {widgetClassPath=&Overlay, bIsVariable=true, parent=SZ_XpBar, widget=Overlay_xp, slot {class=SafeZoneSlot}}
  BladesMenu_XpBar {
   widgetClassPath=/FortniteCoreUI/UI/Progression/BladesMenu_XpBar
   bInherited=true
   bIsVariable=true
   parent=Overlay_xp
   widget=BladesMenu_XpBar
   properties {visibility=Collapsed}
   slot {class=OverlaySlot}
  }
  Overlay_Sidebar {
   widgetClassPath=&Overlay
   bIsVariable=true
   parent=MainOverlay
   widget=Overlay_Sidebar
   slot {class=OverlaySlot, properties {horizontalAlignment=HAlign_Right, verticalAlignment=VAlign_Fill}}
  }
  OldSidebarInjectionSlot {
   widgetClassPath=&NamedSlot
   bInherited=true
   bIsVariable=true
   parent=Overlay_Sidebar
   widget=OldSidebarInjectionSlot
   slot {class=OverlaySlot, properties {horizontalAlignment=HAlign_Right, verticalAlignment=VAlign_Fill}}
  }
  NewSidebarInjectionSlot {
   widgetClassPath=/Game/UI/Plugin/FortPluginWidgetInjectionSlot
   bInherited=true
   bIsVariable=true
   parent=Overlay_Sidebar
   widget=NewSidebarInjectionSlot
   properties {bAutoInjectWidget=false, bIsFocusable=true, pluginId=Sidebar, pluginWidgetId=CranberryWidget}
   slot {class=OverlaySlot, properties {horizontalAlignment=HAlign_Right, verticalAlignment=VAlign_Fill}}
  }
  TopRightWindowControlsWidget {
   widgetClassPath=/FortniteCoreUI/UI/Sidebar/TopRightWindowControlsWidget
   bInherited=true
   bIsVariable=true
   parent=MainOverlay
   widget=TopRightWindowControlsWidget
   properties {activatedVisibility=Visible, bIsFocusable=false, bSupportsActivationFocus=false, visibility=Collapsed}
   slot {class=OverlaySlot, properties {horizontalAlignment=HAlign_Right}}
  }
 ]
}
