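# Lobby dialog that warns an idle party member with a countdown to sitting out.
# Its two buttons call ReadyUp or ResetIdleState on FortPartyMemberIdleVM, and any ready-status change deactivates the widget.
# OnTimeRemainingChanged updates the countdown text and restarts a one-second InterpTime.
# Tick uses InterpTime to smooth the material-driven progress bar between updates, scaled by the dialog viewmodel's DisplayTime.
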WidgetBlueprint /FortniteCoreUI/Frontend/WBP_UI_Lobby_Dialog_PlayerIdle
graphs [
 EventGraph {nodes [
   n1 = OnDialogButtonClicked!() -> (OutputDelegate, then, ButtonIndex:int) {is=title K2Node_CustomEvent}
   > n2 = Switch on Int!(execute:Exec, Selection:int=@n1.ButtonIndex) -> (Default:Exec, 0:Exec, 1:Exec)
   n3 = Get FortPartyMemberIdleVM() -> (FortPartyMemberIdleVM:Fort Party Member Idle VM Object Reference)
   ReadyUp(execute=@n2.0, self:Fort Party Member Idle VM Object Reference=@n3)
   ResetIdleState(execute=@n2.1, self:Fort Party Member Idle VM Object Reference=@n3)
   n4 = OnReadyStatusChanged!() -> (OutputDelegate, then, ReadyStatus:EGameReadiness Enum=NotReady) {is=title K2Node_CustomEvent}
   > n5 = Switch on EGameReadiness!(execute, Selection:EGameReadiness Enum=@n4.ReadyStatus) -> (NotReady:Exec, Ready:Exec, Reevaluating:Exec, SittingOut:Exec, Away:Exec) {is=title K2Node_SwitchEnum}
   DeactivateWidget(execute=[@n5.Away, @n5.NotReady, @n5.Ready, @n5.Reevaluating, @n5.SittingOut])
   > Set InterpTime(InterpTime:double=0) -> (Output_Get:double)
   n6 = Event Tick()
   > TickInterpolation!(execute, self, DeltaTime:double=@n6.InDeltaTime) -> (then)
   Event Construct()
   > Parent: Construct()
   n7 = Create Event()
   n8 = Get WBP_UIKit_Dialog_Internal() -> (WBP_UIKit_Dialog_Internal:WBP UIKit Dialog Internal Object Reference)
   > Bind Event to Dialog Button Clicked Dispatcher(self:WBP UIKit Dialog Internal Object Reference=@n8, Delegate=@n7) {is=title K2Node_AddDelegate}
   n9 = Create Event()
   n10 = Get FortPartyMemberIdleVM()
   > Bind Event to On Ready Status Changed(self:Fort Party Member Idle VM Object Reference=@n10, Delegate=@n9) {is=title K2Node_AddDelegate}
   n11 = Event PreConstruct()
   > Parent: PreConstruct(IsDesignTime=@n11.IsDesignTime)
   > SetupProgressBarMaterial()
   Event Destruct()
   > Parent: Destruct()
   > Unbind Event from Dialog Button Clicked Dispatcher(self:WBP UIKit Dialog Internal Object Reference=@n8, Delegate=@n7) {is=title K2Node_RemoveDelegate}
   > Unbind Event from On Ready Status Changed(self:Fort Party Member Idle VM Object Reference=@n10, Delegate=@n9) {is=title K2Node_RemoveDelegate}
  ]}
 OnTimeRemainingChanged {nodes [
   OnTimeRemainingChanged!() -> (then, TimeRemaining:double=~) {is=title K2Node_FunctionEntry}
   > Set InterpTime(InterpTime=1)
   n1 = Get TimeRemaining() -> (TimeRemaining:double)
   > SetProgress!(execute, self, RemainingSeconds:float=@n1) -> (then)
   > Set Text(
    Text=Format Text!(Format="LOCTEXT(\"Sitting out in {seconds} seconds...\")", seconds:int=Floor(A=@n1) {is=title K2Node_CallFunction}) -> (Result)
    self=Get TextBlock_SittingOutSeconds() -> (TextBlock_SittingOutSeconds:Common Text Object Reference)
   )
  ]}
 SetProgress {nodes [
   n1 = SetProgress!() -> (then, RemainingSeconds:float=~) {is=title K2Node_FunctionEntry}
   > n2 = GetDynamicMaterial(self=Get ProgressBar() -> (ProgressBar:Image Object Reference))
   n3 = Get UIKitDialogViewModel!(self:WBP UIKit Dialog Internal Object Reference=Get WBP_UIKit_Dialog_Internal()) -> (UIKitDialogViewModel:UI Kit Dialog Viewmodel Object Reference)
   > SetScalarParameterValue(
    self=@n2
    ParameterName=LevelProgress
    Value=SafeDivide(A=@n1, B=Get DisplayTime!(self:UI Kit Dialog Viewmodel Object Reference=@n3) -> (DisplayTime:float))
   )
  ]}
 SetupProgressBarMaterial {nodes [
   SetupProgressBarMaterial() {is=title K2Node_FunctionEntry}
   > n1 = GetDynamicMaterial(self=Get ProgressBar())
   > SetScalarParameterValue(self=@n1, ParameterName=Backing_Alpha, Value=0)
   > SetVectorParameterValue(self=@n1, ParameterName=ProgressBar_FillColor, Value=(R=0,G=0.6654,B=1,A=1))
  ]}
 TickInterpolation {nodes [
   TickInterpolation!() -> (then, DeltaTime:double=~) {is=title K2Node_FunctionEntry}
   > Branch(Condition="float > float"(A=Get InterpTime() -> (InterpTime:double)))
   > Set InterpTime(InterpTime=float - float(A=Get InterpTime(), B=Get DeltaTime() -> (DeltaTime:double)))
   n1 = Get InterpTime()
   n2 = Get FortPartyMemberIdleVM()
   > SetProgress(RemainingSeconds=float - float(
     A=Get TimeRemainingBeforeSittingOut!(self:Fort Party Member Idle VM Object Reference=@n2) -> (TimeRemainingBeforeSittingOut:float)
     B=float - float(A=1, B=@n1)
    )) {is=title K2Node_CallFunction}
  ]}
]
variables [
 IdleTime {type=number, default=10}
 InterpTime {type=number}
]
widgets {
 info {
  inheritedWidgetCount=0
  namedSlotCount=1
  parentClass=&CommonActivatableWidget
  rootWidgetClass=/FortUILibrary/Systems/Dialog/WBP_UIKit_Dialog_Internal.WBP_UIKit_Dialog_Internal_C
  widgetCount=5
 }
 widgets [
  WBP_UIKit_Dialog_Internal {
   widgetClassPath=/FortUILibrary/Systems/Dialog/WBP_UIKit_Dialog_Internal
   bIsVariable=true
   widget=WBP_UIKit_Dialog_Internal
   properties {
    blur=2
    designTimePreview_DialogVM=/FortniteCoreUI/Frontend/Dialogs/DialogVM_PlayerIdle.DialogVM_PlayerIdle_C
    buttons {
     each Button Horizontal Fill=false
     each Button Min Width=200
     each Button Size=Medium
     focuse Button=0
     space Between Buttons=30
     cTA Buttons Types [Loud,Regular]
     group Placement {
      horizontal Alignment=HAlign_Center
      vertical Alignment=VAlign_Bottom
      grid {column=1, column Span=0, layer=0, row=2, row Span=0, nudge=<0 0>}
      padding=<ltrb 80 24 80 144>
     }
    }
    dialogSIze {
     height Override {enable=false, value=800}
     max Aspect Ratio {enable=false, value=0}
     max Desired Height {enable=true, value=800}
     max Desired Width {enable=false, value=0}
     min Aspect Ratio {enable=false, value=0}
     min Desired Height {enable=false, value=0}
     min Desired Width {enable=false, value=0}
     width Override {enable=true, value=1000}
    }
    imageProperties {
     material=/FortUILibrary/Systems/Dialog/MaterialInstances/MI_UI_DialogImage
     image Size=<600 0>
     placement {
      horizontal Alignment=HAlign_Fill
      vertical Alignment=VAlign_Fill
      grid {column=0, column Span=0, layer=2, row=1, row Span=0, nudge=<0 0>}
      padding=<ltrb 0 0 0 0>
     }
    }
    textProperties {
     auto Wrap Text=true
     justification=Center
     line Height Percentage=1.2
     mobile Font Size Multiplier=1.3
     transform Policy=None
     wrap Text At=0
     color {colorUseRule=UseColor_Specified, specifiedColor=<rgba 1 1 1 1>}
     font {
      bForceMonospaced=false
      bMaterialIsStencil=false
      fontMaterial=null
      fontObject=&HeadingNow
      letterSpacing=0
      monospacedWidth=1
      size=15.75
      skewAmount=0
      typefaceFontName=74 Regular
      outlineSettings {
       bApplyOutlineToDropShadows=false
       bMiteredCorners=false
       bSeparateFillAlpha=false
       outlineMaterial=null
       outlineSize=0
       outlineColor=<rgba 0 0 0 1>
      }
     }
     placement {
      horizontal Alignment=HAlign_Fill
      vertical Alignment=VAlign_Top
      grid {column=1, column Span=0, layer=0, row=1, row Span=0, nudge=<0 0>}
      padding=<ltrb 80 0 80 0>
     }
    }
    titleProperties {
     auto Wrap Text=true
     justification=Center
     line Height Percentage=1
     mobile Font Size Multiplier=1.3
     transform Policy=ToUpper
     wrap Text At=0
     color {colorUseRule=UseColor_Specified, specifiedColor=<rgba 1 1 1 1>}
     font {
      bForceMonospaced=false
      bMaterialIsStencil=false
      fontMaterial=null
      fontObject=&HeadingNow
      letterSpacing=0
      monospacedWidth=1
      size=25.5
      skewAmount=0
      typefaceFontName=86 Bold Italic
      outlineSettings {
       bApplyOutlineToDropShadows=false
       bMiteredCorners=false
       bSeparateFillAlpha=false
       outlineMaterial=null
       outlineSize=0
       outlineColor=<rgba 0 0 0 1>
      }
     }
     placement {
      horizontal Alignment=HAlign_Fill
      vertical Alignment=VAlign_Top
      grid {column=1, column Span=0, layer=0, row=0, row Span=0, nudge=<0 0>}
      padding=<ltrb 80 80 80 32>
     }
    }
   }
  }
  Overlay_0 {widgetClassPath=&Overlay, namedSlotHost=WBP_UIKit_Dialog_Internal, widget=Overlay_0}
  StackBox_0 {
   widgetClassPath=&StackBox
   parent=Overlay_0
   widget=StackBox_0
   properties {orientation=Orient_Vertical}
   slot {class=OverlaySlot, properties {horizontalAlignment=HAlign_Fill, verticalAlignment=VAlign_Bottom}}
  }
  TextBlock_SittingOutSeconds {
   widgetClassPath=&CommonTextBlock
   bIsVariable=true
   parent=StackBox_0
   widget=TextBlock_SittingOutSeconds
   properties {
    mobileFontSizeMultiplier=1.14
    text="Sitting out in {seconds} seconds..."
    font {
     bForceMonospaced=false
     bMaterialIsStencil=false
     fontMaterial=null
     fontObject=&HeadingNow
     letterSpacing=0
     monospacedWidth=1
     size=15.75
     skewAmount=0
     typefaceFontName=74 Regular
     outlineSettings {
      bApplyOutlineToDropShadows=false
      bMiteredCorners=false
      bSeparateFillAlpha=false
      outlineMaterial=null
      outlineSize=0
      outlineColor=<rgba 0 0 0 1>
     }
    }
    shadowOffset=<0 0>
   }
   slot {class=StackBoxSlot, properties {horizontalAlignment=HAlign_Center, verticalAlignment=VAlign_Bottom}}
  }
  ProgressBar {
   widgetClassPath=&Image
   bIsVariable=true
   parent=StackBox_0
   widget=ProgressBar
   properties {brush {
     bIsDynamicallyLoaded=false
     drawAs=Image
     imageType=FullColor
     mirroring=NoMirror
     resourceName=None
     resourceObject=/FortUILibrary/Foundation/Art_ToolKit/ProgressBars/M_UI_ProgressBar_Master
     tiling=NoTile
     imageSize=<32 8>
     margin=<ltrb 0 0 0 0>
     outlineSettings {
      bUseBrushTransparency=false
      roundingType=HalfHeightRadius
      width=0
      color {colorUseRule=UseColor_Specified, specifiedColor=<rgba 0 0 0 0>}
      cornerRadii=<0 0 0 0>
     }
     tintColor {colorUseRule=UseColor_Specified, specifiedColor=<rgba 1 1 1 1>}
     uVRegion {bIsValid=false, max=<0 0>, min=<0 0>}
    }}
   slot {class=StackBoxSlot, properties {verticalAlignment=VAlign_Bottom, padding=<ltrb 144 12 144 48>}}
  }
 ]
}
