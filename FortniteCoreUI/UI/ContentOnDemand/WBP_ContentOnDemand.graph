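# Content-on-demand toast widget that shows HD texture download progress and completion, built on FortOptionalBundlesInstallWidget.
# Install start reveals it with the Downloading animation, and progress events drive a progress bar, thinned on tablet aspect ratios.
# A successful finish plays Completed with checkmark and pulse, while failure reverses Downloading.
# HasForcedStoppedAnimation swallows the finish callback from StopAllAnimations, so the widget collapses only after animations end with no ongoing installs.
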WidgetBlueprint /FortniteCoreUI/UI/ContentOnDemand/WBP_ContentOnDemand
animations [
 Downloading {
  binding_count=5
  length_seconds=1
  object=&self:Downloading
  sequence {bindings [
    CommonText_Downloading {
     id=#1
     possessed_class=CommonTextBlock
     tracks [
      {
       type=MovieSceneFloatTrack
       display_name=Render Opacity
       sections [{
         type=MovieSceneFloatSection
         end_frame=20
         start_frame=0
         channels [None {type=MovieSceneScriptingFloatChannel, keys [[0,0],[13,0],[19.9997,1]]}]
        }]
      }
      {
       type=MovieScene2DTransformTrack
       sections [{
         type=MovieScene2DTransformSection
         end_frame=20
         start_frame=13
         channels [
          Translation.X {type=MovieSceneScriptingFloatChannel, keys [[13.06,7.5],[19.9997,0]]}
          Translation.Y {type=MovieSceneScriptingFloatChannel, const [0,1]}
          Angle {type=MovieSceneScriptingFloatChannel, const [0,1]}
          Scale.X {type=MovieSceneScriptingFloatChannel, const [1,1]}
          Scale.Y {type=MovieSceneScriptingFloatChannel, const [1,1]}
          Shear.X {type=MovieSceneScriptingFloatChannel, const [0,1]}
          Shear.Y {type=MovieSceneScriptingFloatChannel, const [0,1]}
         ]
        }]
      }
     ]
    }
    WBP_UIKit_ProgressBar {
     id=#2
     possessed_class=WBP_UIKit_ProgressBar_C
     tracks [
      {
       type=MovieScene2DTransformTrack
       sections [{
         type=MovieScene2DTransformSection
         end_frame=10
         start_frame=0
         channels [
          Translation.X {type=MovieSceneScriptingFloatChannel, const [0,1]}
          Translation.Y {type=MovieSceneScriptingFloatChannel, const [0,1]}
          Angle {type=MovieSceneScriptingFloatChannel, const [0,1]}
          Scale.X {type=MovieSceneScriptingFloatChannel, keys [[0,0],[5,0.9877],[10,1]]}
          Scale.Y {type=MovieSceneScriptingFloatChannel, keys [[0,0],[5,0.9877],[10,1]]}
          Shear.X {type=MovieSceneScriptingFloatChannel, const [0,1]}
          Shear.Y {type=MovieSceneScriptingFloatChannel, const [0,1]}
         ]
        }]
      }
      {
       type=MovieSceneFloatTrack
       display_name=Render Opacity
       sections [{
         type=MovieSceneFloatSection
         end_frame=10
         start_frame=0
         channels [None {type=MovieSceneScriptingFloatChannel, keys [[0,0],[2,0],[5,0.9877],[10,1]]}]
        }]
      }
     ]
    }
    UIKit_Backplate {
     id=#3
     possessed_class=WBP_UIKit_Backplate_C
     tracks [
      {
       type=MovieScene2DTransformTrack
       sections [{
         type=MovieScene2DTransformSection
         end_frame=19
         start_frame=0
         channels [
          Translation.X {type=MovieSceneScriptingFloatChannel, default=0}
          Translation.Y {type=MovieSceneScriptingFloatChannel, default=0}
          Angle {type=MovieSceneScriptingFloatChannel, default=0}
          Scale.X {type=MovieSceneScriptingFloatChannel, keys [[0,0],[5,0.0995],[10,0.1115],[13,0.9014],[19.94,1]]}
          Scale.Y {type=MovieSceneScriptingFloatChannel, default=1, keys [[0,0],[5,0.9877],[10,1]]}
          Shear.X {type=MovieSceneScriptingFloatChannel, default=0}
          Shear.Y {type=MovieSceneScriptingFloatChannel, default=0}
         ]
        }]
      }
      {
       type=MovieSceneFloatTrack
       display_name=Render Opacity
       sections [{
         type=MovieSceneFloatSection
         end_frame=10
         start_frame=0
         channels [None {type=MovieSceneScriptingFloatChannel, keys [[0,0],[5,0.9877],[10,1]]}]
        }]
      }
     ]
    }
    CommonText_Complete {
     id=#4
     possessed_class=CommonTextBlock
     tracks [{
       type=MovieSceneFloatTrack
       display_name=Render Opacity
       sections [{type=MovieSceneFloatSection, end_frame=20, start_frame=0, channels [None {type=MovieSceneScriptingFloatChannel, const [0,2]}]}]
      }]
    }
    Overlay_Complete {
     id=#5
     possessed_class=Overlay
     tracks [{
       type=MovieSceneFloatTrack
       display_name=Render Opacity
       sections [{type=MovieSceneFloatSection, end_frame=20, start_frame=0, channels [None {type=MovieSceneScriptingFloatChannel, const [0,2]}]}]
      }]
    }
   ]}
 }
 Completed {
  binding_count=5
  length_seconds=5.003
  object=&self:Completed
  sequence {bindings [
    CommonText_Downloading {
     id=#1
     possessed_class=CommonTextBlock
     tracks [
      {
       type=MovieSceneFloatTrack
       display_name=Render Opacity
       sections [{type=MovieSceneFloatSection, end_frame=10, start_frame=0, channels [None {type=MovieSceneScriptingFloatChannel, keys [[0,1],[10,0]]}]}]
      }
      {
       type=MovieScene2DTransformTrack
       sections [{
         type=MovieScene2DTransformSection
         end_frame=10
         start_frame=0
         channels [
          Translation.X {type=MovieSceneScriptingFloatChannel, keys [[0,0],[10,7.5]]}
          Translation.Y {type=MovieSceneScriptingFloatChannel, const [0,1]}
          Angle {type=MovieSceneScriptingFloatChannel, default=0}
          Scale.X {type=MovieSceneScriptingFloatChannel, default=1}
          Scale.Y {type=MovieSceneScriptingFloatChannel, default=1}
          Shear.X {type=MovieSceneScriptingFloatChannel, default=0}
          Shear.Y {type=MovieSceneScriptingFloatChannel, default=0}
         ]
        }]
      }
     ]
    }
    WBP_UIKit_ProgressBar {
     id=#6
     possessed_class=WBP_UIKit_ProgressBar_C
     tracks [
      {
       type=MovieScene2DTransformTrack
       sections [{
         type=MovieScene2DTransformSection
         end_frame=10
         start_frame=0
         channels [
          Translation.X {type=MovieSceneScriptingFloatChannel, default=0}
          Translation.Y {type=MovieSceneScriptingFloatChannel, default=0}
          Angle {type=MovieSceneScriptingFloatChannel, default=0}
          Scale.X {type=MovieSceneScriptingFloatChannel, keys [[0,0.9877],[10,0]]}
          Scale.Y {type=MovieSceneScriptingFloatChannel, keys [[0,0.9877],[10,0]]}
          Shear.X {type=MovieSceneScriptingFloatChannel, default=0}
          Shear.Y {type=MovieSceneScriptingFloatChannel, default=0}
         ]
        }]
      }
      {
       type=MovieSceneFloatTrack
       display_name=Render Opacity
       sections [{type=MovieSceneFloatSection, end_frame=10, start_frame=0, channels [None {type=MovieSceneScriptingFloatChannel, keys [[0,1],[10,0]]}]}]
      }
     ]
    }
    UIKit_Backplate {
     id=#3
     possessed_class=WBP_UIKit_Backplate_C
     tracks [
      {
       type=MovieScene2DTransformTrack
       sections [{
         type=MovieScene2DTransformSection
         end_frame=100
         start_frame=0
         channels [
          Translation.X {type=MovieSceneScriptingFloatChannel, default=0}
          Translation.Y {type=MovieSceneScriptingFloatChannel, default=0}
          Angle {type=MovieSceneScriptingFloatChannel, default=0}
          Scale.X {type=MovieSceneScriptingFloatChannel, keys [[0,1],[90,1],[100.0593,0]]}
          Scale.Y {type=MovieSceneScriptingFloatChannel, default=1, const [1,1]}
          Shear.X {type=MovieSceneScriptingFloatChannel, default=0}
          Shear.Y {type=MovieSceneScriptingFloatChannel, default=0}
         ]
        }]
      }
      {
       type=MovieSceneFloatTrack
       display_name=Render Opacity
       sections [{type=MovieSceneFloatSection, end_frame=99, start_frame=0, channels [None {type=MovieSceneScriptingFloatChannel, keys [[0,1],[96,1],[99,0]]}]}]
      }
     ]
    }
    CommonText_Complete {
     id=#4
     possessed_class=CommonTextBlock
     tracks [
      {
       type=MovieSceneFloatTrack
       display_name=Render Opacity
       sections [{
         type=MovieSceneFloatSection
         end_frame=97
         start_frame=0
         channels [None {type=MovieSceneScriptingFloatChannel, keys [[0,0],[10,0],[20,1],[90,1],[97,0]]}]
        }]
      }
      {
       type=MovieScene2DTransformTrack
       sections [{
         type=MovieScene2DTransformSection
         end_frame=20
         start_frame=10
         channels [
          Translation.X {type=MovieSceneScriptingFloatChannel, keys [[10,7.5],[20,0]]}
          Translation.Y {type=MovieSceneScriptingFloatChannel, const [0,2]}
          Angle {type=MovieSceneScriptingFloatChannel, default=0}
          Scale.X {type=MovieSceneScriptingFloatChannel, default=1}
          Scale.Y {type=MovieSceneScriptingFloatChannel, default=1}
          Shear.X {type=MovieSceneScriptingFloatChannel, default=0}
          Shear.Y {type=MovieSceneScriptingFloatChannel, default=0}
         ]
        }]
      }
     ]
    }
    Overlay_Complete {
     id=#5
     possessed_class=Overlay
     tracks [
      {
       type=MovieSceneFloatTrack
       display_name=Render Opacity
       sections [{
         type=MovieSceneFloatSection
         end_frame=100
         start_frame=0
         channels [None {type=MovieSceneScriptingFloatChannel, keys [[0,0],[10,0],[15,1],[96,1],[100,0]]}]
        }]
      }
      {
       type=MovieScene2DTransformTrack
       sections [{
         type=MovieScene2DTransformSection
         end_frame=100
         start_frame=10
         channels [
          Translation.X {type=MovieSceneScriptingFloatChannel, default=0}
          Translation.Y {type=MovieSceneScriptingFloatChannel, default=0}
          Angle {type=MovieSceneScriptingFloatChannel, default=0}
          Scale.X {type=MovieSceneScriptingFloatChannel, keys [[10,1],[96,1],[100,0.65]]}
          Scale.Y {type=MovieSceneScriptingFloatChannel, keys [[10,1],[96,1],[100,0.65]]}
          Shear.X {type=MovieSceneScriptingFloatChannel, default=0}
          Shear.Y {type=MovieSceneScriptingFloatChannel, default=0}
         ]
        }]
      }
     ]
    }
   ]}
 }
]
graphs [EventGraph {nodes [
   Event Construct()
   > SetFontMaterial!(
    execute
    self:Text Object Reference=Get CommonText_Downloading() -> (CommonText_Downloading:Common Text Object Reference)
    InMaterial:Material Interface Object Reference=/Game/UI/ContentOnDemand/MI_UI_ContentOnDemand_FontMaterial
   ) -> (then)
   > Branch(Condition=IsTabletAspectRatio!() -> (ReturnValue:bool))
   > Set Thickness_PX!(
    execute
    Thickness_PX:double=1
    self:WBP UIKit Progress Bar Base Object Reference=Get WBP_UIKit_ProgressBar() -> (WBP_UIKit_ProgressBar:WBP UIKit Progress Bar Object Reference)
   ) -> (then, Output_Get:double)
   n1 = Event On Install Progressed!() -> (OutputDelegate, then, Progress:float) {is=title K2Node_Event}
   > SetProgress!(execute, self:WBP UIKit Progress Bar Base Object Reference=Get WBP_UIKit_ProgressBar(), Progress:double=@n1.Progress) -> (then)
   Event On Install Started() {is=title K2Node_Event}
   > SetVisibility(InVisibility=HitTestInvisible)
   > Set HasForcedStoppedAnimation(HasForcedStoppedAnimation:bool=true) -> (Output_Get:bool)
   > StopAllAnimations(self:User Widget Object Reference)
   > PlayAnimation(InAnimation=Get Downloading() -> (Downloading:Widget Animation Object Reference))
   n2 = Event On Install Finished!() -> (OutputDelegate, then, bInstallSuccessful:bool) {is=title K2Node_Event}
   > Set HasForcedStoppedAnimation(HasForcedStoppedAnimation=true)
   > StopAllAnimations()
   > n3 = Branch(Condition=@n2.bInstallSuccessful)
   > PlayAnimation(InAnimation=Get Completed() -> (Completed:Widget Animation Object Reference))
   PlayAnimation(execute=@n3.else, InAnimation=Get Downloading(), PlayMode=Reverse)
   Event OnAnimationFinished!() -> (OutputDelegate, then, Animation:Widget Animation Object Reference) {is=title K2Node_Event}
   > n4 = Branch(Condition=Get HasForcedStoppedAnimation() -> (HasForcedStoppedAnimation:bool))
   > Set HasForcedStoppedAnimation(HasForcedStoppedAnimation=false)
   n5 = Branch(execute=@n4.else, Condition=HasOngoingInstalls!(self:Fort Optional Bundles Install Widget Object Reference) -> (ReturnValue:bool))
   SetVisibility(execute=@n5.else, InVisibility=Collapsed)
  ]}]
variables [HasForcedStoppedAnimation {type=boolean}]
widgets {
 info {inheritedWidgetCount=0, parentClass=/Script/FortniteUI.FortOptionalBundlesInstallWidget, rootWidgetClass=&Overlay, widgetCount=9}
 widgets [
  Overlay_37 {widgetClassPath=&Overlay, widget=Overlay_37}
  WBP_CaptureForPostBufferUpdate {
   widgetClassPath=&WBP_CaptureForPostBufferUpdate
   bIsVariable=true
   parent=Overlay_37
   slot {class=OverlaySlot, properties {horizontalAlignment=HAlign_Fill, verticalAlignment=VAlign_Fill}}
  }
  UIKit_Backplate {
   widgetClassPath=&WBP_UIKit_Backplate
   parent=Overlay_37
   widget=UIKit_Backplate
   properties {brightness=Dark, renderTransformPivot=<1 0.5>}
   slot {class=OverlaySlot, properties {horizontalAlignment=HAlign_Fill, verticalAlignment=VAlign_Fill}}
  }
  CommonText_Downloading {
   widgetClassPath=&CommonTextBlock
   bIsVariable=true
   parent=Overlay_37
   widget=CommonText_Downloading
   properties {
    justification=Center
    style=&TextStyle-Base_C
    text=Downloading HD Textures
    font {
     bForceMonospaced=false
     bMaterialIsStencil=false
     fontMaterial=null
     fontObject=&NotoSans
     letterSpacing=0
     monospacedWidth=1
     size=24
     skewAmount=0
     typefaceFontName=NotoSans-Regular
     outlineSettings {
      bApplyOutlineToDropShadows=false
      bMiteredCorners=false
      bSeparateFillAlpha=false
      outlineMaterial=null
      outlineSize=0
      outlineColor=<rgba 0 0 0 1>
     }
    }
    shadowOffset=<0 0>
   }
   slot {class=OverlaySlot, properties {horizontalAlignment=HAlign_Fill, verticalAlignment=VAlign_Fill, padding=<ltrb 14 4 44 4>}}
  }
  CommonText_Complete {
   widgetClassPath=&CommonTextBlock
   parent=Overlay_37
   widget=CommonText_Complete
   properties {
    justification=Center
    renderOpacity=0
    style=&TextStyle-Base_C
    text=Download Complete
    font {
     bForceMonospaced=false
     bMaterialIsStencil=false
     fontMaterial=null
     fontObject=&NotoSans
     letterSpacing=0
     monospacedWidth=1
     size=24
     skewAmount=0
     typefaceFontName=NotoSans-Regular
     outlineSettings {
      bApplyOutlineToDropShadows=false
      bMiteredCorners=false
      bSeparateFillAlpha=false
      outlineMaterial=null
      outlineSize=0
      outlineColor=<rgba 0 0 0 1>
     }
    }
    shadowOffset=<0 0>
   }
   slot {class=OverlaySlot, properties {horizontalAlignment=HAlign_Fill, verticalAlignment=VAlign_Fill, padding=<ltrb 14 4 44 4>}}
  }
  WBP_UIKit_ProgressBar {
   widgetClassPath=/FortUILibrary/UIKit/ProgressBar/WBP_UIKit_ProgressBar
   bIsVariable=true
   parent=Overlay_37
   widget=WBP_UIKit_ProgressBar
   properties {edgeOffset_PX=6, glowIntensity=0.0763, height=38, progressValue=0.4376, thickness_PX=6, width=38}
   slot {class=OverlaySlot, properties {horizontalAlignment=HAlign_Right, verticalAlignment=VAlign_Center, padding=<ltrb 0 0 3 0>}}
  }
  Overlay_Complete {
   widgetClassPath=&Overlay
   bIsVariable=true
   parent=Overlay_37
   widget=Overlay_Complete
   properties {renderOpacity=0}
   slot {class=OverlaySlot, properties {horizontalAlignment=HAlign_Right, verticalAlignment=VAlign_Center, padding=<ltrb 0 0 3 0>}}
  }
  Image_Pulse {
   widgetClassPath=&Image
   bIsVariable=true
   parent=Overlay_Complete
   widget=Image_Pulse
   properties {
    brush {
     bIsDynamicallyLoaded=false
     drawAs=Image
     imageType=FullColor
     mirroring=NoMirror
     resourceName=None
     resourceObject=/Game/UI/ContentOnDemand/M_UI_ContentOnDemand_CirclePulse
     tiling=NoTile
     imageSize=<38 38>
     margin=<ltrb 0 0 0 0>
     outlineSettings {
      bUseBrushTransparency=false
      roundingType=HalfHeightRadius
      width=0
      color {colorUseRule=UseColor_Specified, specifiedColor=<rgba 0 0 0 0>}
      cornerRadii=<0 0 0 0>
     }
     tintColor {colorUseRule=UseColor_Specified, specifiedColor=<rgba 0.0409 1 0.2307 1>}
     uVRegion {bIsValid=false, max=<0 0>, min=<0 0>}
    }
    renderTransform {angle=0, scale=<1.6 1.65>, shear=<0 0>, translation=<0 0>}
   }
   slot {class=OverlaySlot, properties {horizontalAlignment=HAlign_Center, verticalAlignment=VAlign_Center}}
  }
  Image_Checkmark {
   widgetClassPath=&Image
   bIsVariable=true
   parent=Overlay_Complete
   widget=Image_Checkmark
   properties {brush {
     bIsDynamicallyLoaded=false
     drawAs=Image
     imageType=FullColor
     mirroring=NoMirror
     resourceName=None
     resourceObject=/Game/UI/ContentOnDemand/T_UI_ContentOnDemand_GreenCheckmark
     tiling=NoTile
     imageSize=<22 22>
     margin=<ltrb 0 0 0 0>
     outlineSettings {
      bUseBrushTransparency=false
      roundingType=HalfHeightRadius
      width=0
      color {colorUseRule=UseColor_Specified, specifiedColor=<rgba 0 0 0 0>}
      cornerRadii=<0 0 0 0>
     }
     tintColor {colorUseRule=UseColor_Specified, specifiedColor=<rgba 1 1 1 1>}
     uVRegion {bIsValid=false, max=<0 0>, min=<0 0>}
    }}
   slot {class=OverlaySlot, properties {horizontalAlignment=HAlign_Center, verticalAlignment=VAlign_Center}}
  }
 ]
}
