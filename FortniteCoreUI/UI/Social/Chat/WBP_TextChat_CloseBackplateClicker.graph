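# Text chat backplate widget that closes text chat when its click-catcher button is clicked.
# UpdateChatButtonVisibility creates an empty-styled SimpleCommonButton at runtime in a single-cell filling grid when chat is active, and removes it when inactive.
# Clicking the button calls SetTextChatActive(false) on the FortTextChatWrapperVM.
# AddButton calls RemoveButton first, which unbinds, removes and clears any previous button.
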WidgetBlueprint /FortniteCoreUI/UI/Social/Chat/WBP_TextChat_CloseBackplateClicker
graphs [
 AddButton {nodes [
   AddButton() {is=title K2Node_FunctionEntry}
   > RemoveButton()
   > n1 = Construct Object from Class!(
    execute
    Class=/Game/Athena/UI/Matchmaking/SimpleCommonButton.SimpleCommonButton_C
    self
    Style:Common Button Style Class Reference=/FortUILibrary/Foundation/Buttons/ButtonStyles/ButtonStyle-Empty.ButtonStyle-Empty_C
    bApplyAlphaOnDisable:bool=True
    bLocked:bool=False
    bSelectable:bool=False
    bShouldSelectUponReceivingFocus:bool=False
    bInteractableWhenSelected:bool=False
    bToggleable:bool=False
    bTriggerClickedAfterSelection:bool=False
    bDisplayInputActionWhenNotInteractable:bool=false
    bHideInputActionWithKeyboard:bool=true
    bShouldUseFallbackDefaultInputAction:bool=True
    bRequiresHold:bool=False
    HoldData:Common UIHold Data Class Reference
    ClickMethod:EButtonClickMethod Enum=DownAndUp
    InputPriority:int=0
    TriggeringInputAction:Data Table Row Handle Structure
   ) -> (then, ReturnValue:Simple Common Button Object Reference)
   > n2 = Set ClickCatcherButtonRef(ClickCatcherButtonRef:Simple Common Button Object Reference=@n1) -> (Output_Get:Simple Common Button Object Reference)
   > Branch(Condition=IsValid(Object=@n2))
   n3 = Get ClickCatcherButtonRef() -> (ClickCatcherButtonRef:Simple Common Button Object Reference)
   > AddChildToGrid!(
    execute
    self:Grid Panel Object Reference=Get Grid_HoldsClickCatcher() -> (Grid_HoldsClickCatcher:Grid Panel Object Reference)
    Content:Widget Object Reference=@n3
    InRow:int=0
    InColumn:int=0
   ) -> (then, ReturnValue:Grid Slot Object Reference)
   > Bind Event to On Button Clicked(self:Common Button Legacy Object Reference=@n3, Delegate=Create Event()) {is=title K2Node_AddDelegate}
  ]}
 EventGraph {nodes [
   OnCatcherButtonClicked!() -> (OutputDelegate, then, Button:Common Button Legacy Object Reference) {is=title K2Node_CustomEvent}
   > SetTextChatActive!(
    execute
    self:Fort Text Chat Wrapper VM Object Reference=Get FortTextChatWrapperVM() -> (FortTextChatWrapperVM:Fort Text Chat Wrapper VM Object Reference)
    bTextChatActive:bool=false
   ) -> (then)
  ]}
 RemoveButton {nodes [
   RemoveButton() {is=title K2Node_FunctionEntry}
   > n1 = Get!(execute:Exec) -> (then:Exec, else:Exec, ClickCatcherButtonRef:Simple Common Button Object Reference)
   > Unbind Event from On Button Clicked(self:Common Button Legacy Object Reference=@n1, Delegate=Create Event()) {is=title K2Node_RemoveDelegate}
   > RemoveWidget!(execute, Widget:Widget Object Reference=Get ClickCatcherButtonRef()) -> (then)
   > Set ClickCatcherButtonRef()
  ]}
 UpdateChatButtonVisibility {nodes [
   n1 = UpdateChatButtonVisibility!() -> (then, IsTextChatActive:bool=~) {is=title K2Node_FunctionEntry}
   > n2 = Branch(Condition=@n1)
   > AddButton() {is=title K2Node_CallFunction}
   RemoveButton(execute=@n2.else) {is=title K2Node_CallFunction}
  ]}
]
variables [ClickCatcherButtonRef {type=SimpleCommonButton_C, default=null}]
widgets {
 info {inheritedWidgetCount=0, parentClass=&CommonUserWidget, rootWidgetClass=&GridPanel, widgetCount=1}
 widgets [Grid_HoldsClickCatcher {widgetClassPath=&GridPanel, bIsVariable=true, widget=Grid_HoldsClickCatcher, properties {columnFill [1], rowFill [1]}}]
}
