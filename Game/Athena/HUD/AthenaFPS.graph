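# AthenaFPS is a HUD element widget blueprint in the Athena HUD, derived from FortHUDElementWidget, that hosts an FPS ticker.
# An AthenaFPSTicker widget serves as the root and is exposed as a variable, with a TutorialHighlight widget also listed in the tree.
# The event graph is empty and no variables are declared, so any display logic lives outside this asset.
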WidgetBlueprint /Game/Athena/HUD/AthenaFPS
graphs [EventGraph {links [], nodes []}]
variables []
widgets {
 info {
  inheritedWidgetCount=1
  parentClass=&FortHUDElementWidget
  rootWidgetClass=&/AthenaFPSTicker.AthenaFPSTicker_C
  widgetCount=2
 }
 widgets [
  AthenaFPSTicker {widgetClassPath=&/AthenaFPSTicker, bIsVariable=true, widget=AthenaFPSTicker}
  TutorialHighlight {widgetClassPath=&TutorialHighlight}
 ]
}
