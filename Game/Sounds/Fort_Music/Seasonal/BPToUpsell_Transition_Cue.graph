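# Sound cue that plays the Season 24 Battle Pass outro transition music, by its name for the move from the Battle Pass to an upsell screen.
# A single wave player feeds the S24_BP_Outro_Transition wave straight to the output, with no randomization or mixing.
# The cue routes through the MediaCinematics sound class with a raised volume and a very large max distance.
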SoundCue /Game/Sounds/Fort_Music/Seasonal/BPToUpsell_Transition_Cue
root=WavePlayer_1
nodes [{id=WavePlayer_1, properties {soundWaveAssetPtr=/Game/Sounds/Fort_Music/Seasonal/S24/S24_BP_Outro_Transition}}]
properties {maxDistance=2097152, soundClassObject=/Game/Sounds/FortSoundClassesAndModes/MediaCinematics, volumeMultiplier=1.4}
