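# Frontend top bar widget (AthenaTopBar child) holding navigation tabs, a sidebar avatar button and window chrome.
# Maps tabs to EFortUIFeature through tab view models, drives onboarding coach marks, and adapts layout for mobile and Evo.
# Blocked playtime forces the Play tab, disables other tabs and clears their bangs.
# Gamepad sidebar input uses a visual-only prompt hack and a throbber timed to the chat double-tap delay.
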WidgetBlueprint /Game/UI/Frontend/TopBar/TopBar
animations [Anim_WaitThrobber {
  binding_count=1
  length_seconds=1
  object=&self:Anim_WaitThrobber
  sequence {bindings [WaitThrobber {
     id=#1
     possessed_class=Image
     tracks [{
       type=MovieSceneFloatTrack
       display_name=Render Opacity
       sections [{
         type=MovieSceneFloatSection
         end_frame=20
         start_frame=0
         channels [None {type=MovieSceneScriptingFloatChannel, keys [[0,1],[18.1817,1],[20,0]]}]
        }]
      }]
    }]}
 }]
graphs [
 BP_GetActiveTab {nodes [
   BP_GetActiveTab() {is=title K2Node_FunctionEntry}
   n1 = Get WBP_UIKit_Tabs_TopNavBar() -> (WBP_UIKit_Tabs_TopNavBar:WBP UIKit Tabs Top Nav Bar Object Reference)
   > n2 = Cast To FortTabViewModel!(
    execute
    Object=GetTabViewModelByID!(
     self:UIKit Tab Group Object Reference=@n1
     TabNameID:Name=GetSelectedTabId!(self:Common Tab List Widget Base Object Reference=@n1) -> (ReturnValue:Name=None)
    ) -> (ReturnValue:UI Kit Tab Group Button View Model Object Reference)
   ) -> (then, CastFailed, AsFort Tab Viewmodel:Fort Tab Viewmodel Object Reference) {is=title K2Node_DynamicCast}
   > Return Node!(
    execute
    ReturnValue:EFortUIFeature Enum=GetLinkedUIFeature!(self:Fort Tab Viewmodel Object Reference=@n2) -> (ReturnValue:EFortUIFeature Enum=ShowHome)
   )
   Return Node(execute=@n2.CastFailed, ReturnValue=Max_None)
  ]}
 BP_GetActiveTabForwardOnRemoval {nodes [
   BP_GetActiveTabForwardOnRemoval() {is=title K2Node_FunctionEntry}
   n1 = Get WBP_UIKit_Tabs_TopNavBar()
   > n2 = Cast To FortTabViewModel(Object=GetTabViewModelByID(self=@n1, TabNameID=GetSelectedTabId(self=@n1) -> (ReturnValue=None)))
   > Return Node(ReturnValue=GetForwardOnRemoval!(self:Fort Tab Viewmodel Object Reference=@n2) -> (ReturnValue:EFortUIFeature Enum=ShowHome))
   Return Node(execute=@n2.CastFailed, ReturnValue=Max_None)
  ]}
 CloseTabCoachMarks {nodes [
   CloseTabCoachMarks() {is=title K2Node_FunctionEntry}
   > Branch(Condition="Not Equal (Enum)"!(
     A:EFortUIFeature Enum=GetActiveTab!(self:Athena Top Bar Object Reference) -> (ReturnValue:EFortUIFeature Enum=ShowHome)
     B:EFortUIFeature Enum=ShowPlay
    ) -> (ReturnValue:bool))
   > n1 = For Each Loop(Array:Array of Fort Sidebar Onboard Tooltip Widget Object References=Get TabCoachMarks() -> (TabCoachMarks:Array of Fort Sidebar Onboard Tooltip Widget Object References)) -> (Array Element:Fort Sidebar Onboard Tooltip Widget Object Reference)
   Clear(execute=@n1.Completed, TargetArray:Array of Fort Sidebar Onboard Tooltip Widget Object References=Get TabCoachMarks())
   Branch(execute=@n1.LoopBody, Condition=IsActivated!(self:Common Activatable Widget Object Reference=@n1) -> (ReturnValue:bool))
   > DeactivateWidget(self=@n1)
  ]}
 EventGraph {nodes [
   n1 = "On Input Method Changed (SidebarVisualOnlyButtonHack)"!() -> (OutputDelegate, then, bUsingGamepad:bool) {is=title K2Node_ComponentBoundEvent}
   > n2 = Cast To FortPlayerController(Object=GetOwningPlayer())
   > Branch(Condition=AND Boolean(
     A=UseMessageWidgetV3!(self:Fort Player Controller Object Reference=@n2) -> (ReturnValue:bool)
     B=GetIsShowButtonBindHackForSidebarRollbackEnabled!() -> (ReturnValue:bool)
    ))
   > SetVisibility(
    self=Get SidebarVisualOnlyButtonHack() -> (SidebarVisualOnlyButtonHack:Common Action Widget Object Reference)
    InVisibility=Select(Option 0=Collapsed, Option 1=HitTestInvisible, Index=@n1.bUsingGamepad)
   )
   > SetInputAction!(
    execute
    self:Common Action Widget Object Reference=Get SidebarVisualOnlyButtonHack()
    InputActionRow:Data Table Row Handle Structure=Select(
     Option 0:Data Table Row Handle Structure=Make Data Table Row Handle!(DataTable:Data Table Object Reference, RowName:Name=None) -> (DataTableRowHandle:Data Table Row Handle Structure) {is=title K2Node_MakeStruct}
     Option 1:Data Table Row Handle Structure=Get SidebarInputAction() -> (SidebarInputAction:Data Table Row Handle Structure)
     Index=@n1.bUsingGamepad
    ) -> (ReturnValue:Data Table Row Handle Structure)
   ) -> (then)
   "On Clicked (WBP_SidebarAvatarButton)"!() -> (OutputDelegate, then, Button:Common Button Base Object Reference) {is=title K2Node_ComponentBoundEvent}
   > HandleSocialButtonClicked(self:Athena Top Bar Object Reference)
   n3 = OnTabCreating!() -> (OutputDelegate, then, TabId:Name, TabButton:Common Button Base Object Reference) {is=title K2Node_CustomEvent}
   n4 = "Equal (Enum)"(
    A:EPlaytimeState Enum=Get PlaytimeState!(self:Player Playtime State View Model Object Reference=Get PlayerPlaytimeStateVM() -> (PlayerPlaytimeStateVM:Player Playtime State View Model Object Reference)) -> (PlaytimeState:EPlaytimeState Enum=Unblocked)
    B:EPlaytimeState Enum=PlayTime_Blocked
   )
   > SetPlayTimeOut!(execute, self, IsPlayTimeOut:bool=@n4) -> (then)
   > n5 = Branch(Condition=@n4)
   HandleCoachMarkForTab!(execute=@n5.else, self, TabId:Name=@n3.TabId) -> (then)
   n6 = OnTabSelectionChangedUpdateContent!() -> (OutputDelegate, then, TabId:Name) {is=title K2Node_CustomEvent}
   > Branch(Condition=GetIsDiscoverInLobbyEnabled!(self:Fort Runtime Options Object Reference=GetRuntimeOptions(), ContextWidget:Widget Object Reference=Self-Reference()) -> (ReturnValue:bool))
   > n7 = Cast To FortTabViewModel(Object=GetTabViewModelByID(self=Get WBP_UIKit_Tabs_TopNavBar(), TabNameID=@n6.TabId))
   > HandleTopNavSelectionChanged!(execute, self:Athena Top Bar Object Reference, Feature:EFortUIFeature Enum=GetLinkedUIFeature(self=@n7) -> (ReturnValue=ShowHome)) -> (then)
   > CloseTabCoachMarks() {is=title K2Node_CallFunction}
   n8 = OnTabClicked!() -> (OutputDelegate, then, TabID:Name, TabVM:MVVM Base Viewmodel Object Reference) {is=title K2Node_CustomEvent}
   > HandleClickedCoachMarkForTab!(execute, self, TabId:Name=@n8.TabID) -> (then)
   EnhancedInputAction IA_OpenSideBarMenu!() -> (
    Triggered:Exec
    Started:Exec
    Ongoing:Exec
    Canceled:Exec
    Completed:Exec
    ActionValue:bool
    ElapsedSeconds:double=~
    TriggeredSeconds:double=~
    InputAction:Input Action Object Reference=/Game/Input/EnhancedInput/FortPlayerController/IA_OpenSideBarMenu
   ) {is=title K2Node_EnhancedInputAction}
   > n9 = Switch on ECommonInputType!(execute, Selection:ECommonInputType Enum=GetCurrentInputType(self=Get CommonInputSubsystem())) -> (MouseAndKeyboard:Exec, Gamepad:Exec, Touch:Exec, Count:Exec) {is=title K2Node_SwitchEnum}
   n10 = GetConsoleVariableFloatValue!(execute=@n9.Gamepad, VariableName:String=fro.TextChatDoubleTapDelay) -> (then, ReturnValue:float)
   > Branch(Condition="float > float"(A=@n10))
   > PlayAnimationForward(InAnimation=Get Anim_WaitThrobber() -> (Anim_WaitThrobber:Widget Animation Object Reference), PlaybackSpeed=float / float(A=1, B=@n10))
   Event On Activated()
   > SetVisibility(InVisibility=SelfHitTestInvisible)
   > SetBumperInputEnabled!(execute, self, AreBumpersEnabled:bool=true) -> (then)
   Event On Deactivated()
   > SetVisibility(InVisibility=Collapsed)
   Event Construct()
   n11 = Get WBP_UIKit_Tabs_TopNavBar()
   > Bind Event to On UIKit Tab Clicked(self:UIKit Tab Group Object Reference=@n11, Delegate=Create Event()) {is=title K2Node_AddDelegate}
   > Bind Event to On Tab Selected(self:Common Tab List Widget Base Object Reference=@n11, Delegate=Create Event()) {is=title K2Node_AddDelegate}
   > Bind Event to On Tab Button Creation(self:Common Tab List Widget Base Object Reference=@n11, Delegate=Create Event()) {is=title K2Node_AddDelegate}
   > SetBumperInputEnabled(AreBumpersEnabled=true)
   > SetProfileVisibility()
   > SetMobileLayout()
   > n12 = Cast To FortPlayerController(Object=GetOwningPlayer())
   > n13 = Branch(Condition=UseMessageWidgetV3(self=@n12))
   > Branch(Condition=GetIsShowButtonBindHackForSidebarRollbackEnabled())
   n14 = GetCurrentInputType(self=Get CommonInputSubsystem!() -> (ReturnValue:Common Input Object Reference) {is=title K2Node_GetBlueprintContextEx})
   > Branch(Condition="Equal (Enum)"(A:ECommonInputType Enum=@n14, B:ECommonInputType Enum=Gamepad))
   n15 = Get SidebarVisualOnlyButtonHack()
   > SetInputAction(self=@n15, InputActionRow=Get SidebarInputAction())
   > SetVisibility(
    self=@n15
    InVisibility=Select!(
     MouseAndKeyboard:ESlateVisibility Enum=Collapsed
     Gamepad:ESlateVisibility Enum=HitTestInvisible
     Touch:ESlateVisibility Enum=Collapsed
     Count:ESlateVisibility Enum=Collapsed
     Index:ECommonInputType Enum=@n14
    ) -> (ReturnValue:ESlateVisibility Enum)
   )
   RemoveFromParent(execute=@n13.else, self=Get SidebarVisualOnlyButtonHack())
   > SetTriggeringInputAction!(
    execute
    self:Common Button Base Object Reference=Get WBP_SidebarAvatarButton() -> (WBP_SidebarAvatarButton:WBP Sidebar Avatar Button Object Reference)
    InputActionRow:"Data Table Row Handle Structure (by ref)"=Get SidebarInputAction()
   ) -> (then)
   Event PreConstruct()
   > UpdateTabsAlignment()
   n16 = Event OnSetSelectedTab!() -> (OutputDelegate, then, SelectedTab:EFortUIFeature Enum=ShowHome) {is=title K2Node_Event}
   > SelectTabFromFeature!(execute, self, Feature:EFortUIFeature Enum=@n16.SelectedTab) -> (then)
   Event OnActivityRequested() {is=title K2Node_Event}
   > SelectTabFromFeature(Feature=ShowPlay)
   > NavigateToFeature!(execute, Feature:EFortUIFeature Enum=ShowPlay, bClearPrimaryContentStack:bool=true) -> (then)
  ]}
 GetIsMPLobbyEnabled {nodes [
   GetIsMPLobbyEnabled() {is=title K2Node_FunctionEntry}
   > Return Node(ReturnValue:bool=GetIsDiscoverInLobbyEnabled(self=GetRuntimeOptions(), ContextWidget=Self-Reference()))
  ]}
 GetIsPlayTimeOut {nodes [
   n1 = GetIsPlayTimeOut!() -> (then, PlaytimeState:EPlaytimeState Enum=Unblocked) {is=title K2Node_FunctionEntry}
   > Return Node!(execute, IsPlaytimeOut:bool="Equal (Enum)"(A:EPlaytimeState Enum=@n1, B:EPlaytimeState Enum=PlayTime_Blocked))
  ]}
 GetTabsWrapperHAlignment {nodes [
   GetTabsWrapperHAlignment() {is=title K2Node_FunctionEntry}
   n1 = GetGameInstance!() -> (ReturnValue:Game Instance Object Reference)
   > n2 = Is Valid(InputObject=@n1)
   > n3 = Cast To FortGameInstance!(execute, Object=@n1) -> (then, CastFailed, AsFort Game Instance:Fort Game Instance Object Reference) {is=title K2Node_DynamicCast}
   > Return Node!(
    execute
    Alignment:EHorizontalAlignment Enum=Select!(
     Option 0:EHorizontalAlignment Enum=HAlign_Left
     Option 1:EHorizontalAlignment Enum=HAlign_Center
     Index:bool=IsEvoEnabled!(self:Fort Game Instance Object Reference=@n3) -> (ReturnValue:bool)
    ) -> (ReturnValue:EHorizontalAlignment Enum)
   )
   Return Node(execute=[@n3.CastFailed, @n2.Is Not Valid], Alignment=HAlign_Left)
  ]}
 HandleClickedCoachMarkForTab {nodes [
   n1 = HandleClickedCoachMarkForTab!() -> (then, TabId:Name) {is=title K2Node_FunctionEntry}
   > Branch(Condition=NOT Boolean(A="Equal (Name)"(A=@n1)))
   > n2 = Cast To FortTabViewModel(Object=GetTabViewModelByID(self=Get WBP_UIKit_Tabs_TopNavBar(), TabNameID=@n1))
   > n3 = Switch on EFortUIFeature!(execute, Selection:EFortUIFeature Enum=GetLinkedUIFeature(self=@n2) -> (ReturnValue=ShowHome)) -> (
    ShowHome:Exec
    ShowPlay:Exec
    ShowCommand:Exec
    ShowHeros:Exec
    ShowSquads:Exec
    ShowArmory:Exec
    ShowLocker:Exec
    ShowMultiProductLocker:Exec
    ShowMyLoadout:Exec
    ShowSkillTree:Exec
    ShowStore:Exec
    ShowQuests:Exec
    ShowMainStore:Exec
    ShowContextHelpMenu:Exec
    ShowCampaign:Exec
    ShowActiveBoost:Exec
    ShowJournal:Exec
    ShowPartyBar:Exec
    ShowChatWindow:Exec
    ShowFriendsMenu:Exec
    ShowObjectives:Exec
    ShowRatingIconsInTopbar:Exec
    ShowAntiAddictionMessage:Exec
    ShowMinorShutdownMessage:Exec
    ShowHealthWarningScreen:Exec
    ShowSimplifiedHUD:Exec
    LargeHeaderFooterButtons:Exec
    ShowAthenaFavoriting:Exec
    ShowAthenaItemRandomization:Exec
    ShowBattlePass:Exec
    ShowNewBattlePass:Exec
    ShowDynamicBattlePass:Exec
    ShowBattlePassFAQ:Exec
    ShowPoblano:Exec
    ShowCareerTabV2:Exec
    ShowCareerTabV3:Exec
    ShowAthenaItemShop:Exec
    ShowMultiProductItemShop:Exec
    ShowShowdown:Exec
    ShowSpecialEvent:Exec
    ShowSpecialEventTeaser:Exec
    ShowSparksSpecialEvent:Exec
    ShowJunoSpecialEvent:Exec
    ShowDelMarSpecialEvent:Exec
    ShowWinterfest:Exec
    ShowSpatialUI:Exec
    ShowSocial:Exec
    ShowQuestsV3:Exec
    ShowHordeRushProgression:Exec
    ShowAccountBoosts:Exec
    ShowCustomerSupport:Exec
    ShowGlobalChat:Exec
    ShowEULA:Exec
    ShowEndOfZoneCinematic:Exec
    ShowOnboardingCinematics:Exec
    ShowFounderBannerIcons:Exec
    ShowCurrentRegionInLobby:Exec
    ShowPrerollLlamas:Exec
    EnableFoundersDailyRewards:Exec
    EnableMatchmakingRegionSetting:Exec
    EnableLanguageSetting:Exec
    EnableFriendCodeSetting:Exec
    EnableEarlyAccessLoadingScreenBanner:Exec
    EnableAlterationModifications:Exec
    EnableSchematicRarityUpgrade:Exec
    EnableMissionActivationVote:Exec
    EnableLtmRetrieveTheData:Exec
    EnableUpgradesVideos:Exec
    ShowPreviewTestTab:Exec
    EnableSearchTabInTopNavigationBar:Exec
    EnableDiscoverTabInTopNavigationBar:Exec
    EnableFollowingTabInTopNavigationBar:Exec
    EnableLibraryTabInTopNavigationBar:Exec
    HestiaUI:Exec
    PlaceholderUIFeature1:Exec
    PlaceholderUIFeature2:Exec
    PlaceholderUIFeature3:Exec
    PlaceholderUIFeature4:Exec
    PlaceholderUIFeature5:Exec
    ShowStoreItemShop:Exec
    Max_None:Exec
   ) {is=title K2Node_SwitchEnum}
   SetSeenOnboardings!(
    execute=@n3.ShowCareerTabV3
    InOnboardingSeenIds:Array of Gameplay Tag Structures=Make Array("[0]"=(TagName="UI.Tutorials.SeenSocialLeaderboard"))
   ) -> (then, ReturnValue:bool)
  ]}
 HandleCoachMarkForTab {nodes [
   n1 = HandleCoachMarkForTab!() -> (then, TabId:Name) {is=title K2Node_FunctionEntry}
   > Branch(Condition=NOT Boolean(A="Equal (Name)"(A=@n1)))
   > n2 = Cast To FortTabViewModel(Object=GetTabViewModelByID(self=Get WBP_UIKit_Tabs_TopNavBar(), TabNameID=@n1))
   > n3 = Switch on EFortUIFeature(Selection=GetLinkedUIFeature(self=@n2) -> (ReturnValue=ShowHome))
   n4 = Construct Object from Class(execute=@n3.ShowCareerTabV3, Class=&/WBP_CareerV3_CoachMark.WBP_CareerV3_CoachMark_C) -> (ReturnValue:WBP Career V3 Coach Mark Object Reference)
   > ShowOnboardingCoachMark!(execute, self, TooltipWidget:Fort Sidebar Onboard Tooltip Widget Object Reference=@n4, TabId:Name=Get TabId() -> (TabId:Name=None)) -> (then)
   n5 = Construct Object from Class(Class=&/WBP_SearchForCreators_CoachMark.WBP_SearchForCreators_CoachMark_C) -> (ReturnValue:WBP Search For Creators Coach Mark Object Reference)
   > ShowOnboardingCoachMark(TooltipWidget=@n5, TabId=Get TabId() -> (TabId=None))
  ]}
 SelectTabFromFeature {nodes [
   SelectTabFromFeature!() -> (then, Feature:EFortUIFeature Enum=ShowHome) {is=title K2Node_FunctionEntry}
   > Branch(Condition=GetIsDiscoverInLobbyEnabled(self=GetRuntimeOptions(), ContextWidget=Self-Reference()))
   > n1 = For Each Loop(Array:Array of UI Kit Tab Group Button View Model Object References=Get Tabs!(self:UI Kit Tab Group View Model Object Reference=Get FortTabGroupViewModel() -> (FortTabGroupViewModel:Fort Tab Group Viewmodel Object Reference)) -> (Tabs:Array of UI Kit Tab Group Button View Model Object References)) -> (Array Element:UI Kit Tab Group Button View Model Object Reference)
   > n2 = Cast To FortTabViewModel(Object=@n1)
   > Branch(Condition="Equal (Enum)"(
     A:EFortUIFeature Enum=GetLinkedUIFeature(self=@n2) -> (ReturnValue=ShowHome)
     B:EFortUIFeature Enum=Get Feature() -> (Feature:EFortUIFeature Enum=ShowHome)
    ))
   > n3 = SelectTabByIndex!(execute, self:UIKit Tab Group Object Reference=Get WBP_UIKit_Tabs_TopNavBar(), Index:int=@n1.Array Index, bSuppressClickFeedback:bool=true) -> (then, ReturnValue:bool)
   > Branch(Condition=@n3)
   > Return Node!(execute)
  ]}
 SetBumperInputEnabled {nodes [
   n1 = SetBumperInputEnabled!() -> (then, AreBumpersEnabled:bool=~) {is=title K2Node_FunctionEntry}
   > SetInputActionsType!(
    execute
    self:WBP UIKit Tabs Base Object Reference=Get WBP_UIKit_Tabs_TopNavBar()
    InputActions:E_UI_TabsInputActionType Enum=Select(Option 0:E_UI_TabsInputActionType Enum=NewEnumerator2, Option 1:E_UI_TabsInputActionType Enum=NewEnumerator0, Index=@n1) -> (ReturnValue:E_UI_TabsInputActionType Enum)
    InputActionsWithMouseScroll:bool=false
   ) -> (then)
  ]}
 SetMainNavVisibility {nodes [
   n1 = SetMainNavVisibility!() -> (then, NewVisibility:ESlateVisibility Enum=Visible) {is=title K2Node_FunctionEntry}
   > SetVisibility(self=Get MainNav() -> (MainNav:Grid Panel Object Reference), InVisibility=@n1)
  ]}
 SetMobileLayout {nodes [
   SetMobileLayout() {is=title K2Node_FunctionEntry}
   > Branch(Condition=IsMobileGame())
   > SetGridSlotPadding!(
    execute
    Widget:Widget Object Reference=Get MainNav()
    Padding_Left:float=48
    Padding_Top:float=24
    Padding_Right:float=48
    Padding_Bottom:float=0
   ) -> (then)
   > SetGridSlotPadding(Widget=Get Tabs_Wrapper() -> (Tabs_Wrapper:Overlay Object Reference), Padding_Left=0, Padding_Top=-16, Padding_Right=0, Padding_Bottom=0)
   > UpdateTabsAlignment()
   n1 = Get WBP_UIKit_Tabs_TopNavBar()
   > Set IconSize!(execute, IconSize:Vector 2D Structure=(X=72,Y=72), self:WBP UIKit Tabs Base Object Reference=@n1) -> (then, Output_Get:Vector 2D Structure)
   > Set!(
    execute
    TabGroupPadding_Left:float=8
    TabGroupPadding_Top:float=8
    TabGroupPadding_Right:float=8
    TabGroupPadding_Bottom:float=8
    self:WBP UIKit Tabs Base Object Reference=@n1
   ) -> (then, Output_Get:Margin Structure)
   n2 = SlotAsOverlaySlot(Widget=@n1)
   > SetHorizontalAlignment(self=@n2, InHorizontalAlignment=HAlign_Left)
   > SetVerticalAlignment(self=@n2, InVerticalAlignment=VAlign_Top)
  ]}
 SetPlayTimeOut {nodes [
   SetPlayTimeOut!() -> (then, IsPlayTimeOut:bool=~) {is=title K2Node_FunctionEntry}
   > n1 = Branch(Condition=Get IsPlayTimeOut() -> (IsPlayTimeOut:bool))
   > SelectTabFromFeature(Feature=ShowPlay) {is=title K2Node_CallFunction}
   > n2 = NavigateToFeature(Feature=ShowPlay, bClearPrimaryContentStack=true)
   n3 = For Loop(
    execute=[@n2.then, @n1.else]
    LastIndex=GetTabCount!(self:Common Tab List Widget Base Object Reference=Get WBP_UIKit_Tabs_TopNavBar()) -> (ReturnValue:int=0)
   )
   n4 = Get WBP_UIKit_Tabs_TopNavBar()
   SetRenderOpacity(
    execute=@n3.Completed
    self=[
     Get Next_Prompt!(self:WBP UIKit Tabs Base Object Reference=@n4) -> (Next_Prompt:Grid Panel Object Reference)
     Get Prev_Prompt!(self:WBP UIKit Tabs Base Object Reference=@n4) -> (Prev_Prompt:Grid Panel Object Reference)
    ]
    InOpacity=SelectFloat(A=0.5, B=1, bPickA=Get IsPlayTimeOut())
   )
   n5 = Get WBP_UIKit_Tabs_TopNavBar()
   n6 = GetTabIDbyIndex!(self:UIKit Tab Group Object Reference=@n5, Index:int=@n3) -> (ReturnValue:Name=None)
   n7 = Cast To FortTabViewModel(execute=@n3.LoopBody, Object=GetTabViewModelByID(self=@n5, TabNameID=@n6))
   > n8 = Cast To WBP_UIKit_Button_TopNavBar!(
    execute
    Object=GetTabButtonBaseByID!(self:Common Tab List Widget Base Object Reference=Get WBP_UIKit_Tabs_TopNavBar(), TabNameID:Name=@n6) -> (ReturnValue:Common Button Base Object Reference)
   ) -> (then, CastFailed, AsWBP UIKit Button Top Nav Bar:WBP UIKit Button Top Nav Bar Object Reference) {is=title K2Node_DynamicCast}
   > SetBangType!(
    execute
    self:WBP UIKit Button Top Nav Bar Object Reference=@n8
    EBangType:EFortBangType Enum=Select(
     Option 0:EFortBangType Enum=GetBangType!(self:Fort Tab Viewmodel Object Reference=@n7) -> (ReturnValue:EFortBangType Enum=Invalid)
     Option 1:EFortBangType Enum=Invalid
     Index=Get IsPlayTimeOut()
    ) -> (ReturnValue:EFortBangType Enum)
   ) -> (then)
   > Branch(Condition="Not Equal (Enum)"(A=GetLinkedUIFeature(self=@n7) -> (ReturnValue=ShowHome), B=ShowPlay))
   n9 = Get WBP_UIKit_Tabs_TopNavBar()
   > SetTabEnabled!(execute, self:Common Tab List Widget Base Object Reference=@n9, TabNameID:Name=@n6, bEnable:bool=true) -> (then)
   > SetTabEnabled(self=@n9, TabNameID=@n6, bEnable=NOT Boolean(A=Get IsPlayTimeOut()))
  ]}
 SetProfileVisibility {nodes [
   SetProfileVisibility() {is=title K2Node_FunctionEntry}
   > n1 = Cast To FortGameInstance(Object=GetGameInstance!(self:Widget Object Reference) -> (ReturnValue))
   > n2 = GetConsoleVariableBoolValue!(execute, VariableName:String=Fort.Sidemenu.UseGlobalSidebar) -> (then, ReturnValue:bool)
   n3 = IsEvoEnabled(self=@n1)
   > SetVisibility(
    self=Get ProfileGridPanel() -> (ProfileGridPanel:Grid Panel Object Reference)
    InVisibility=Select(Option 0:ESlateVisibility Enum=SelfHitTestInvisible, Option 1:ESlateVisibility Enum=Collapsed, Index=AND Boolean(A=@n2, B=@n3)) -> (ReturnValue:ESlateVisibility Enum)
   )
  ]}
 SetTopTabsVisibility {nodes [
   n1 = SetTopTabsVisibility!() -> (then, NewVisibility:ESlateVisibility Enum=Visible) {is=title K2Node_FunctionEntry}
   > SetVisibility(self=Get MainNav(), InVisibility=@n1)
  ]}
 ShowOnboardingCoachMark {nodes [
   n1 = ShowOnboardingCoachMark!() -> (then, TooltipWidget:Fort Sidebar Onboard Tooltip Widget Object Reference, TabId:Name) {is=title K2Node_FunctionEntry}
   > SetOwningPlayer!(execute, self:User Widget Object Reference=@n1, LocalPlayerController:Player Controller Object Reference=GetOwningPlayer()) -> (then)
   > Branch(Condition=ShouldShowOnboarding!(self:Fort Sidebar Onboard Tooltip Widget Object Reference=@n1) -> (ReturnValue:bool))
   n2 = Get WBP_UIKit_Tabs_TopNavBar()
   > n3 = AddChildToGrid!(
    execute
    self:Grid Panel Object Reference=Get TabsContainer!(self:WBP UIKit Tabs Base Object Reference=@n2) -> (TabsContainer:Grid Panel Object Reference)
    Content:Widget Object Reference=@n1
    InRow:int=0
    InColumn:int=GetTabIndexByID!(self:UIKit Tab Group Object Reference=@n2, TabNameID:Name=@n1.TabId) -> (ReturnValue:int=0)
   ) -> (then, ReturnValue:Grid Slot Object Reference)
   > Set HorizontalAlignment!(execute, HorizontalAlignment:EHorizontalAlignment Enum=HAlign_Center, self:Grid Slot Object Reference=@n3) -> (then, Output_Get:EHorizontalAlignment Enum=HAlign_Fill)
   > ActivateWidget(self=@n1)
   > Add(
    TargetArray:Array of Fort Sidebar Onboard Tooltip Widget Object References=Get TabCoachMarks()
    NewItem:"Fort Sidebar Onboard Tooltip Widget Object Reference (by ref)"=@n1
   ) {is=title K2Node_CallArrayFunction}
  ]}
 UpdateTabsAlignment {nodes [
   UpdateTabsAlignment() {is=title K2Node_FunctionEntry}
   > SetGridSlotAlignment!(
    execute
    Widget:Widget Object Reference=Get Tabs_Wrapper()
    Horizontal:EHorizontalAlignment Enum=GetTabsWrapperHAlignment!(self) -> (Alignment:EHorizontalAlignment Enum=HAlign_Fill) {is=title K2Node_CallFunction}
    Vertical:EVerticalAlignment Enum=Select(Option 0:EVerticalAlignment Enum=VAlign_Center, Option 1:EVerticalAlignment Enum=VAlign_Top, Index=IsMobileGame()) -> (ReturnValue:EVerticalAlignment Enum)
   ) -> (then)
  ]}
]
variables [
 SidebarInputAction {type=DataTableRowHandle, default {dataTable=&InputActionDataTable, rowName=MainMenu}}
 TabCoachMarks {type="array<FortSidebarOnboardTooltipWidget>", default []}
]
widgets {
 info {inheritedWidgetCount=0, parentClass=/Script/FortniteUI.AthenaTopBar, rootWidgetClass=&SafeZone, widgetCount=16}
 widgets [
  SafeZone {widgetClassPath=&SafeZone, bIsVariable=true, widget=SafeZone, properties {padBottom=false}}
  GridPanel_0 {
   widgetClassPath=&GridPanel
   parent=SafeZone
   widget=GridPanel_0
   properties {columnFill [1], rowFill [1]}
   slot {class=SafeZoneSlot, properties {padding=<ltrb 0 0 0 8>}}
  }
  TopTitlebarArea {
   widgetClassPath=/Script/UMG.WindowTitleBarArea
   parent=GridPanel_0
   widget=TopTitlebarArea
   slot {class=GridSlot, properties {verticalAlignment=VAlign_Top}}
  }
  CommonVisibilityWidget_0 {
   widgetClassPath=&CommonVisibilityWidgetLegacy
   parent=TopTitlebarArea
   widget=CommonVisibilityWidget_0
   properties {
    bShowForAndroid=false
    bShowForErebus=false
    bShowForIOS=false
    bShowForMac=false
    bShowForPS4=false
    bShowForPS5=false
    bShowForSwitch2=false
    bShowForXBox=false
    bShowForXSX=false
    padding=<ltrb 0 0 0 0>
   }
   slot {class=WindowTitleBarAreaSlot}
  }
  SizeBox_3 {
   widgetClassPath=&SizeBox
   parent=CommonVisibilityWidget_0
   widget=SizeBox_3
   properties {bOverride_HeightOverride=true, heightOverride=24}
   slot {class=BorderSlot, properties {padding=<ltrb 0 0 0 0>}}
  }
  Overlay_0 {widgetClassPath=&Overlay, parent=SizeBox_3, widget=Overlay_0, slot {class=SizeBoxSlot}}
  ChromeButtonFill {
   widgetClassPath=&Image
   parent=Overlay_0
   widget=ChromeButtonFill
   properties {
    visibility=SelfHitTestInvisible
    brush {
     bIsDynamicallyLoaded=false
     drawAs=Image
     imageType=NoImage
     mirroring=NoMirror
     resourceName=None
     resourceObject=/Game/UI/Foundation/UI_Refresh/M_UI_Bar_Shell_Backing
     tiling=NoTile
     imageSize=<32 32>
     margin=<ltrb 0 0 0 0>
     outlineSettings {
      bUseBrushTransparency=false
      roundingType=HalfHeightRadius
      width=0
      color {colorUseRule=UseColor_Specified, specifiedColor=<rgba 0 0 0 0>}
      cornerRadii=<0 0 0 0>
     }
     tintColor {colorUseRule=UseColor_Specified, specifiedColor=<rgba 1 1 1 1>}
     uVRegion {bIsValid=false, max=<0 0>, min=<0 0>}
    }
    colorAndOpacity=<rgba 1 1 1 0>
   }
   slot {class=OverlaySlot, properties {horizontalAlignment=HAlign_Fill, verticalAlignment=VAlign_Fill, padding=<ltrb -128 -88 -128 0>}}
  }
  windowchromebuttons {
   widgetClassPath=/Game/UI/Foundation/Shell/windowchromebuttons
   parent=Overlay_0
   widget=windowchromebuttons
   slot {class=OverlaySlot, properties {horizontalAlignment=HAlign_Right, verticalAlignment=VAlign_Fill}}
  }
  MainNav {
   widgetClassPath=&GridPanel
   bIsVariable=true
   parent=GridPanel_0
   widget=MainNav
   properties {columnFill [1], rowFill [1]}
   slot {class=GridSlot, properties {verticalAlignment=VAlign_Top, padding=<ltrb 80 40 80 0>}}
  }
  Tabs_Wrapper {
   widgetClassPath=&Overlay
   bIsVariable=true
   parent=MainNav
   widget=Tabs_Wrapper
   slot {class=GridSlot, properties {horizontalAlignment=HAlign_Left, verticalAlignment=VAlign_Center}}
  }
  WBP_UIKit_Tabs_TopNavBar {
   widgetClassPath=&/WBP_UIKit_Tabs_TopNavBar
   bIsVariable=true
   parent=Tabs_Wrapper
   widget=WBP_UIKit_Tabs_TopNavBar
   properties {inputActions=None, promptBackplateCornerRadius=Backplate, promptBackplateMinWidth=48}
   slot {class=OverlaySlot, properties {horizontalAlignment=HAlign_Fill, verticalAlignment=VAlign_Fill}}
  }
  ProfileGridPanel {
   widgetClassPath=&GridPanel
   bIsVariable=true
   parent=MainNav
   widget=ProfileGridPanel
   properties {columnFill [0,0], rowFill [1]}
   slot {class=GridSlot, properties {horizontalAlignment=HAlign_Right, verticalAlignment=VAlign_Center}}
  }
  VisualAttachment {
   widgetClassPath=&CommonVisualAttachment
   parent=ProfileGridPanel
   widget=VisualAttachment
   properties {contentAnchor=<0.5 0.5>}
   slot {class=GridSlot, properties {horizontalAlignment=HAlign_Center, layer=1, verticalAlignment=VAlign_Center}}
  }
  WaitThrobber {
   widgetClassPath=&Image
   parent=VisualAttachment
   widget=WaitThrobber
   properties {
    renderOpacity=0
    visibility=HitTestInvisible
    brush {
     bIsDynamicallyLoaded=false
     drawAs=Image
     imageType=FullColor
     mirroring=NoMirror
     resourceName=None
     resourceObject=/FortUILibrary/UIKit/Throbber/MI_UIKit_Monochrome_Throbber
     tiling=NoTile
     imageSize=<50 50>
     margin=<ltrb 0 0 0 0>
     outlineSettings {
      bUseBrushTransparency=false
      roundingType=HalfHeightRadius
      width=0
      color {colorUseRule=UseColor_Specified, specifiedColor=<rgba 0 0 0 0>}
      cornerRadii=<0 0 0 0>
     }
     tintColor {colorUseRule=UseColor_Specified, specifiedColor=<rgba 1 1 1 1>}
     uVRegion {bIsValid=false, max=<0 0>, min=<0 0>}
    }
   }
   slot {class=SizeBoxSlot, properties {horizontalAlignment=HAlign_Center, verticalAlignment=VAlign_Center}}
  }
  SidebarVisualOnlyButtonHack {
   widgetClassPath=&CommonActionWidget
   bIsVariable=true
   parent=ProfileGridPanel
   widget=SidebarVisualOnlyButtonHack
   properties {designTimeKey=Gamepad_Special_Right, visibility=Collapsed}
   slot {class=GridSlot, properties {layer=2}}
  }
  WBP_SidebarAvatarButton {
   widgetClassPath=&/WBP_SidebarAvatarButton
   bIsVariable=true
   parent=ProfileGridPanel
   widget=WBP_SidebarAvatarButton
   properties {triggeringInputAction {dataTable=null, rowName=None}}
   slot {class=GridSlot, properties {column=1}}
  }
 ]
}
