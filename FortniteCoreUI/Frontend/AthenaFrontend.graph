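# Battle Royale frontend (lobby) UI state widget, hosting the primary content stack, tabbed top bar, bottom bar, text chat and mic toggle.
# A visibility-mode switch shows or hides bars and tabs, while input, push-to-talk and mobile checks adjust chat and mic layout.
# It pushes UIKit dialogs for download failures, cellular confirmation and idle players, disabling top-bar bumpers during modals.
# Timers poll trailer playback time to trigger transition sounds.
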WidgetBlueprint /FortniteCoreUI/Frontend/AthenaFrontend
graphs [
 BP_PushBulkCosmeticsCellularConfirmDialog {nodes [
   BP_PushBulkCosmeticsCellularConfirmDialog() {is=title K2Node_FunctionEntry}
   > n1 = Push Simple UIKit Modal Dialog!(
    execute
    Context:User Widget Object Reference=Self-Reference()
    VM Class:UI Kit Dialog Viewmodel Class Reference=/FortniteCoreUI/Frontend/Dialogs/DialogVM_Lobby_DownloadFailed_NetworkChangedToCellular.DialogVM_Lobby_DownloadFailed_NetworkChangedToCellular_C
    OnDialogButtonClicked:"Delegate (by ref)"=Create Event()
   ) -> (then, DialogActivatableWidget:WBP UIKit Dialog Object Reference, DialogVM:UI Kit Dialog Viewmodel Object Reference)
   > Fire Event Main View Opened!(execute, View:Widget Object Reference=@n1, AdditionalInfo:Flow Analytics Additional Info Structure, bIsOverlay:bool=true) -> (then)
   > Return Node!(execute, ReturnValue:UI Kit Dialog Viewmodel Object Reference=@n1.DialogVM)
  ]}
 BP_PushIslandDownloadFailedDialog {nodes [
   n1 = BP_PushIslandDownloadFailedDialog!() -> (then, DownloadResult:"Selected Island Download Result Structure (by ref)") {is=title K2Node_FunctionEntry}
   > n2 = Switch on ESelectedIslandDownloadResult!(
    execute
    Selection:ESelectedIslandDownloadResult Enum=Break Selected Island Download Result!(SelectedIslandDownloadResult:"Selected Island Download Result Structure (by ref)"=@n1) -> (SourceTitle:String, Result:ESelectedIslandDownloadResult Enum) {is=title K2Node_BreakStruct}.Result
   ) -> (
    Unknown:Exec
    Succeeded:Exec
    Cancelled:Exec
    UserCancelled:Exec
    Paused:Exec
    Reprioritized:Exec
    Error:Exec
    ErrorNotEnoughStorage:Exec
    ErrorCellularNotAllowed:Exec
   ) {is=title K2Node_SwitchEnum}
   Return Node(execute=[@n2.Cancelled, @n2.Paused, @n2.Succeeded, @n2.Unknown, @n2.UserCancelled])
   n3 = Push Simple UIKit Modal Dialog(
    execute=@n2.ErrorCellularNotAllowed
    Context=Self-Reference()
    VM Class=&DialogVM_Lobby_DownloadFailed_NetworkChangedToCellular_C
    OnDialogButtonClicked=Create Event()
   )
   > Fire Event Main View Opened(View=@n3, bIsOverlay=true)
   > Return Node(ReturnValue=@n3.DialogVM)
   n4 = Push Simple UIKit Modal Dialog(
    execute=@n2.ErrorNotEnoughStorage
    Context=Self-Reference()
    VM Class=/FortniteCoreUI/Frontend/Dialogs/DialogVM_Lobby_DownloadFailed_NotEnoughStorage.DialogVM_Lobby_DownloadFailed_NotEnoughStorage_C
    OnDialogButtonClicked=Create Event()
   )
   > Fire Event Main View Opened(View=@n4, bIsOverlay=true)
   n5 = GetDownloadResultErrorInfo!(DownloadResult:"Selected Island Download Result Structure (by ref)"=@n1) -> (ReturnValue_MissingDiskSize:Integer64)
   > Set DialogInfoText!(
    execute
    DialogInfoText:Text=Format Text!(
     Format=Get DialogInfoText!(self:UI Kit Dialog Viewmodel Object Reference=@n4.DialogVM) -> (DialogInfoText:Text)
     0:Text=AsMemory!(
      NumBytes:Integer64=@n5
      UnitStandard:EMemoryUnitStandard Enum=SI
      bUseGrouping:bool=true
      MinimumIntegralDigits:int=1
      MaximumIntegralDigits:int=324
      MinimumFractionalDigits:int=0
      MaximumFractionalDigits:int=2
     ) -> (ReturnValue:Text)
    ) -> (Result)
    self:UI Kit Dialog Viewmodel Object Reference=@n4.DialogVM
   ) -> (then, Output_Get:Text)
   > Return Node(ReturnValue=@n4.DialogVM)
   n6 = Push Simple UIKit Modal Dialog(
    execute=@n2.Error
    Context=Self-Reference()
    VM Class=/FortniteCoreUI/Frontend/Dialogs/DialogVM_Lobby_DownloadFailed.DialogVM_Lobby_DownloadFailed_C
    OnDialogButtonClicked=Create Event()
   )
   > Fire Event Main View Opened(View=@n6, bIsOverlay=true)
   > Return Node(ReturnValue=@n6.DialogVM)
  ]}
 BulkCosmeticsCellularConfirmDialog_ButtonClicked {nodes [
   n1 = BulkCosmeticsCellularConfirmDialog_ButtonClicked!() -> (then, ButtonIndex:int) {is=title K2Node_FunctionEntry}
   > FireEventElementInteractedByName!(
    execute
    Element Name:String=Select(Option 0:String=CellularNotAllowedModal_Continue_Button, Option 1:String=CellularNotAllowedModal_CancelDownloads_Button, Index:int=@n1) -> (ReturnValue:String)
    WidgetCallingFrom:Widget Object Reference=Self-Reference()
   ) -> (then, Success:bool)
   > n2 = Switch on Int!(execute:Exec, Selection:int=@n1) -> (Default:Exec, 0:Exec, 1:Exec)
   n3 = BulkCosmeticsAcceptCellularDownload(execute=@n2.0, self:Athena UIState Widget Frontend Object Reference)
   BulkCosmeticsCellularConfirmDialog_Closed(execute=[@n3.then, @n4.then])
   n4 = BulkCosmeticsRejectCellularDownload(execute=@n2.1, self:Athena UIState Widget Frontend Object Reference)
  ]}
 BulkCosmeticsCellularConfirmDialog_Closed {nodes [
   BulkCosmeticsCellularConfirmDialog_Closed() {is=title K2Node_FunctionEntry}
   > n1 = Cast To FortPlayerController(Object=GetOwningPlayer())
   > "Fire Event Sub View Opened (Non-Localized)"!(
    execute
    FortPC:Fort Player Controller Object Reference=@n1
    ViewName:String
    AdditionalInfo:Flow Analytics Additional Info Structure
    bIsOverlay:bool=false
   ) -> (then)
  ]}
 DownloadFailedDialog_CellularNotAllowed_ButtonClicked {nodes [
   n1 = DownloadFailedDialog_CellularNotAllowed_ButtonClicked!() -> (then, ButtonIndex:int) {is=title K2Node_FunctionEntry}
   > FireEventElementInteractedByName(
    Element Name=Select(Option 0=CellularNotAllowedModal_Continue_Button, Option 1=CellularNotAllowedModal_CancelDownloads_Button, Index=@n1)
    WidgetCallingFrom=Self-Reference()
   )
   > n2 = Switch on Int(Selection=@n1)
   n3 = ResumeIslandDownload(execute=@n2.0, self:Athena UIState Widget Frontend Object Reference)
   DownloadFailedDialog_Closed(execute=[@n3.then, @n4.then])
   n4 = StopListeningForIslandDownloadConnectionResumed(execute=@n2.1, self:Athena UIState Widget Frontend Object Reference)
  ]}
 DownloadFailedDialog_Closed {nodes [
   DownloadFailedDialog_Closed() {is=title K2Node_FunctionEntry}
   > n1 = Cast To FortPlayerController(Object=GetOwningPlayer())
   > "Fire Event Sub View Opened (Non-Localized)"(FortPC=@n1, bIsOverlay=false)
  ]}
 DownloadFailedDialog_Error_ButtonClicked {nodes [
   DownloadFailedDialog_Error_ButtonClicked!() -> (then, ButtonIndex:int) {is=title K2Node_FunctionEntry}
   > FireEventElementInteractedByName(Element Name=DownloadFailedModal_OK_Button, WidgetCallingFrom=Self-Reference())
   > DownloadFailedDialog_Closed() {is=title K2Node_CallFunction}
  ]}
 DownloadFailedDialog_NotEnoughStorage_ButtonClicked {nodes [
   n1 = DownloadFailedDialog_NotEnoughStorage_ButtonClicked!() -> (then, ButtonIndex:int) {is=title K2Node_FunctionEntry}
   > FireEventElementInteractedByName(Element Name=NotEnoughStorageModal_GotIt_Button, WidgetCallingFrom=Self-Reference())
   > n2 = Switch on Int(Selection=@n1)
   DownloadFailedDialog_Closed(execute=[@n2.0, @n2.1])
   Reroute Node(InputPin:String=Select(Option 0=NotEnoughStorageModal_DeviceStorage_Button, Option 1=NotEnoughStorageModal_Cancel_Button, Index=@n1)) -> (OutputPin:String)
  ]}
 EventGraph {nodes [
   "On Clicked (DecoyButton_MobileMicToggle)"!() -> (OutputDelegate, then, Button:Common Button Base Object Reference) {is=title K2Node_ComponentBoundEvent}
   > n1 = Branch(Condition=Get bSupportsTouchHudMicButton!(self:Fort Runtime Options Object Reference=GetRuntimeOptions()) -> (bSupportsTouchHudMicButton:bool))
   n2 = Get VoiceChannelManagementVM() -> (VoiceChannelManagementVM:Voice Channel Management VM Object Reference)
   > SetSelfIsMuted!(
    execute
    self:Voice Channel Management VM Object Reference=@n2
    bMuteSelf:bool=NOT Boolean(A=GetSelfIsMuted!(self:Voice Channel Management VM Object Reference=@n2) -> (ReturnValue:bool))
   ) -> (then)
   n3 = Get AthenaHUDContext!() -> (ReturnValue:Athena HUDContext Object Reference) {is=title K2Node_GetBlueprintContextEx}
   SetPTTState!(
    execute=@n1.else
    self:Athena HUDContext Object Reference=@n3
    NewPushToTalkState:EPTTState Enum=Select!(
     Enabled:EPTTState Enum=MicDisabled
     MicDisabled:EPTTState Enum=Enabled
     Index:EPTTState Enum=GetPTTState!(self:Athena HUDContext Object Reference=@n3) -> (ReturnValue:EPTTState Enum=Enabled)
    ) -> (ReturnValue:EPTTState Enum)
   ) -> (then)
   n4 = OnOpenTextChat!() -> (OutputDelegate, then, IsTextChatOpen:bool=~) {is=title K2Node_CustomEvent}
   > n5 = Set IsChatOpen(IsChatOpen:bool=@n4.IsTextChatOpen) -> (Output_Get:bool)
   > Branch(Condition=IsUsingCompactUI())
   > SetChatFullscreenMobile!(execute, self, ShouldChatBeFullscreen:bool=@n5) -> (then)
   n6 = InputChanged!() -> (OutputDelegate, then, bNewInputType:ECommonInputType Enum) {is=title K2Node_CustomEvent}
   > n7 = Set IsTouch(IsTouch:bool=Select!(MouseAndKeyboard:bool=false, Gamepad:bool=false, Touch:bool=true, Count:bool=false, Index:ECommonInputType Enum=@n6.bNewInputType) -> (ReturnValue:bool)) -> (Output_Get:bool)
   n8 = Get AthenaHUDContext()
   > SetBottomButtonVisibility!(
    execute
    self:WBP Bottom Bar Decoy Button Object Reference=Get DecoyButton_MobileMicToggle() -> (DecoyButton_MobileMicToggle:WBP Bottom Bar Decoy Button Object Reference)
    ShouldButtonBeVisible:bool=AND Boolean(A=@n7, B=GetCanPTT!(self:Athena HUDContext Object Reference=@n8) -> (ReturnValue:bool))
   ) -> (then)
   OnToggleMainMenu() {is=title K2Node_CustomEvent}
   > n9 = Branch(Condition=IsActivated!(self:Common Activatable Widget Object Reference=Get DualBladeMenu_Widget() -> (DualBladeMenu_Widget:Fort Dual Blade Menu Object Reference)) -> (ReturnValue:bool))
   n10 = Get DualBladeMenu_Widget()
   > DeactivateWidget(self=@n10)
   ActivateWidget(execute=@n9.else, self=@n10)
   n11 = OnUIKitDialogModalStateChanged!() -> (OutputDelegate, then, IsUIKitModalActive:bool=~) {is=title K2Node_CustomEvent}
   > n12 = Branch(Condition=@n11.IsUIKitModalActive)
   > n13 = Cast To TopBar!(execute, Object=Get TopBar_Widget() -> (TopBar_Widget:Athena Top Bar Object Reference)) -> (then, CastFailed, AsTop Bar:Top Bar Object Reference) {is=title K2Node_DynamicCast}
   > SetBumperInputEnabled!(execute, self:Top Bar Object Reference=@n13, AreBumpersEnabled:bool=false) -> (then)
   n14 = Cast To TopBar(execute=@n12.else, Object=Get TopBar_Widget())
   > SetBumperInputEnabled(self=@n14, AreBumpersEnabled=true)
   n15 = On Can PTT() {is=title K2Node_CustomEvent}
   > UpdateMicIconState!(execute, self, IsMicOn:EPTTState Enum=GetPTTState(self=Get AthenaHUDContext()) -> (ReturnValue=Enabled)) -> (then)
   n16 = Event DisplayDebugFlowText!() -> (OutputDelegate, then, DebugFlowText:"Text (by ref)") {is=title K2Node_Event}
   n17 = Get CommonTextBlock_FrontendFlowDebug() -> (CommonTextBlock_FrontendFlowDebug:Common Text Object Reference)
   > "SetText (Text)"(self=@n17, InText=@n16.DebugFlowText)
   > SetVisibility(self=@n17, InVisibility=HitTestInvisible)
   n18 = Event OnSetFrontendVisibilityMode!() -> (OutputDelegate, then, InFrontendVisibilityMode:EFrontendVisibilityMode Enum=Normal) {is=title K2Node_Event}
   > n19 = Set FrontendVisibilityMode(FrontendVisibilityMode:EFrontendVisibilityMode Enum=@n18.InFrontendVisibilityMode) -> (Output_Get:EFrontendVisibilityMode Enum=Normal)
   > n20 = Switch on EFrontendVisibilityMode!(execute, Selection:EFrontendVisibilityMode Enum=@n19) -> (Normal:Exec, HideTopTabsOnly:Exec, HideTopTabsOnlyWithoutBottomBar:Exec, OnlyBottom:Exec, OnlyTop:Exec, OnlyTitleBar:Exec, Empty:Exec) {is=title K2Node_SwitchEnum}
   n21 = Get TopBar_Widget()
   SetVisibility(
    execute=@n20.HideTopTabsOnly
    self=[
     Get SafeZone_Bottom() -> (SafeZone_Bottom:Safe Zone Object Reference)
     Get BottomBar() -> (BottomBar:Common User Widget Object Reference)
     @n21
    ]
    InVisibility=SelfHitTestInvisible
   )
   n22 = Get TopBar_Widget()
   > ActivateWidget(self=@n22)
   > n23 = Cast To TopBar(Object=@n22)
   > SetTopTabsVisibility!(execute, self:Top Bar Object Reference=@n23, NewVisibility:ESlateVisibility Enum=Hidden) -> (then)
   n24 = Get BottomBar()
   SetVisibility(execute=@n20.HideTopTabsOnlyWithoutBottomBar, self=[Get SafeZone_Bottom(), @n24], InVisibility=Collapsed)
   n25 = Get TopBar_Widget()
   > SetVisibility(self=@n25, InVisibility=SelfHitTestInvisible)
   > ActivateWidget(self=@n22)
   > n26 = Cast To TopBar(Object=@n22)
   > SetTopTabsVisibility(self=@n26, NewVisibility=Hidden)
   SetVisibility(execute=@n20.OnlyBottom, self=Get SafeZone_Bottom(), InVisibility=Collapsed)
   > SetVisibility(self=@n24, InVisibility=SelfHitTestInvisible)
   > n27 = DeactivateWidget(self=@n25)
   n28 = Cast To TopBar(execute=[@n27.then, @n35.then, @n34.then, @n29.then], Object=Get TopBar_Widget())
   > SetTopTabsVisibility(self=@n28, NewVisibility=SelfHitTestInvisible)
   > SetMainNavVisibility!(execute, self:Top Bar Object Reference=@n28, NewVisibility:ESlateVisibility Enum=SelfHitTestInvisible) -> (then)
   SetVisibility(execute=@n20.OnlyTop, self=Get TopBar_Widget(), InVisibility=SelfHitTestInvisible)
   > SetVisibility(self=@n24, InVisibility=Collapsed)
   > n29 = ActivateWidget(self=@n25)
   n30 = Get SafeZone_Bottom()
   n31 = Get BottomBar()
   SetVisibility(execute=@n20.OnlyTitleBar, self=[@n30, @n31], InVisibility=Collapsed)
   > SetVisibility(self=@n25, InVisibility=SelfHitTestInvisible)
   n32 = Get TopBar_Widget()
   > ActivateWidget(self=@n32)
   > n33 = Cast To TopBar(Object=@n32)
   > SetMainNavVisibility(self=@n33, NewVisibility=Hidden)
   SetVisibility(execute=@n20.Normal, self=[Get BottomBar(), Get SafeZone_Bottom()], InVisibility=SelfHitTestInvisible)
   > n34 = ActivateWidget(self=@n21)
   SetVisibility(execute=@n20.Empty, self=[@n30, @n31], InVisibility=Collapsed)
   > n35 = DeactivateWidget(self=@n25)
   Event StopTrailerTransitionAudioCheck() {is=title K2Node_Event}
   > n36 = Set Timer by Event(Delegate=Create Event(), Time=0, bLooping=false)
   > Add(
    TargetArray:Array of Timer Handle Structures=Get AudioTimerHandles() -> (AudioTimerHandles:Array of Timer Handle Structures)
    NewItem:"Timer Handle Structure (by ref)"=@n36
   ) {is=title K2Node_CallArrayFunction}
   Event Construct()
   n37 = Create Event()
   n38 = Get WBP_Wrapper_Text_Chat!(self:WBP Text Chat Input Listener Object Reference=Get WBP_TextChat_InputListener() -> (WBP_TextChat_InputListener:WBP Text Chat Input Listener Object Reference)) -> (WBP_Wrapper_Text_Chat:WBP Wrapper Text Chat Object Reference)
   > Bind Event to Open Text Chat(self:WBP Wrapper Text Chat Object Reference=@n38, Delegate=@n37) {is=title K2Node_AddDelegate}
   > n39 = Cast To FortPlayerController(Object=GetOwningPlayer())
   > Bind Event to On Toggle Main Menu Delegate(self:Fort Player Controller Object Reference=@n39, Delegate=Create Event()) {is=title K2Node_AddDelegate}
   > SetChatButtonVisibility!(execute, self, ShouldButtonBeVisible:bool=true) -> (then)
   > SetVisibility(
    self=Get WBP_TextChat_InputListener()
    InVisibility=Select!(
     Option 0:ESlateVisibility Enum=Collapsed
     Option 1:ESlateVisibility Enum=SelfHitTestInvisible
     Index:bool=Get IsUseMessageWidgetV3() -> (IsUseMessageWidgetV3:bool)
    ) -> (ReturnValue:ESlateVisibility Enum)
   )
   n40 = Get AthenaHUDContext()
   > Bind Event to On Can PTTChange(self:Athena HUDContext Object Reference=@n40, Delegate=@n15) {is=title K2Node_AddDelegate}
   n41 = Create Event()
   > Bind Event to On PTTState Change(self:Athena HUDContext Object Reference=@n40, Delegate=@n41) {is=title K2Node_AddDelegate}
   > UpdateMicIconState(IsMicOn=GetPTTState(self=Get AthenaHUDContext()) -> (ReturnValue=Enabled))
   n42 = Create Event()
   n43 = Get CommonInputSubsystem()
   > Bind Event to On Input Method Changed(self=@n43, Delegate=@n42)
   > SetMobileStyle() {is=title K2Node_CallFunction}
   > Initialize(self:Voice Channel Management VM Object Reference=Get VoiceChannelManagementVM())
   > n44 = StartListeningToEvent!(
    execute:Exec
    Target:Gameplay Event Router Component Object Reference=GetGlobalEventRouter()
    EventType:Script Struct Object Reference=/FortUILibrary/Systems/Dialog/BPS_UIKitDialog_ModalStateChange
    Context:Object Reference=GetOwningPlayer()
    EventBubblingRule:EEventBubblingRule Enum
   ) -> (
    then:Exec
    AsyncTaskProxy:Async Action Start Listening To Event Object Reference
    OnEventReceived:Exec
    Payload:BPS UIKit Dialog Modal State Change Structure
   )
   > Set UIKitDialogOpenorCloseEventRef(UIKitDialogOpenorCloseEventRef:Async Action Start Listening To Event Object Reference=@n44) -> (Output_Get:Async Action Start Listening To Event Object Reference)
   OnUIKitDialogModalStateChanged!(
    execute=@n44.OnEventReceived
    self
    IsUIKitModalActive:bool=Break BPS UIKit Dialog Modal State Change!(BPS_UIKitDialog_ModalStateChange:"BPS UIKit Dialog Modal State Change Structure (by ref)"=@n44.Payload) -> (bActive_4_#1:bool=~) {is=title K2Node_BreakStruct}
   ) -> (then) {is=title K2Node_CallFunction}
   n45 = Event StartTrailerTransitionAudioCheck!() -> (OutputDelegate, then, MediaPlayer:Media Player Object Reference) {is=title K2Node_Event}
   > n46 = Set MovieMediaPlayer(MovieMediaPlayer:Media Player Object Reference=@n45.MediaPlayer) -> (Output_Get:Media Player Object Reference)
   > Bind Event to On Media Opened(self:Media Player Object Reference=@n46, Delegate=Create Event()) {is=title K2Node_AddDelegate}
   > Bind Event to On End Reached(self:Media Player Object Reference=@n46, Delegate=Create Event()) {is=title K2Node_AddDelegate}
   Event HideDebugFlowText() {is=title K2Node_Event}
   > SetVisibility(self=Get CommonTextBlock_FrontendFlowDebug(), InVisibility=Collapsed)
   Event OnInitialized()
   > SetIsMessageWidgetV3()
   n47 = Get IsUseMessageWidgetV3()
   > n48 = Branch(Condition=AND Boolean(A=IsMobileGame(), B=NOT Boolean(A=@n47)))
   SetChatWidgetAlignment(execute=@n48.else)
   > ActivateWidget(self=Get WBP_Wrapper_Text_Chat(self=Get WBP_TextChat_InputListener()))
   n49 = Get CommonInputSubsystem!() -> (ReturnValue:Common Input Object Reference) {is=title K2Node_GetBlueprintContextEx}
   > Set IsTouch(IsTouch="Equal (Enum)"(A=GetCurrentInputType(self=@n49), B=Touch))
   Event Destruct()
   > Unbind Event from Open Text Chat(self:WBP Wrapper Text Chat Object Reference=@n38, Delegate=@n37) {is=title K2Node_RemoveDelegate}
   > Unbind Event from On Input Method Changed(self=Get CommonInputSubsystem() {is=title K2Node_GetSubsystem}, Delegate=Create Event())
   > Unbind Event from On Can PTTChange(self:Athena HUDContext Object Reference=@n40, Delegate=@n15) {is=title K2Node_RemoveDelegate}
   > Unbind Event from On PTTState Change(self:Athena HUDContext Object Reference=@n40, Delegate=@n41) {is=title K2Node_RemoveDelegate}
   > Unbind Event from On Input Method Changed(self=@n43, Delegate=@n42)
   > StopListeningToEvent(self:Async Action Start Listening To Event Object Reference=Get UIKitDialogOpenorCloseEventRef() -> (UIKitDialogOpenorCloseEventRef:Async Action Start Listening To Event Object Reference))
   n50 = Event OnSetTextChatEnabled!() -> (OutputDelegate, then, bEnabled:bool) {is=title K2Node_Event}
   > n51 = Branch(Condition=@n50.bEnabled)
   n52 = MakeLiteralGameplayTag!(Value:Gameplay Tag Structure=(TagName="UI.TextChat.Disabled")) -> (ReturnValue:Gameplay Tag Structure)
   n53 = Get EnhancedInputLocalPlayerSubsystem() -> (ReturnValue:Enhanced Input Local Player Subsystem Object Reference) {is=title K2Node_GetSubsystem}
   > RemoveTagFromInputMode!(
    execute
    self:Enhanced Input Subsystem Interface Interface=@n53
    TagToRemove:"Gameplay Tag Structure (by ref)"=@n52
    Options:"Modify Context Options Structure (by ref)"=(bIgnoreAllPressedKeysUntilRelease=True,bForceImmediately=False,bNotifyUserSettings=False)
   ) -> (then)
   n54 = Get WBP_TextChat_InputListener()
   n55 = Get WBP_Wrapper_Text_Chat(self=@n54)
   > ActivateWidget(self=@n55)
   AddTagToInputMode!(
    execute=@n51.else
    self:Enhanced Input Subsystem Interface Interface=@n53
    TagToAdd:"Gameplay Tag Structure (by ref)"=@n52
    Options:"Modify Context Options Structure (by ref)"=(bIgnoreAllPressedKeysUntilRelease=True,bForceImmediately=False,bNotifyUserSettings=False)
   ) -> (then)
   > OnToggleTextChatRequested!(execute, self:WBP Text Chat Input Listener Object Reference=@n54, ForceClose:bool=true) -> (then)
   > DeactivateWidget(self=@n55)
   n56 = Event BP_OnPartyMemberIdleStateChanged!() -> (OutputDelegate, then, IdleState:EPartyMemberIdleState Enum=Active) {is=title K2Node_Event}
   > n57 = Switch on EPartyMemberIdleState!(execute, Selection:EPartyMemberIdleState Enum=@n56.IdleState) -> (Active:Exec, Idle:Exec) {is=title K2Node_SwitchEnum}
   n58 = QueueModalWidgetClassAsync!(
    execute:Exec=@n57.Idle
    WidgetClass:Common Activatable Widget Class Reference=&/WBP_UI_Lobby_Dialog_PlayerIdle.WBP_UI_Lobby_Dialog_PlayerIdle_C
   ) -> (then:Exec, OnModalWidgetQueued:Exec, ActivatableWidget:Common Activatable Widget Object Reference) {is=title K2Node_AsyncAction}
   n59 = Cast To WBP_UI_Lobby_Dialog_PlayerIdle!(execute=@n58.OnModalWidgetQueued, Object=@n58) -> (then, CastFailed, AsWBP UI Lobby Dialog Player Idle:WBP UI Lobby Dialog Player Idle Object Reference) {is=title K2Node_DynamicCast}
   > n60 = Construct Object from Class(Class=/FortniteCoreUI/Frontend/Dialogs/DialogVM_PlayerIdle.DialogVM_PlayerIdle_C) -> (ReturnValue:Dialog VM Player Idle Object Reference)
   > SetUIKitDialogViewModel!(
    execute
    self:WBP UIKit Dialog Internal Object Reference=Get WBP_UIKit_Dialog_Internal!(self:WBP UI Lobby Dialog Player Idle Object Reference=@n59) -> (WBP_UIKit_Dialog_Internal:WBP UIKit Dialog Internal Object Reference)
    Viewmodel:UI Kit Dialog Viewmodel Object Reference=@n60
   ) -> (then)
  ]}
 OnTrailerMediaEnded {nodes [
   OnTrailerMediaEnded() {is=title K2Node_FunctionEntry}
   n1 = Get MovieMediaPlayer() -> (MovieMediaPlayer:Media Player Object Reference)
   > Unbind all Events from On Media Opened!(execute:Exec, self:Media Player Object Reference=@n1) -> (then:Exec) {is=title K2Node_ClearDelegate}
   > Unbind all Events from On End Reached!(execute:Exec, self:Media Player Object Reference=@n1) -> (then:Exec) {is=title K2Node_ClearDelegate}
   > n2 = For Each Loop(Array:Array of Timer Handle Structures=Get AudioTimerHandles()) -> (Array Element:Timer Handle Structure)
   > Clear and Invalidate Timer by Handle(Handle=@n2)
  ]}
 OnTrailerMediaOpened {nodes [
   OnTrailerMediaOpened!() -> (then, OpenedUrl:String) {is=title K2Node_FunctionEntry}
   > Unbind all Events from On Media Opened(self=Get MovieMediaPlayer())
   > n1 = Set Timer by Event(Delegate=Create Event(), Time=0.1, bLooping=true)
   > Add(TargetArray=Get AudioTimerHandles(), NewItem=@n1)
   n2 = GetRuntimeOptions()
   n3 = Get TransitionAudioData() -> (TransitionAudioData:Array of S12 Cine Transition Data Structures)
   > Add!(
    execute
    TargetArray:Array of S12 Cine Transition Data Structures=@n3
    NewItem_Asset_5_#2:Sound Base Object Reference=Get TransitionSound01() -> (TransitionSound01:Sound Base Object Reference)
    NewItem_Offset_8_#3:float=float - float(
     A=Get FTUESeasonTrailerBoundary!(self:Fort Runtime Options Object Reference=@n2) -> (FTUESeasonTrailerBoundary:float)
     B=Get VideoDurationOffsetFromTransition!(self:Fort Runtime Options Object Reference=@n2) -> (VideoDurationOffsetFromTransition:float)
    ) {is=title K2Node_CallFunction}
   ) -> (then, ReturnValue)
   n4 = GetDuration!(self:Media Player Object Reference=Get MovieMediaPlayer()) -> (ReturnValue:Timespan Structure)
   n5 = GetRuntimeOptions()
   > Add(
    TargetArray=@n3
    NewItem_Asset_5_#2=Get TransitionSound02() -> (TransitionSound02:Sound Base Object Reference)
    NewItem_Offset_8_#3=float - float(
     A=Get Total Seconds!(A:Timespan Structure=@n4) -> (ReturnValue:double) {is=title K2Node_CallFunction}
     B=Get VideoDurationOffsetFromEnd!(self:Fort Runtime Options Object Reference=@n5) -> (VideoDurationOffsetFromEnd:float)
    )
   )
   > Set TransitionIndex(TransitionIndex:int=1) -> (Output_Get:int=0)
   > OnTrailerTransitionAudioCheck()
  ]}
 OnTrailerTransitionAudioCheck {nodes [
   OnTrailerTransitionAudioCheck() {is=title K2Node_FunctionEntry}
   > n1 = Get!(execute:Exec) -> (then:Exec, else:Exec, MovieMediaPlayer:Media Player Object Reference)
   n2 = GetTime!(self:Media Player Object Reference=@n1) -> (ReturnValue:Timespan Structure)
   n3 = "Get (a ref)"!(Array:Array of S12 Cine Transition Data Structures=Get TransitionAudioData(), Dimension 1=Get TransitionIndex() -> (TransitionIndex:int=0)) -> (Output_Asset_5_#2:Sound Base Object Reference, Output_Offset_8_#3:float=~)
   > Branch(Condition="float >= float"(A=Get Total Seconds(A=@n2), B=@n3.Output_Offset_8_#3) {is=title K2Node_CallFunction})
   > PlaySound2D(Sound=@n3)
   > n4 = Set TransitionIndex(TransitionIndex=int + int(A=Get TransitionIndex() -> (TransitionIndex=0), B=1) -> (ReturnValue=0) {is=title K2Node_CommutativeAssociativeBinaryOperator}) -> (Output_Get=0)
   n5 = Get TransitionAudioData()
   > Branch(Condition="Equal (Integer)"(A=@n4, B=Length(TargetArray:Array of S12 Cine Transition Data Structures=@n5)) {is=title K2Node_CallFunction})
   > n6 = Set Timer by Event(Delegate=Create Event(), Time=0, bLooping=false)
   > Add!(execute, TargetArray:Array of Timer Handle Structures=Get AudioTimerHandles(), NewItem:"Timer Handle Structure (by ref)"=@n6) -> (then, ReturnValue)
  ]}
 PlayTrailer {nodes [PlayTrailer() {is=title K2Node_FunctionEntry}]}
 SetChatButtonVisibility {nodes [
   n1 = SetChatButtonVisibility!() -> (then, ShouldButtonBeVisible:bool=~) {is=title K2Node_FunctionEntry}
   > n2 = Branch(Condition=Get IsUseMessageWidgetV3())
   > SetBottomBarButtonsVisible!(execute, self:WBP Wrapper Text Chat Object Reference=Get WBP_Wrapper_Text_Chat(self=Get WBP_TextChat_InputListener()), Visible:bool=@n1) -> (then)
   SetBottomButtonVisibility(
    execute=@n2.else
    self=Get Button_HoldToChat() -> (Button_HoldToChat:WBP Bottom Bar Decoy Button Object Reference)
    ShouldButtonBeVisible=@n1
   )
  ]}
 SetChatFullscreenMobile {nodes [
   n1 = SetChatFullscreenMobile!() -> (then, ShouldChatBeFullscreen:bool=~) {is=title K2Node_FunctionEntry}
   > n2 = Branch(Condition=@n1)
   n3 = Get SafeZone_Bottom()
   > n4 = SetSidesToPad!(execute, self:Safe Zone Object Reference=@n3, InPadLeft:bool=false, InPadRight:bool=false, InPadTop:bool=false, InPadBottom:bool=false) -> (then)
   n5 = SlotAsOverlaySlot(Widget=@n3)
   SetPadding!(
    execute=[@n4.then, @n12.then]
    self:Overlay Slot Object Reference=@n5
    InPadding_Left:float=0
    InPadding_Top:float=0
    InPadding_Right:float=0
    InPadding_Bottom:float=0
   ) -> (then)
   > SetVerticalAlignment(
    self=@n5
    InVerticalAlignment=Select(Option 0:EVerticalAlignment Enum=VAlign_Bottom, Option 1:EVerticalAlignment Enum=VAlign_Fill, Index=@n1) -> (ReturnValue:EVerticalAlignment Enum)
   )
   > SetHorizontalAlignment(
    self=@n5
    InHorizontalAlignment=Select(Option 0:EHorizontalAlignment Enum=HAlign_Right, Option 1:EHorizontalAlignment Enum=HAlign_Fill, Index=@n1) -> (ReturnValue:EHorizontalAlignment Enum)
   )
   n6 = Get VerticalBox_Bottom() -> (VerticalBox_Bottom:Vertical Box Object Reference)
   > SetSize!(
    execute
    self:Horizontal Box Slot Object Reference=SlotAsHorizontalBoxSlot!(Widget:Widget Object Reference=@n6) -> (ReturnValue:Horizontal Box Slot Object Reference)
    InSize_Value:float=1
    InSize_SizeRule:ESlateSizeRule Enum=Select(Option 0:ESlateSizeRule Enum=Automatic, Option 1:ESlateSizeRule Enum=Fill, Index=@n1) -> (ReturnValue:ESlateSizeRule Enum)
   ) -> (then)
   > SetRenderTranslation(self=@n6, Translation=Select(Option 0:Vector 2D Structure, Option 1:Vector 2D Structure, Index=@n1) -> (ReturnValue:Vector 2D Structure))
   n7 = Get Overlay_ChatPanel() -> (Overlay_ChatPanel:Overlay Object Reference)
   n8 = SlotAsVerticalBoxSlot(Widget=@n7)
   > SetSize(
    self:Vertical Box Slot Object Reference=@n8
    InSize_Value=1
    InSize_SizeRule=Select(Option 0:ESlateSizeRule Enum=Automatic, Option 1:ESlateSizeRule Enum=Fill, Index=@n1) -> (ReturnValue:ESlateSizeRule Enum)
   )
   n9 = IsUsingInputType!(InputType:ECommonInputType Enum=Gamepad, PlayerController:Player Controller Object Reference=GetOwningPlayer()) -> (ReturnValue:bool)
   > SetPadding!(
    execute
    self:Vertical Box Slot Object Reference=@n8
    InPadding:Margin Structure=Select!(
     Option 0:Margin Structure=Get P_MobilePosition() -> (P_MobilePosition:Margin Structure)
     Option 1_Left:float=0
     Option 1_Top:float=SelectFloat(A=48, B=24, bPickA=@n9)
     Option 1_Right:float=24
     Option 1_Bottom:float=0
     Index=@n1
    ) -> (ReturnValue:Margin Structure)
   ) -> (then)
   > SetRenderTranslation(self=@n7)
   > SetVerticalAlignment(
    self=SlotAsOverlaySlot(Widget=Get WBP_TextChat_InputListener())
    InVerticalAlignment=Select!(Option 0:EVerticalAlignment Enum=VAlign_Top, Option 1:EVerticalAlignment Enum=VAlign_Fill, Index=@n1) -> (ReturnValue:EVerticalAlignment Enum)
   )
   n10 = SlotAsOverlaySlot(Widget=Get WBP_TextChat_InputListener())
   > SetVerticalAlignment(self=@n10, InVerticalAlignment=Select(Option 0=VAlign_Top, Option 1=VAlign_Fill, Index=@n1))
   > SetHorizontalAlignment(
    self=@n10
    InHorizontalAlignment=Select(Option 0:EHorizontalAlignment Enum=HAlign_Left, Option 1:EHorizontalAlignment Enum=HAlign_Fill, Index=@n1) -> (ReturnValue:EHorizontalAlignment Enum)
   )
   > n11 = Branch(Condition=Get ShouldChatBeFullscreen() -> (ShouldChatBeFullscreen:bool))
   > SetGridSlotAlignment!(
    execute
    Widget:Widget Object Reference=Get WBP_Wrapper_Text_Chat(self=Get WBP_TextChat_InputListener())
    Horizontal:EHorizontalAlignment Enum=HAlign_Fill
    Vertical:EVerticalAlignment Enum=VAlign_Fill
   ) -> (then)
   SetGridSlotAlignment(execute=@n11.else, Widget=Get WBP_Wrapper_Text_Chat(self=Get WBP_TextChat_InputListener()), Horizontal=HAlign_Left, Vertical=VAlign_Bottom)
   n12 = SetSidesToPad(execute=@n2.else, self=@n3, InPadLeft=true, InPadRight=true, InPadTop=false, InPadBottom=true)
  ]}
 SetChatWidgetAlignment {nodes [
   SetChatWidgetAlignment() {is=title K2Node_FunctionEntry}
   n1 = SlotAsOverlaySlot(Widget=Get SafeZone_Bottom())
   n2 = GetIsDiscoverInLobbyEnabled!(self:Fort Runtime Options Object Reference=GetRuntimeOptions(), ContextWidget:Widget Object Reference=Self-Reference()) -> (ReturnValue:bool)
   > SetPadding(
    self:Overlay Slot Object Reference=@n1
    InPadding=Select(
     Option 0:Margin Structure=Make Margin(Left=16, Top=0, Right=0, Bottom=16)
     Option 1:Margin Structure=Make Margin(Left=0, Top=0, Right=80, Bottom=96)
     Index=@n2
    ) -> (ReturnValue:Margin Structure)
   )
   > Set HorizontalAlignment!(
    execute
    HorizontalAlignment:EHorizontalAlignment Enum=Select(Option 0:EHorizontalAlignment Enum=HAlign_Left, Option 1:EHorizontalAlignment Enum=HAlign_Right, Index=@n2) -> (ReturnValue:EHorizontalAlignment Enum)
    self:Overlay Slot Object Reference=@n1
   ) -> (then, Output_Get:EHorizontalAlignment Enum=HAlign_Fill)
  ]}
 SetIsMessageWidgetV3 {nodes [
   SetIsMessageWidgetV3() {is=title K2Node_FunctionEntry}
   > n1 = Cast To FortPlayerController(Object=GetOwningPlayer())
   > Set IsUseMessageWidgetV3(IsUseMessageWidgetV3:bool=UseMessageWidgetV3!(self:Fort Player Controller Object Reference=@n1) -> (ReturnValue:bool)) -> (Output_Get:bool)
  ]}
 SetMobileStyle {nodes [
   SetMobileStyle() {is=title K2Node_FunctionEntry}
   > Branch(Condition=IsMobileGame())
   n1 = Get SafeZone_Bottom()
   > Set PadBottom!(execute, PadBottom:bool=true, self:Safe Zone Object Reference=@n1) -> (then, Output_Get:bool)
   > SetPadding!(
    execute
    self=SlotAsOverlaySlot(Widget=@n1)
    InPadding_Left:float=0
    InPadding_Top:float=0
    InPadding_Right:float=0
    InPadding_Bottom:float=0
   ) -> (then)
   n2 = Get SafeZone_BuildWatermark() -> (SafeZone_BuildWatermark:Safe Zone Object Reference)
   > Set PadTop!(execute, PadTop:bool=false, self:Safe Zone Object Reference=@n2) -> (then, Output_Get:bool)
   > Set PadBottom(PadBottom=false, self=@n2)
   n3 = SlotAsOverlaySlot(Widget=@n2)
   > SetHorizontalAlignment(self=@n3, InHorizontalAlignment=HAlign_Center)
   > SetVerticalAlignment(self=@n3, InVerticalAlignment=VAlign_Bottom)
   > SetPadding(self=@n3, InPadding_Left=0, InPadding_Top=0, InPadding_Right=0, InPadding_Bottom=120)
   > AddChild(
    self=Get Mobile_ScaleBox_BuildWatermark() -> (Mobile_ScaleBox_BuildWatermark:Scale Box Object Reference)
    Content=Get BuildWatermark() -> (BuildWatermark:Build Watermark Object Reference)
   )
   > SetPadding(
    self=SlotAsOverlaySlot(Widget=Get SafeZone_FPS() -> (SafeZone_FPS:Safe Zone Object Reference))
    InPadding_Left=20
    InPadding_Top=0
    InPadding_Right=0
    InPadding_Bottom=5
   )
   n4 = SlotAsOverlaySlot(Widget=Get SafeZone_Temperature() -> (SafeZone_Temperature:Safe Zone Object Reference))
   > SetHorizontalAlignment(self=@n4, InHorizontalAlignment=HAlign_Left)
   > SetVerticalAlignment(self=@n4, InVerticalAlignment=VAlign_Top)
   > SetPadding(self=@n4, InPadding_Left=20, InPadding_Top=130, InPadding_Right=0, InPadding_Bottom=5)
   > SetRenderTranslation(self=Get VerticalBox_Bottom(), Translation=(X=0,Y=-25))
   > SetPadding!(
    execute
    self:Vertical Box Slot Object Reference=SlotAsVerticalBoxSlot(Widget=Get Overlay_ChatPanel())
    InPadding:Margin Structure=Get P_MobilePosition()
   ) -> (then)
  ]}
 SetOverlayTitleBarVisibility {nodes [
   n1 = SetOverlayTitleBarVisibility!() -> (then, Visible:bool=~) {is=title K2Node_FunctionEntry}
   > n2 = Branch(Condition=@n1)
   > SetWindowTitleBarState!(
    execute
    TitleBarContent:Widget Object Reference
    Mode:EWindowTitleBarMode Enum=Overlay
    bTitleBarDragEnabled:bool=true
    bWindowButtonsVisible:bool=true
    bTitleBarVisible:bool=true
   ) -> (then)
   SetWindowTitleBarState(execute=@n2.else, Mode=Overlay, bTitleBarDragEnabled=false, bWindowButtonsVisible=false, bTitleBarVisible=false)
  ]}
 SetVoiceChannelSpeakerContainerVisibility {nodes [
   n1 = SetVoiceChannelSpeakerContainerVisibility!() -> (then, ShouldBeVisible:bool=~) {is=title K2Node_FunctionEntry}
   > SetVisibility(
    self=Get FortVoiceChannelSpeakerContainer() -> (FortVoiceChannelSpeakerContainer:Fort Voice Channel Speaker Container Object Reference)
    InVisibility=Select(Option 0:ESlateVisibility Enum=Collapsed, Option 1:ESlateVisibility Enum=SelfHitTestInvisible, Index=@n1) -> (ReturnValue:ESlateVisibility Enum)
   )
  ]}
 UpdateMicIconState {nodes [
   n1 = UpdateMicIconState!() -> (then, IsMicOn:EPTTState Enum=Enabled) {is=title K2Node_FunctionEntry}
   > Broadcast Event!(
    execute
    self:Gameplay Event Router Component Object Reference=GetGlobalEventRouter()
    EventData:BPS Bottom Bar Mic Toggle State Changed Structure=Make BPS Bottom Bar Mic Toggle State Changed!(IsMicOn_2_#4:bool=Select!(Enabled:bool=true, MicDisabled:bool=false, Index:EPTTState Enum=@n1) -> (ReturnValue:bool)) -> (BPS_BottomBarMicToggleStateChanged:BPS Bottom Bar Mic Toggle State Changed Structure) {is=title K2Node_MakeStruct}
    EventContext:Object Reference=GetOwningPlayer()
   ) -> (then, ReturnValue:bool)
   n2 = Get AthenaHUDContext()
   > SetBottomButtonVisibility(self=Get DecoyButton_MobileMicToggle(), ShouldButtonBeVisible=AND Boolean(A=GetCanPTT(self=@n2), B=Get IsTouch() -> (IsTouch:bool)))
  ]}
]
variables [
 AudioTimerHandles {type="array<TimerHandle>", default []}
 Button_HoldToChat {type=WBP_BottomBarDecoyButton_C, default=null}
 FrontendVisibilityMode {type=EFrontendVisibilityMode, default=Normal}
 IsChatOpen {type=boolean}
 IsTouch {type=boolean}
 IsUseMessageWidgetV3 {type=boolean}
 MovieMediaPlayer {type=MediaPlayer, default=null}
 P_MobilePosition {type=Margin, default=<ltrb 0 0 48 130>}
 StartedPlayingVideo {type=boolean}
 TransitionAudioData {type="array<S12_CineTransitionData>", default []}
 TransitionIndex {type=integer}
 TransitionSound01 {type=SoundBase, default=/Game/Sounds/Fort_Music/Seasonal/CineToBP_Transition_Cue}
 TransitionSound02 {type=SoundBase, default=/Game/Sounds/Fort_Music/Seasonal/BPToUpsell_Transition_Cue}
 UIKitDialogOpenorCloseEventRef {type=AsyncAction_StartListeningToEvent, default=null}
]
widgets {
 info {
  inheritedWidgetCount=5
  namedSlotCount=1
  parentClass=/Script/FortniteUI.AthenaUIStateWidget_Frontend
  rootWidgetClass=&Overlay
  widgetCount=33
 }
 widgets [
  Overlay_4 {widgetClassPath=&Overlay, widget=Overlay_4}
  Overlay_0 {
   widgetClassPath=&Overlay
   parent=Overlay_4
   widget=Overlay_0
   slot {class=OverlaySlot, properties {horizontalAlignment=HAlign_Fill, verticalAlignment=VAlign_Fill}}
  }
  WidgetStack_PrimaryContent {
   widgetClassPath=/Script/CommonUI.CommonActivatableWidgetStack
   bInherited=true
   parent=Overlay_0
   widget=WidgetStack_PrimaryContent
   properties {
    navigation=&self:WidgetTree.WidgetStack_PrimaryContent.WidgetNavigation_0
    rootContentWidgetClass=&/AthenaTabsScreen.AthenaTabsScreen_C
    transitionCurveType=CubicInOut
    transitionDuration=0
   }
   slot {class=OverlaySlot, properties {horizontalAlignment=HAlign_Fill, verticalAlignment=VAlign_Fill}}
  }
  TopBar_Widget {
   widgetClassPath=/Game/UI/Frontend/TopBar/TopBar
   bInherited=true
   parent=Overlay_0
   widget=TopBar_Widget
   properties {tabsInfo [
     {
      bForceImage=false
      bInteractAnalytic=false
      bOverride_MobileIconBrush=true
      bPrimaryPlayerOnly=true
      bShowInSplitscreenV2=true
      bangType=Invalid
      buttonGroup=Core
      calendarEventName=""
      displayName=PLAY
      forwardOnRemoval=Max_None
      linkedUIFeature=ShowPlay
      overrideTabWidget=null
      sortOrder=1000
      tabId=Lobby
      viewModel=null
      iconBrush {
       bIsDynamicallyLoaded=false
       drawAs=Image
       imageType=NoImage
       mirroring=NoMirror
       resourceName=None
       resourceObject=null
       tiling=NoTile
       imageSize=<64 64>
       margin=<ltrb 0 0 0 0>
       outlineSettings {
        bUseBrushTransparency=false
        roundingType=HalfHeightRadius
        width=0
        color {colorUseRule=UseColor_Specified, specifiedColor=<rgba 0 0 0 0>}
        cornerRadii=<0 0 0 1>
       }
       tintColor {colorUseRule=UseColor_Specified, specifiedColor=<rgba 1 1 1 1>}
       uVRegion {bIsValid=false, max=<0 0>, min=<0 0>}
      }
      mobileIconBrush {
       bIsDynamicallyLoaded=false
       drawAs=Image
       imageType=FullColor
       mirroring=NoMirror
       resourceName=None
       resourceObject=/FortUILibrary/Textures/Icons/T_UI_Icon_UIKit_Play_SDF
       tiling=NoTile
       imageSize=<32 32>
       margin=<ltrb 0 0 0 0>
       outlineSettings {
        bUseBrushTransparency=false
        roundingType=HalfHeightRadius
        width=0
        color {colorUseRule=UseColor_Specified, specifiedColor=<rgba 0 0 0 0>}
        cornerRadii=<0 0 0 1>
       }
       tintColor {colorUseRule=UseColor_Specified, specifiedColor=<rgba 1 1 1 1>}
       uVRegion {bIsValid=false, max=<0 0>, min=<0 0>}
      }
      persistentTabObjects []
     }
     {
      bForceImage=false
      bInteractAnalytic=true
      bOverride_MobileIconBrush=true
      bPrimaryPlayerOnly=true
      bShowInSplitscreenV2=false
      bangType=Invalid
      buttonGroup=Game
      calendarEventName=""
      displayName=COMPETE
      forwardOnRemoval=Max_None
      linkedUIFeature=ShowShowdown
      overrideTabWidget=null
      sortOrder=4000
      tabId=AthenaCompete
      viewModel=null
      iconBrush {
       bIsDynamicallyLoaded=false
       drawAs=Image
       imageType=NoImage
       mirroring=NoMirror
       resourceName=None
       resourceObject=null
       tiling=NoTile
       imageSize=<64 64>
       margin=<ltrb 0 0 0 0>
       outlineSettings {
        bUseBrushTransparency=false
        roundingType=HalfHeightRadius
        width=0
        color {colorUseRule=UseColor_Specified, specifiedColor=<rgba 0 0 0 0>}
        cornerRadii=<0 0 0 1>
       }
       tintColor {colorUseRule=UseColor_Specified, specifiedColor=<rgba 1 1 1 1>}
       uVRegion {bIsValid=false, max=<0 0>, min=<0 0>}
      }
      mobileIconBrush {
       bIsDynamicallyLoaded=false
       drawAs=Image
       imageType=FullColor
       mirroring=NoMirror
       resourceName=None
       resourceObject=/FortUILibrary/Textures/Icons/T_UI_Icon_UIKit_Compete_SDF
       tiling=NoTile
       imageSize=<32 32>
       margin=<ltrb 0 0 0 0>
       outlineSettings {
        bUseBrushTransparency=false
        roundingType=HalfHeightRadius
        width=0
        color {colorUseRule=UseColor_Specified, specifiedColor=<rgba 0 0 0 0>}
        cornerRadii=<0 0 0 1>
       }
       tintColor {colorUseRule=UseColor_Specified, specifiedColor=<rgba 1 1 1 1>}
       uVRegion {bIsValid=false, max=<0 0>, min=<0 0>}
      }
      persistentTabObjects []
     }
     {
      bForceImage=false
      bInteractAnalytic=true
      bOverride_MobileIconBrush=true
      bPrimaryPlayerOnly=false
      bShowInSplitscreenV2=false
      bangType=MultiProductItemShop
      buttonGroup=Core
      calendarEventName=""
      displayName=SHOP
      forwardOnRemoval=Max_None
      linkedUIFeature=ShowMultiProductItemShop
      overrideTabWidget=null
      sortOrder=6001
      tabId=MultiProductItemShop
      viewModel=null
      iconBrush {
       bIsDynamicallyLoaded=false
       drawAs=Image
       imageType=NoImage
       mirroring=NoMirror
       resourceName=None
       resourceObject=null
       tiling=NoTile
       imageSize=<64 64>
       margin=<ltrb 0 0 0 0>
       outlineSettings {
        bUseBrushTransparency=false
        roundingType=HalfHeightRadius
        width=0
        color {colorUseRule=UseColor_Specified, specifiedColor=<rgba 0 0 0 0>}
        cornerRadii=<0 0 0 1>
       }
       tintColor {colorUseRule=UseColor_Specified, specifiedColor=<rgba 1 1 1 1>}
       uVRegion {bIsValid=false, max=<0 0>, min=<0 0>}
      }
      mobileIconBrush {
       bIsDynamicallyLoaded=false
       drawAs=Image
       imageType=FullColor
       mirroring=NoMirror
       resourceName=None
       resourceObject=/FortUILibrary/Textures/Icons/T_UI_Icon_UIKit_Shop_SDF
       tiling=NoTile
       imageSize=<32 32>
       margin=<ltrb 0 0 0 0>
       outlineSettings {
        bUseBrushTransparency=false
        roundingType=HalfHeightRadius
        width=0
        color {colorUseRule=UseColor_Specified, specifiedColor=<rgba 0 0 0 0>}
        cornerRadii=<0 0 0 1>
       }
       tintColor {colorUseRule=UseColor_Specified, specifiedColor=<rgba 1 1 1 1>}
       uVRegion {bIsValid=false, max=<0 0>, min=<0 0>}
      }
      persistentTabObjects []
     }
     {
      bForceImage=false
      bInteractAnalytic=true
      bOverride_MobileIconBrush=true
      bPrimaryPlayerOnly=false
      bShowInSplitscreenV2=false
      bangType=Invalid
      buttonGroup=PinnedRight
      calendarEventName=""
      displayName=V-BUCKS
      forwardOnRemoval=Max_None
      linkedUIFeature=ShowMainStore
      overrideTabWidget=null
      sortOrder=9000
      tabId=AthenaStore
      viewModel=/Script/FortniteUI.FortMTXTabViewModel
      iconBrush {
       bIsDynamicallyLoaded=false
       drawAs=Image
       imageType=NoImage
       mirroring=NoMirror
       resourceName=None
       resourceObject=/Game/UI/Foundation/Textures/Icons/Navigation/t_icon_nav_mobile_vbuxstore_128
       tiling=NoTile
       imageSize=<64 64>
       margin=<ltrb 0 0 0 0>
       outlineSettings {
        bUseBrushTransparency=false
        roundingType=HalfHeightRadius
        width=0
        color {colorUseRule=UseColor_Specified, specifiedColor=<rgba 0 0 0 0>}
        cornerRadii=<0 0 0 1>
       }
       tintColor {colorUseRule=UseColor_Specified, specifiedColor=<rgba 1 1 1 1>}
       uVRegion {bIsValid=false, max=<0 0>, min=<0 0>}
      }
      mobileIconBrush {
       bIsDynamicallyLoaded=false
       drawAs=Image
       imageType=FullColor
       mirroring=NoMirror
       resourceName=None
       resourceObject=/FortUILibrary/Textures/Icons/T_UI_Icon_UIKit_VBuck_SDF
       tiling=NoTile
       imageSize=<64 64>
       margin=<ltrb 0 0 0 0>
       outlineSettings {
        bUseBrushTransparency=false
        roundingType=HalfHeightRadius
        width=0
        color {colorUseRule=UseColor_Specified, specifiedColor=<rgba 0 0 0 0>}
        cornerRadii=<0 0 0 1>
       }
       tintColor {colorUseRule=UseColor_Specified, specifiedColor=<rgba 1 1 1 1>}
       uVRegion {bIsValid=false, max=<0 0>, min=<0 0>}
      }
      persistentTabObjects []
     }
    ]}
   slot {class=OverlaySlot, properties {horizontalAlignment=HAlign_Fill}}
  }
  SafeZone_ContentOnDemand {
   widgetClassPath=&SafeZone
   bIsVariable=true
   parent=Overlay_0
   widget=SafeZone_ContentOnDemand
   properties {padBottom=false, padLeft=false}
   slot {class=OverlaySlot, properties {horizontalAlignment=HAlign_Right, padding=<ltrb 0 96 144 0>}}
  }
  WrapBox_0 {
   widgetClassPath=&WrapBox
   parent=SafeZone_ContentOnDemand
   widget=WrapBox_0
   properties {horizontalAlignment=HAlign_Right}
   slot {class=SafeZoneSlot, properties {hAlign=HAlign_Right, vAlign=VAlign_Top}}
  }
  WBP_ContentOnDemand {
   widgetClassPath=/FortniteCoreUI/UI/ContentOnDemand/WBP_ContentOnDemand
   bIsVariable=true
   parent=WrapBox_0
   widget=WBP_ContentOnDemand
   properties {visibility=Collapsed}
   slot {class=WrapBoxSlot, properties {horizontalAlignment=HAlign_Right, verticalAlignment=VAlign_Top, padding=<ltrb 0 0 0 4>}}
  }
  WBP_ContentOnDemand_IslandDownload {
   widgetClassPath=/FortniteCoreUI/UI/ContentOnDemand/WBP_ContentOnDemand_IslandDownload
   bIsVariable=true
   parent=WrapBox_0
   widget=WBP_ContentOnDemand_IslandDownload
   properties {visibility=Collapsed}
   slot {class=WrapBoxSlot, properties {horizontalAlignment=HAlign_Right, verticalAlignment=VAlign_Top}}
  }
  CommonBorder_175 {
   widgetClassPath=&CommonBorder
   parent=Overlay_0
   widget=CommonBorder_175
   properties {
    horizontalAlignment=HAlign_Left
    verticalAlignment=VAlign_Bottom
    background {
     bIsDynamicallyLoaded=false
     drawAs=Box
     imageType=NoImage
     mirroring=NoMirror
     resourceName=None
     resourceObject=null
     tiling=NoTile
     imageSize=<128 32>
     margin=<ltrb 0.1 0.7965 1 0.2611>
     outlineSettings {
      bUseBrushTransparency=false
      roundingType=HalfHeightRadius
      width=0
      color {colorUseRule=UseColor_Specified, specifiedColor=<rgba 0 0 0 0>}
      cornerRadii=<0 0 0 0>
     }
     tintColor {colorUseRule=UseColor_Specified, specifiedColor=<rgba 0.1667 0.1667 0.1667 1>}
     uVRegion {bIsValid=false, max=<0 0>, min=<0 0>}
    }
   }
   slot {class=OverlaySlot, properties {padding=<ltrb 0 0 0 200>}}
  }
  CommonTextBlock_FrontendFlowDebug {
   widgetClassPath=&CommonTextBlock
   bIsVariable=true
   parent=CommonBorder_175
   widget=CommonTextBlock_FrontendFlowDebug
   properties {
    text="DEBUG: Going through frontend flow \"default\""
    visibility=Collapsed
    font {
     bForceMonospaced=false
     bMaterialIsStencil=false
     fontMaterial=null
     fontObject=&BurbankSmall
     letterSpacing=0
     monospacedWidth=1
     size=14
     skewAmount=0
     typefaceFontName=Bold
     outlineSettings {
      bApplyOutlineToDropShadows=false
      bMiteredCorners=false
      bSeparateFillAlpha=false
      outlineMaterial=null
      outlineSize=0
      outlineColor=<rgba 0 0 0 1>
     }
    }
    shadowOffset=<0 0>
   }
   slot {class=BorderSlot, properties {horizontalAlignment=HAlign_Left, verticalAlignment=VAlign_Bottom}}
  }
  VerticalBox_0 {
   widgetClassPath=&VerticalBox
   parent=Overlay_0
   widget=VerticalBox_0
   slot {class=OverlaySlot, properties {horizontalAlignment=HAlign_Right}}
  }
  Spacer_110 {widgetClassPath=&Spacer, parent=VerticalBox_0, widget=Spacer_110, properties {size=<1 128>}, slot {class=VerticalBoxSlot}}
  BottomBar {
   widgetClassPath=/Game/UI/Frontend/BottomBar/WBP_AthenaBottomBarWidget
   bInherited=true
   bIsVariable=true
   parent=Overlay_4
   widget=BottomBar
   properties {display Owner Actions Only=false}
   slot {class=OverlaySlot, properties {horizontalAlignment=HAlign_Right, verticalAlignment=VAlign_Bottom}}
  }
  WBP_CaptureForPostBufferUpdate {widgetClassPath=&WBP_CaptureForPostBufferUpdate, bIsVariable=true, namedSlotHost=BottomBar, slot=~}
  SafeZone_BuildWatermark {
   widgetClassPath=&SafeZone
   bIsVariable=true
   parent=Overlay_4
   widget=SafeZone_BuildWatermark
   slot {class=OverlaySlot, properties {horizontalAlignment=HAlign_Right, verticalAlignment=VAlign_Bottom, padding=<ltrb 0 0 0 62>}}
  }
  Overlay_1 {
   widgetClassPath=&Overlay
   parent=SafeZone_BuildWatermark
   widget=Overlay_1
   slot {class=SafeZoneSlot, properties {hAlign=HAlign_Center, vAlign=VAlign_Center}}
  }
  BuildWatermark {
   widgetClassPath=/Game/UI/Foundation/Watermark/BuildWatermark
   bIsVariable=true
   parent=Overlay_1
   widget=BuildWatermark
   properties {elementTag {tagName=HUD.Center.BuildWatermark}, hUDElementTag {gameplayTags []}}
   slot {class=OverlaySlot}
  }
  Mobile_ScaleBox_BuildWatermark {
   widgetClassPath=&ScaleBox
   bIsVariable=true
   parent=Overlay_1
   widget=Mobile_ScaleBox_BuildWatermark
   properties {stretch=UserSpecified, userSpecifiedScale=1.3}
   slot {class=OverlaySlot, properties {horizontalAlignment=HAlign_Fill, verticalAlignment=VAlign_Center}}
  }
  SafeZone_FPS {
   widgetClassPath=&SafeZone
   bIsVariable=true
   parent=Overlay_4
   widget=SafeZone_FPS
   properties {padBottom=false, padTop=false}
   slot {class=OverlaySlot, properties {verticalAlignment=VAlign_Bottom, padding=<ltrb 0 0 0 66>}}
  }
  AthenaFPS {
   widgetClassPath=/Game/Athena/HUD/AthenaFPS
   bIsVariable=true
   parent=SafeZone_FPS
   widget=AthenaFPS
   properties {renderOpacity=0.5}
   slot {class=SafeZoneSlot}
  }
  AthenaWatchers {
   widgetClassPath=/Game/Athena/HUD/AthenaWatchers
   parent=Overlay_4
   widget=AthenaWatchers
   slot {class=OverlaySlot, properties {verticalAlignment=VAlign_Bottom, padding=<ltrb 20 0 0 150>}}
  }
  SafeZone_Temperature {
   widgetClassPath=&SafeZone
   bIsVariable=true
   parent=Overlay_4
   widget=SafeZone_Temperature
   properties {padBottom=false, padTop=false}
   slot {class=OverlaySlot, properties {horizontalAlignment=HAlign_Right, verticalAlignment=VAlign_Bottom}}
  }
  AthenaTemperature {
   widgetClassPath=/FortniteCoreUI/Athena/HUD/AthenaTemperature
   bIsVariable=true
   parent=SafeZone_Temperature
   widget=AthenaTemperature
   slot {class=SafeZoneSlot}
  }
  WBP_TextChat_CloseBackplateClicker {
   widgetClassPath=/FortniteCoreUI/UI/Social/Chat/WBP_TextChat_CloseBackplateClicker
   bIsVariable=true
   parent=Overlay_4
   widget=WBP_TextChat_CloseBackplateClicker
   slot {class=OverlaySlot, properties {horizontalAlignment=HAlign_Fill, verticalAlignment=VAlign_Fill}}
  }
  SafeZone_Bottom {
   widgetClassPath=&SafeZone
   bIsVariable=true
   parent=Overlay_4
   widget=SafeZone_Bottom
   properties {padTop=false}
   slot {class=OverlaySlot, properties {verticalAlignment=VAlign_Bottom, padding=<ltrb 16 0 0 16>}}
  }
  VerticalBox_Bottom {widgetClassPath=&VerticalBox, bIsVariable=true, parent=SafeZone_Bottom, widget=VerticalBox_Bottom, slot {class=SafeZoneSlot}}
  Overlay_ChatPanel {widgetClassPath=&Overlay, bIsVariable=true, parent=VerticalBox_Bottom, widget=Overlay_ChatPanel, slot {class=VerticalBoxSlot}}
  WBP_TextChat_InputListener {
   widgetClassPath=/FortniteCoreUI/TextChat/UI/WBP_TextChat_InputListener
   bIsVariable=true
   parent=Overlay_ChatPanel
   widget=WBP_TextChat_InputListener
   slot {class=OverlaySlot, properties {horizontalAlignment=HAlign_Fill}}
  }
  DualBladeMenu_Widget {
   widgetClassPath=/FortniteCoreUI/Frontend/MainMenu/DualBladeMenu
   bInherited=true
   bIsVariable=true
   parent=Overlay_4
   widget=DualBladeMenu_Widget
   slot {class=OverlaySlot, properties {horizontalAlignment=HAlign_Fill, verticalAlignment=VAlign_Fill}}
  }
  FortVoiceChannelSpeakerContainer {
   widgetClassPath=/Game/UI/Social/FortVoiceChannelSpeakerContainer
   bIsVariable=true
   parent=Overlay_4
   widget=FortVoiceChannelSpeakerContainer
   slot {class=OverlaySlot, properties {verticalAlignment=VAlign_Bottom, padding=<ltrb 0 0 0 32>}}
  }
  Overlay_DynamicExtensions {
   widgetClassPath=&Overlay
   bInherited=true
   parent=Overlay_4
   widget=Overlay_DynamicExtensions
   slot {class=OverlaySlot, properties {horizontalAlignment=HAlign_Fill, verticalAlignment=VAlign_Fill}}
  }
  FrontendEventNotifications {
   widgetClassPath=&/FrontendEventNotifications
   bIsVariable=true
   parent=Overlay_4
   widget=FrontendEventNotifications
   slot {class=OverlaySlot, properties {padding=<ltrb 30 100 0 0>}}
  }
  DecoyButton_MobileMicToggle {
   widgetClassPath=/Game/UI/Frontend/BottomBar/WBP_BottomBarDecoyButton
   bIsVariable=true
   parent=Overlay_4
   widget=DecoyButton_MobileMicToggle
   properties {triggeringInputAction {dataTable=&InputActionDataTable, rowName=MobileMicToggle}}
   slot {class=OverlaySlot}
  }
 ]
}
