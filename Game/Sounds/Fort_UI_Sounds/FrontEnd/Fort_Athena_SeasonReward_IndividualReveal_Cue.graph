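# Sound cue for the UI sound of revealing an individual season reward in the Athena front end.
# A single wave player feeds the output directly, reusing the map hover UI wave at reduced volume in the Fort_UI sound class.
# A concurrency override allows only one instance at a time and stops the oldest when a new reveal plays.
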SoundCue /Game/Sounds/Fort_UI_Sounds/FrontEnd/Fort_Athena_SeasonReward_IndividualReveal_Cue
root=WavePlayer_1
nodes [{id=WavePlayer_1, properties {soundWaveAssetPtr=/Game/Sounds/Fort_UI_Sounds/FrontEnd/New_FrontEnd/Fortnite_UI_Map_Hover_01}}]
properties {
 bOverrideConcurrency=true
 maxDistance=2097152
 soundClassObject=&Fort_UI
 volumeMultiplier=0.5
 concurrencyOverrides {maxCount=1, resolutionRule=StopOldest}
}
