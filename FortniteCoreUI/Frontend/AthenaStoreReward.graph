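# AthenaStoreReward is a FortAthenaReward widget showing a cosmetic reward card with locked, claimed, owned and rating-gate indicators.
# When the represented item is set, it sets Owned/Equipped text and pushes rarity colors into the flare and reveal overlay dynamic materials.
# It then shows a cosmetic gate warning icon if the item has rating restrictions for the local player.
# PlayIntro plays a flare and overlay reveal animation with a sound.
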WidgetBlueprint /FortniteCoreUI/Frontend/AthenaStoreReward
animations [
 Intro {
  binding_count=2
  display_label=""
  length_seconds=0.75
  object=&self:Intro
  sequence {bindings [
    RarityFlare {
     id=#1
     possessed_class=Image
     tracks [
      {
       type=MovieScene2DTransformTrack
       sections [{
         type=MovieScene2DTransformSection
         end_frame=15
         start_frame=0
         channels [
          Translation.X {type=MovieSceneScriptingFloatChannel, const [0,1]}
          Translation.Y {type=MovieSceneScriptingFloatChannel, const [0,1]}
          Angle {type=MovieSceneScriptingFloatChannel, const [0,1]}
          Scale.X {type=MovieSceneScriptingFloatChannel, keys [[0,0.75],[3,1.25],[15,0.75]]}
          Scale.Y {type=MovieSceneScriptingFloatChannel, keys [[0,0.75],[3,1.25],[15,0.75]]}
          Shear.X {type=MovieSceneScriptingFloatChannel, const [0,1]}
          Shear.Y {type=MovieSceneScriptingFloatChannel, const [0,1]}
         ]
        }]
      }
      {
       type=MovieSceneFloatTrack
       display_name=Render Opacity
       sections [{type=MovieSceneFloatSection, end_frame=15, start_frame=0, channels [None {type=MovieSceneScriptingFloatChannel, keys [[0,0],[3,1],[15,0]]}]}]
      }
      {
       type=MovieSceneByteTrack
       display_name=Visibility
       sections [{type=MovieSceneByteSection, end_frame=15, start_frame=0, channels [None {type=MovieSceneScriptingByteChannel, keys [[0,4],[15,1]]}]}]
      }
     ]
    }
    RevealOverlay {
     id=#2
     possessed_class=Image
     tracks [
      {
       type=MovieSceneFloatTrack
       display_name=Render Opacity
       sections [{type=MovieSceneFloatSection, end_frame=15, start_frame=0, channels [None {type=MovieSceneScriptingFloatChannel, keys [[6.9997,1],[15,0]]}]}]
      }
      {
       type=MovieSceneByteTrack
       display_name=Visibility
       sections [{type=MovieSceneByteSection, end_frame=15, start_frame=0, channels [None {type=MovieSceneScriptingByteChannel, keys [[0,4],[15,1]]}]}]
      }
     ]
    }
   ]}
 }
 Fade {
  binding_count=1
  length_seconds=0.1
  object=&self:Fade
  sequence {bindings [Overlay_0 {
     id=#3
     possessed_class=Overlay
     tracks [{
       type=MovieSceneFloatTrack
       display_name=Render Opacity
       sections [{type=MovieSceneFloatSection, end_frame=2, start_frame=0, channels [None {type=MovieSceneScriptingFloatChannel, keys [[0,0],[2,1]]}]}]
      }]
    }]}
 }
]
graphs [
 EnableRarityFlare {nodes [
   EnableRarityFlare() {is=title K2Node_FunctionEntry}
   n1 = Get RarityFlare() -> (RarityFlare:Image Object Reference)
   > SetVisibility(self=@n1, InVisibility=SelfHitTestInvisible)
   > SetRenderScale!(execute, self:Widget Object Reference=@n1, Scale:Vector 2D Structure=(X=1.5,Y=1.5)) -> (then)
  ]}
 EventGraph {nodes [
   PlayIntro() {is=title K2Node_CustomEvent}
   > PlayAnimation(InAnimation=Get Intro())
   > PlaySound2D(Sound=Get SoundOnReveal() -> (SoundOnReveal:Sound Base Object Reference))
   n1 = Event OnRepresentedItemSet!() -> (OutputDelegate, then, RepresentedItem:Fort Item Object Reference, IsEquipped:bool) {is=title K2Node_Event}
   > Set Text(
    Text=Select(Option 0:Text="LOCTEXT(\"OWNED\")", Option 1:Text="LOCTEXT(\"EQUIPPED\")", Index=@n1.IsEquipped) -> (ReturnValue:Text)
    self=Get OwnedText() -> (OwnedText:Common Text Object Reference)
   )
   > n2 = GetRarityData!(execute, Rarity:EFortRarity Enum=GetRarity!(self:Fort Item Object Reference=@n1.RepresentedItem) -> (ReturnValue:EFortRarity Enum=Common)) -> (then, ReturnValue:Fort Rarity Item Data Structure)
   > n3 = GetDynamicMaterial(self=Get RarityFlare())
   n4 = Break Fort Rarity Item Data!(FortRarityItemData:"Fort Rarity Item Data Structure (by ref)"=@n2) -> (
    Name:Text
    Color1:Linear Color Structure
    Color2:Linear Color Structure
    Color3:Linear Color Structure
    Color4:Linear Color Structure
    Color5:Linear Color Structure
    Radius:float=~
    Falloff:float=~
    Brightness:float=~
    Roughness:float=~
    Glow:float=~
   ) {is=title K2Node_BreakStruct}
   > SetVectorParameterValue(self=@n3, ParameterName=Rarity Color 1, Value=@n4.Color2)
   > SetVectorParameterValue(self=@n3, ParameterName=Rarity Color 2, Value=@n4.Color3)
   > n5 = GetDynamicMaterial(self=Get RevealOverlay() -> (RevealOverlay:Image Object Reference))
   > SetVectorParameterValue(self=@n5, ParameterName=RGB_1, Value=@n4.Color1)
   > SetVectorParameterValue(self=@n5, ParameterName=RGB_2, Value=@n4.Color2)
   > SetVectorParameterValue(self=@n5, ParameterName=RGB_3, Value=@n4.Color3)
   > SetVectorParameterValue(self=@n5, ParameterName=RGB_4, Value=@n4.Color3)
   > UpdateCosmeticGateWarning!(
    execute
    self
    Item:Fort Item Definition Object Reference=GetItemDefinition!(self:Fort Item Object Reference=@n1.RepresentedItem) -> (ReturnValue:Fort Item Definition Object Reference)
   ) -> (then)
  ]}
 UpdateCosmeticGateWarning {nodes [
   n1 = UpdateCosmeticGateWarning!() -> (then, Item:Fort Item Definition Object Reference) {is=title K2Node_FunctionEntry}
   > n2 = Cast To AthenaCosmeticItemDefinition!(execute, Object=@n1) -> (then, CastFailed, AsAthena Cosmetic Item Definition:Athena Cosmetic Item Definition Object Reference) {is=title K2Node_DynamicCast}
   > n3 = HasRatingRestrictions!(
    execute
    LocalPlayer:Local Player Object Reference=GetOwningLocalPlayer!(self:Widget Object Reference) -> (ReturnValue:Local Player Object Reference)
    CosmeticItemGatingInterface:"Cosmetic Item Gating Interface Interface (by ref)"=@n2
   ) -> (then, ReturnValue:bool)
   > n4 = Branch(Condition=@n3)
   n5 = Get Image_CosmeticGate() -> (Image_CosmeticGate:Image Object Reference)
   > SetVisibility(self=@n5, InVisibility=SelfHitTestInvisible)
   SetVisibility(execute=[@n4.else, @n2.CastFailed], self=@n5, InVisibility=Collapsed)
  ]}
]
variables [
 SoundOnReveal {type=SoundBase, default=/Game/Sounds/Fort_UI_Sounds/FrontEnd/Fort_Athena_SeasonReward_IndividualReveal_Cue}
 StrongItemToRepresent {type=FortItem, default=null}
]
widgets {
 info {inheritedWidgetCount=6, parentClass=/Script/FortniteUI.FortAthenaReward, rootWidgetClass=&Overlay, widgetCount=16}
 widgets [
  Overlay_0 {widgetClassPath=&Overlay, widget=Overlay_0}
  RarityFlare {
   widgetClassPath=&Image
   bIsVariable=true
   parent=Overlay_0
   widget=RarityFlare
   properties {
    visibility=Collapsed
    brush {
     bIsDynamicallyLoaded=false
     drawAs=Image
     imageType=NoImage
     mirroring=NoMirror
     resourceName=None
     resourceObject=/Game/UI/Frontend/Rewards/materials/M_UI_SeasonReward_Rarity-Flares
     tiling=NoTile
     imageSize=<32 32>
     margin=<ltrb 0 0 0 0>
     outlineSettings {
      bUseBrushTransparency=false
      roundingType=HalfHeightRadius
      width=0
      color {colorUseRule=UseColor_Specified, specifiedColor=<rgba 0 0 0 0>}
      cornerRadii=<0 0 0 0>
     }
     tintColor {colorUseRule=UseColor_Specified, specifiedColor=<rgba 1 1 1 1>}
     uVRegion {bIsValid=false, max=<0 0>, min=<0 0>}
    }
   }
   slot {class=OverlaySlot, properties {horizontalAlignment=HAlign_Fill, verticalAlignment=VAlign_Fill, padding=<ltrb -64 -64 -64 -64>}}
  }
  ItemCard {
   widgetClassPath=/FortniteCoreUI/Athena/UI/Frontend/CosmeticItemCard/FortCosmeticItemCard
   bInherited=true
   bIsVariable=true
   parent=Overlay_0
   widget=ItemCard
   properties {visibility=Visible}
   slot {class=OverlaySlot, properties {horizontalAlignment=HAlign_Center, verticalAlignment=VAlign_Center}}
  }
  Widget_UnearnedTintOverlay {
   widgetClassPath=&Image
   bInherited=true
   bIsVariable=true
   parent=Overlay_0
   widget=Widget_UnearnedTintOverlay
   properties {renderOpacity=0.4, visibility=SelfHitTestInvisible, colorAndOpacity=<rgba 0 0.1167 0.7 1>}
   slot {class=OverlaySlot, properties {horizontalAlignment=HAlign_Fill, verticalAlignment=VAlign_Fill}}
  }
  Widget_LockedImage {
   widgetClassPath=/FortUILibrary/UIKit/StatusIndicator/WBP_UIKit_StatusIndicator
   bInherited=true
   bIsVariable=true
   parent=Overlay_0
   widget=Widget_LockedImage
   properties {type=Utility, icon=/FortUILibrary/Textures/MaterialInstances/MI_UIKit_Icon_Locked, overrideStatusIndicatorVisibility=true}
   slot {class=OverlaySlot, properties {horizontalAlignment=HAlign_Center, verticalAlignment=VAlign_Center}}
  }
  Widget_ClaimedImage {
   widgetClassPath=&Image
   bInherited=true
   bIsVariable=true
   parent=Overlay_0
   widget=Widget_ClaimedImage
   properties {
    visibility=SelfHitTestInvisible
    brush {
     bIsDynamicallyLoaded=false
     drawAs=Image
     imageType=NoImage
     mirroring=NoMirror
     resourceName=None
     resourceObject=/FortUILibrary/Styles/Border/Materials/M_CircledCheckMark
     tiling=NoTile
     imageSize=<36 36>
     margin=<ltrb 0 0 0 0>
     outlineSettings {
      bUseBrushTransparency=false
      roundingType=HalfHeightRadius
      width=0
      color {colorUseRule=UseColor_Specified, specifiedColor=<rgba 0 0 0 0>}
      cornerRadii=<0 0 0 0>
     }
     tintColor {colorUseRule=UseColor_Specified, specifiedColor=<rgba 1 1 1 1>}
     uVRegion {bIsValid=false, max=<0 0>, min=<0 0>}
    }
    renderTransform {angle=0, scale=<1 1>, shear=<0 0>, translation=<8 -8>}
   }
   slot {class=OverlaySlot, properties {horizontalAlignment=HAlign_Right}}
  }
  RevealOverlay {
   widgetClassPath=&Image
   bIsVariable=true
   parent=Overlay_0
   widget=RevealOverlay
   properties {
    visibility=Collapsed
    brush {
     bIsDynamicallyLoaded=false
     drawAs=Box
     imageType=NoImage
     mirroring=NoMirror
     resourceName=None
     resourceObject=/FortUILibrary/Styles/Border/Materials/M_Border_SeasonReward_ColorOverlay
     tiling=NoTile
     imageSize=<48 48>
     margin=<ltrb 0.5 0.5 0.5 0.5>
     outlineSettings {
      bUseBrushTransparency=false
      roundingType=HalfHeightRadius
      width=0
      color {colorUseRule=UseColor_Specified, specifiedColor=<rgba 0 0 0 0>}
      cornerRadii=<0 0 0 0>
     }
     tintColor {colorUseRule=UseColor_Specified, specifiedColor=<rgba 1 1 1 1>}
     uVRegion {bIsValid=false, max=<0 0>, min=<0 0>}
    }
   }
   slot {class=OverlaySlot, properties {horizontalAlignment=HAlign_Fill, verticalAlignment=VAlign_Fill}}
  }
  Widget_OwnedBorder {
   widgetClassPath=&Overlay
   bInherited=true
   bIsVariable=true
   parent=Overlay_0
   widget=Widget_OwnedBorder
   slot {class=OverlaySlot, properties {horizontalAlignment=HAlign_Fill}}
  }
  baking {
   widgetClassPath=&Image
   bIsVariable=true
   parent=Widget_OwnedBorder
   widget=baking
   properties {brush {
     bIsDynamicallyLoaded=false
     drawAs=Image
     imageType=NoImage
     mirroring=NoMirror
     resourceName=None
     resourceObject=/Game/UI/Frontend/ItemShop/Com/MI_TopBarOwnedTag_Inst
     tiling=NoTile
     imageSize=<32 32>
     margin=<ltrb 0 0 0 0>
     outlineSettings {
      bUseBrushTransparency=false
      roundingType=HalfHeightRadius
      width=0
      color {colorUseRule=UseColor_Specified, specifiedColor=<rgba 0 0 0 0>}
      cornerRadii=<0 0 0 0>
     }
     tintColor {colorUseRule=UseColor_Specified, specifiedColor=<rgba 1 1 1 1>}
     uVRegion {bIsValid=false, max=<0 0>, min=<0 0>}
    }}
   slot {class=OverlaySlot, properties {horizontalAlignment=HAlign_Fill, verticalAlignment=VAlign_Fill}}
  }
  SizeBox_0 {
   widgetClassPath=&SizeBox
   parent=Widget_OwnedBorder
   widget=SizeBox_0
   properties {
    bOverride_HeightOverride=true
    bOverride_WidthOverride=true
    heightOverride=20
    widthOverride=100
    renderTransform {angle=0, scale=<1 1>, shear=<0 0>, translation=<0 1>}
   }
   slot {class=OverlaySlot}
  }
  HorizontalBox_1 {
   widgetClassPath=&HorizontalBox
   parent=SizeBox_0
   widget=HorizontalBox_1
   properties {renderTransform {angle=-2, scale=<1 1>, shear=<-11 0>, translation=<0 0>}}
   slot {class=SizeBoxSlot, properties {horizontalAlignment=HAlign_Left, padding=<ltrb 5 0 0 0>}}
  }
  ScaleBox_0 {
   widgetClassPath=&ScaleBox
   parent=HorizontalBox_1
   widget=ScaleBox_0
   properties {stretch=ScaleToFitX, stretchDirection=DownOnly}
   slot {class=HorizontalBoxSlot, properties {horizontalAlignment=HAlign_Left, padding=<ltrb 0 0 2 0>, size {sizeRule=Fill, value=1}}}
  }
  OwnedText {
   widgetClassPath=&CommonTextBlock
   bIsVariable=true
   parent=ScaleBox_0
   widget=OwnedText
   properties {
    scrollStyle=&TextScrollStyle-Base_C
    text=OWNED
    font {
     bForceMonospaced=false
     bMaterialIsStencil=false
     fontMaterial=null
     fontObject=&BurbankSmall
     letterSpacing=0
     monospacedWidth=1
     size=14
     skewAmount=0
     typefaceFontName=Bold
     outlineSettings {
      bApplyOutlineToDropShadows=false
      bMiteredCorners=false
      bSeparateFillAlpha=false
      outlineMaterial=null
      outlineSize=0
      outlineColor=<rgba 0 0 0 1>
     }
    }
    shadowColorAndOpacity=<rgba 0 0 0 0.402>
   }
   slot {class=ScaleBoxSlot, properties {horizontalAlignment=HAlign_Left}}
  }
  Check {
   widgetClassPath=&Image
   bIsVariable=true
   parent=HorizontalBox_1
   widget=Check
   properties {
    brush {
     bIsDynamicallyLoaded=false
     drawAs=Image
     imageType=NoImage
     mirroring=NoMirror
     resourceName=None
     resourceObject=/Game/UI/Foundation/Textures/Icons/Checks/T-Icon-Check-32
     tiling=NoTile
     imageSize=<16 16>
     margin=<ltrb 0 0 0 0>
     outlineSettings {
      bUseBrushTransparency=false
      roundingType=HalfHeightRadius
      width=0
      color {colorUseRule=UseColor_Specified, specifiedColor=<rgba 0 0 0 0>}
      cornerRadii=<0 0 0 0>
     }
     tintColor {colorUseRule=UseColor_Specified, specifiedColor=<rgba 1 1 1 1>}
     uVRegion {bIsValid=false, max=<0 0>, min=<0 0>}
    }
    colorAndOpacity=<rgba 0.07 0.7 0.175 0.7>
    renderTransform {angle=0, scale=<1 1>, shear=<0 0>, translation=<0 -2>}
   }
   slot {class=HorizontalBoxSlot, properties {horizontalAlignment=HAlign_Left, verticalAlignment=VAlign_Center}}
  }
  Image_CosmeticGate {
   widgetClassPath=&Image
   bIsVariable=true
   parent=Overlay_0
   widget=Image_CosmeticGate
   properties {
    visibility=Collapsed
    brush {
     bIsDynamicallyLoaded=false
     drawAs=Image
     imageType=FullColor
     mirroring=NoMirror
     resourceName=None
     resourceObject=/Game/Athena/UI/Frontend/CosmeticGating/Materials/MI_UI_CosmeticGate_WarningIcon
     tiling=NoTile
     imageSize=<40 40>
     margin=<ltrb 0 0 0 0>
     outlineSettings {
      bUseBrushTransparency=false
      roundingType=HalfHeightRadius
      width=0
      color {colorUseRule=UseColor_Specified, specifiedColor=<rgba 0 0 0 0>}
      cornerRadii=<0 0 0 0>
     }
     tintColor {colorUseRule=UseColor_Specified, specifiedColor=<rgba 1 1 1 1>}
     uVRegion {bIsValid=false, max=<0 0>, min=<0 0>}
    }
   }
   slot {class=OverlaySlot, properties {verticalAlignment=VAlign_Bottom, padding=<ltrb -6 0 0 -8>}}
  }
  InputActionWidget {widgetClassPath=&CommonActionWidget, bInherited=true, widget=null}
 ]
}
