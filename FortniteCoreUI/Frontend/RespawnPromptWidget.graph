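# Modal confirmation widget, based on FortRespawnPromptWidget, that asks the player whether they want to respawn.
# On activation it plays a short Intro fade and scale animation on the modal border, along with a whoosh sound, over a blurred, dimmed lightbox.
# Respawn casts the owning player to FortPlayerController, calls ForcedSuicide, then pops the panel from the modal layer.
# Cancel, the touch-only close button and a failed cast just pop the panel.
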WidgetBlueprint /FortniteCoreUI/Frontend/RespawnPromptWidget
animations [Intro {
  binding_count=1
  length_seconds=0.25
  object=&self:Intro
  sequence {bindings [CommonBorder_80 {
     id=#1
     possessed_class=CommonBorder
     tracks [
      {
       type=MovieSceneFloatTrack
       display_name=Render Opacity
       sections [{type=MovieSceneFloatSection, end_frame=5, start_frame=0, channels [None {type=MovieSceneScriptingFloatChannel, keys [[0,0],[2,0],[5,1]]}]}]
      }
      {
       type=MovieScene2DTransformTrack
       sections [{
         type=MovieScene2DTransformSection
         end_frame=5
         start_frame=0
         channels [
          Scale.X {type=MovieSceneScriptingFloatChannel, keys [[2,1.15],[4,1]]}
          Scale.Y {type=MovieSceneScriptingFloatChannel, keys [[2,1.15],[4,1]]}
         ]
        }]
      }
     ]
    }]}
 }]
graphs [EventGraph {nodes [
   n1 = "On Button Clicked (Button_CloseTouch)"!() -> (OutputDelegate, then, Button:Common Button Legacy Object Reference) {is=title K2Node_ComponentBoundEvent}
   PopActivatablePanelInModalLayer!(execute=[@n3.then, @n1.then, @n4.then, @n2.CastFailed], Panel:Common Activatable Panel Legacy Object Reference=Self-Reference()) -> (then)
   "On Clicked (ApplyButton)"!() -> (OutputDelegate, then, Button:Common Button Base Object Reference) {is=title K2Node_ComponentBoundEvent}
   > n2 = Cast To FortPlayerController(Object=GetOwningPlayer())
   > n3 = ForcedSuicide(self:Fort Player Controller Object Reference=@n2)
   n4 = "On Clicked (CancelButton)"!() -> (OutputDelegate, then, Button:Common Button Base Object Reference) {is=title K2Node_ComponentBoundEvent}
   Event On Activated()
   > PlayAnimationForward(InAnimation=Get Intro())
   > SpawnSound2D!(
    execute
    Sound:Sound Base Object Reference=Get IntroSound() -> (IntroSound:Sound Base Object Reference)
    VolumeMultiplier:float=1
    PitchMultiplier:float=1
    StartTime:float=0
    ConcurrencySettings:Sound Concurrency Object Reference
    bPersistAcrossLevelTransition:bool=false
    bAutoDestroy:bool=true
   ) -> (then, ReturnValue:Audio Component Object Reference)
  ]}]
variables [IntroSound {type=SoundBase, default=/Game/Sounds/Fort_UI_Sounds/FrontEnd/Whooshes/Toasts/UI_Whoosh_Generic_Toast_Appear_Cue}]
widgets {
 info {
  inheritedWidgetCount=0
  namedSlotCount=1
  parentClass=/Script/FortniteUI.FortRespawnPromptWidget
  rootWidgetClass=&BackgroundBlur
  widgetCount=19
 }
 widgets [
  BackgroundBlur_2 {widgetClassPath=&BackgroundBlur, widget=BackgroundBlur_2, properties {blurStrength=15}}
  Lightbox {
   widgetClassPath=/Game/UI/Foundation/Lightbox/Lightbox
   bIsVariable=true
   parent=BackgroundBlur_2
   widget=Lightbox
   properties {dimAmount=0.75, backgroundColor=<rgba 0 0.0006 0.0075 1>}
   slot {class=BackgroundBlurSlot, properties {padding=<ltrb 0 0 0 0>}}
  }
  RootOverlay {widgetClassPath=&Overlay, namedSlotHost=Lightbox, widget=RootOverlay}
  TouchToCloseZone {
   widgetClassPath=&CommonBorder
   parent=RootOverlay
   widget=TouchToCloseZone
   properties {horizontalAlignment=HAlign_Center, verticalAlignment=VAlign_Center, padding=<ltrb 0 0 0 0>}
   slot {class=OverlaySlot, properties {horizontalAlignment=HAlign_Fill, verticalAlignment=VAlign_Fill}}
  }
  SizeBox_4 {
   widgetClassPath=&SizeBox
   parent=TouchToCloseZone
   widget=SizeBox_4
   properties {
    bOverride_MaxAspectRatio=true
    bOverride_MinDesiredHeight=true
    bOverride_MinDesiredWidth=true
    maxAspectRatio=2.5
    minDesiredHeight=600
    minDesiredWidth=1000
   }
   slot {class=BorderSlot, properties {horizontalAlignment=HAlign_Center, verticalAlignment=VAlign_Center, padding=<ltrb 0 0 0 0>}}
  }
  CommonBorder_80 {
   widgetClassPath=&CommonBorder
   parent=SizeBox_4
   widget=CommonBorder_80
   properties {style=/FortUILibrary/Styles/Backgrounds/_BorderStyle_Modal._BorderStyle_Modal_C}
   slot {class=SizeBoxSlot}
  }
  Overlay_0 {widgetClassPath=&Overlay, parent=CommonBorder_80, widget=Overlay_0, slot {class=BorderSlot}}
  WBP_CaptureForPostBufferUpdate {
   widgetClassPath=&WBP_CaptureForPostBufferUpdate
   bIsVariable=true
   parent=Overlay_0
   slot {class=OverlaySlot, properties {horizontalAlignment=HAlign_Fill, verticalAlignment=VAlign_Fill}}
  }
  VerticalBox_0 {
   widgetClassPath=&VerticalBox
   parent=Overlay_0
   widget=VerticalBox_0
   slot {class=OverlaySlot, properties {horizontalAlignment=HAlign_Fill, verticalAlignment=VAlign_Fill}}
  }
  CommonBorder_Modal {
   widgetClassPath=&CommonBorder
   parent=VerticalBox_0
   widget=CommonBorder_Modal
   properties {style=/FortUILibrary/Styles/Backgrounds/_BorderStyle_ModalHeader._BorderStyle_ModalHeader_C, padding=<ltrb 40 0 4 0>}
   slot {class=VerticalBoxSlot}
  }
  HorizontalBox_1 {
   widgetClassPath=&HorizontalBox
   parent=CommonBorder_Modal
   widget=HorizontalBox_1
   slot {class=BorderSlot, properties {padding=<ltrb 40 0 4 0>}}
  }
  Spacer_0 {widgetClassPath=&Spacer, parent=HorizontalBox_1, widget=Spacer_0, properties {size=<0 85>}, slot {class=HorizontalBoxSlot}}
  Title {
   widgetClassPath=&CommonTextBlock
   bIsVariable=true
   parent=HorizontalBox_1
   widget=Title
   properties {
    style=/FortUILibrary/Styles/Typography/_TextStyle_SubH2._TextStyle_SubH2_C
    text=RESPAWN
    font {
     bForceMonospaced=false
     bMaterialIsStencil=false
     fontMaterial=null
     fontObject=&Burbank
     letterSpacing=0
     monospacedWidth=1
     size=33
     skewAmount=0
     typefaceFontName=BurbankRegular-Black
     outlineSettings {
      bApplyOutlineToDropShadows=false
      bMiteredCorners=false
      bSeparateFillAlpha=false
      outlineMaterial=null
      outlineSize=0
      outlineColor=<rgba 0 0 0 1>
     }
    }
    shadowOffset=<0 0>
    strikeBrush {
     bIsDynamicallyLoaded=false
     drawAs=NoDrawType
     imageType=NoImage
     mirroring=NoMirror
     resourceName=None
     resourceObject=null
     tiling=NoTile
     imageSize=<32 32>
     margin=<ltrb 0 0 0 0>
     outlineSettings {
      bUseBrushTransparency=false
      roundingType=HalfHeightRadius
      width=0
      color {colorUseRule=UseColor_Specified, specifiedColor=<rgba 0 0 0 0>}
      cornerRadii=<0 0 0 0>
     }
     tintColor {colorUseRule=UseColor_Specified, specifiedColor=<rgba 1 1 1 1>}
     uVRegion {bIsValid=false, max=<0 0>, min=<0 0>}
    }
   }
   slot {class=HorizontalBoxSlot, properties {horizontalAlignment=HAlign_Left, verticalAlignment=VAlign_Center, padding=<ltrb 8 0 20 0>}}
  }
  VisibilityBorder_Touch {
   widgetClassPath=&CommonHardwareVisibilityBorder
   parent=HorizontalBox_1
   widget=VisibilityBorder_Touch
   properties {
    horizontalAlignment=HAlign_Right
    visibilityQuery {
     autoDescription=" ANY( Input.Touch )"
     tokenStreamVersion=0
     userDescription=if using touch
     queryTokenStream [0,1,1,1,0]
     tagDictionary [{tagName=Input.Touch}]
    }
   }
   slot {class=HorizontalBoxSlot, properties {horizontalAlignment=HAlign_Right, size {sizeRule=Fill, value=1}}}
  }
  Button_CloseTouch {
   widgetClassPath=/FortUILibrary/Foundation/Buttons/CloseButton
   bIsVariable=true
   parent=VisibilityBorder_Touch
   widget=Button_CloseTouch
   properties {
    onButtonBaseClicked=(&self:WidgetTree.Button_CloseTouch.HandleOnButtonClicked)
    onButtonBaseDoubleClicked=(&self:WidgetTree.Button_CloseTouch.HandleOnButtonDoubleClicked)
    onButtonBaseHovered=(&self:WidgetTree.Button_CloseTouch.HandleOnButtonHovered)
    onButtonBaseUnhovered=(&self:WidgetTree.Button_CloseTouch.HandleOnButtonUnhovered)
    onSelectedChangedBase=(&self:WidgetTree.Button_CloseTouch.HandleOnSelectedChanged)
    renderTransform {angle=0, scale=<0.8 0.8>, shear=<0 0>, translation=<0 0>}
   }
   slot {class=BorderSlot, properties {horizontalAlignment=HAlign_Right, padding=<ltrb 0 0 0 0>}}
  }
  Text_CurrentTeam {
   widgetClassPath=&CommonTextBlock
   bIsVariable=true
   parent=VerticalBox_0
   widget=Text_CurrentTeam
   properties {
    lineHeightPercentage=1.29
    style=/FortUILibrary/Styles/Typography/_TextStyle_Body1._TextStyle_Body1_C
    text=Are you sure you want to respawn?
    font {
     bForceMonospaced=false
     bMaterialIsStencil=false
     fontMaterial=null
     fontObject=&BurbankSmall
     letterSpacing=0
     monospacedWidth=1
     size=18
     skewAmount=0
     typefaceFontName=Bold
     outlineSettings {
      bApplyOutlineToDropShadows=false
      bMiteredCorners=false
      bSeparateFillAlpha=false
      outlineMaterial=null
      outlineSize=0
      outlineColor=<rgba 0 0 0 1>
     }
    }
    shadowOffset=<0 0>
    strikeBrush {
     bIsDynamicallyLoaded=false
     drawAs=NoDrawType
     imageType=NoImage
     mirroring=NoMirror
     resourceName=None
     resourceObject=null
     tiling=NoTile
     imageSize=<32 32>
     margin=<ltrb 0 0 0 0>
     outlineSettings {
      bUseBrushTransparency=false
      roundingType=HalfHeightRadius
      width=0
      color {colorUseRule=UseColor_Specified, specifiedColor=<rgba 0 0 0 0>}
      cornerRadii=<0 0 0 0>
     }
     tintColor {colorUseRule=UseColor_Specified, specifiedColor=<rgba 1 1 1 1>}
     uVRegion {bIsValid=false, max=<0 0>, min=<0 0>}
    }
   }
   slot {class=VerticalBoxSlot, properties {padding=<ltrb 48 34 48 0>}}
  }
  HorizontalBox_6 {
   widgetClassPath=&HorizontalBox
   parent=VerticalBox_0
   widget=HorizontalBox_6
   slot {class=VerticalBoxSlot, properties {verticalAlignment=VAlign_Bottom, padding=<ltrb 48 0 48 32>, size {sizeRule=Fill, value=1}}}
  }
  ApplyButton {
   widgetClassPath=&WBP_UIKit_Button_Regular
   bIsVariable=true
   parent=HorizontalBox_6
   widget=ApplyButton
   properties {
    text=Respawn
    pressedSlateSoundOverride {resourceObject=/Game/Sounds/Fort_UI_Sounds/FrontEnd/Whooshes/Toasts/UI_Whoosh_Generic_Toast_Close_Cue}
    triggeringInputAction {dataTable=&InputActionDataTable, rowName=Apply}
   }
   slot {class=HorizontalBoxSlot, properties {size {sizeRule=Fill, value=1}}}
  }
  CancelButton {
   widgetClassPath=&WBP_UIKit_Button_Regular
   bIsVariable=true
   parent=HorizontalBox_6
   widget=CancelButton
   properties {
    text=Cancel
    pressedSlateSoundOverride {resourceObject=&UI_Whoosh_Generic_Toast_Close_Cue}
    triggeringInputAction {dataTable=&InputActionDataTable, rowName=Cancel}
   }
   slot {class=HorizontalBoxSlot, properties {padding=<ltrb 48 0 0 0>, size {sizeRule=Fill, value=1}}}
  }
 ]
}
